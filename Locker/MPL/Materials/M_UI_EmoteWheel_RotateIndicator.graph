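# Translucent UI material for the emote wheel's rotate indicator icon, with an animated radial shading sweep.
# Mirrored widget UVs become radial coordinates, which a looping Time value offsets to form a clamped, moving gradient that is scaled and biased into a subtle brightness.
# An SDF icon texture passes through an anti-aliasing function with Threshold and Smoothing parameters, and the resulting inner mask blends in that brightness.
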Material /Locker/MPL/Materials/M_UI_EmoteWheel_RotateIndicator
graph {
 properties {
  bUsedWithStaticMesh=false
  blendMode=BLEND_Translucent
  materialDomain=MD_UI
  pixelDepthOffsetMode=PDOM_Legacy
  usedShadingModels=Default Lit
 }
 nodes [
  Constant() {r=1}
  MaterialExpressionMaterialFunctionCall_3 = MaterialFunctionCall() {materialFunction=&GetUserInterfaceUV}
  n1 = Convert(None=@MaterialExpressionMaterialFunctionCall_3.Normalized UV) {
   nodeName=Break Float2
   convertInputs [{
     type=Vector2
     expressionInput {
      expression=&self:MaterialExpressionMaterialFunctionCall_3
      inputName=None
      mask=0
      maskA=0
      maskB=0
      maskG=0
      maskR=0
      outputIndex=2
     }
    }]
   convertOutputs [{type=Scalar}, {type=Scalar}]
  }
  n2 = MaterialFunctionCall!(Vector or UVs=AppendVector(A=OneMinus(None=@n1), B=@n1)) {materialFunction=/FortUILibrary/MaterialFunctions/Basics/MF_UI_CheaperVectorToRadial}
  n3 = Time() {bOverride_Period=true, period=4}
  MaterialExpressionOneMinus_1 = OneMinus(None=Subtract(A=Add(A=@n2.Radial Coordinates), B=Multiply(A=@n3) {constB=2}))
  n4 = Convert(None=@MaterialExpressionOneMinus_1) {
   nodeName=Break Float2
   convertInputs [{
     type=Vector2
     expressionInput {
      expression=&self:MaterialExpressionOneMinus_1
      inputName=None
      mask=0
      maskA=0
      maskB=0
      maskG=0
      maskR=0
      outputIndex=0
     }
    }]
   convertOutputs [{type=Scalar}, {type=Scalar}]
  }
  n5 = Saturate(None=LinearInterpolate(A=@n4, Alpha=@n4))
  n6 = Multiply(A=TextureSample(UVs=@MaterialExpressionMaterialFunctionCall_3.Normalized UV) {samplerType=SAMPLERTYPE_LinearColor, texture=/Locker/MPL/Materials/Textures/T_UI_Icon_RotateIndicator_SDF}.RGB) {constB=2}
  LinearInterpolate(
   B=Add(A=Multiply(A=@n5) {constB=0.35}) {constB=0.2}
   Alpha=MaterialFunctionCall(
    DistanceField=Subtract(A=@n6)
    Threshold=ScalarParameter() {defaultValue=-0.25, group=Distance Field, parameterName=Threshold}
    Smoothing=ScalarParameter() {defaultValue=0.1, group=Distance Field, parameterName=Smoothing}
   ) {materialFunction=&MF_UI_AntiAliasing}.Inner
  )
 ]
}
