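# UI-domain translucent material that draws the slot ring of the emote wheel as rounded, spaced radial segments.
# Offset, scaled UVs are rotated by half a segment and converted to radial coordinates, then split by the Subdivisions parameter.
# A rounded-box function in that polar space shapes each segment from thickness, padding, spacing and corner-radius parameters.
# It also builds a SelectedIndex slot mask and a backplate color from the blurred Slate post buffer.
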Material /Locker/MPL/Materials/M_UI_EmoteWheel_Slots
graph {
 properties {
  bAutomaticallySetUsageInEditor=false
  bUsedWithStaticMesh=false
  blendMode=BLEND_Translucent
  materialDomain=MD_UI
  usedShadingModels=Default Lit
 }
 nodes [
  n1 = NamedRerouteUsage()
  n2 = Floor(None=ScalarParameter() {desc=Index, group=Settings, parameterName=SelectedIndex})
  n3 = Abs(None=Subtract(A=Ceil(None=@n1), B=Add(A=@n2)))
  NamedRerouteDeclaration(None=OneMinus(None=Saturate(None=@n3))) {name=Selected Emote Slot Mask, nodeColor=<rgba 0 0 0 1>}
  n4 = NamedRerouteDeclaration(None=Floor(None=ScalarParameter() {defaultValue=8, group=Settings, parameterName=Subdivisions})) {name=Subdivisions, nodeColor=<rgba 0 0 0 1>}
  n5 = MaterialFunctionCall() {materialFunction=&GetUserInterfaceUV}
  n6 = ComponentMask(None=VectorParameter() {group=UV, parameterName=UVOffset, defaultValue=<rgba 0.75 -0.2 0 1>}.RGB) {g=true, r=true}
  n7 = VectorParameter() {group=UV, parameterName=UVScale, defaultValue=<rgba 0.4 0.4 0 1>}
  MaterialExpressionNamedRerouteDeclaration_3 = NamedRerouteDeclaration(None=Multiply(A=Add(A=@n5.Normalized UV, B=@n6), B=ComponentMask(None=@n7.RGB) {g=true, r=true})) {name=UV, nodeColor=<rgba 0 0 0 1>}
  n8 = Convert(None=@MaterialExpressionNamedRerouteDeclaration_3) {
   nodeName=Break Float2
   convertInputs [{
     type=Vector2
     expressionInput {
      expression=&self:MaterialExpressionNamedRerouteDeclaration_3
      inputName=None
      mask=0
      maskA=0
      maskB=0
      maskG=0
      maskR=0
      outputIndex=0
     }
    }]
   convertOutputs [{type=Scalar}, {type=Scalar}]
  }
  n9 = OneMinus(None=@n8)
  n10 = Max(A=NamedRerouteUsage()) {constB=0}
  MaterialExpressionMaterialFunctionCall_2 = MaterialFunctionCall!(Vector or UVs=Rotator(Coordinate=AppendVector(A=@n9, B=@n8), Time=Divide(B=@n10) {constA=-1}) {speed=3.1416}) {materialFunction=/FortUILibrary/MaterialFunctions/Basics/MF_UI_CheaperVectorToRadial}
  n11 = Convert(None=@MaterialExpressionMaterialFunctionCall_2.Radial Coordinates) {
   nodeName=Break Float2
   convertInputs [{
     type=Vector2
     expressionInput {
      expression=&self:MaterialExpressionMaterialFunctionCall_2
      inputName=None
      mask=0
      maskA=0
      maskB=0
      maskG=0
      maskR=0
      outputIndex=0
     }
    }]
   convertOutputs [{type=Scalar}, {type=Scalar}]
  }
  n12 = Frac(None=NamedRerouteDeclaration(None=Multiply(A=@n4, B=@n11) {constB=8}) {name=Subdivised Angle, nodeColor=<rgba 0 0 0 1>})
  n13 = Min(A=@n12, B=OneMinus(None=@n12))
  NamedRerouteDeclaration(None=DotProduct(A=NamedRerouteUsage(), B=NamedRerouteDeclaration(None=@n13) {name=Linear Radial Segments, nodeColor=<rgba 0 0 0 1>})) {name=Radial Segments, nodeColor=<rgba 0 0 0 1>}
  n14 = NamedRerouteUsage()
  n15 = OneMinus(None=NamedRerouteUsage())
  n16 = Add(
   A=Add(A=NamedRerouteUsage(), B=Multiply(A=@n15) {constB=-0.5, desc=Set Pivot Point Offset})
   B=NamedRerouteDeclaration(None=ScalarParameter() {defaultValue=0.025, desc=Ring Size, group=Settings, parameterName=DonutPadding, sliderMax=1}) {name=Donut Padding, nodeColor=<rgba 0 0 0 1>}
  ) {constB=0}
  n17 = NamedRerouteUsage()
  n18 = Subtract(
   A=NamedRerouteDeclaration(None=ScalarParameter() {defaultValue=0.375, desc=Ring Size, group=Main Segments, parameterName=MainRingThickness, sliderMax=1}) {name=Main Thickness, nodeColor=<rgba 0 0 0 1>}
   B=NamedRerouteUsage()
  )
  n19 = ScalarParameter() {defaultValue=0.15, group=Main Segments, parameterName=MainCornerRadius, sliderMax=1}
  NamedRerouteDeclaration(None=MaterialFunctionCall(UVs=AppendVector(A=@n14, B=@n16), Size=AppendVector(A=@n17, B=@n18), CornerRoundness=Multiply(A=@n18, B=@n19), KeepAspectRatio=StaticBool()) {materialFunction=&MF_UI_Box}.Fill) {name=Main Segments Mask, nodeColor=<rgba 0 0 0 1>}
  NamedRerouteDeclaration(None=Multiply(A=@n11) {constB=2}) {name=Vector Length, nodeColor=<rgba 0 0 0 1>}
  n20 = Multiply(A=ScalarParameter() {defaultValue=0.2, desc=Segment Spacing, group=Settings, parameterName=SegmentSpacing, sliderMax=1}) {constB=0.25}
  n21 = Multiply(A=NamedRerouteUsage(), B=@n20)
  NamedRerouteDeclaration(None=OneMinus(None=Multiply(A=@n21, B=@n20))) {name=Spacing, nodeColor=<rgba 0 0 0 1>}
  n22 = OneMinus(None=NamedRerouteUsage())
  NamedRerouteDeclaration(None=Add(A=NamedRerouteUsage(), B=Multiply(A=@n22) {constB=-1})) {name=Spaced Radial Segments, nodeColor=<rgba 0 0 0 1>}
  NamedRerouteDeclaration(None=MaterialFunctionCall!(Post Buffer Blur=MaterialFunctionCall!("UVs (Optional)") {materialFunction=/Engine/Functions/UserInterface/GetSlatePost0}.LinearRGB) {materialFunction=/FortUILibrary/MaterialFunctions/CoreUI/MF_UI_CoreUI_BackplateB}.Result) {name=Blur Color, nodeColor=<rgba 0 0 0 1>}
  NamedRerouteDeclaration(None=Multiply(A=NamedRerouteUsage(), B=NamedRerouteUsage())) {name=ButtonBackground, nodeColor=<rgba 1 0 0.9609 1>}
  NamedRerouteUsage()
  NamedRerouteUsage()
 ]
}
