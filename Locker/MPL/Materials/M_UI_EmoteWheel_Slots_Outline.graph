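# UI material that draws the animated outline and glow for emote wheel slots, with alpha-composite blending.
# Radial UVs are split into Subdivisions segments, shaped by a rounded-box SDF into outline and glow masks, alongside a SelectedIndex slot mask.
# Panning flow-map and polar rift noise, oscillating colors and a sine sheen add animation, with hover parameters and a UseFlatColor switch.
# Polar noise samples use explicit derivatives for mip selection.
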Material /Locker/MPL/Materials/M_UI_EmoteWheel_Slots_Outline
graph {
 properties {
  bAutomaticallySetUsageInEditor=false
  bUsedWithStaticMesh=false
  blendMode=BLEND_AlphaComposite
  materialDomain=MD_UI
  usedShadingModels=Default Lit
 }
 nodes [
  n1 = NamedRerouteUsage()
  Multiply(A=NamedRerouteUsage(), B=Multiply(A=VertexColor().A, B=@n1))
  n2 = NamedRerouteUsage()
  n3 = Floor(None=ScalarParameter() {desc=Index, group=Settings, parameterName=SelectedIndex})
  n4 = Abs(None=Subtract(A=Ceil(None=@n2), B=Add(A=@n3)))
  NamedRerouteDeclaration(None=OneMinus(None=Saturate(None=@n4))) {name=Selected Emote Slot Mask, nodeColor=<rgba 1 0.4723 0 1>}
  n5 = ScreenPosition()
  n6 = Add(A=NamedRerouteUsage())
  n7 = Time()
  n8 = Multiply(A=@n7) {constB=0.05}
  n9 = Add(A=Divide(A=@n5.ViewportUV, B=Multiply(A=@n6) {constB=2}), B=AppendVector(A=Multiply(A=@n7) {constB=-0.1}, B=@n8))
  n10 = NamedRerouteUsage()
  n11 = Divide(A=@n5.ViewportUV, B=Multiply(A=@n6) {constB=3})
  n12 = AppendVector(A=Multiply(A=@n7) {constB=0.025}, B=@n8)
  n13 = AppendVector(
   A=TextureSample(UVs=NamedRerouteDeclaration(None=@n9) {name=PanningUVs, nodeColor=<rgba 0 0 0 1>}, Tex=@n10).R
   B=TextureSample(UVs=Add(A=@n11, B=@n12), Tex=@n10).R
  )
  n14 = Constant() {r=0.2}
  n15 = MaterialFunctionCall!(
   UVs=NamedRerouteUsage()
   Texture Object=NamedRerouteUsage()
   Grayscale Flow Map=@n13
   Flow Direction=OneMinus(None=ComponentMask(None=@n13) {r=true})
   Flow Intensity=@n14
   Flow Speed=@n14
  ) {materialFunction=/FortUILibrary/MaterialFunctions/Effects/MF_UI_GrayscaleFlowMap_New}
  NamedRerouteDeclaration(None=SmoothStep(Value=ComponentMask(None=@n15.Distorted Texture) {r=true})) {name=Flowing Noise Mask, nodeColor=<rgba 0 0 0 1>}
  n16 = MaterialFunctionCall(
   UVs=NamedRerouteUsage()
   Rotation Center=Constant2Vector() {g=0.5, r=0.5}
   "Rotation Angle (0-1)"=ScalarParameter() {defaultValue=0.25, group=Anim, parameterName=SineRotationAngle, sliderMax=1}
  ) {materialFunction=&CustomRotator}
  n17 = ScalarParameter() {defaultValue=0.5, group=Anim, parameterName=Anim_Highlight, sliderMax=1}
  n18 = Sine(None=Add(A=ComponentMask(None=@n16.Rotated Values) {r=true}, B=LinearInterpolate(Alpha=@n17) {constA=1, constB=-1})) {period=2}
  n19 = Multiply(
   A=LinearInterpolate(Alpha=NamedRerouteUsage()) {constA=0.333}
   B=MaterialFunctionCall("Value (0-1)"=Saturate(None=@n18)) {materialFunction=&MF_UI_EaseCurves}.EaseInCubic
  )
  n20 = NamedRerouteUsage()
  NamedRerouteDeclaration(None=Multiply(A=Multiply(A=@n19, B=@n20), B=ScalarParameter() {defaultValue=0.5, group=Alpha, parameterName=Sheen_Alpha, sliderMax=1})) {name=Sheen Effect, nodeColor=<rgba 0 0 0 1>}
  n21 = NamedRerouteDeclaration(None=Floor(None=ScalarParameter() {defaultValue=8, group=Settings, parameterName=Subdivisions})) {name=Subdivisions, nodeColor=<rgba 0 0 0 1>}
  n22 = MaterialFunctionCall() {materialFunction=&GetUserInterfaceUV}
  n23 = ComponentMask(None=VectorParameter() {group=UV, parameterName=UVOffset, defaultValue=<rgba 0.75 -0.2 0 1>}.RGB) {g=true, r=true}
  n24 = VectorParameter() {group=UV, parameterName=UVScale, defaultValue=<rgba 0.4 0.4 0 1>}
  MaterialExpressionNamedRerouteDeclaration_3 = NamedRerouteDeclaration(None=Multiply(A=Add(A=@n22.Normalized UV, B=@n23), B=ComponentMask(None=@n24.RGB) {g=true, r=true})) {name=UV, nodeColor=<rgba 0 0 0 1>}
  n25 = Convert(None=@MaterialExpressionNamedRerouteDeclaration_3) {
   nodeName=Break Float2
   convertInputs [{
     type=Vector2
     expressionInput {
      expression=&self:MaterialExpressionNamedRerouteDeclaration_3
      inputName=None
      mask=0
      maskA=0
      maskB=0
      maskG=0
      maskR=0
      outputIndex=0
     }
    }]
   convertOutputs [{type=Scalar}, {type=Scalar}]
  }
  n26 = OneMinus(None=@n25)
  n27 = Max(A=NamedRerouteUsage()) {constB=0}
  MaterialExpressionMaterialFunctionCall_2 = MaterialFunctionCall!(Vector or UVs=Rotator(Coordinate=AppendVector(A=@n26, B=@n25), Time=Divide(B=@n27) {constA=-1}) {speed=3.1416}) {materialFunction=/FortUILibrary/MaterialFunctions/Basics/MF_UI_CheaperVectorToRadial}
  n28 = Convert(None=@MaterialExpressionMaterialFunctionCall_2.Radial Coordinates) {
   nodeName=Break Float2
   convertInputs [{
     type=Vector2
     expressionInput {
      expression=&self:MaterialExpressionMaterialFunctionCall_2
      inputName=None
      mask=0
      maskA=0
      maskB=0
      maskG=0
      maskR=0
      outputIndex=0
     }
    }]
   convertOutputs [{type=Scalar}, {type=Scalar}]
  }
  n29 = Frac(None=NamedRerouteDeclaration(None=Multiply(A=@n21, B=@n28) {constB=8}) {name=Subdivised Angle, nodeColor=<rgba 0 0 0 1>})
  n30 = Min(A=@n29, B=OneMinus(None=@n29))
  NamedRerouteDeclaration(None=DotProduct(A=NamedRerouteUsage(), B=NamedRerouteDeclaration(None=@n30) {name=Linear Radial Segments, nodeColor=<rgba 0 0 0 1>})) {name=Radial Segments, nodeColor=<rgba 0 0 0 1>}
  n31 = NamedRerouteUsage()
  n32 = Sine(None=Multiply(A=Time()) {constB=0.125})
  n33 = LinearInterpolate(
   A=NamedRerouteUsage()
   B=AppendVector(A=OneMinus(None=@n31), B=@n31)
   Alpha=LinearInterpolate(Alpha=ConstantBiasScale(None=@n32)) {constA=0.95, constB=0.85}
  )
  n34 = NamedRerouteUsage()
  n35 = MaterialFunctionCall(Vector or UVs=NamedRerouteDeclaration(None=MaterialFunctionCall!(Value=@n33, Scale=@n34, Height=@n34) {materialFunction=/FortUILibrary/MaterialFunctions/Basics/MF_UI_PingPong}.Result) {name=PrepPolarUVs, nodeColor=<rgba 0 0 0 1>}) {materialFunction=&MF_UI_CheaperVectorToRadial}
  n36 = ComponentMask(None=@n35.Radial Coordinates) {r=true}
  n37 = ComponentMask(None=@n35.Radial Coordinates) {g=true}
  n38 = AppendVector(A=Multiply(A=@n36) {constB=3}, B=Multiply(A=@n37) {constB=2.75})
  n39 = Time()
  n40 = Constant() {r=-0.05}
  n41 = AppendVector(A=Constant() {r=0.1}, B=@n40)
  n42 = Constant() {r=0.5}
  n43 = Multiply(A=@n37) {constB=1.5}
  n44 = NamedRerouteUsage()
  n45 = TextureSample!(
   UVs=Panner(Coordinate=AppendVector(A=@n36, B=@n43), Speed=@n40)
   Tex=NamedRerouteDeclaration(None=TextureObject() {desc=Rift Noise Texture, samplerType=SAMPLERTYPE_Grayscale, texture=/FortUILibrary/UIKit/Buttons/Textures/LowResBlurredNoise}) {name=Rift Noise Texture, nodeColor=<rgba 0 0 0 1>}
   "DDX(UVs)"=NamedRerouteDeclaration(None=DDX(Value=@n44)) {name=Polar DDX, nodeColor=<rgba 0 0 0 1>}
   "DDY(UVs)"=NamedRerouteDeclaration(None=DDY(Value=@n44)) {name=Polar DDY, nodeColor=<rgba 0 0 0 1>}
   Apply View MipBias
  ) {mipValueMode=TMVM_Derivative}
  n46 = TextureSample(
   UVs=Add(A=Panner(Coordinate=@n38, Time=@n39, Speed=@n41), B=Multiply(A=@n45.R, B=@n42))
   Tex=NamedRerouteUsage()
   "DDX(UVs)"=NamedRerouteUsage()
   "DDY(UVs)"=NamedRerouteUsage()
  ) {mipValueMode=TMVM_Derivative}
  NamedRerouteDeclaration(None=SmoothStep(Value=DotProduct(A=@n46.RGB, B=@n45.RGB))) {name=Rift Pulse Mask, nodeColor=<rgba 0 0 0 1>}
  n47 = Rotator(Coordinate=NamedRerouteUsage(), Time=Multiply(A=Time()) {constB=0.333}) {speed=6.2832}
  NamedRerouteDeclaration(None=LinearInterpolate(
    A=Constant3Vector() {constant=<rgba 1 1 1 1>}
    B=MaterialFunctionCall!(PrimaryUV, SecondaryUV, UVMix, PixelSize) {materialFunction=/FortUILibrary/MaterialFunctions/CoreUI/MF_UI_CoreUI_OscillatingColor}.Result
    Alpha=ComponentMask(None=@n47) {r=true}
   ) {constA=1}) {name=RiftOscillatingColors, nodeColor=<rgba 0 0 0 1>}
  n48 = NamedRerouteUsage()
  n49 = Add(A=NamedRerouteUsage(), B=Multiply(A=OneMinus(None=@n48)) {constB=-0.5, desc=Set Pivot Point Offset})
  n50 = NamedRerouteDeclaration(None=ScalarParameter() {defaultValue=0.025, desc=Ring Size, group=Settings, parameterName=DonutPadding, sliderMax=1}) {name=Donut Padding, nodeColor=<rgba 0 0 0 1>}
  n51 = Subtract(
   A=NamedRerouteDeclaration(None=ScalarParameter() {defaultValue=0.375, desc=Ring Size, group=Main Segments, parameterName=MainRingThickness, sliderMax=1}) {name=Main Thickness, nodeColor=<rgba 0 0 0 1>}
   B=NamedRerouteUsage()
  )
  n52 = MaterialFunctionCall(
   UVs=AppendVector(A=NamedRerouteUsage(), B=Add(A=@n49, B=@n50) {constB=0})
   Size=AppendVector(A=NamedRerouteUsage(), B=@n51)
   CornerRoundness=Multiply(A=@n51, B=ScalarParameter() {defaultValue=0.15, group=Main Segments, parameterName=MainCornerRadius, sliderMax=1})
  ) {materialFunction=&MF_UI_Box}
  NamedRerouteDeclaration(None=@n52.Fill) {name=Main Segments Mask, nodeColor=<rgba 0 0 0 1>}
  NamedRerouteDeclaration(None=Multiply(A=@n28) {constB=2}) {name=Vector Length, nodeColor=<rgba 0 0 0 1>}
  n53 = Multiply(A=ScalarParameter() {defaultValue=0.2, desc=Segment Spacing, group=Settings, parameterName=SegmentSpacing, sliderMax=1}) {constB=0.25}
  n54 = Multiply(A=NamedRerouteUsage(), B=@n53)
  NamedRerouteDeclaration(None=OneMinus(None=Multiply(A=@n54, B=@n53))) {name=Spacing, nodeColor=<rgba 0 0 0 1>}
  n55 = OneMinus(None=NamedRerouteUsage())
  NamedRerouteDeclaration(None=Add(A=NamedRerouteUsage(), B=Multiply(A=@n55) {constB=-1})) {name=Spaced Radial Segments, nodeColor=<rgba 0 0 0 1>}
  n56 = NamedRerouteDeclaration(None=LinearInterpolate(Alpha=NamedRerouteUsage()) {constA=0.5}) {name=Reduced Flow Noise, nodeColor=<rgba 0 0 0 1>}
  n57 = NamedRerouteUsage()
  NamedRerouteDeclaration(None=Multiply(A=Multiply(A=@n56, B=@n57), B=ScalarParameter() {defaultValue=2, group=Anim, parameterName=Anim_Hover_Outline, sliderMax=1})) {name=Outline Alpha, nodeColor=<rgba 0 0 0 1>}
  n58 = LinearInterpolate(
   A=NamedRerouteUsage()
   B=LinearInterpolate(
    A=Constant3Vector() {constant=<rgba 0.6654 1 1 1>}
    B=NamedRerouteUsage()
    Alpha=Constant() {desc="RiftEnergyOscillatingColorAmount`", r=0.5}
   )
   Alpha=NamedRerouteUsage()
  )
  n59 = LinearInterpolate(
   A=LinearInterpolate(A=NamedRerouteUsage(), B=Constant3Vector() {constant=<rgba 0.5029 0.5029 1 1>}, Alpha=Constant() {desc=RiftEnergyColorTintMix, r=0.8})
   B=Constant3Vector() {desc=Outline Color, constant=<rgba 1 1 1 1>}
   Alpha=OneMinus(None=NamedRerouteUsage())
  )
  n60 = NamedRerouteUsage()
  NamedRerouteDeclaration(None=StaticSwitchParameter(
    True=VectorParameter() {group=Color, parameterName=Color_Outline, defaultValue=<rgba 1 1 1 1>}.RGB
    False=LinearInterpolate(A=@n58, B=@n59, Alpha=@n60)
   ) {parameterName=UseFlatColor}) {name=FinalColor, nodeColor=<rgba 0 0 0 1>}
  NamedRerouteDeclaration(None=@n52.SDFOutline) {name=Main Segment Outline, nodeColor=<rgba 1 0 0.1172 1>}
  n61 = Add(A=Multiply(A=NamedRerouteUsage(), B=NamedRerouteUsage())) {constB=0.13}
  n62 = SmoothStep(Value=OneMinus(None=Saturate(None=@n61))) {constMax=0.95, constMin=0.9, constValue=0.01}
  NamedRerouteDeclaration(None=Multiply(A=Saturate(None=@n62)) {constB=10}) {name=ButtonOutline, nodeColor=<rgba 0 0.4375 1 1>}
  n63 = MaterialFunctionCall() {materialFunction=&GetUserInterfaceUV}
  NamedRerouteDeclaration(None=@n63.Normalized UV) {name=NormalizedUV, nodeColor=<rgba 0 0 0 1>}
  NamedRerouteDeclaration(None=@n63.Pixel Size) {name=Pixel Size, nodeColor=<rgba 0 0 0 1>}
  n64 = MaterialFunctionCall!(Dimensions) {materialFunction=/FortUILibrary/MaterialFunctions/Basics/MF_UI_GetAspectRatio}
  n65 = ComponentMask(None=@n64.Max Ratio) {r=true}
  n66 = ComponentMask(None=@n64.Max Ratio) {g=true}
  NamedRerouteDeclaration(None=MaterialFunctionCall!(
    Value PX=NamedRerouteUsage()
    UV Pixel Size=NamedRerouteUsage()
    X or Y=MaterialFunctionCall!(A=@n65, B=@n66) {materialFunction=/FortUILibrary/MaterialFunctions/Basics/MF_UI_BooleanOperations}.IsGreater
   ) {materialFunction=/FortUILibrary/MaterialFunctions/Basics/MF_UI_PixelsToNormalized}.Normalized) {name=Normalized Padding, nodeColor=<rgba 0 0 0 1>}
  n67 = Multiply(A=NamedRerouteUsage(), B=NamedRerouteUsage())
  n68 = OneMinus(None=Saturate(None=Add(A=@n67) {constB=0.1}))
  NamedRerouteDeclaration(None=Multiply(A=SmoothStep(Value=@n68) {constMax=0.95, constMin=0.9, constValue=0.01}) {constB=0.1}) {name=Outline Glow Mask, nodeColor=<rgba 0.004 0.0467 0.0529 0.502>}
  n69 = NamedRerouteUsage()
  n70 = NamedRerouteUsage()
  n71 = Saturate(None=Add(A=Multiply(A=@n69, B=@n70)) {constB=0.1})
  NamedRerouteDeclaration(None=Multiply(A=SmoothStep(Min=@n71) {constMax=0.95, constMin=0.9, constValue=0.01}) {constB=0.1}) {name=Hover Sheen Mask, nodeColor=<rgba 0.0312 0 1 1>}
  n72 = NamedRerouteUsage()
  n73 = SmoothStep(Min=Multiply(A=@n72) {constB=-1}, Max=@n72, Value=NamedRerouteUsage()) {constMax=0.05}
  n74 = NamedRerouteUsage()
  NamedRerouteDeclaration(None=Multiply(A=OneMinus(None=@n73), B=Saturate(None=@n74))) {name=Glow Mask, nodeColor=<rgba 0 0 0 1>}
  n75 = NamedRerouteUsage()
  n76 = ScalarParameter() {defaultValue=1, group=Alpha, parameterName=Glow_Alpha, sliderMax=1}
  n77 = ScalarParameter() {defaultValue=1, group=Alpha, parameterName=Glow_Alpha, sliderMax=1}
  n78 = Sine(None=Multiply(A=Time(), B=Constant() {desc=Glow_Alpha_Change_Speed, r=0.25}))
  n79 = LinearInterpolate(
   B=LinearInterpolate(A=Add(A=@n77) {constB=0.05}, B=Add(A=@n77) {constB=-0.05}, Alpha=ConstantBiasScale(None=@n78))
   Alpha=ScalarParameter() {defaultValue=1, group=Anim, parameterName=Anim_Hover_Fill, sliderMax=1}
  ) {constA=1}
  n80 = NamedRerouteUsage()
  n81 = Multiply(A=NamedRerouteUsage(), B=ScalarParameter() {defaultValue=1, group=Anim, parameterName=Anim_Hover_Energy, sliderMax=1})
  n82 = LinearInterpolate(Alpha=NamedRerouteUsage()) {constA=0.75}
  n83 = Multiply(A=LinearInterpolate(A=Multiply(A=@n79, B=@n80), Alpha=Multiply(A=@n81, B=@n82)), B=NamedRerouteUsage())
  n84 = LinearInterpolate(A=NamedRerouteDeclaration(None=Multiply(A=@n76, B=@n83)) {name=Rift Glow Alpha, nodeColor=<rgba 0 0 0 1>}, Alpha=NamedRerouteUsage())
  NamedRerouteDeclaration(None=Multiply(A=LinearInterpolate(A=@n75, Alpha=@n84), B=ScalarParameter() {defaultValue=1, group=Alpha, parameterName=FInalAlpha, sliderMax=1})) {name=FinalAlpha, nodeColor=<rgba 0 0 0 1>}
  Reroute(None=@n1)
 ]
}
