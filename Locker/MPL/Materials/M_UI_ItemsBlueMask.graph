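# Translucent UI material, named as a blue mask for item widgets, that builds a parameterized color gradient and a procedural mask.
# An inverted vertical gradient plus left and right exponential radial gradients, each with tunable tiling, offset, radius and density, blends two mask colors and their alphas.
# Separately, a horizontal gradient drives a three-color left-center-right blend.
# The graph shows no connections from either branch to material outputs.
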Material /Locker/MPL/Materials/M_UI_ItemsBlueMask
graph {
 properties {
  bUsedWithStaticMesh=false
  blendMode=BLEND_Translucent
  materialDomain=MD_UI
  pixelDepthOffsetMode=PDOM_Legacy
  usedShadingModels=Default Lit
 }
 nodes [
  n1 = VectorParameter() {group=MaskColors, parameterName=MaskColor1, sortPriority=1}
  n2 = VectorParameter() {group=MaskColors, parameterName=MaskColor2, sortPriority=2, defaultValue=<rgba 1 1 1 1>}
  n3 = MaterialFunctionCall() {materialFunction=&LinearGradient}
  n4 = Multiply(
   A=TextureCoordinate()
   B=AppendVector(
    A=ScalarParameter() {defaultValue=10, group=RadialGadientLeft, parameterName=TilingRadialGadientLeftX, sortPriority=1}
    B=ScalarParameter() {defaultValue=0.5, group=RadialGadientLeft, parameterName=TilingRadialGadientLeftY, sortPriority=2}
   )
  )
  n5 = AppendVector(
   A=ScalarParameter() {group=RadialGadientLeft, parameterName=OffsetRadialGadientLeftX, sortPriority=3}
   B=ScalarParameter() {group=RadialGadientLeft, parameterName=OffsetRadialGadientLeftY, sortPriority=4}
  )
  n6 = ScalarParameter() {defaultValue=0.5, group=RadialGadientLeft, parameterName=RadialGadientLeftRadius, sortPriority=5}
  n7 = ScalarParameter() {defaultValue=1, group=RadialGadientLeft, parameterName=RadialGadientLeftDensity, sortPriority=6}
  n8 = ScalarParameter() {defaultValue=8, group=RadialGadientRight, parameterName=OffsetRadialGadientRightX, sortPriority=3}
  n9 = ScalarParameter() {group=RadialGadientRight, parameterName=OffsetRadialGadientRightY, sortPriority=4}
  n10 = TextureCoordinate()
  n11 = AppendVector(
   A=ScalarParameter() {defaultValue=8, group=RadialGadientRight, parameterName=TilingRadialGadientRightX, sortPriority=1}
   B=ScalarParameter() {defaultValue=0.5, group=RadialGadientRight, parameterName=TilingRadialGadientRightY, sortPriority=2}
  )
  n12 = Add(
   A=Add(
    A=OneMinus(None=@n3.VGradient)
    B=MaterialFunctionCall(UVs=@n4, CenterPosition=@n5, Radius=@n6, Density=@n7) {materialFunction=&RadialGradientExponential}.RadialGradientExponential
   )
   B=MaterialFunctionCall(
    UVs=Multiply(A=@n10, B=@n11)
    CenterPosition=AppendVector(A=@n8, B=@n9)
    Radius=ScalarParameter() {defaultValue=0.5, group=RadialGadientRight, parameterName=RadialGadientRightRadius, sortPriority=5}
    Density=ScalarParameter() {defaultValue=1, group=RadialGadientRight, parameterName=RadialGadientRightDensity, sortPriority=6}
   ) {materialFunction=&RadialGradientExponential}.RadialGradientExponential
  )
  LinearInterpolate(A=LinearInterpolate(A=@n1.RGB, B=@n2.RGB, Alpha=@n12), B=LinearInterpolate(A=@n1.A, B=@n2.A, Alpha=@n12))
  n13 = VectorParameter() {group=GradientColors, parameterName=ColorGradientLeft, sortPriority=1, defaultValue=<rgba 0.0316 0 1 1>}
  n14 = VectorParameter() {group=GradientColors, parameterName=ColorGradientCenter, sortPriority=2, defaultValue=<rgba 0.2479 1 0.6136 1>}
  n15 = VectorParameter() {group=GradientColors, parameterName=ColorGradientRight, sortPriority=3, defaultValue=<rgba 0.5106 0.3431 1 1>}
  MaterialFunctionCall(
   A=@n13.RGB
   B=@n14.RGB
   C=@n15.RGB
   Alpha=MaterialFunctionCall() {materialFunction=&LinearGradient}.UGradient
  ) {materialFunction=&3ColorBlend}
 ]
}
