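# Activatable locker widget that frames the item preview between two flexible spacers, inside nested size boxes.
# UpdateCategoryPreviewScale reads a category view model's slot info to scale the preview's width and shift the frame's render translation.
# It relays preview-shown events through OnPreviewShown and broadcasts HideItemDescription when a touch player rotates or zooms the item.
# SetMobileStyle widens the left spacer's fill share on mobile.
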WidgetBlueprint /Locker/MPL/WBP_CollectionCameraFrame
graphs [
 EventGraph {nodes [
   n1 = "On Item Rotation Changed (FortItemPreviewWidget)"() {is=title K2Node_ComponentBoundEvent}
   Branch(execute=[@n1.then, @n2.then], Condition=IsUsingTouch(PlayerController=GetOwningPlayer()))
   > Call Hide Item Description!(execute:Exec, self:WBP Collection Camera Frame Object Reference) -> (then:Exec) {is=title K2Node_CallDelegate}
   n2 = "On Item Zoom Level Changed (FortItemPreviewWidget)"() {is=title K2Node_ComponentBoundEvent}
   n3 = "On Preview Item Shown (FortItemPreviewWidget)"!() -> (OutputDelegate, then, ItemDefinition:Fort Account Item Definition Object Reference) {is=title K2Node_ComponentBoundEvent}
   > Call On Preview Shown!(execute:Exec, self:WBP Collection Camera Frame Object Reference, ItemDef:Fort Account Item Definition Object Reference=@n3.ItemDefinition) -> (then:Exec) {is=title K2Node_CallDelegate}
   Event PreConstruct()
   Event Construct()
   Event Tick()
  ]}
 HideItemDescription {nodes [HideItemDescription() {is=title K2Node_FunctionEntry}]}
 OnPreviewShown {nodes [OnPreviewShown!() -> (then, ItemDef:Fort Account Item Definition Object Reference) {is=title K2Node_FunctionEntry}]}
 SetMobileStyle {nodes [
   SetMobileStyle()
   > Branch(Condition=IsMobileGame())
   > SetSize!(
    execute
    self:Horizontal Box Slot Object Reference=SlotAsHorizontalBoxSlot!(Widget:Widget Object Reference=Get Camera_Spacer_Left() -> (Camera_Spacer_Left:Spacer Object Reference)) -> (ReturnValue:Horizontal Box Slot Object Reference)
    InSize:Slate Child Size Structure=Make Slate Child Size!(Value:float=0.24, SizeRule:ESlateSizeRule Enum=Fill) -> (SlateChildSize:Slate Child Size Structure) {is=title K2Node_MakeStruct}
   ) -> (then)
  ]}
 UpdateCategoryPreviewScale {nodes [
   n1 = UpdateCategoryPreviewScale!() -> (then, VM:Fort Locker Category Item VM Object Reference) {is=title K2Node_FunctionEntry}
   > Is Valid(InputObject=@n1)
   n2 = GetSlotInfo!(self:Fort Locker Category Item VM Object Reference=@n1) -> (
    ReturnValue_CategoryName:Text
    ReturnValue_CategoryNamePlural:Text
    ReturnValue_CategoryNameWithPluralForms:Text
    ReturnValue_CategoryDescription:Text
    ReturnValue_DefaultEmptyTexture:Soft Object Reference
    ReturnValue_OverrideShuffleTexture:Texture 2D Soft Object Reference
    ReturnValue_CategoryIcon:Material Instance Object Reference
    ReturnValue_CategorySoundOverrides:Map of Gameplay Tag Structures to Sound Base Soft Object References
    ReturnValue_bShowInCategoryTab:bool=~
    ReturnValue_PreviewWindowScale:float=~
    ReturnValue_PreviewWindowOffset_X:double
    ReturnValue_PreviewWindowOffset_Y:double
    ReturnValue_PreviewWindowOffset_Z:double
    ReturnValue_bUsePluralCategoryInItemList:bool=~
   )
   > Set WidthOverride!(
    execute
    WidthOverride:float=float * float(A=@n2.ReturnValue_PreviewWindowScale, B=1024) {is=title K2Node_PromotableOperator}
    self:Size Box Object Reference=Get SizeBox_PreviewWidget() -> (SizeBox_PreviewWidget:Size Box Object Reference)
   ) -> (then, Output_Get:float)
   > SetRenderTranslation!(
    execute
    self=Get Widget_CameraFrame() -> (Widget_CameraFrame:Size Box Object Reference)
    Translation_X:double=@n2.ReturnValue_PreviewWindowOffset_X
    Translation_Y:double=@n2.ReturnValue_PreviewWindowOffset_Y
   ) -> (then)
  ]}
]
variables [
 HideItemDescription {}
 OnPreviewShown {}
]
widgets {
 info {inheritedWidgetCount=0, parentClass=&CommonActivatableWidget, rootWidgetClass=&SizeBox, widgetCount=6}
 widgets [
  Widget_CameraFrame {widgetClassPath=&SizeBox, bIsVariable=true, widget=Widget_CameraFrame}
  HorizontalBox_0 {widgetClassPath=&HorizontalBox, parent=Widget_CameraFrame, widget=HorizontalBox_0, slot {class=SizeBoxSlot}}
  Camera_Spacer_Left {
   widgetClassPath=&Spacer
   bIsVariable=true
   parent=HorizontalBox_0
   widget=Camera_Spacer_Left
   slot {class=HorizontalBoxSlot, properties {size {sizeRule=Fill, value=0}}}
  }
  SizeBox_PreviewWidget {
   widgetClassPath=&SizeBox
   bIsVariable=true
   parent=HorizontalBox_0
   widget=SizeBox_PreviewWidget
   properties {bOverride_HeightOverride=true, bOverride_WidthOverride=true, heightOverride=1024, widthOverride=1024}
   slot {class=HorizontalBoxSlot}
  }
  FortItemPreviewWidget {
   widgetClassPath=/FortniteCoreUI/Frontend/ItemPreview/WBP_ItemPreview
   bIsVariable=true
   parent=SizeBox_PreviewWidget
   widget=FortItemPreviewWidget
   properties {
    bSpecialPreviewActionsEnabled=false
    bSupportsActivationFocus=false
    itemViewObject=&self:WidgetTree.FortItemPreviewWidget.ItemViewObject
   }
   slot {class=SizeBoxSlot, properties {padding=<ltrb 0 0 0 96>}}
  }
  Camera_Spacer_Right {
   widgetClassPath=&Spacer
   bIsVariable=true
   parent=HorizontalBox_0
   widget=Camera_Spacer_Right
   slot {class=HorizontalBoxSlot, properties {size {sizeRule=Fill, value=1}}}
  }
 ]
}
