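# WBP_MPL_CustomBanner is a user widget in the Locker's MPL folder that wraps a PlayerBanner using a large texture and a borderless rounded banner material.
# PlayHoveredAnimation forwards the hover and instant flags to the inner banner.
# UpdateCorners gets the dynamic material of the banner's image and sets each of the four corner radius parameters, rounded or square, from a separate boolean.
# The event graph is empty.
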WidgetBlueprint /Locker/MPL/WBP_MPL_CustomBanner
graphs [
 EventGraph {links [], nodes []}
 PlayHoveredAnimation {nodes [
   n1 = PlayHoveredAnimation!() -> (then, IsHovered:bool=~, Instant:bool=~) {is=title K2Node_FunctionEntry}
   > PlayHoveredAnimation!(
    execute
    self:Player Banner Object Reference=Get PlayerBanner() -> (PlayerBanner:Player Banner Object Reference)
    IsHovered:bool=@n1
    Instant:bool=@n1.Instant
   ) -> (then) {is=title K2Node_CallFunction}
  ]}
 UpdateCorners {nodes [
   n1 = UpdateCorners!() -> (then, Round TopLeft:bool=~, Round TopRight:bool=~, Round BottomLeft:bool=~, Round BottomRight:bool=~) {is=title K2Node_FunctionEntry}
   > n2 = GetDynamicMaterial(self=Get BannerImage!(self:Player Banner Object Reference=Get PlayerBanner()) -> (BannerImage:Fort Lazy Image Object Reference))
   > SetScalarParameterValue(self=@n2, ParameterName=TopLeftCornerRadius, Value=SelectFloat(A=20, B=0, bPickA=@n1))
   > SetScalarParameterValue(self=@n2, ParameterName=TopRightCornerRadius, Value=SelectFloat(A=20, B=0, bPickA=@n1.Round TopRight))
   > SetScalarParameterValue(self=@n2, ParameterName=BottomLeftCornerRadius, Value=SelectFloat(A=20, B=0, bPickA=@n1.Round BottomLeft))
   > SetScalarParameterValue(self=@n2, ParameterName=BottomRightCornerRadius, Value=SelectFloat(A=20, B=0, bPickA=@n1.Round BottomRight))
  ]}
]
variables []
widgets {
 info {inheritedWidgetCount=0, parentClass=&UserWidget, rootWidgetClass=&Overlay, widgetCount=2}
 widgets [
  Overlay_0 {widgetClassPath=&Overlay, widget=Overlay_0}
  PlayerBanner {
   widgetClassPath=/Game/UI/Foundation/Profile/PlayerBanner
   bIsVariable=true
   parent=Overlay_0
   widget=PlayerBanner
   properties {bUseLargeTexture=true, bannerMaterialInstance=/Game/UI/Foundation/Banner/M_BorderlessRounded_Banner_Inst}
   slot {class=OverlaySlot, properties {horizontalAlignment=HAlign_Fill, verticalAlignment=VAlign_Fill}}
  }
 ]
}
