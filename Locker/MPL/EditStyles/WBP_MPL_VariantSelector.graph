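# Locker "Edit Style" screen where players change cosmetic variants beside an item preview.
# Engaging a variant adjusts channels, equips the item asynchronously, refreshes conflicts and logs changed channels for analytics sent on back or deactivation.
# Filter changes update preview overrides and can auto-equip; the touch-only Equip button enables for owned, unequipped selections.
# It also shows conflict notices and quest navigation; Randomize and Reset only fire analytics or debug prints.
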WidgetBlueprint /Locker/MPL/EditStyles/WBP_MPL_VariantSelector
graphs [
 AnalyticsEventVariantEquipped {nodes [
   AnalyticsEventVariantEquipped() {is=title K2Node_FunctionEntry}
   > n1 = Values!(execute, TargetMap:Map of Strings to Strings=Get ChangedVariants() -> (ChangedVariants:Map of Strings to Strings)) -> (then, Values:Array of Strings)
   > Branch(Condition=Is Valid Index!(TargetArray=@n1, IndexToTest:int) -> (ReturnValue:bool))
   > Set AdditionalInfo(AdditionalInfo:Flow Analytics Additional Info Structure="Add String (Non-Localized)"!(
     Info:"Flow Analytics Additional Info Structure (by ref)"=Make Flow Analytics Additional Info!() -> (FlowAnalyticsAdditionalInfo:Flow Analytics Additional Info Structure) {is=title K2Node_MakeStruct}
     Key:String=Interaction
     Value:String=Edited Variant
    ) -> (ReturnValue:Flow Analytics Additional Info Structure)) -> (Output_Get:Flow Analytics Additional Info Structure)
   > n2 = For Each Loop(Array=@n1)
   n3 = Cast To FortPlayerController(execute=@n2.Completed, Object=GetOwningPlayer())
   n4 = Get ItemDefinition!(self:Fort Item VM Object Reference=Get CosmeticItem!(self:Fort Variant Item VM Object Reference=Get Variant Item() -> (Variant Item:Fort Variant Item VM Object Reference)) -> (CosmeticItem:Fort Item VM Object Reference)) -> (ItemDefinition:Fort Item Definition Object Reference)
   > "Fire Event Element Interacted (Non-Localized)"!(
    execute
    FortPC:Fort Player Controller Object Reference=@n3
    ElementName:String="To String (Text)"(InText="To Text (Object)"!(InObj:Object Reference=@n4) -> (ReturnValue:Text))
    AdditionalInfo:Flow Analytics Additional Info Structure=Get AdditionalInfo() -> (AdditionalInfo:Flow Analytics Additional Info Structure)
   ) -> (then)
   n5 = Keys!(execute=@n2.LoopBody, TargetMap:Map of Strings to Strings=Get ChangedVariants()) -> (then, Keys:Array of Strings)
   n6 = "Get (a copy)"(Array=@n5, Dimension 1=@n2.Array Index)
   > Set AdditionalInfo(AdditionalInfo="Add String (Non-Localized)"(Info=Get AdditionalInfo(), Key=@n6, Value=@n2))
   "To Text (String)"(InString=@n6)
   "To Text (String)"(InString=@n2)
  ]}
 BP_GetDesiredFocusTarget {nodes [
   Get Desired Focus Target()
   > Return Node!(execute, ReturnValue:Widget Object Reference=GetFocusTarget!(self) -> (Target:Widget Object Reference))
  ]}
 BP_GetDesiredInputConfig {nodes [
   Get Desired Input Config() {is=title K2Node_FunctionEntry}
   > Return Node(ReturnValue:UIInput Config Structure=Make UIInput Config!(
     bIgnoreMoveInput:bool=False
     bIgnoreLookInput:bool=False
     InputMode:ECommonInputMode Enum=All
     MouseCaptureMode:EMouseCaptureMode Enum=CaptureDuringMouseDown
     MouseLockMode:EMouseLockMode Enum=DoNotLock
     bHideCursorDuringViewportCapture:bool=True
    ) -> (UIInputConfig:UIInput Config Structure) {is=title K2Node_MakeStruct})
  ]}
 CheckForMimosaQuestStatus {nodes [
   CheckForMimosaQuestStatus!() -> (then, ItemDefinition:Item Definition Base Object Reference, ConflictText:Text, IsLocked:bool=~, IsOwned:bool=~, ShowLockedAsOwned:bool=~) {is=title K2Node_FunctionEntry}
   n1 = Get ItemDefinition!() -> (ItemDefinition:Item Definition Base Object Reference)
   > n2 = Branch(Condition="Equal (Enum)"(
     A:EFortItemType Enum=GetItemType!(ItemDefinition:Item Definition Base Object Reference=@n1) -> (ReturnValue:EFortItemType Enum=WorldItem)
     B:EFortItemType Enum=CosmeticMimosa
    ))
   n3 = IsVisible(self=Get InfoBox_Conflict() -> (InfoBox_Conflict:WBP UIKit Info Box Object Reference))
   n4 = Get ConflictText() -> (ConflictText:Text)
   > n5 = Branch(Condition=AND Boolean(A=NOT Boolean(A=@n3), B=TextIsEmpty(InText=@n4)))
   n6 = NOT Boolean(A=Get IsOwned() -> (IsOwned:bool))
   n7 = Get ShowLockedAsOwned() -> (ShowLockedAsOwned:bool)
   > Return Node!(execute, ShowISB:bool=OR Boolean(A=Get IsLocked() -> (IsLocked:bool), B=AND Boolean(A=@n6, B=@n7)))
   Return Node(execute=@n5.else, ShowISB=false)
   Return Node(execute=@n2.else, ShowISB=false)
  ]}
 Convert_FortItemVMToSelectedConnextedQuestCategoryTag {nodes [
   n1 = Convert_FortItemVMToSelectedConnextedQuestCategoryTag!() -> (then, InItemVM:Fort Item VM Object Reference) {is=title K2Node_FunctionEntry}
   > Return Node!(
    execute
    OutGameplayTags:Gameplay Tag Structure=GetConnectedQuestCategoryTag!(ItemDefinition:Item Definition Base Object Reference=Get ItemDefinition!(self:Fort Item VM Object Reference=@n1) -> (ItemDefinition:Fort Item Definition Object Reference)) -> (ReturnValue:Gameplay Tag Structure)
   )
  ]}
 EventGraph {nodes [
   n1 = "On Clicked (Button_Back)"!() -> (OutputDelegate, then, Button:Common Button Base Object Reference) {is=title K2Node_ComponentBoundEvent}
   > Fire Event Element Interacted(Element=@n1.Button)
   > AnalyticsEventVariantEquipped() {is=title K2Node_CallFunction}
   > Clear!(execute, TargetMap:Map of Strings to Strings=Get ChangedVariants()) -> (then) {is=title K2Node_CallFunction}
   > n2 = Branch(Condition=IsInSubGame!(SubGame:ESubGame Enum=Campaign) -> (ReturnValue:bool))
   > PopContentWidget!(execute) -> (then)
   DeactivateWidget(execute=@n2.else)
   "On Clicked (Button_Equip)"!() -> (OutputDelegate, then, Button:Common Button Base Object Reference) {is=title K2Node_ComponentBoundEvent}
   > EngageAllSelectedItems(self:WBP UIKit Variant Selector Object Reference=Get WBP_UIKit_VariantSelector() -> (WBP_UIKit_VariantSelector:WBP UIKit Variant Selector Object Reference))
   "On Clicked (Button_Reset)"!() -> (OutputDelegate, then, Button:Common Button Base Object Reference) {is=title K2Node_ComponentBoundEvent}
   > n3 = Construct Object from Class(Class=/Locker/MPL/Data/DialogVM_LockerVariantReset.DialogVM_LockerVariantReset_C) -> (ReturnValue:Dialog VM Locker Variant Reset Object Reference)
   > n4 = Create Widget!(
    execute:Exec
    Class:User Widget Class Reference=/FortUILibrary/UIKit/Dialogs/WBP_UIKit_Dialog.WBP_UIKit_Dialog_C
    OwningPlayer:Player Controller Object Reference
    UIKitDialogViewModel:UI Kit Dialog Viewmodel Object Reference=@n3
   ) -> (then:Exec, ReturnValue:WBP UIKit Dialog Object Reference)
   n5 = OnModalResponse!() -> (OutputDelegate, then, ButtonIndex:int) {is=title K2Node_CustomEvent}
   n6 = Get Dialog_Internal!(self:WBP UIKit Dialog Object Reference=@n4) -> (Dialog_Internal:WBP UIKit Dialog Internal Object Reference)
   Bind Event to Dialog Button Clicked Dispatcher(execute=@n4.then, self:WBP UIKit Dialog Internal Object Reference=@n6, Delegate=@n5) {is=title K2Node_AddDelegate}
   > TryPushWidgetAsModalPanel!(execute, PushWidget:User Widget Object Reference=@n4) -> (then)
   "On Clicked (Button_Randomize)"!() -> (OutputDelegate, then, Button:Common Button Base Object Reference) {is=title K2Node_ComponentBoundEvent}
   > FireEventElementInteractedByName!(execute, Element Name:String=Variant Random Clicked, WidgetCallingFrom:Widget Object Reference=Self-Reference()) -> (then, Success:bool)
   > PrintString!(execute, InString=RANDOMIZE, bPrintToScreen, bPrintToLog, TextColor, Duration, Key) -> (then)
   n7 = "On Rotation Requested (WBP_UIKit_VariantSelector)"!() -> (OutputDelegate, then, RotationOffset:Rotator, Snap:bool) {is=title K2Node_ComponentBoundEvent}
   > SetRotationOffset!(
    execute
    self:Fort Item Preview Widget Object Reference=Get FortItemPreviewWidget() -> (FortItemPreviewWidget:WBP Item Preview Object Reference)
    InRotationOffset:"Rotator (by ref)"=@n7.RotationOffset
   ) -> (then)
   n8 = "On Variant Item Engaged (WBP_UIKit_VariantSelector)"!() -> (
    OutputDelegate
    then
    VariantItem:Fort Variant Item VM Object Reference
    VariantChannel:Fort Variant Channel VM Object Reference
    AdditionalInfo:String
   ) {is=title K2Node_ComponentBoundEvent}
   AdjustVariantChannels!(
    execute=[@n8.then, @n17.then]
    self:Fort Variant Selector VM Object Reference=Get FortVariantSelectorVM() -> (FortVariantSelectorVM:Fort Variant Selector VM Object Reference)
    InBaseChannel:Fort Variant Channel VM Object Reference=@n8.VariantChannel
   ) -> (then)
   n9 = Get FortItemSelectionVM() -> (FortItemSelectionVM:Fort Item Selection VM Object Reference)
   > n10 = EquipItem!(
    execute:Exec
    Item:Fort Item VM Object Reference=Get SelectedItem!(self:Fort Item Selection VM Object Reference=@n9) -> (SelectedItem:Fort Item VM Object Reference)
    SlotIndex:int=Get SlotIndex!(self:Fort Locker Category Item VM Object Reference=Get SelectedCategory!(self:Fort Item Selection VM Object Reference=@n9) -> (SelectedCategory:Fort Locker Category Item VM Object Reference)) -> (SlotIndex:int=0)
    Variants:Array of Mcp Variant Channel Info Structures=Get CurrentVariantOverrides!(self:Fort Variant Selector VM Object Reference=Get FortVariantSelectorVM()) -> (CurrentVariantOverrides:Array of Mcp Variant Channel Info Structures)
    bApplyToAllSlots:bool=false
    bAllowMocking:bool=true
   ) -> (then:Exec, OnOperationCompleted:Exec, Success:bool=~) {is=title K2Node_AsyncAction}
   UpdateEquipButtonState(execute=[@n10.OnOperationCompleted, @n17.else])
   > UpdateVariantChannelConflicts(self:Fort Variant Selector VM Object Reference=Get FortVariantSelectorVM())
   n11 = Get FilterChangedRecently() -> (FilterChangedRecently:bool)
   > n12 = Branch(Condition=AND Boolean(A=@n10, B=NOT Boolean(A=@n11)))
   > n13 = RecentlyEditedVariants!(
    execute
    self
    VariantItem:Fort Variant Item VM Object Reference=@n8.VariantItem
    VariantChannel:Fort Variant Channel VM Object Reference=@n8.VariantChannel
    AdditionalInfo:String=@n8.AdditionalInfo
   ) -> (then)
   Set FilterChangedRecently(execute=[@n13.then, @n12.else], FilterChangedRecently:bool=false) -> (Output_Get:bool)
   n14 = "On Conflict Text Updated (WBP_UIKit_VariantSelector)"!() -> (OutputDelegate, then, ConflictText:Text, ConflictType:E_VariantConflictType Enum=NewEnumerator6) {is=title K2Node_ComponentBoundEvent}
   n15 = Get InfoBox_Conflict()
   > SetText(self:WBP UIKit Info Box Object Reference=@n15, InText:Text=@n14.ConflictText)
   > SetIconImage!(
    execute
    self:WBP UIKit Info Box Object Reference=@n15
    Material:Material Interface Object Reference=Select!(
     NewEnumerator6:Material Interface Object Reference=/FortUILibrary/Textures/MaterialInstances/MI_UIKit_Icon_Information
     NewEnumerator4:Material Interface Object Reference=/FortUILibrary/Textures/MaterialInstances/MI_UIKit_Icon_AgeGatedAlert
     NewEnumerator3:Material Interface Object Reference=&MI_UIKit_Icon_AgeGatedAlert
     NewEnumerator5:Material Interface Object Reference=/FortUILibrary/Textures/MaterialInstances/MI_UIKit_Icon_Locked
     NewEnumerator1:Material Interface Object Reference=&MI_UIKit_Icon_AgeGatedAlert
     NewEnumerator2:Material Interface Object Reference=&MI_UIKit_Icon_Locked
     NewEnumerator0:Material Interface Object Reference=/FortUILibrary/Textures/MaterialInstances/MI_UIKit_Icon_Warning
     Index:E_VariantConflictType Enum=@n14.ConflictType
    ) -> (ReturnValue:Material Interface Object Reference)
   ) -> (then)
   > Set Visibility(
    Visibility=Select!(
     Option 0:ESlateVisibility Enum=SelfHitTestInvisible
     Option 1:ESlateVisibility Enum=Collapsed
     Index:bool=TextIsEmpty(InText=@n14.ConflictText)
    ) -> (ReturnValue:ESlateVisibility Enum)
    self=Get InfoBox_Conflict()
   )
   "On Filter Changed (WBP_UIKit_VariantSelector)"() {is=title K2Node_ComponentBoundEvent}
   > SetVariantOverrides!(
    execute
    self:Fort Item Preview Widget Object Reference=Get FortItemPreviewWidget()
    InVariantOverrides:Array of Mcp Variant Channel Info Structures=Get CurrentVariantOverrides(self=Get FortVariantSelectorVM())
   ) -> (then)
   > Set FilterChangedRecently(FilterChangedRecently=true)
   > n16 = Sequence()
   > n17 = Branch(Condition=Get bShouldEquipOnTabChange!(self:Fort Variant Selector VM Object Reference=Get FortVariantSelectorVM()) -> (bShouldEquipOnTabChange:bool))
   Delay(execute=@n16.then_1, Duration=0.2)
   > SelectEntryIfNoEntryIsSelected()
   "On Variant Item Previewed (WBP_UIKit_VariantSelector)"!() -> (OutputDelegate, then, VariantItem:Fort Variant Item VM Object Reference) {is=title K2Node_ComponentBoundEvent}
   > UpdateEquipButtonState()
   "On Clicked (Button_GoToQuests)"!() -> (OutputDelegate, then, Button:Common Button Base Object Reference) {is=title K2Node_ComponentBoundEvent}
   n18 = Get SelectedItem(self=Get FortItemSelectionVM())
   > NavigateToQuestScreen!(
    execute
    Context:Widget Object Reference=Self-Reference()
    StartingCategoryTag:"Gameplay Tag Structure (by ref)"=GetConnectedQuestCategoryTag(ItemDefinition=Get ItemDefinition(self=@n18))
   ) -> (then)
   n19 = OnInputChanged!() -> (OutputDelegate, then, InputType:ECommonInputType Enum) {is=title K2Node_CustomEvent}
   > n20 = Switch on ECommonInputType!(execute, Selection:ECommonInputType Enum=@n19.InputType) -> (MouseAndKeyboard:Exec, Gamepad:Exec, Touch:Exec, Count:Exec) {is=title K2Node_SwitchEnum}
   UpdateEquipButtonState(execute=@n20.Touch)
   Event Construct()
   n21 = Create Event()
   > Bind Event to On Input Method Changed(self=Get CommonInputSubsystem(), Delegate=@n21)
   > SetCompactUIStyle() {is=title K2Node_CallFunction}
   > SetVisibility(
    self=[
     Get Button_Randomize() -> (Button_Randomize:WBP UIKit Button Quiet Object Reference)
     Get Button_Reset() -> (Button_Reset:WBP UIKit Button Quiet Object Reference)
    ]
    InVisibility=Select(Option 0=Collapsed, Option 1=Visible, Index=Get TempShowRandomize() -> (TempShowRandomize:bool))
   )
   > n22 = StartListeningToEvent!(
    execute:Exec
    Target:Gameplay Event Router Component Object Reference=GetGlobalEventRouter()
    EventType:Script Struct Object Reference=/FortniteCoreUI/UI/Social/Chat/TextChat/BPS_TextChatActiveStateChange
    Context:Object Reference=GetOwningPlayer()
    EventBubblingRule:EEventBubblingRule Enum
   ) -> (
    then:Exec
    AsyncTaskProxy:Async Action Start Listening To Event Object Reference
    OnEventReceived:Exec
    Payload:BPS Text Chat Active State Change Structure
   )
   > Set ChatEventWatch(ChatEventWatch:Async Action Start Listening To Event Object Reference=@n22) -> (Output_Get:Async Action Start Listening To Event Object Reference)
   SetVisibility(
    execute=@n22.OnEventReceived
    self=Get WBP_ActionBarContainer() -> (WBP_ActionBarContainer:WBP Action Bar Container Object Reference)
    InVisibility=Select(
     Option 0=SelfHitTestInvisible
     Option 1=Hidden
     Index=Break BPS Text Chat Active State Change!(BPS_TextChatActiveStateChange:"BPS Text Chat Active State Change Structure (by ref)"=@n22.Payload) -> (bActive_4_#1:bool=~) {is=title K2Node_BreakStruct}
    )
   )
   Event Destruct()
   > Unbind Event from On Input Method Changed(self=Get CommonInputSubsystem(), Delegate=@n21)
   > StopListeningToEvent(self:Async Action Start Listening To Event Object Reference=Get ChatEventWatch() -> (ChatEventWatch:Async Action Start Listening To Event Object Reference))
   Event On Activated()
   > Fire Event Main View Opened!(execute, View:Widget Object Reference=Self-Reference(), AdditionalInfo:Flow Analytics Additional Info Structure, bIsOverlay:bool=false) -> (then)
   n23 = Get FortItemSelectionVM()
   > UpdateFlowriderSubView!(
    execute
    self
    SubViewString:String=GetAnalyticsName!(self:Fort Locker Category Item VM Object Reference=Get SelectedCategory(self=@n23)) -> (ReturnValue:String)
   ) -> (then)
   > Set RecentlyEditedList(RecentlyEditedList:String) -> (Output_Get:String)
   > n24 = Sequence()
   > Broadcast Event!(
    execute
    self:Gameplay Event Router Component Object Reference=GetGlobalEventRouter()
    EventData:BPS Set Text Chat Active State Structure=Make BPS Set Text Chat Active State!(bSetActive_5_#1:bool=False) -> (BPS_SetTextChatActiveState:BPS Set Text Chat Active State Structure) {is=title K2Node_MakeStruct}
    EventContext:Object Reference=GetOwningPlayer()
   ) -> (then, ReturnValue:bool)
   Set FortItemVM!(
    execute=@n24.then_1
    FortItemVM:Fort Item VM Object Reference=Get SelectedItem(self=Get FortItemSelectionVM())
    self:WBP UIKit Variant Selector Object Reference=Get WBP_UIKit_VariantSelector()
   ) -> (then, Output_Get:Fort Item VM Object Reference)
   > Clear and Invalidate Timer by Handle(Handle=Get Delay Preview Activation Handler() -> (Delay Preview Activation Handler:Timer Handle Structure))
   > n25 = Set Timer by Event(Delegate=Create Event(self=Get FortItemPreviewWidget()), Time=0.2, bLooping=false)
   > Set Delay Preview Activation Handler(Delay Preview Activation Handler:Timer Handle Structure=@n25) -> (Output_Get:Timer Handle Structure)
   Event On Deactivated()
   > Clear and Invalidate Timer by Handle(Handle=Get Delay Preview Activation Handler())
   > AnalyticsEventVariantEquipped()
   > Clear(TargetMap=Get ChangedVariants())
   > DeactivateWidget(self=Get FortItemPreviewWidget())
   > Set FortItemVM(self=Get WBP_UIKit_VariantSelector())
   > n26 = Branch(Condition=IsInZone!(self:Fort Global UIContext Object Reference=Get FortGlobalUIContext() -> (ReturnValue:Fort Global UIContext Object Reference) {is=title K2Node_GetSubsystem}) -> (ReturnValue:bool))
   Broadcast Event(
    execute=@n26.else
    self=GetGlobalEventRouter()
    EventData=Make BPS Set Text Chat Active State(bSetActive_5_#1=true)
    EventContext=GetOwningPlayer()
   )
   DeactivateWidget(execute=@n28.then, self=@n4)
   n27 = PrintString(execute=@n31.then, InString=RESET)
   n28 = Unbind all Events from Dialog Button Clicked Dispatcher!(execute:Exec=[@n27.then, @n30.then], self:WBP UIKit Dialog Internal Object Reference=@n6) -> (then:Exec) {is=title K2Node_ClearDelegate}
   n29 = Self-Reference()
   n30 = FireEventElementInteractedByName(execute=@n32.1, Element Name=Variant Random Reset Cancelled, WidgetCallingFrom=@n29)
   n31 = FireEventElementInteractedByName(execute=@n32.0, Element Name=Variant Random Reset Clicked, WidgetCallingFrom=@n29)
   n32 = Switch on Int!(execute:Exec=@n5.then, Selection:int=@n5.ButtonIndex) -> (Default:Exec, 0:Exec, 1:Exec)
  ]}
 GetFocusTarget {nodes [
   GetFocusTarget() {is=title K2Node_FunctionEntry}
   > Return Node!(
    execute
    Target:Widget Object Reference=GetDesiredFocusTarget!(self:WBP UIKit Variant Selector Object Reference=Get WBP_UIKit_VariantSelector()) -> (ReturnValue:Widget Object Reference)
   )
  ]}
 GetPrimaryContentSetup {nodes [
   GetPrimaryContentSetup() {is=title K2Node_FunctionEntry}
   > Return Node!(execute, ReturnValue:Primary Content Setup Structure=Get PrimaryContentSetup() -> (PrimaryContentSetup:Primary Content Setup Structure))
  ]}
 RecentlyEditedVariants {nodes [
   n1 = RecentlyEditedVariants!() -> (then, VariantItem:Fort Variant Item VM Object Reference, VariantChannel:Fort Variant Channel VM Object Reference, AdditionalInfo:String) {is=title K2Node_FunctionEntry}
   > Set Variant Item(Variant Item:Fort Variant Item VM Object Reference=@n1) -> (Output_Get:Fort Variant Item VM Object Reference)
   n2 = Get Source String from Text!(Text:Text=Get RowDisplayName!(self:Fort Variant Channel VM Object Reference=@n1.VariantChannel) -> (RowDisplayName:Text)) -> (ReturnValue:String) {is=title K2Node_CallFunction}
   n3 = Split(SourceString=@n2, InStr=" ( ", SearchDir=FromStart)
   n4 = Split(SourceString=@n3.RightS, InStr=" )", SearchDir=FromStart)
   n5 = "To String (Text)"(InText=SelectText(A="To Text (String)"(InString=@n4), B="To Text (String)"(InString=@n1.AdditionalInfo), bPickA=IsEmpty(InString=@n1.AdditionalInfo)))
   n6 = Get AdditionalInfo() -> (AdditionalInfo:String)
   n7 = "Equal Exactly (String)"(A=@n3.RightS, B="Object Spray )")
   > Add!(
    execute
    TargetMap:Map of Strings to Strings=Get ChangedVariants()
    Key:"String (by ref)"=ToLower!(SourceString:String=SelectString(A=@n3, B=@n2, bPickA=@n3.ReturnValue)) -> (ReturnValue:String)
    Value:"String (by ref)"=ToLower(SourceString=SelectString(A=@n6, B=@n5, bPickA=@n7))
   ) -> (then) {is=title K2Node_CallFunction}
  ]}
 SelectEntryIfNoEntryIsSelected {nodes [
   SelectEntryIfNoEntryIsSelected() {is=title K2Node_FunctionEntry}
   n1 = Get WBP_UIKit_VariantSelector()
   > n2 = For Each Loop with Break!(
    Exec:Exec
    Array:Array of User Widget Object References=GetAllEntries!(self:Dynamic Entry Box Base Object Reference=Get EntryBox_ChannelRows!(self:WBP UIKit Variant Selector Object Reference=@n1) -> (EntryBox_ChannelRows:Dynamic Entry Box Object Reference)) -> (ReturnValue:Array of User Widget Object References)
    Break:Exec=@n6.then
   ) -> (LoopBody:Exec, Array Element:User Widget Object Reference, Array Index:int, Completed:Exec)
   n3 = Branch(execute=@n2.Completed, Condition=Get HasSelectedEntry() -> (HasSelectedEntry:bool))
   > Return Node!(execute)
   n4 = Get FortItemSelectionVM()
   Set SelectedItem!(execute=@n3.else, SelectedItem:Fort Item VM Object Reference=Get SelectedItem(self=@n4), self:Fort Item Selection VM Object Reference=@n4) -> (then, Output_Get:Fort Item VM Object Reference)
   > Return Node()
   n5 = Set HasSelectedEntry(
    execute=@n2.LoopBody
    HasSelectedEntry:bool=OR Boolean(A=HasAnyUserFocusedDescendants!(self:Widget Object Reference=@n2) -> (ReturnValue:bool), B=Get HasSelectedEntry())
   ) -> (Output_Get:bool)
   > n6 = Branch(Condition=@n5)
  ]}
 SetCompactUIStyle {nodes [
   SetCompactUIStyle() {is=title K2Node_FunctionEntry}
   > Branch(Condition=IsMobileGame())
   > SetSidesToPad!(
    execute
    self:Safe Zone Object Reference=Get SafeZone() -> (SafeZone:Safe Zone Object Reference)
    InPadLeft:bool=true
    InPadRight:bool=true
    InPadTop:bool=true
    InPadBottom:bool=true
   ) -> (then)
  ]}
 SetHeaders {nodes [
   n1 = SetHeaders!() -> (then, SubCategory:Fort Locker Category Item VM Object Reference) {is=title K2Node_FunctionEntry}
   n2 = Get EquippedItem!(self:Fort Locker Category Item VM Object Reference=@n1) -> (EquippedItem:Fort Item VM Object Reference)
   n3 = Get ItemDefinition(self=@n2)
   > n4 = HasMatchingGameplayTag!(execute:Exec, self:Object Reference=@n3, TagToCheck:Gameplay Tag Structure=(TagName="Cosmetics.Source.DefaultItem")) -> (then:Exec, ReturnValue:bool) {is=title K2Node_Message}
   n5 = "Equal (Enum)"(A=GetItemType(ItemDefinition=@n3) -> (ReturnValue=WorldItem), B=AthenaCharacter)
   > SetHeader!(
    execute
    self:WBP MPLocker Heading Object Reference=Get Heading() -> (Heading:WBP MPLocker Heading Object Reference)
    Text:Text=SelectText(A="LOCTEXT(\"PICK OUTFIT\")", B="LOCTEXT(\"EDIT STYLE\")", bPickA=AND Boolean(A=@n4, B=@n5))
   ) -> (then)
   > SetSubHeader!(
    execute
    self:WBP MPLocker Heading Object Reference=Get Heading()
    Text:Text=GetItemCategoryName!(self:Fort Item Details VM Object Reference=Get ItemDetails!(self:Fort Item VM Object Reference=@n2) -> (ItemDetails:Fort Item Details VM Object Reference)) -> (ReturnValue:Text)
   ) -> (then)
  ]}
 ShowMimosaQuestsISB {nodes [
   n1 = ShowMimosaQuestsISB!() -> (then, ShowISB:bool=~) {is=title K2Node_FunctionEntry}
   > SetVisibility(
    self=Get InfoBox_MimosaQuests() -> (InfoBox_MimosaQuests:WBP UIKit Info Box Object Reference)
    InVisibility=Select(Option 0=Collapsed, Option 1=SelfHitTestInvisible, Index=@n1)
   )
  ]}
 UpdateEquipButtonState {nodes [
   UpdateEquipButtonState() {is=title K2Node_FunctionEntry}
   n1 = Get CommonInputSubsystem()
   > Branch(Condition="Equal (Enum)"(A:ECommonInputType Enum=GetCurrentInputType(self=@n1), B:ECommonInputType Enum=Touch))
   > SetIsEnabled(self=Get Button_Equip() -> (Button_Equip:WBP UIKit Button Quiet Object Reference), bInIsEnabled=false)
   > n2 = For Each Loop(Array:Array of Fort Variant Set VM Object References=Get VariantSets!(self:Fort Variant Selector VM Object Reference=Get FortVariantSelectorVM()) -> (VariantSets:Array of Fort Variant Set VM Object References)) -> (Array Element:Fort Variant Set VM Object Reference)
   > Is Valid(InputObject=@n2)
   > n3 = For Each Loop(Array:Array of Fort Variant Channel VM Object References=Get VariantChannels!(self:Fort Variant Set VM Object Reference=@n2) -> (VariantChannels:Array of Fort Variant Channel VM Object References)) -> (Array Element:Fort Variant Channel VM Object Reference)
   > Is Valid(InputObject=@n3)
   n4 = GetSelectedVariantItem!(self:Fort Variant Channel VM Object Reference=@n3) -> (ReturnValue:Fort Variant Item VM Object Reference)
   n5 = Get bEquipped!(self:Fort Variant Item VM Object Reference=@n4) -> (bEquipped:bool)
   > Branch(Condition=AND Boolean!(A=IsValid(Object=@n4), B=Get bOwned!(self:Fort Variant Item VM Object Reference=@n4) -> (bOwned:bool), C:bool=NOT Boolean(A=@n5)) -> (ReturnValue))
   > SetIsEnabled(self=Get Button_Equip(), bInIsEnabled=true)
   > Return Node()
  ]}
 UpdateFlowriderSubView {nodes [
   UpdateFlowriderSubView!() -> (then, SubViewString:String) {is=title K2Node_FunctionEntry}
   n1 = GetDisplayName!(self:Fort Locker Category Group VM Object Reference=Get SelectedCategoryGroup!(self:Fort Item Selection VM Object Reference=Get FortItemSelectionVM()) -> (SelectedCategoryGroup:Fort Locker Category Group VM Object Reference)) -> (ReturnValue:"Text (by ref)")
   > "Fire Event Sub View Opened (Non-Localized)"!(
    execute
    FortPC:Fort Player Controller Object Reference=GetFortPlayerControllerFromActor!(Actor:Actor Object Reference=GetOwningPlayer()) -> (ReturnValue:Fort Player Controller Object Reference)
    ViewName:String=Append!(A=Get Source String from Text(Text=@n1), B=" / ", C:String=Get SubViewString() -> (SubViewString:String)) -> (ReturnValue)
    AdditionalInfo:Flow Analytics Additional Info Structure
    bIsOverlay:bool=false
   ) -> (then)
  ]}
 VM_UpdateGoToQuestsButtonVisibility {nodes [
   n1 = VM_UpdateGoToQuestsButtonVisibility!() -> (then, InGameplayTags:Gameplay Tag Structure) {is=title K2Node_FunctionEntry}
   > n2 = Cast To FortPlayerController(Object=GetOwningPlayer())
   > n3 = PlayerHasQuestsForCategory!(execute, FortPC:Fort Player Controller Object Reference=@n2, QuestCategoryTag:"Gameplay Tag Structure (by ref)"=@n1) -> (then, ReturnValue:bool)
   n4 = NOT Boolean(A=IsInZone(self=Get FortGlobalUIContext()))
   n5 = IsGameplayTagValid!(GameplayTag:Gameplay Tag Structure=@n1) -> (ReturnValue:bool)
   > SetVisibility(
    self=Get Button_GoToQuests() -> (Button_GoToQuests:WBP UIKit Button Quiet Object Reference)
    InVisibility=Select(Option 0=Collapsed, Option 1=Visible, Index=AND Boolean(A=@n3, B=@n5, C=@n4))
   )
  ]}
 VariantRowsSet {nodes [VariantRowsSet() {is=title K2Node_FunctionEntry}]}
]
variables [
 AdditionalInfo {type=FlowAnalyticsAdditionalInfo, default {}}
 bShowCapeRestriction {type=boolean, default=true}
 ChangedVariants {type="map<string>", default {}}
 ChatEventWatch {type=AsyncAction_StartListeningToEvent, default=null}
 DefaultFilterIcon {type=Texture2D, default=/Game/UI/Foundation/Textures/Icons/Locker/T_Icon_Spray_Adjustments_Tab_128}
 Delay Preview Activation Handler {type=TimerHandle}
 FilterChangedRecently {type=boolean}
 LastSelectedTabIndex {type=integer}
 PrimaryContentSetup {
  type=PrimaryContentSetup
  default {bEnableNewTextChat=false, bShowBottomBar=false, bShowChatWidget=false, bShowTopBar=false, bTabSupportsInterface=false}
 }
 RecentlyEditedList {type=string, default=""}
 Section Contents {type="array<FlowAnalyticsImpressionSectionContent>", default []}
 TempShowRandomize {type=boolean}
 Variant Item {type=FortVariantItemVM, default=null}
 VariantRowsSet {}
]
widgets {
 info {
  inheritedWidgetCount=0
  namedSlotCount=1
  parentClass=&CommonActivatableWidget
  rootWidgetClass=&SafeZone
  widgetCount=28
 }
 widgets [
  SafeZone {widgetClassPath=&SafeZone, bIsVariable=true, widget=SafeZone, properties {padBottom=false, padTop=false}}
  SizeBox_0 {
   widgetClassPath=&SizeBox
   parent=SafeZone
   widget=SizeBox_0
   properties {bOverride_WidthOverride=true, widthOverride=1920}
   slot {class=SafeZoneSlot, properties {hAlign=HAlign_Center}}
  }
  Root {widgetClassPath=&GridPanel, parent=SizeBox_0, widget=Root, properties {columnFill [1,0], rowFill [1,0]}, slot {class=SizeBoxSlot}}
  WBP_CaptureForPostBufferUpdate {
   widgetClassPath=&WBP_CaptureForPostBufferUpdate
   bIsVariable=true
   parent=Root
   slot {class=GridSlot, properties {layer=-20}}
  }
  MainContents {
   widgetClassPath=&GridPanel
   bIsVariable=true
   parent=Root
   widget=MainContents
   properties {columnFill [0,0.8,0.8], rowFill [0,1,0]}
   slot {class=GridSlot, properties {layer=1, padding=<ltrb 64 64 64 64>}}
  }
  Heading {
   widgetClassPath=/FortniteCoreUI/Customization/MPL/WBP_MPLocker_Heading
   bIsVariable=true
   parent=MainContents
   widget=Heading
   slot {class=GridSlot, properties {column=2, padding=<ltrb 32 0 32 0>}}
  }
  Text_NoCapes {
   widgetClassPath=&CommonTextBlock
   bIsVariable=true
   parent=MainContents
   widget=Text_NoCapes
   properties {
    autoWrapText=true
    clipping=OnDemand
    mobileFontSizeMultiplier=1.3
    text="Cape Included: Feel free to wear other back-bling, but NO CAPES!"
    visibility=Collapsed
    font {
     bForceMonospaced=false
     bMaterialIsStencil=false
     fontMaterial=null
     fontObject=&HeadingNow
     letterSpacing=0
     monospacedWidth=1
     size=18
     skewAmount=0
     typefaceFontName=75 Medium Italic
     outlineSettings {
      bApplyOutlineToDropShadows=false
      bMiteredCorners=false
      bSeparateFillAlpha=false
      outlineMaterial=null
      outlineSize=0
      outlineColor=<rgba 0 0 0 1>
     }
    }
    margin=<ltrb 0 5 0 10>
    shadowOffset=<0 0>
   }
   slot {class=GridSlot, properties {column=2, verticalAlignment=VAlign_Top, padding=<ltrb 0 -64 0 0>}}
  }
  WBP_UIKit_VariantSelector {
   widgetClassPath=/FortUILibrary/UIKit/VariantSelector/WBP_UIKit_VariantSelector
   bIsVariable=true
   parent=MainContents
   widget=WBP_UIKit_VariantSelector
   properties {showAgeGateWarnings=true, variantPropertyFilterTag {tagName=Cosmetics.Queries.ExcludeComingSoon}}
   slot {class=GridSlot, properties {column=2, columnSpan=1, row=1, rowSpan=1, padding=<ltrb 32 16 15 16>}}
  }
  WBP_ActionBarContainer {
   widgetClassPath=/Game/UI/Frontend/BottomBar/WBP_ActionBarContainer
   bIsVariable=true
   parent=MainContents
   widget=WBP_ActionBarContainer
   slot {
    class=GridSlot
    properties {column=1, columnSpan=2, horizontalAlignment=HAlign_Right, row=2, rowSpan=1, verticalAlignment=VAlign_Bottom, padding=<ltrb 0 0 15 0>}
   }
  }
  Temp_ButtonBar {widgetClassPath=&HorizontalBox, namedSlotHost=WBP_ActionBarContainer, widget=Temp_ButtonBar}
  Button_Randomize {
   widgetClassPath=/FortUILibrary/UIKit/Buttons/WBP_UIKit_Button_Quiet
   bIsVariable=true
   parent=Temp_ButtonBar
   widget=Button_Randomize
   properties {
    bHideInputActionWithKeyboard=true
    text=Randomize
    visibility=Visible
    triggeringInputAction {dataTable=&InputActionDataTable, rowName=GenericLeftTrigger}
   }
   slot {class=HorizontalBoxSlot, properties {horizontalAlignment=HAlign_Right, verticalAlignment=VAlign_Bottom, padding=<ltrb 16 0 16 0>}}
  }
  Button_Reset {
   widgetClassPath=&WBP_UIKit_Button_Quiet
   bIsVariable=true
   parent=Temp_ButtonBar
   widget=Button_Reset
   properties {
    bHideInputActionWithKeyboard=true
    text=" Reset"
    visibility=Visible
    triggeringInputAction {dataTable=&InputActionDataTable, rowName=GenericRightTrigger}
   }
   slot {class=HorizontalBoxSlot, properties {horizontalAlignment=HAlign_Right, verticalAlignment=VAlign_Bottom, padding=<ltrb 0 0 16 0>}}
  }
  VisibilityBorder_Equip_Touch {
   widgetClassPath=&CommonHardwareVisibilityBorder
   parent=Temp_ButtonBar
   widget=VisibilityBorder_Equip_Touch
   properties {
    padding=<ltrb 0 0 16 0>
    visibilityQuery {
     autoDescription=" ALL( Input.Touch )"
     tokenStreamVersion=0
     userDescription=""
     queryTokenStream [0,1,2,1,0]
     tagDictionary [{tagName=Input.Touch}]
    }
   }
   slot {class=HorizontalBoxSlot, properties {horizontalAlignment=HAlign_Right, verticalAlignment=VAlign_Center}}
  }
  Button_Equip {
   widgetClassPath=&WBP_UIKit_Button_Quiet
   bIsVariable=true
   parent=VisibilityBorder_Equip_Touch
   widget=Button_Equip
   properties {text=Equip, visibility=Visible, triggeringInputAction {dataTable=&InputActionDataTable, rowName=GenericAccept}}
   slot {class=BorderSlot, properties {padding=<ltrb 0 0 16 0>}}
  }
  WBP_AthenaBottomBar_WithNotifications {
   widgetClassPath=/Game/UI/Frontend/BottomBar/WBP_AthenaBottomBar_WithNotifications
   bIsVariable=true
   parent=Temp_ButtonBar
   widget=WBP_AthenaBottomBar_WithNotifications
   properties {visibility=Collapsed}
   slot {class=HorizontalBoxSlot, properties {padding=<ltrb 24 0 16 0>}}
  }
  Button_GoToQuests {
   widgetClassPath=&WBP_UIKit_Button_Quiet
   bIsVariable=true
   parent=Temp_ButtonBar
   widget=Button_GoToQuests
   properties {text=Go to Quests, visibility=Collapsed, triggeringInputAction {dataTable=&InputActionDataTable, rowName=GoToQuests}}
   slot {class=HorizontalBoxSlot, properties {horizontalAlignment=HAlign_Right, verticalAlignment=VAlign_Bottom, padding=<ltrb 0 0 24 0>}}
  }
  Button_PreviewAction {
   widgetClassPath=&WBP_UIKit_Button_Quiet
   bIsVariable=true
   parent=Temp_ButtonBar
   widget=Button_PreviewAction
   properties {
    bAutoUpdateFromInputAction=true
    text="[PH] Preview"
    visibility=Collapsed
    triggeringInputAction {dataTable=&InputActionDataTable, rowName=Edit}
   }
   slot {class=HorizontalBoxSlot, properties {horizontalAlignment=HAlign_Right, verticalAlignment=VAlign_Bottom, padding=<ltrb 0 0 24 0>}}
  }
  Button_PreviewAction_1 {
   widgetClassPath=&WBP_UIKit_Button_Quiet
   bIsVariable=true
   parent=Temp_ButtonBar
   widget=Button_PreviewAction_1
   properties {
    bAutoUpdateFromInputAction=true
    text="[PH] Preview"
    visibility=Collapsed
    triggeringInputAction {dataTable=&InputActionDataTable, rowName=Edit}
   }
   slot {class=HorizontalBoxSlot, properties {horizontalAlignment=HAlign_Right, verticalAlignment=VAlign_Bottom, padding=<ltrb 0 0 24 0>}}
  }
  Button_Back {
   widgetClassPath=&WBP_UIKit_Button_Quiet
   bIsVariable=true
   parent=Temp_ButtonBar
   widget=Button_Back
   properties {text=Back, visibility=Visible, triggeringInputAction {dataTable=&InputActionDataTable, rowName=GenericBack}}
   slot {class=HorizontalBoxSlot, properties {horizontalAlignment=HAlign_Right, verticalAlignment=VAlign_Bottom}}
  }
  VerticalBox_0 {
   widgetClassPath=&VerticalBox
   parent=MainContents
   widget=VerticalBox_0
   slot {class=GridSlot, properties {column=1, horizontalAlignment=HAlign_Left, rowSpan=3, verticalAlignment=VAlign_Bottom}}
  }
  InfoBox_MimosaQuests {
   widgetClassPath=/FortUILibrary/UIKit/InfoBox/WBP_UIKit_InfoBox
   bIsVariable=true
   parent=VerticalBox_0
   widget=InfoBox_MimosaQuests
   properties {
    iconMaterial=&MI_UIKit_Icon_Locked
    previewText=Unlockable through Quests. Go to Quests to see your current progress.
    use Preview Text=true
    visibility=Collapsed
   }
   slot {class=VerticalBoxSlot, properties {padding=<ltrb 0 0 0 16>}}
  }
  InfoBox_Conflict {
   widgetClassPath=&WBP_UIKit_InfoBox
   bIsVariable=true
   parent=VerticalBox_0
   widget=InfoBox_Conflict
   properties {visibility=Collapsed}
   slot {class=VerticalBoxSlot}
  }
  Widget_CameraFrame {widgetClassPath=&SizeBox, bIsVariable=true, parent=Root, widget=Widget_CameraFrame, slot {class=GridSlot}}
  HorizontalBox {widgetClassPath=&HorizontalBox, parent=Widget_CameraFrame, widget=HorizontalBox, slot {class=SizeBoxSlot}}
  Camera_Spacer_Left {
   widgetClassPath=&Spacer
   bIsVariable=true
   parent=HorizontalBox
   widget=Camera_Spacer_Left
   slot {class=HorizontalBoxSlot, properties {size {sizeRule=Fill, value=0}}}
  }
  SizeBox {
   widgetClassPath=&SizeBox
   parent=HorizontalBox
   widget=SizeBox
   properties {bOverride_HeightOverride=true, bOverride_WidthOverride=true, heightOverride=970, widthOverride=970}
   slot {class=HorizontalBoxSlot}
  }
  FortItemPreviewWidget {
   widgetClassPath=/FortniteCoreUI/Frontend/ItemPreview/WBP_ItemPreview
   bIsVariable=true
   parent=SizeBox
   widget=FortItemPreviewWidget
   properties {
    bSupportsActivationFocus=false
    itemViewObject=&self:WidgetTree.FortItemPreviewWidget.ItemViewObject
    previewActionButtons [
     &self:WidgetTree.Button_PreviewAction
     &self:WidgetTree.Button_PreviewAction_1
    ]
    screenTag {tagName=UI.ItemPreview.Screen.Locker.VariantSelector}
   }
   slot {class=SizeBoxSlot, properties {padding=<ltrb 0 52 0 64>}}
  }
  Camera_Spacer_Right {
   widgetClassPath=&Spacer
   bIsVariable=true
   parent=HorizontalBox
   widget=Camera_Spacer_Right
   slot {class=HorizontalBoxSlot, properties {size {sizeRule=Fill, value=1}}}
  }
 ]
}
