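# Locker widget wrapping a Common List View of grouped item rows, each drawn by WBP_LockerGroupTiles_Default.
# When the item count changes on entry generation, it scales the mouse-wheel scroll multiplier by remapping that count between configurable bounds.
# UpdateIsEmoteWheel sets emote-wheel row spacing and applies compact-UI scrollbar padding.
# It forwards ToggleSelectedContext to displayed rows and resets the player controller's touch-press state when list touches end.
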WidgetBlueprint /Locker/MPL/ItemLists/WBP_LockerGroupListView_Default
graphs [
 EventGraph {nodes [
   "On Entries Generated (LockerGroupListView)"!() -> (OutputDelegate, then, NumEntries:int=0) {is=title K2Node_ComponentBoundEvent}
   n1 = GetNumItems!(self:List View Object Reference=Get LockerGroupListView() -> (LockerGroupListView:Common List View Object Reference)) -> (ReturnValue:int=0)
   > n2 = Branch(Condition="Equal (Integer)"(A=Get LastItemCount() -> (LastItemCount:int=0), B=@n1))
   Set LastItemCount(execute=@n2.else, LastItemCount:int=@n1) -> (Output_Get:int=0)
   n3 = "To Float (Integer)"(InInt=@n1)
   n4 = "To Float (Integer)"(InInt=Get MinScrollItems() -> (MinScrollItems:int=0))
   n5 = "To Float (Integer)"(InInt=Get MaxScrollItems() -> (MaxScrollItems:int=0))
   > SetWheelScrollMultiplier!(
    execute
    self:List View Base Object Reference=Get LockerGroupListView()
    NewWheelScrollMultiplier:float=Lerp(
     A=Get MinMouseMultiplier() -> (MinMouseMultiplier:double)
     B=Get MaxMouseMultiplier() -> (MaxMouseMultiplier:double)
     Alpha=RemapValueToInterval!(Input:double=@n3, InputLow:double=@n4, InputHigh:double=@n5) -> (Result:double)
    )
   ) -> (then)
   "On List View Touch End (LockerGroupListView)"() {is=title K2Node_ComponentBoundEvent}
   > ResetInputTouchPressed(self:Fort Player Controller Object Reference=GetFortPlayerControllerFromActor!(Actor:Actor Object Reference=GetOwningPlayer()) -> (ReturnValue:Fort Player Controller Object Reference))
   n6 = Event UpdateIsEmoteWheel!() -> (OutputDelegate, then, IsEmoteWheel:bool) {is=title K2Node_Event}
   n7 = Get LockerGroupListView()
   > Set VerticalEntrySpacing!(
    execute
    VerticalEntrySpacing:float=Select(
     Option 0:float=GetVerticalEntrySpacing!(self:List View Object Reference=@n7) -> (ReturnValue:float)
     Option 1:float=24
     Index=@n6.IsEmoteWheel
    ) -> (ReturnValue:float)
    self:List View Object Reference=@n7
   ) -> (then, Output_Get:float)
   > SetCompactUIStyle() {is=title K2Node_CallFunction}
   Event ToggleSelectedContext() {is=title K2Node_Event}
   > TriggerRowSelectedContext()
  ]}
 GetEntryObject {nodes [
   GetEntryObject() {is=title K2Node_FunctionEntry}
   > Return Node!(execute, Object:Object Reference)
  ]}
 GetGroupListView {nodes [
   GetGroupListView() {is=title K2Node_FunctionEntry}
   > Return Node!(execute, ListView:Common List View Object Reference=Get LockerGroupListView())
  ]}
 GetListView {nodes [
   GetListView() {is=title K2Node_FunctionEntry}
   > Return Node!(execute, TileView:Fort Tile View Object Reference)
  ]}
 RemapValueTo_0-1 {nodes [
   n1 = Inputs!() -> (Input:double=~, Input Low:double=~, Input high:double=~)
   n2 = float - float(A=@n1.Input high, B=@n1.Input Low)
   n3 = float - float(A=@n1, B=@n1.Input Low)
   Outputs!(Result:double="Clamp (Float)"(Value=float / float(A=@n3, B=@n2), Min=0, Max=1))
  ]}
 SetCompactUIStyle {nodes [
   SetCompactUIStyle() {is=title K2Node_FunctionEntry}
   > Branch(Condition=IsUsingCompactUI())
   > Set!(
    execute
    ScrollBarPadding_Left:float=40
    ScrollBarPadding_Top:float=124
    ScrollBarPadding_Right:float=0
    ScrollBarPadding_Bottom:float=92
    self:List View Object Reference=Get LockerGroupListView()
   ) -> (then, Output_Get:Margin Structure)
  ]}
 TriggerRowSelectedContext {nodes [
   TriggerRowSelectedContext() {is=title K2Node_FunctionEntry}
   > n1 = Set CachedDisplayedEntryWidgets(CachedDisplayedEntryWidgets:Array of User Widget Object References=GetDisplayedEntryWidgets!(self:List View Base Object Reference=Get LockerGroupListView()) -> (ReturnValue:Array of User Widget Object References)) -> (Output_Get:Array of User Widget Object References)
   > n2 = For Each Loop(Array:Array of User Widget Object References=@n1) -> (Array Element:User Widget Object Reference)
   > ToggleSelectedContext!(execute:Exec, self:Object Reference=@n2) -> (then:Exec) {is=title K2Node_Message}
  ]}
]
variables [
 LastItemCount {type=integer}
 MaxMouseMultiplier {type=number, default=4}
 MaxScrollItems {type=integer, default=18}
 MinMouseMultiplier {type=number, default=1}
 MinScrollItems {type=integer, default=6}
]
widgets {
 info {inheritedWidgetCount=0, parentClass=&CommonActivatableWidget, rootWidgetClass=&CommonListView, widgetCount=1}
 widgets [LockerGroupListView {
   widgetClassPath=&CommonListView
   bIsVariable=true
   widget=LockerGroupListView
   properties {
    bReturnFocusToSelection=true
    clipping=Inherit
    entryWidgetClass=&/WBP_LockerGroupTiles_Default.WBP_LockerGroupTiles_Default_C
    navigation=&self:WidgetTree.LockerGroupListView.WidgetNavigation_0
    scrollIntoViewAlignment=TopOrLeft
    verticalEntrySpacing=24
    scrollBarPadding=<ltrb 18 124 0 92>
    scrollBarStyle {
     thickness=12
     draggedThumbImage {
      bIsDynamicallyLoaded=false
      drawAs=RoundedBox
      imageType=FullColor
      mirroring=NoMirror
      resourceName=../../../Engine/Content/Slate/Common/Scrollbar_Thumb.png
      resourceObject=null
      tiling=NoTile
      imageSize=<12 12>
      margin=<ltrb 0.25 0.25 0.25 0.25>
      outlineSettings {
       bUseBrushTransparency=false
       roundingType=FixedRadius
       width=2
       color {colorUseRule=UseColor_Specified, specifiedColor=<rgba 0 0 0 0>}
       cornerRadii=<6 6 6 6>
      }
      tintColor {colorUseRule=UseColor_Specified, specifiedColor=<rgba 1 1 1 1>}
      uVRegion {bIsValid=false, max=<0 0>, min=<0 0>}
     }
     horizontalBackgroundImage {
      bIsDynamicallyLoaded=false
      drawAs=NoDrawType
      imageType=NoImage
      mirroring=NoMirror
      resourceName=None
      resourceObject=null
      tiling=NoTile
      imageSize=<0 0>
      margin=<ltrb 0 0 0 0>
      outlineSettings {
       bUseBrushTransparency=false
       roundingType=HalfHeightRadius
       width=0
       color {colorUseRule=UseColor_Specified, specifiedColor=<rgba 0 0 0 0>}
       cornerRadii=<0 0 0 0>
      }
      tintColor {colorUseRule=UseColor_Specified, specifiedColor=<rgba 1 1 1 1>}
      uVRegion {bIsValid=false, max=<0 0>, min=<0 0>}
     }
     horizontalBottomSlotImage {
      bIsDynamicallyLoaded=false
      drawAs=NoDrawType
      imageType=FullColor
      mirroring=NoMirror
      resourceName=../../../Engine/Content/Slate/Common/Scrollbar_Background_Horizontal.png
      resourceObject=null
      tiling=NoTile
      imageSize=<8 8>
      margin=<ltrb 0 0 0 0>
      outlineSettings {
       bUseBrushTransparency=false
       roundingType=HalfHeightRadius
       width=0
       color {colorUseRule=UseColor_Specified, specifiedColor=<rgba 0 0 0 0>}
       cornerRadii=<0 0 0 0>
      }
      tintColor {colorUseRule=UseColor_Specified, specifiedColor=<rgba 1 1 1 1>}
      uVRegion {bIsValid=false, max=<0 0>, min=<0 0>}
     }
     horizontalTopSlotImage {
      bIsDynamicallyLoaded=false
      drawAs=NoDrawType
      imageType=FullColor
      mirroring=NoMirror
      resourceName=../../../Engine/Content/Slate/Common/Scrollbar_Background_Horizontal.png
      resourceObject=null
      tiling=NoTile
      imageSize=<8 8>
      margin=<ltrb 0 0 0 0>
      outlineSettings {
       bUseBrushTransparency=false
       roundingType=HalfHeightRadius
       width=0
       color {colorUseRule=UseColor_Specified, specifiedColor=<rgba 0 0 0 0>}
       cornerRadii=<0 0 0 0>
      }
      tintColor {colorUseRule=UseColor_Specified, specifiedColor=<rgba 1 1 1 1>}
      uVRegion {bIsValid=false, max=<0 0>, min=<0 0>}
     }
     hoveredThumbImage {
      bIsDynamicallyLoaded=false
      drawAs=RoundedBox
      imageType=FullColor
      mirroring=NoMirror
      resourceName=../../../Engine/Content/Slate/Common/Scrollbar_Thumb.png
      resourceObject=null
      tiling=NoTile
      imageSize=<12 12>
      margin=<ltrb 0.25 0.25 0.25 0.25>
      outlineSettings {
       bUseBrushTransparency=false
       roundingType=FixedRadius
       width=2
       color {colorUseRule=UseColor_Specified, specifiedColor=<rgba 0 0 0 0>}
       cornerRadii=<6 6 6 6>
      }
      tintColor {colorUseRule=UseColor_Specified, specifiedColor=<rgba 1 1 1 1>}
      uVRegion {bIsValid=false, max=<0 0>, min=<0 0>}
     }
     normalThumbImage {
      bIsDynamicallyLoaded=false
      drawAs=RoundedBox
      imageType=FullColor
      mirroring=NoMirror
      resourceName=../../../Engine/Content/Slate/Common/Scrollbar_Thumb.png
      resourceObject=null
      tiling=NoTile
      imageSize=<12 12>
      margin=<ltrb 0.25 0.25 0.25 0.25>
      outlineSettings {
       bUseBrushTransparency=false
       roundingType=FixedRadius
       width=2
       color {colorUseRule=UseColor_Specified, specifiedColor=<rgba 0 0 0 0>}
       cornerRadii=<6 6 6 6>
      }
      tintColor {colorUseRule=UseColor_Specified, specifiedColor=<rgba 1 1 1 0.75>}
      uVRegion {bIsValid=false, max=<0 0>, min=<0 0>}
     }
     verticalBackgroundImage {
      bIsDynamicallyLoaded=false
      drawAs=Box
      imageType=NoImage
      mirroring=NoMirror
      resourceName=None
      resourceObject=/FortUILibrary/UIKit/ScrollBar/MaterialInstances/MI_UIKit_ScrollbarBackground
      tiling=NoTile
      imageSize=<12 12>
      margin=<ltrb 0.5 0.5 0.5 0.5>
      outlineSettings {
       bUseBrushTransparency=false
       roundingType=HalfHeightRadius
       width=0
       color {colorUseRule=UseColor_Specified, specifiedColor=<rgba 0 0 0 0>}
       cornerRadii=<0 0 0 0>
      }
      tintColor {colorUseRule=UseColor_Specified, specifiedColor=<rgba 1 1 1 1>}
      uVRegion {bIsValid=false, max=<0 0>, min=<0 0>}
     }
     verticalBottomSlotImage {
      bIsDynamicallyLoaded=false
      drawAs=NoDrawType
      imageType=FullColor
      mirroring=NoMirror
      resourceName=../../../Engine/Content/Slate/Common/Scrollbar_Background_Vertical.png
      resourceObject=null
      tiling=NoTile
      imageSize=<8 8>
      margin=<ltrb 0 0 0 0>
      outlineSettings {
       bUseBrushTransparency=false
       roundingType=HalfHeightRadius
       width=0
       color {colorUseRule=UseColor_Specified, specifiedColor=<rgba 0 0 0 0>}
       cornerRadii=<0 0 0 0>
      }
      tintColor {colorUseRule=UseColor_Specified, specifiedColor=<rgba 1 1 1 1>}
      uVRegion {bIsValid=false, max=<0 0>, min=<0 0>}
     }
     verticalTopSlotImage {
      bIsDynamicallyLoaded=false
      drawAs=NoDrawType
      imageType=FullColor
      mirroring=NoMirror
      resourceName=../../../Engine/Content/Slate/Common/Scrollbar_Background_Vertical.png
      resourceObject=null
      tiling=NoTile
      imageSize=<8 8>
      margin=<ltrb 0 0 0 0>
      outlineSettings {
       bUseBrushTransparency=false
       roundingType=HalfHeightRadius
       width=0
       color {colorUseRule=UseColor_Specified, specifiedColor=<rgba 0 0 0 0>}
       cornerRadii=<0 0 0 0>
      }
      tintColor {colorUseRule=UseColor_Specified, specifiedColor=<rgba 1 1 1 1>}
      uVRegion {bIsValid=false, max=<0 0>, min=<0 0>}
     }
    }
   }
  }]
}
