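# Context button widget for the Locker's collection sort/filter, derived from the generic UIKit button, that restyles itself for touch input.
# On initialization and on input method changes it calls ToggleTouchStyle, which checks the current CommonInput type.
# That function switches visibility of the background and main slots and picks a quiet or white icon material depending on touch.
# It also holds a mobile-touch check that sets a fixed icon size and centers button content.
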WidgetBlueprint /Locker/MPL/WBP_CollectionSortFilter_ContextButton
graphs [
 EventGraph {nodes [
   n1 = Event PreConstruct()
   > Parent: PreConstruct(IsDesignTime=@n1.IsDesignTime)
   Event Construct()
   > Parent: Construct()
   Event On Hovered()
   > Parent: On Hovered() {is=title K2Node_CallParentFunction}
   Event On Unhovered()
   > Parent: On Unhovered() {is=title K2Node_CallParentFunction}
   n2 = Event OnInitialized()
   ToggleTouchStyle(execute=[@n2.then, @n3.then])
   n3 = Event On Input Method Changed!() -> (OutputDelegate, then, CurrentInputType:ECommonInputType Enum) {is=title K2Node_Event}
   Event Tick()
  ]}
 ToggleTouchStyle {nodes [
   n1 = "Equal (Enum)"(A=GetCurrentInputType(self=Get CommonInputSubsystem()), B=Touch)
   SetVisibility(
    self=Get Slot_Background() -> (Slot_Background:Named Slot Object Reference)
    InVisibility=Select(Option 0=SelfHitTestInvisible, Option 1=Hidden, Index=@n1)
   )
   > SetVisibility(
    self=Get Slot_Main() -> (Slot_Main:Named Slot Object Reference)
    InVisibility=Select(Option 0=SelfHitTestInvisible, Option 1=Hidden, Index=@n1)
   )
   > SetIconMaterial!(
    execute
    self:WBP UIKit Button Generic Base Object Reference
    Material Instance:Material Instance Object Reference=Select(
     Option 0:Material Instance Object Reference=/FortUILibrary/UIKit/Buttons/MaterialInstances/CTAButtons/MI_UIKit_IconQuiet
     Option 1:Material Instance Object Reference=/FortUILibrary/UIKit/Buttons/MaterialInstances/CTAButtons/MI_UIKit_IconWhite
     Index=@n1
    ) -> (ReturnValue:Material Instance Object Reference)
   ) -> (then)
   ToggleTouchStyle() {is=title K2Node_FunctionEntry}
   n2 = GetCurrentInputType(self=Get CommonInputSubsystem())
   > SetVisibility(
    self=Get Slot_Main()
    InVisibility=Select(Option 0:ESlateVisibility Enum=SelfHitTestInvisible, Option 1:ESlateVisibility Enum=Collapsed, Index="Equal (Enum)"(A=@n2, B=Touch)) -> (ReturnValue:ESlateVisibility Enum)
   )
   n3 = GetOwningPlayer()
   > Branch(Condition=AND Boolean(A=IsMobileGame(), B=IsUsingTouch(PlayerController=@n3)))
   > Set IconSize(IconSize:Vector 2D Structure=(X=48,Y=48)) -> (Output_Get:Vector 2D Structure)
   > SetButtonContentAlignment!(
    execute
    self:WBP UIKit Button Generic Base Object Reference
    HorizontalAlignment:EHorizontalAlignment Enum=HAlign_Center
    VerticalAlignment:EVerticalAlignment Enum=VAlign_Center
   ) -> (then)
  ]}
]
variables []
widgets {
 info {
  inheritedWidgetCount=0
  parentClass=/FortUILibrary/UIKit/Buttons/WBP_UIKit_Button_Generic.WBP_UIKit_Button_Generic_C
  rootWidgetClass=null
  widgetCount=0
 }
 widgets []
}
