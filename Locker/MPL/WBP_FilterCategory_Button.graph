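# Locker filter category button, built on the UIKit generic modular button, with a selection pip and a new-items indicator.
# A short Selected animation plays forward on select and in reverse on deselect, raising the pip's opacity and stretching its spacer vertically.
# UpdateState passes HasNewItems to the StatusIndicator's bang state with a zero count.
# The Pip border is collapsed by default, and nothing here makes it visible.
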WidgetBlueprint /Locker/MPL/WBP_FilterCategory_Button
animations [Selected {
  binding_count=2
  length_seconds=0.15
  object=&self:Selected
  sequence {bindings [
    Pip {
     id=#1
     possessed_class=Border
     tracks [{
       type=MovieSceneColorTrack
       display_name=Brush Color
       sections [{
         type=MovieSceneColorSection
         end_frame=3
         start_frame=0
         channels [
          Color.R {type=MovieSceneScriptingFloatChannel}
          Color.G {type=MovieSceneScriptingFloatChannel}
          Color.B {type=MovieSceneScriptingFloatChannel}
          Color.A {type=MovieSceneScriptingFloatChannel, keys [[0,0.5],[3,0.8]]}
         ]
        }]
      }]
    }
    PipSpacer {
     id=#2
     possessed_class=Spacer
     tracks [{
       type=MovieSceneDoubleVectorTrack
       display_name=Size
       sections [{
         type=MovieSceneDoubleVectorSection
         end_frame=3
         start_frame=0
         channels [
          Vector.X {type=MovieSceneScriptingDoubleChannel}
          Vector.Y {type=MovieSceneScriptingDoubleChannel, keys [[0,12],[3,24]]}
         ]
        }]
      }]
    }
   ]}
 }]
graphs [
 EventGraph {nodes [
   n1 = Event PreConstruct()
   > Parent: PreConstruct(IsDesignTime=@n1.IsDesignTime)
   Event Construct()
   > Parent: Construct()
   Event On Selected()
   > Parent: On Selected() {is=title K2Node_CallParentFunction}
   n2 = Get Selected() -> (Selected:Widget Animation Object Reference)
   > PlayAnimationForward(InAnimation=@n2)
   Event On Deselected()
   > Parent: On Deselected() {is=title K2Node_CallParentFunction}
   > PlayAnimationReverse(InAnimation=@n2)
   Event Tick()
  ]}
 UpdateState {nodes [
   n1 = UpdateState!() -> (then, HasNewItems:bool=~) {is=title K2Node_FunctionEntry}
   > Set Bang State!(
    execute
    self:Fort Bang Wrapper NUI Object Reference=Get StatusIndicator() -> (StatusIndicator:WBP UIKit Status Indicator Object Reference)
    bEnabled:bool=@n1
    Count:int=0
   ) -> (then) {is=title K2Node_CallFunction}
  ]}
]
variables []
widgets {
 info {
  inheritedWidgetCount=0
  namedSlotCount=1
  parentClass=/FortUILibrary/Systems/ModularButton/WBP_UIKit_ButtonGeneric_Base.WBP_UIKit_ButtonGeneric_Base_C
  rootWidgetClass=null
  widgetCount=2
 }
 widgets [
  Pip {
   widgetClassPath=&Border
   bIsVariable=true
   widget=Pip
   properties {
    horizontalAlignment=HAlign_Left
    verticalAlignment=VAlign_Top
    visibility=Collapsed
    background {
     bIsDynamicallyLoaded=false
     drawAs=Image
     imageType=FullColor
     mirroring=NoMirror
     resourceName=None
     resourceObject=/FortUILibrary/UIKit/PipList/MI_UIKit_Pip_Entry
     tiling=NoTile
     imageSize=<12 12>
     margin=<ltrb 0.5 0.5 0.5 0.5>
     outlineSettings {
      bUseBrushTransparency=false
      roundingType=HalfHeightRadius
      width=0
      color {colorUseRule=UseColor_Specified, specifiedColor=<rgba 0 0 0 0>}
      cornerRadii=<0 0 0 0>
     }
     tintColor {colorUseRule=UseColor_Specified, specifiedColor=<rgba 1 1 1 1>}
     uVRegion {bIsValid=false, max=<0 0>, min=<0 0>}
    }
    brushColor=<rgba 1 1 1 0.5>
    padding=<ltrb 0 0 0 0>
   }
  }
  PipSpacer {
   widgetClassPath=&Spacer
   bIsVariable=true
   parent=Pip
   widget=PipSpacer
   properties {size=<12 12>}
   slot {class=BorderSlot, properties {horizontalAlignment=HAlign_Left, verticalAlignment=VAlign_Top, padding=<ltrb 0 0 0 0>}}
  }
 ]
}
