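# Locker UI widget that displays an empty warning, laid out as a horizontal box with two CommonUI text blocks.
# A fixed prefix block reads "(Empty)", beside a variable block meant for the selected filter and category.
# On Construct, SelectedFilterLabel and SelectedCategoryLabel are formatted as a localized "{FilterLabel} {CategoryLabel}" into the variable block.
# That block starts Collapsed and nothing in this graph reveals it or updates it after Construct.
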WidgetBlueprint /Locker/MPL/WBP_LockerEmptyWarning
graphs [EventGraph {nodes [
   Event Construct()
   > "SetText (Text)"(
    self=Get CommonTextBlock_Variables() -> (CommonTextBlock_Variables:Common Text Object Reference)
    InText=Format Text!(
     Format="LOCTEXT(\"{FilterLabel} {CategoryLabel}\")"
     FilterLabel:Text=Get SelectedFilterLabel() -> (SelectedFilterLabel:Text)
     CategoryLabel:Text=Get SelectedCategoryLabel() -> (SelectedCategoryLabel:Text)
    ) -> (Result)
   )
  ]}]
variables [
 SelectedCategoryLabel {type=string, default=""}
 SelectedFilterLabel {type=string, default=""}
]
widgets {
 info {inheritedWidgetCount=0, parentClass=&UserWidget, rootWidgetClass=&HorizontalBox, widgetCount=3}
 widgets [
  HorizontalBox_Root {widgetClassPath=&HorizontalBox, widget=HorizontalBox_Root}
  CommonTextBlock_FixedPrefix {
   widgetClassPath=&CommonTextBlock
   parent=HorizontalBox_Root
   widget=CommonTextBlock_FixedPrefix
   properties {
    lineHeightPercentage=1.28
    text=(Empty)
    colorAndOpacity {colorUseRule=UseColor_Specified, specifiedColor=<rgba 1 1 1 0.7>}
    font {
     bForceMonospaced=false
     bMaterialIsStencil=false
     fontMaterial=null
     fontObject=&HeadingNow
     letterSpacing=0
     monospacedWidth=1
     size=18
     skewAmount=0
     typefaceFontName=74 Regular
     outlineSettings {
      bApplyOutlineToDropShadows=false
      bMiteredCorners=false
      bSeparateFillAlpha=false
      outlineMaterial=null
      outlineSize=0
      outlineColor=<rgba 0 0 0 1>
     }
    }
    margin=<ltrb 0 8 0 0>
    shadowOffset=<0 0>
    strikeBrush {
     bIsDynamicallyLoaded=false
     drawAs=NoDrawType
     imageType=NoImage
     mirroring=NoMirror
     resourceName=None
     resourceObject=null
     tiling=NoTile
     imageSize=<32 32>
     margin=<ltrb 0 0 0 0>
     outlineSettings {
      bUseBrushTransparency=false
      roundingType=HalfHeightRadius
      width=0
      color {colorUseRule=UseColor_Specified, specifiedColor=<rgba 0 0 0 0>}
      cornerRadii=<0 0 0 0>
     }
     tintColor {colorUseRule=UseColor_Specified, specifiedColor=<rgba 1 1 1 1>}
     uVRegion {bIsValid=false, max=<0 0>, min=<0 0>}
    }
   }
   slot {class=HorizontalBoxSlot, properties {horizontalAlignment=HAlign_Center, verticalAlignment=VAlign_Center}}
  }
  CommonTextBlock_Variables {
   widgetClassPath=&CommonTextBlock
   bIsVariable=true
   parent=HorizontalBox_Root
   widget=CommonTextBlock_Variables
   properties {
    text=FILTERLABEL CATEGORYLABEL
    visibility=Collapsed
    font {
     bForceMonospaced=false
     bMaterialIsStencil=false
     fontMaterial=null
     fontObject=&HeadingNow
     letterSpacing=0
     monospacedWidth=1
     size=24
     skewAmount=0
     typefaceFontName=75 Medium
     outlineSettings {
      bApplyOutlineToDropShadows=false
      bMiteredCorners=false
      bSeparateFillAlpha=false
      outlineMaterial=null
      outlineSize=0
      outlineColor=<rgba 0 0 0 1>
     }
    }
    shadowOffset=<0 0>
    strikeBrush {
     bIsDynamicallyLoaded=false
     drawAs=NoDrawType
     imageType=NoImage
     mirroring=NoMirror
     resourceName=None
     resourceObject=null
     tiling=NoTile
     imageSize=<32 32>
     margin=<ltrb 0 0 0 0>
     outlineSettings {
      bUseBrushTransparency=false
      roundingType=HalfHeightRadius
      width=0
      color {colorUseRule=UseColor_Specified, specifiedColor=<rgba 0 0 0 0>}
      cornerRadii=<0 0 0 0>
     }
     tintColor {colorUseRule=UseColor_Specified, specifiedColor=<rgba 1 1 1 1>}
     uVRegion {bIsValid=false, max=<0 0>, min=<0 0>}
    }
   }
   slot {class=HorizontalBoxSlot, properties {horizontalAlignment=HAlign_Center, verticalAlignment=VAlign_Center}}
  }
 ]
}
