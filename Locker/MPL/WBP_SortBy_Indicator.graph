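# Locker widget that displays the current sort type under a "Sort By:" header.
# A vertical box stacks two right-justified, uppercase CommonTextBlocks: a slightly translucent header and the sort type name, which auto-collapses when empty.
# UpdateSortTypeText sets the name from the text it receives; the event graph has no logic, so updates depend on that function being called.
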WidgetBlueprint /Locker/MPL/WBP_SortBy_Indicator
graphs [
 EventGraph {nodes [
   Event PreConstruct()
   Event Construct()
   Event Tick()
  ]}
 UpdateSortTypeText {nodes [
   n1 = UpdateSortTypeText!() -> (then, NewParam:Text) {is=title K2Node_FunctionEntry}
   > "SetText (Text)"(self=Get SortTypeName() -> (SortTypeName:Common Text Object Reference), InText=@n1)
  ]}
]
variables []
widgets {
 info {inheritedWidgetCount=0, parentClass=&UserWidget, rootWidgetClass=&VerticalBox, widgetCount=3}
 widgets [
  VerticalBox_14 {widgetClassPath=&VerticalBox, widget=VerticalBox_14}
  SortByHeader {
   widgetClassPath=&CommonTextBlock
   bIsVariable=true
   parent=VerticalBox_14
   widget=SortByHeader
   properties {
    autoWrapText=true
    bIsScrollingEnabled=false
    justification=Right
    lineHeightPercentage=0.9
    mobileFontSizeMultiplier=1.6
    scrollOrientation=Orient_Vertical
    scrollStyle=/FortUILibrary/Styles/Typography/Foundation/TextScrollStyle-NoFade.TextScrollStyle-NoFade_C
    text=Sort By:
    textOverflowPolicy=Ellipsis
    textTransformPolicy=ToUpper
    wrappingPolicy=AllowPerCharacterWrapping
    colorAndOpacity {colorUseRule=UseColor_Specified, specifiedColor=<rgba 1 1 1 0.7>}
    font {
     bForceMonospaced=false
     bMaterialIsStencil=false
     fontMaterial=null
     fontObject=&HeadingNow
     letterSpacing=0
     monospacedWidth=1
     size=32.25
     skewAmount=0
     typefaceFontName=HeadingNow-86BoldItalic
     outlineSettings {
      bApplyOutlineToDropShadows=false
      bMiteredCorners=false
      bSeparateFillAlpha=false
      outlineMaterial=null
      outlineSize=0
      outlineColor=<rgba 0 0 0 1>
     }
    }
    shadowOffset=<0 0>
   }
   slot {class=VerticalBoxSlot}
  }
  SortTypeName {
   widgetClassPath=&CommonTextBlock
   bIsVariable=true
   parent=VerticalBox_14
   widget=SortTypeName
   properties {
    autoWrapText=true
    bAutoCollapseWithEmptyText=true
    bIsScrollingEnabled=false
    justification=Right
    lineHeightPercentage=0.9
    scrollOrientation=Orient_Vertical
    scrollStyle=&TextScrollStyle-NoFade_C
    text="[PH] Sort Type Name"
    textOverflowPolicy=Ellipsis
    textTransformPolicy=ToUpper
    wrappingPolicy=AllowPerCharacterWrapping
    font {
     bForceMonospaced=false
     bMaterialIsStencil=false
     fontMaterial=null
     fontObject=&HeadingNow
     letterSpacing=0
     monospacedWidth=1
     size=32.25
     skewAmount=0
     typefaceFontName=HeadingNow-86BoldItalic
     outlineSettings {
      bApplyOutlineToDropShadows=false
      bMiteredCorners=false
      bSeparateFillAlpha=false
      outlineMaterial=null
      outlineSize=0
      outlineColor=<rgba 0 0 0 1>
     }
    }
    shadowOffset=<0 0>
   }
   slot {class=VerticalBoxSlot}
  }
 ]
}
