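# WBP_ContextMenu_Title is a UMG widget that displays a title heading within a context menu, apparently for the Locker screens.
# A padded CommonBorder root wraps a single grey CommonTextBlock, Text_Title, which uses a base scroll style and holds placeholder text.
# SetContextTitle takes a Text input and sets it on Text_Title; the event graph's PreConstruct, Construct and Tick events are empty.
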WidgetBlueprint /Locker/MPL/WBP_ContextMenu_Title
graphs [
 EventGraph {nodes [
   Event PreConstruct()
   Event Construct()
   Event Tick()
  ]}
 SetContextTitle {nodes [
   n1 = SetContextTitle!() -> (then, InText:Text) {is=title K2Node_FunctionEntry}
   > "SetText (Text)"(self=Get Text_Title() -> (Text_Title:Common Text Object Reference), InText=@n1)
  ]}
]
variables []
widgets {
 info {inheritedWidgetCount=0, parentClass=&UserWidget, rootWidgetClass=&CommonBorder, widgetCount=2}
 widgets [
  CommonBorder_0 {widgetClassPath=&CommonBorder, widget=CommonBorder_0, properties {padding=<ltrb 15 12 8 8>}}
  Text_Title {
   widgetClassPath=&CommonTextBlock
   bIsVariable=true
   parent=CommonBorder_0
   widget=Text_Title
   properties {
    scrollStyle=&TextScrollStyle-Base_C
    text="[PH] Context Title"
    colorAndOpacity {colorUseRule=UseColor_Specified, specifiedColor=<rgba 0.5776 0.5776 0.5776 1>}
    font {
     bForceMonospaced=false
     bMaterialIsStencil=false
     fontMaterial=null
     fontObject=&HeadingNow
     letterSpacing=0
     monospacedWidth=1
     size=18
     skewAmount=0
     typefaceFontName=75 Medium
     outlineSettings {
      bApplyOutlineToDropShadows=false
      bMiteredCorners=false
      bSeparateFillAlpha=false
      outlineMaterial=null
      outlineSize=0
      outlineColor=<rgba 0 0 0 1>
     }
    }
    shadowOffset=<0 0>
    shapedTextOptions {bOverride_TextFlowDirection=true, bOverride_TextShapingMethod=false, textFlowDirection=Culture, textShapingMethod=Auto}
   }
   slot {class=BorderSlot, properties {padding=<ltrb 15 12 8 8>}}
  }
 ]
}
