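# Sort context menu widget for the Locker, built as a child of WBP_UIKit_ContextMenu with no widgets of its own.
# PreConstruct creates a WBP_ContextMenu_Title reading "Sort by:" and adds it to the top-left cell of the inherited root grid, filling horizontally.
# Construct calls the parent, then pins the SB_MaxWidth size box to a fixed width by setting both min and max from one variable.
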WidgetBlueprint /Locker/MPL/WBP_Sort_ContextMenu
graphs [EventGraph {nodes [
   Event PreConstruct()
   > n1 = Create Widget!(
    execute:Exec
    Class:User Widget Class Reference=&/WBP_ContextMenu_Title.WBP_ContextMenu_Title_C
    OwningPlayer:Player Controller Object Reference
   ) -> (then:Exec, ReturnValue:WBP Context Menu Title Object Reference)
   > SetContextTitle!(execute, self:WBP Context Menu Title Object Reference=@n1, InText:Text="LOCTEXT(\"Sort by:\")") -> (then)
   > n2 = AddChildToGrid!(
    execute
    self:Grid Panel Object Reference=Get Root() -> (Root:Grid Panel Object Reference)
    Content:Widget Object Reference=@n1
    InRow:int=0
    InColumn:int=0
   ) -> (then, ReturnValue:Grid Slot Object Reference)
   > SetHorizontalAlignment(self:Grid Slot Object Reference=@n2, InHorizontalAlignment=HAlign_Fill)
   Event Construct()
   > Parent: Construct()
   n3 = Get SB_MaxWidth() -> (SB_MaxWidth:Size Box Object Reference)
   n4 = Get In Max Desired Width() -> (In Max Desired Width:float)
   > SetMaxDesiredWidth!(execute, self:Size Box Object Reference=@n3, InMaxDesiredWidth:float=@n4) -> (then)
   > SetMinDesiredWidth!(execute, self:Size Box Object Reference=@n3, InMinDesiredWidth:float=@n4) -> (then)
   Event Tick()
  ]}]
variables [In Max Desired Width {type=number, default=360}]
widgets {
 info {
  inheritedWidgetCount=0
  parentClass=/FortUILibrary/Systems/ContextMenu/WBP_UIKit_ContextMenu.WBP_UIKit_ContextMenu_C
  rootWidgetClass=null
  widgetCount=0
 }
 widgets []
}
