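# Subcategory carousel widget for the locker emote wheel, overlaying a heading and icon slot on an arrow-navigated tab group with invisible-named button class.
# PopulateCategories rebuilds tabs from category item view models flagged for tab display, keyed by analytics name.
# Tab selection is rebroadcast through OnTabSelect.
# SetupSpacing pads the heading box by arrow-button width plus desktop or mobile padding, deferred three ticks after Construct, presumably awaiting cached geometry.
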WidgetBlueprint /Locker/MPL/EmoteWheel/WBP_MPLEmotes_SubcategoryCarousel
graphs [
 EventGraph {nodes [
   n1 = "On UIKit Tab Selected (Tabs_Subcategory)"!() -> (OutputDelegate, then, TabID:Name=None, TabVM:MVVM Base Viewmodel Object Reference) {is=title K2Node_ComponentBoundEvent}
   > Call On Tab Select!(execute:Exec, self:WBP MPLEmotes Subcategory Carousel Object Reference, TabID:Name=@n1.TabID) -> (then:Exec) {is=title K2Node_CallDelegate}
   n2 = Event PreConstruct()
   > Branch(Condition=@n2.IsDesignTime)
   > SetupSpacing!(execute, self, UseMobileSpacing:bool=false) -> (then)
   Event Construct()
   > Parent: Construct()
   > DelayUntilNextTick!(execute) -> (then)
   > DelayUntilNextTick()
   > DelayUntilNextTick()
   > SetupSpacing(UseMobileSpacing=IsMobileGame())
  ]}
 OnTabSelect {nodes [OnTabSelect!() -> (then, TabID:Name) {is=title K2Node_FunctionEntry}]}
 PopulateCategories {nodes [
   n1 = PopulateCategories!() -> (then, Categories:Array of Fort Locker Category Item VM Object References) {is=title K2Node_FunctionEntry}
   > RemoveTabs(self:WBP UIKit Tab Group Base Object Reference=Get Tabs_Subcategory() -> (Tabs_Subcategory:WBP UIKit Tab Group Object Reference))
   > n2 = For Each Loop(Array:Array of Fort Locker Category Item VM Object References=@n1) -> (Array Element:Fort Locker Category Item VM Object Reference)
   n3 = Get Category!(self:Fort Locker Category Item VM Object Reference=@n2) -> (Category:Fort Locker Category VM Object Reference)
   > Branch(Condition=Get!(self:Fort Locker Category Config Object Reference=GetConfig!(self:Fort Locker Category VM Object Reference=@n3) -> (ReturnValue:Fort Locker Category Config Object Reference)) -> (
     SlotInfo_CategoryName:Text
     SlotInfo_CategoryNamePlural:Text
     SlotInfo_CategoryNameWithPluralForms:Text
     SlotInfo_CategoryDescription:Text
     SlotInfo_DefaultEmptyTexture:Soft Object Reference
     SlotInfo_OverrideShuffleTexture:Texture 2D Soft Object Reference
     SlotInfo_CategoryIcon:Material Instance Object Reference
     SlotInfo_CategorySoundOverrides:Map of Gameplay Tag Structures to Sound Base Soft Object References
     SlotInfo_bShowInCategoryTab:bool=~
     SlotInfo_PreviewWindowScale:float=~
     SlotInfo_PreviewWindowOffset:Vector 3f
     SlotInfo_bUsePluralCategoryInItemList:bool=~
    ).SlotInfo_bShowInCategoryTab)
   > AddTab!(
    execute
    self:WBP UIKit Tab Group Base Object Reference=Get Tabs_Subcategory()
    Tab VM:MVVM Base Viewmodel Object Reference=@n2
    Tab ID:Name=String To Name(InString=GetAnalyticsName!(self:Fort Locker Category Item VM Object Reference=@n2) -> (ReturnValue:String))
    Index:int=-1
   ) -> (then)
  ]}
 SetupSpacing {nodes [
   SetupSpacing!() -> (then, UseMobileSpacing:bool=~) {is=title K2Node_FunctionEntry}
   n1 = Get Prev_ArrowButton!(self:WBP UIKit Tab Group Base Object Reference=Get Tabs_Subcategory()) -> (Prev_ArrowButton:WBP UIKit Button Generic Base Object Reference)
   n2 = BreakVector2D(InVec=GetLocalSize!(Geometry:"Geometry Structure (by ref)"=GetCachedGeometry!(self:Widget Object Reference=@n1) -> (ReturnValue:"Geometry Structure (by ref)")) -> (ReturnValue:Vector 2D Structure))
   n3 = Select(
    Option 0:double=float + float(A=@n2, B=Get BumpersPadding() -> (BumpersPadding:double))
    Option 1:double=float + float(A=@n2, B=Get MobileBumpersPadding() -> (MobileBumpersPadding:double))
    Index=Get UseMobileSpacing() -> (UseMobileSpacing:bool)
   ) -> (ReturnValue:double)
   > Set Padding!(
    execute
    Padding:Margin Structure=Make Margin(Left=@n3, Top=0, Right=@n3, Bottom=0)
    self:Overlay Slot Object Reference=SlotAsOverlaySlot(Widget=Get HorizontalBox() -> (HorizontalBox:Horizontal Box Object Reference))
   ) -> (then, Output_Get:Margin Structure)
  ]}
]
variables [
 BumpersPadding {type=number, default=24}
 MobileBumpersPadding {type=number, default=24}
 OnTabSelect {}
]
widgets {
 info {inheritedWidgetCount=0, parentClass=&UserWidget, rootWidgetClass=&Overlay, widgetCount=5}
 widgets [
  Overlay_1 {widgetClassPath=&Overlay, bIsVariable=true, widget=Overlay_1, properties {bOverride_Cursor=true}}
  Tabs_Subcategory {
   widgetClassPath=/FortUILibrary/UIKit/Tabs/WBP_UIKit_TabGroup
   bIsVariable=true
   parent=Overlay_1
   widget=Tabs_Subcategory
   properties {
    animateReservedPromptSpace=false
    arrowButtonsClass=/FortniteCoreUI/Frontend/WBP_UIKit_Button_TabNavArrow_FrontEnd.WBP_UIKit_Button_TabNavArrow_FrontEnd_C
    buttonClass=/Locker/MPL/WBP_SubCategoryCarousel_Button_invisible.WBP_SubCategoryCarousel_Button_invisible_C
    reservedPromptSpace=64
    tabGroupHorizontalAlignment=HAlign_Fill
    tabGroupVerticalAlignment=VAlign_Fill
    tabSpacing=4
    useArrowNavButtons=true
    useBackplate=false
    usePromptBackplate=false
    initialTabsData [{bEnableTabId=false, tabId=None, tabViewModel=null}, {bEnableTabId=false, tabId=None, tabViewModel=null}]
    minButtonSize=<0 40>
    nextPromptPadding=<ltrb 10 0 0 0>
    prevPromptPadding=<ltrb 0 0 10 0>
    tabGroupPadding=<ltrb 0 0 0 0>
   }
   slot {class=OverlaySlot, properties {horizontalAlignment=HAlign_Fill, verticalAlignment=VAlign_Center, padding=<ltrb 10 0 10 0>}}
  }
  HorizontalBox {
   widgetClassPath=&HorizontalBox
   bIsVariable=true
   parent=Overlay_1
   widget=HorizontalBox
   slot {class=OverlaySlot, properties {horizontalAlignment=HAlign_Center, verticalAlignment=VAlign_Center, padding=<ltrb 56 0 56 0>}}
  }
  WBP_MPLocker_Heading {
   widgetClassPath=/FortniteCoreUI/Customization/MPL/WBP_MPLocker_Heading
   bIsVariable=true
   parent=HorizontalBox
   widget=WBP_MPLocker_Heading
   properties {show Sub Heading?=false, useNewLockerTextSize=true}
   slot {class=HorizontalBoxSlot, properties {verticalAlignment=VAlign_Center}}
  }
  IconSlot {
   widgetClassPath=&NamedSlot
   bIsVariable=true
   parent=HorizontalBox
   widget=IconSlot
   slot {class=HorizontalBoxSlot, properties {padding=<ltrb 12 0 0 0>}}
  }
 ]
}
