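# Locker widget that shows a subcategory tab carousel for item selection, switching to a dedicated emotes carousel when the selected category group is Emotes.
# PopulateCategories builds tabs from category view models flagged for display, keyed by analytics name.
# Both carousels forward selection through the TabSelect dispatcher; on switching to Emotes, the selected category's tab is reselected.
# It also tracks input type, applies edge-fade scroll styling and enlarges buttons on mobile.
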WidgetBlueprint /Locker/MPL/WBP_SubcategoryCarousel
graphs [
 EventGraph {nodes [
   n1 = "On UIKit Tab Selected (Tabs_Subcategory)"!() -> (OutputDelegate, then, TabID:Name=None, TabVM:MVVM Base Viewmodel Object Reference) {is=title K2Node_ComponentBoundEvent}
   > Call Tab Select!(execute:Exec, self:WBP Subcategory Carousel Object Reference, TabID:Name=@n1.TabID) -> (then:Exec) {is=title K2Node_CallDelegate}
   n2 = "On Tab Select (MPLEmotes_SubcategoryCarousel)"!() -> (OutputDelegate, then, TabID:Name=None) {is=title K2Node_ComponentBoundEvent}
   > Call Tab Select(TabID=@n2.TabID)
   n3 = Event On Input Method Changed!() -> (OutputDelegate, then, bNewInputType:ECommonInputType Enum) {is=title K2Node_CustomEvent}
   > Set CurrentInputType(CurrentInputType:ECommonInputType Enum=@n3.bNewInputType) -> (Output_Get:ECommonInputType Enum)
   n4 = Event PreConstruct()
   > SetCompactUIStyle() {is=title K2Node_CallFunction}
   > Branch(Condition=@n4.IsDesignTime)
   > SetupEmotesSubcategoryCarousel()
   Event Construct()
   n5 = Get CommonInputSubsystem()
   > n6 = Branch(Condition=IsValid(Object=@n5))
   > Bind Event to On Input Method Changed(self=@n5, Delegate=@n3)
   > n7 = Set CurrentInputType(CurrentInputType=GetCurrentInputType(self=@n5))
   Set!(
    execute=[@n7.then, @n6.else]
    WidgetStyle_BarThickness:float=9
    WidgetStyle_TopShadowBrush:Slate Brush Structure=(TintColor=(SpecifiedColor=(R=1,G=1,B=1,A=1),ColorUseRule=UseColor_Specified),DrawAs=Image,Tiling=NoTile,Mirroring=NoMirror,ImageType=NoImage,ImageSize=(X=32,Y=32),Margin=(Left=0,Top=0,Right=0,Bottom=0),ResourceObject=None,OutlineSettings=(CornerRadii=(X=0,Y=0,Z=0,W=0),Color=(SpecifiedColor=(R=0,G=0,B=0,A=0),ColorUseRule=UseColor_Specified),Width=0,RoundingType=HalfHeightRadius,bUseBrushTransparency=False),UVRegion=(Min=(X=0,Y=0),Max=(X=0,Y=0),bIsValid=False),bIsDynamicallyLoaded=False,ResourceName="")
    WidgetStyle_BottomShadowBrush:Slate Brush Structure=(TintColor=(SpecifiedColor=(R=1,G=1,B=1,A=1),ColorUseRule=UseColor_Specified),DrawAs=Image,Tiling=NoTile,Mirroring=NoMirror,ImageType=NoImage,ImageSize=(X=32,Y=32),Margin=(Left=0,Top=0,Right=0,Bottom=0),ResourceObject=None,OutlineSettings=(CornerRadii=(X=0,Y=0,Z=0,W=0),Color=(SpecifiedColor=(R=0,G=0,B=0,A=0),ColorUseRule=UseColor_Specified),Width=0,RoundingType=HalfHeightRadius,bUseBrushTransparency=False),UVRegion=(Min=(X=0,Y=0),Max=(X=0,Y=0),bIsValid=False),bIsDynamicallyLoaded=False,ResourceName="")
    WidgetStyle_LeftShadowBrush_TintColor:Slate Color Structure=(SpecifiedColor=(R=1,G=1,B=1,A=1),ColorUseRule=UseColor_Specified)
    WidgetStyle_LeftShadowBrush_DrawAs:ESlateBrushDrawType Enum=Image
    WidgetStyle_LeftShadowBrush_Tiling:ESlateBrushTileType Enum=NoTile
    WidgetStyle_LeftShadowBrush_Mirroring:ESlateBrushMirrorType Enum=NoMirror
    WidgetStyle_LeftShadowBrush_ImageSize:Vector 2D Structure=(X=32,Y=32)
    WidgetStyle_LeftShadowBrush_Margin:Margin Structure=(Left=0,Top=0,Right=0,Bottom=0)
    WidgetStyle_LeftShadowBrush_ResourceObject:Object Reference=/Locker/MPL/Materials/M_UI_ItemsFade_Left
    WidgetStyle_LeftShadowBrush_OutlineSettings:Slate Brush Outline Settings Structure=(CornerRadii=(X=0,Y=0,Z=0,W=0),Color=(SpecifiedColor=(R=0,G=0,B=0,A=0),ColorUseRule=UseColor_Specified),Width=0,RoundingType=HalfHeightRadius,bUseBrushTransparency=False)
    WidgetStyle_RightShadowBrush_TintColor:Slate Color Structure=(SpecifiedColor=(R=1,G=1,B=1,A=1),ColorUseRule=UseColor_Specified)
    WidgetStyle_RightShadowBrush_DrawAs:ESlateBrushDrawType Enum=Image
    WidgetStyle_RightShadowBrush_Tiling:ESlateBrushTileType Enum=NoTile
    WidgetStyle_RightShadowBrush_Mirroring:ESlateBrushMirrorType Enum=NoMirror
    WidgetStyle_RightShadowBrush_ImageSize:Vector 2D Structure=(X=32,Y=32)
    WidgetStyle_RightShadowBrush_Margin:Margin Structure=(Left=0,Top=0,Right=0,Bottom=0)
    WidgetStyle_RightShadowBrush_ResourceObject:Object Reference=/Locker/MPL/Materials/M_UI_ItemsFade_Right
    WidgetStyle_RightShadowBrush_OutlineSettings:Slate Brush Outline Settings Structure=(CornerRadii=(X=0,Y=0,Z=0,W=0),Color=(SpecifiedColor=(R=0,G=0,B=0,A=0),ColorUseRule=UseColor_Specified),Width=0,RoundingType=HalfHeightRadius,bUseBrushTransparency=False)
    WidgetStyle_HorizontalScrolledContentPadding:Margin Structure=(Left=0,Top=0,Right=1,Bottom=0)
    WidgetStyle_VerticalScrolledContentPadding:Margin Structure=(Left=0,Top=0,Right=0,Bottom=1)
    self:Scroll Box Object Reference=Get ScrollBox_Tabs!(self:WBP UIKit Tab Group Base Object Reference=Get Tabs_Subcategory() -> (Tabs_Subcategory:WBP UIKit Tab Group Object Reference)) -> (ScrollBox_Tabs:Scroll Box Object Reference)
   ) -> (then, Output_Get:Scroll Box Style Structure)
   > SetupEmotesSubcategoryCarousel()
   Event Destruct()
   n8 = Get CommonInputSubsystem()
   > Branch(Condition=IsValid(Object=@n8))
   > Unbind Event from On Input Method Changed(self=@n8, Delegate=@n3)
  ]}
 GetCarouselTabName {nodes [
   n1 = GetCarouselTabName!() -> (then, CategoryItemVM:Fort Locker Category Item VM Object Reference) {is=title K2Node_FunctionEntry}
   > n2 = Branch(Condition=GetSlotInfo!(self:Fort Locker Category Item VM Object Reference=@n1) -> (
     ReturnValue_CategoryName:Text
     ReturnValue_CategoryNamePlural:Text
     ReturnValue_CategoryNameWithPluralForms:Text
     ReturnValue_CategoryDescription:Text
     ReturnValue_DefaultEmptyTexture:Soft Object Reference
     ReturnValue_OverrideShuffleTexture:Texture 2D Soft Object Reference
     ReturnValue_CategoryIcon:Material Instance Object Reference
     ReturnValue_CategorySoundOverrides:Map of Gameplay Tag Structures to Sound Base Soft Object References
     ReturnValue_bShowInCategoryTab:bool=~
     ReturnValue_PreviewWindowScale:float=~
     ReturnValue_PreviewWindowOffset:Vector 3f
     ReturnValue_bUsePluralCategoryInItemList:bool=~
    ).ReturnValue_bUsePluralCategoryInItemList)
   n3 = Get CategoryItemVM() -> (CategoryItemVM:Fort Locker Category Item VM Object Reference)
   > Return Node!(execute, TabName:Text=GetCategoryDisplayNamePlural!(self:Fort Locker Category Item VM Object Reference=@n3) -> (ReturnValue:Text))
   Return Node(execute=@n2.else, TabName=GetCategoryDisplayName!(self:Fort Locker Category Item VM Object Reference=@n3) -> (ReturnValue:Text))
  ]}
 PopulateCategories {nodes [
   n1 = PopulateCategories!() -> (then, Categories:Array of Fort Locker Category Item VM Object References) {is=title K2Node_FunctionEntry}
   > RemoveTabs(self:WBP UIKit Tab Group Base Object Reference=Get Tabs_Subcategory())
   > n2 = For Each Loop(Array:Array of Fort Locker Category Item VM Object References=@n1) -> (Array Element:Fort Locker Category Item VM Object Reference)
   n3 = Get Category!(self:Fort Locker Category Item VM Object Reference=@n2) -> (Category:Fort Locker Category VM Object Reference)
   > Branch(Condition=Get!(self:Fort Locker Category Config Object Reference=GetConfig!(self:Fort Locker Category VM Object Reference=@n3) -> (ReturnValue:Fort Locker Category Config Object Reference)) -> (
     SlotInfo_CategoryName:Text
     SlotInfo_CategoryNamePlural:Text
     SlotInfo_CategoryNameWithPluralForms:Text
     SlotInfo_CategoryDescription:Text
     SlotInfo_DefaultEmptyTexture:Soft Object Reference
     SlotInfo_OverrideShuffleTexture:Texture 2D Soft Object Reference
     SlotInfo_CategoryIcon:Material Instance Object Reference
     SlotInfo_CategorySoundOverrides:Map of Gameplay Tag Structures to Sound Base Soft Object References
     SlotInfo_bShowInCategoryTab:bool=~
     SlotInfo_PreviewWindowScale:float=~
     SlotInfo_PreviewWindowOffset:Vector 3f
     SlotInfo_bUsePluralCategoryInItemList:bool=~
    ).SlotInfo_bShowInCategoryTab)
   > AddTab!(
    execute
    self:WBP UIKit Tab Group Base Object Reference=Get Tabs_Subcategory()
    Tab VM:MVVM Base Viewmodel Object Reference=@n2
    Tab ID:Name=String To Name(InString=GetAnalyticsName!(self:Fort Locker Category Item VM Object Reference=@n2) -> (ReturnValue:String))
    Index:int=-1
   ) -> (then)
  ]}
 SetCompactUIStyle {nodes [
   SetCompactUIStyle() {is=title K2Node_FunctionEntry}
   > Branch(Condition=IsMobileGame())
   > Set!(execute, MinButtonSize_X:int=0, MinButtonSize_Y:int=80, self:WBP UIKit Tab Group Base Object Reference=Get Tabs_Subcategory()) -> (then, Output_Get:Int Point Structure)
  ]}
 SetInteractionsEnabled {nodes [
   n1 = SetInteractionsEnabled!() -> (then, EnableInteractions:bool=~) {is=title K2Node_FunctionEntry}
   > n2 = For Loop(LastIndex=GetTabCount!(self:Common Tab List Widget Base Object Reference=Get Tabs_Subcategory()) -> (ReturnValue:int=0))
   n3 = Get Tabs_Subcategory()
   > SetTabInteractionEnabled!(
    execute
    self:Common Tab List Widget Base Object Reference=@n3
    TabNameID:Name=GetTabIDbyIndex!(self:UIKit Tab Group Object Reference=@n3, Index:int=@n2) -> (ReturnValue:Name=None)
    bEnable:bool=@n1
   ) -> (then)
  ]}
 SetupEmotesSubcategoryCarousel {nodes [
   SetupEmotesSubcategoryCarousel() {is=title K2Node_FunctionEntry}
   n1 = Get SelectedCategoryGroup!(self:Fort Item Selection VM Object Reference=Get FortItemSelectionVM() -> (FortItemSelectionVM:Fort Item Selection VM Object Reference)) -> (SelectedCategoryGroup:Fort Locker Category Group VM Object Reference)
   > n2 = Branch(Condition="Equal (Name)"(A=GetUniqueName!(self:Fort Locker Category Group VM Object Reference=@n1) -> (ReturnValue:Name=None), B=LoadoutSchema_Emotes))
   > n3 = Sequence()
   > SetActiveWidgetIndex(
    self:Common Visibility Switcher Object Reference=Get CommonVisibilitySwitcher() -> (CommonVisibilitySwitcher:Common Visibility Switcher Object Reference)
    Index=1
   )
   n4 = Get SelectedCategory!(self:Fort Item Selection VM Object Reference=Get FortItemSelectionVM()) -> (SelectedCategory:Fort Locker Category Item VM Object Reference)
   SelectTabByID!(
    execute=@n3.then_1
    self:Common Tab List Widget Base Object Reference=Get Tabs_Subcategory!(self:WBP MPLEmotes Subcategory Carousel Object Reference=Get MPLEmotes_SubcategoryCarousel() -> (MPLEmotes_SubcategoryCarousel:WBP MPLEmotes Subcategory Carousel Object Reference)) -> (Tabs_Subcategory)
    TabNameID:Name=String To Name(InString=GetAnalyticsName(self=@n4))
    bSuppressClickFeedback:bool=true
   ) -> (then, ReturnValue:bool)
   SetActiveWidgetIndex(execute=@n2.else, self=Get CommonVisibilitySwitcher(), Index=0)
  ]}
 TabSelect {nodes [TabSelect!() -> (then, TabID:Name) {is=title K2Node_FunctionEntry}]}
]
variables [
 CurrentInputType {type=ECommonInputType, default=MouseAndKeyboard}
 Tab Name ID {type=string, default=None}
 TabSelect {}
]
widgets {
 info {inheritedWidgetCount=0, namedSlotCount=1, parentClass=&UserWidget, rootWidgetClass=&SizeBox, widgetCount=8}
 widgets [
  SizeBox {widgetClassPath=&SizeBox, bIsVariable=true, widget=SizeBox, properties {bOverride_WidthOverride=true, widthOverride=850}}
  CommonVisibilitySwitcher {
   widgetClassPath=&CommonVisibilitySwitcher
   bIsVariable=true
   parent=SizeBox
   widget=CommonVisibilitySwitcher
   slot {class=SizeBoxSlot}
  }
  DefaultSubcategory_VerticalBox {
   widgetClassPath=&VerticalBox
   parent=CommonVisibilitySwitcher
   widget=DefaultSubcategory_VerticalBox
   slot {class=CommonVisibilitySwitcherSlot}
  }
  WBP_MPLocker_Heading {
   widgetClassPath=/FortniteCoreUI/Customization/MPL/WBP_MPLocker_Heading
   bIsVariable=true
   parent=DefaultSubcategory_VerticalBox
   widget=WBP_MPLocker_Heading
   properties {show Sub Heading?=false, useNewLockerTextSize=true}
   slot {class=VerticalBoxSlot, properties {padding=<ltrb 25 -8 25 0>}}
  }
  SubCategoryCarousel_Root {
   widgetClassPath=&Overlay
   bIsVariable=true
   parent=DefaultSubcategory_VerticalBox
   widget=SubCategoryCarousel_Root
   properties {bOverride_Cursor=true}
   slot {class=VerticalBoxSlot, properties {size {sizeRule=Fill, value=1}}}
  }
  Tabs_Subcategory {
   widgetClassPath=/FortUILibrary/UIKit/Tabs/WBP_UIKit_TabGroup
   bIsVariable=true
   parent=SubCategoryCarousel_Root
   widget=Tabs_Subcategory
   properties {
    animateReservedPromptSpace=false
    arrowButtonsClass=/FortniteCoreUI/Frontend/WBP_UIKit_Button_TabNavArrow_FrontEnd.WBP_UIKit_Button_TabNavArrow_FrontEnd_C
    buttonClass=&/WBP_SubCategoryCarousel_Button.WBP_SubCategoryCarousel_Button_C
    reservedPromptSpace=64
    tabGroupHorizontalAlignment=HAlign_Fill
    tabGroupVerticalAlignment=VAlign_Fill
    tabSpacing=4
    useArrowNavButtons=true
    useBackplate=false
    usePromptBackplate=false
    initialTabsData [
     {
      bEnableTabId=false
      tabId=None
      tabViewModel=&self:WidgetTree.Tabs_Subcategory.UIKitSelectableButtonViewModel_0
     }
     {
      bEnableTabId=false
      tabId=None
      tabViewModel=&self:WidgetTree.Tabs_Subcategory.UIKitSelectableButtonViewModel_2
     }
    ]
    minButtonSize=<0 40>
    nextPromptPadding=<ltrb 10 0 0 0>
    prevPromptPadding=<ltrb 0 0 10 0>
    tabGroupPadding=<ltrb 0 0 0 0>
   }
   slot {class=OverlaySlot, properties {horizontalAlignment=HAlign_Fill, verticalAlignment=VAlign_Center, padding=<ltrb 25 4 25 0>}}
  }
  MPLEmotes_SubcategoryCarousel {
   widgetClassPath=/Locker/MPL/EmoteWheel/WBP_MPLEmotes_SubcategoryCarousel
   bIsVariable=true
   parent=CommonVisibilitySwitcher
   widget=MPLEmotes_SubcategoryCarousel
   slot {class=CommonVisibilitySwitcherSlot, properties {horizontalAlignment=HAlign_Center}}
  }
  WBP_EmoteIndicator {
   widgetClassPath=/FortUILibrary/UIKit/EmoteWheel/WBP_EmoteIndicator
   bIsVariable=true
   namedSlotHost=MPLEmotes_SubcategoryCarousel
   widget=WBP_EmoteIndicator
  }
 ]
}
