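# Locker item-selection widget that shows an item's title and description with a warning box for access restrictions or items equipped in another category.
# Show and hide functions check render opacity and manage a timer before triggering fade animations, with the hide also sliding and scaling the panel.
# Helper functions derive warning visibility, icon and message from restriction state and the locker equip view model.
# On mobile, initialization applies compact padding and slot layout.
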WidgetBlueprint /Locker/MPL/WBP_ItemSelection_ItemDetails
animations [
 Anim_ShowItemDetails {
  binding_count=1
  length_seconds=0.35
  object=&self:Anim_ShowItemDetails
  sequence {bindings [ItemDetails_VerticalBox {
     id=#1
     possessed_class=VerticalBox
     tracks [
      {
       type=MovieSceneFloatTrack
       display_name=Render Opacity
       sections [{type=MovieSceneFloatSection, end_frame=7, start_frame=0, channels [None {type=MovieSceneScriptingFloatChannel, keys [[0,0],[7,1]]}]}]
      }
      {
       type=MovieScene2DTransformTrack
       sections [{
         type=MovieScene2DTransformSection
         end_frame=7
         start_frame=0
         channels [
          Translation.X {type=MovieSceneScriptingFloatChannel, const [0,2]}
          Translation.Y {type=MovieSceneScriptingFloatChannel, const [0,2]}
          Scale.X {type=MovieSceneScriptingFloatChannel, const [1,2]}
          Scale.Y {type=MovieSceneScriptingFloatChannel, const [1,2]}
         ]
        }]
      }
     ]
    }]}
 }
 Anim_HideItemDetails {
  binding_count=1
  length_seconds=0.35
  object=&self:Anim_HideItemDetails
  sequence {bindings [ItemDetails_VerticalBox {
     id=#2
     possessed_class=VerticalBox
     tracks [
      {
       type=MovieSceneFloatTrack
       display_name=Render Opacity
       sections [{type=MovieSceneFloatSection, end_frame=7, start_frame=0, channels [None {type=MovieSceneScriptingFloatChannel, keys [[0,1],[7,0]]}]}]
      }
      {
       type=MovieScene2DTransformTrack
       sections [{
         type=MovieScene2DTransformSection
         end_frame=7
         start_frame=0
         channels [
          Translation.X {type=MovieSceneScriptingFloatChannel, keys [[0,0],[7,-32]]}
          Translation.Y {type=MovieSceneScriptingFloatChannel, keys [[0,0],[7,64]]}
          Scale.X {type=MovieSceneScriptingFloatChannel, keys [[0,1],[7,1.2]]}
          Scale.Y {type=MovieSceneScriptingFloatChannel, keys [[0,1],[7,1.2]]}
         ]
        }]
      }
     ]
    }]}
 }
]
graphs [
 EventGraph {nodes [
   n1 = ToggleItemDescription!() -> (OutputDelegate, then, show:bool=~) {is=title K2Node_CustomEvent}
   > n2 = Branch(Condition=@n1.show)
   > StopAnimation(InAnimation=Get Anim_HideItemDetails() -> (Anim_HideItemDetails:Widget Animation Object Reference))
   n3 = Get Anim_ShowItemDetails() -> (Anim_ShowItemDetails:Widget Animation Object Reference)
   > n4 = IsAnimationPlayingForward!(execute, self:User Widget Object Reference, InAnimation:Widget Animation Object Reference=@n3) -> (then, ReturnValue:bool)
   n5 = GetOwningPlayer()
   > Branch(Condition=OR Boolean(A=@n4, B=IsUsingTouch(PlayerController=@n5)))
   > StopAnimation(InAnimation=@n3)
   > PlayAnimationForward(InAnimation=@n3)
   n6 = Get Anim_HideItemDetails()
   n7 = Branch(execute=@n2.else, Condition=IsAnimationPlaying(InAnimation=@n6))
   > n8 = IsAnimationPlayingForward(InAnimation=@n6)
   > n9 = Branch(Condition=@n8)
   n10 = StopAnimation(execute=@n9.else, InAnimation=@n6)
   PlayAnimationForward(execute=[@n10.then, @n7.else], InAnimation=@n6)
   n11 = Event PreConstruct()
   > SetVisibility(
    self=Get AccessRestrictionWarning() -> (AccessRestrictionWarning:WBP UIKit Info Box Object Reference)
    InVisibility=Select(Option 0=Collapsed, Option 1=HitTestInvisible, Index=@n11.IsDesignTime)
   )
   Event Construct()
   > ClearMaxDesiredHeight(self:Size Box Object Reference=Get SizeBoxTitle!(self:WBP UIKit Item Description Base Object Reference=Get WBP_UIKit_DescriptionWithSubtitle() -> (WBP_UIKit_DescriptionWithSubtitle:WBP UIKit Description With Subtitle Object Reference)) -> (SizeBoxTitle:Size Box Object Reference))
   Event OnInitialized()
   > SetCompactUIStyle() {is=title K2Node_CallFunction}
  ]}
 GetAccessRestrictionIconVis {nodes [
   GetAccessRestrictionIconVis!() -> (then, Restriction:EItemAccessRestriction Enum=UnsatisfiedExclusiveItem, IsEquippedInOtherCategory:bool=~) {is=title K2Node_FunctionEntry}
   > Return Node!(
    execute
    ShowIcon:bool=Select!(
     UnsatisfiedExclusiveItem:bool=false
     LockedEmote:bool=false
     CosmeticRestrictions:bool=false
     AllArchived:bool=false
     RatingGated:bool=true
     UnsupportedProduct:bool=false
     AdditionalInfo:bool=false
     AdditionalInfoWarning:bool=true
     Unknown:bool=Get IsEquippedInOtherCategory() -> (IsEquippedInOtherCategory:bool)
     Index:EItemAccessRestriction Enum=Get Restriction() -> (Restriction:EItemAccessRestriction Enum=UnsatisfiedExclusiveItem)
    ) -> (ReturnValue:bool)
   )
  ]}
 GetAccessRestrictionWarningVis {nodes [
   GetAccessRestrictionWarningVis!() -> (then, Restriction:EItemAccessRestriction Enum=UnsatisfiedExclusiveItem, IsEquippedInOtherCategory:bool=~) {is=title K2Node_FunctionEntry}
   n1 = "Equal (Enum)"(A:EItemAccessRestriction Enum=Get Restriction() -> (Restriction=UnsatisfiedExclusiveItem), B:EItemAccessRestriction Enum=Unknown)
   n2 = NOT Boolean(A=@n1)
   n3 = AND Boolean(A=Get IsEquippedInOtherCategory(), B=@n1)
   > Return Node!(
    execute
    Visibility:ESlateVisibility Enum="To Visibility (Boolean)"!(
     bIsVisible:bool=OR Boolean(A=@n3, B=@n2)
     TrueVisibility:ESlateVisibility Enum=HitTestInvisible
     FalseVisibility:ESlateVisibility Enum=Collapsed
    ) -> (ReturnValue:ESlateVisibility Enum=Visible)
   )
  ]}
 GetEquippedInOtherCategoryMessage {nodes [
   GetEquippedInOtherCategoryMessage() {is=title K2Node_FunctionEntry}
   n1 = Get FortLockerEquipUtilityVM() -> (FortLockerEquipUtilityVM:Fort Locker Equip Utility VM Object Reference)
   > Return Node!(
    execute
    Result:Text=Format Text!(
     Format="LOCTEXT(\"This item is also equipped in {Category}.\")"
     Category:Text=GetItemEquippedCategoryNames!(self:Fort Locker Equip Utility VM Object Reference=@n1) -> (ReturnValue:Text)
    ) -> (Result)
   )
  ]}
 GetRestrictionMessage {nodes [
   GetRestrictionMessage!() -> (then, RestrictionMessage:Text, IsEquippedinOtherCategory:bool=~, Restriction:EItemAccessRestriction Enum=UnsatisfiedExclusiveItem) {is=title K2Node_FunctionEntry}
   n1 = "Equal (Enum)"(A=Get Restriction() -> (Restriction=UnsatisfiedExclusiveItem), B=Unknown)
   n2 = Get IsEquippedinOtherCategory() -> (IsEquippedinOtherCategory:bool)
   > Return Node(Result=SelectText(
     A=GetEquippedInOtherCategoryMessage!(self) -> (Result:Text) {is=title K2Node_CallFunction}
     B=Get RestrictionMessage() -> (RestrictionMessage:Text)
     bPickA=AND Boolean(A=@n2, B=@n1)
    ))
  ]}
 GetWarningIcon {nodes [
   GetWarningIcon!() -> (then, HasAccessRestriction:bool=~, MoreDetailedFigureIncoming:bool=~, IsEquippedInOtherCategory:bool=~) {is=title K2Node_FunctionEntry}
   n1 = NOT Boolean(A=Get HasAccessRestriction() -> (HasAccessRestriction:bool))
   > n2 = Branch(Condition=AND Boolean(A=Get IsEquippedInOtherCategory(), B=@n1))
   > Return Node!(execute, Material:Material Interface Object Reference=/FortUILibrary/Textures/MaterialInstances/MI_UIKit_Icon_LobbyImage)
   n3 = Branch(execute=@n2.else, Condition=AND Boolean(A=@n1, B=Get MoreDetailedFigureIncoming() -> (MoreDetailedFigureIncoming:bool)))
   > Return Node(Material=/Game/Athena/UI/Frontend/CosmeticGating/Materials/MI_ItemJunoStyle)
   Return Node(execute=@n3.else, Material=/Game/Athena/UI/Frontend/CosmeticGating/Materials/MI_UI_CosmeticGate_WarningIcon)
  ]}
 HideItemDescription {nodes [
   HideItemDescription() {is=title K2Node_FunctionEntry}
   > Clear and Invalidate Timer by Handle(Handle=Get ItemDescriptionHideTimer() -> (ItemDescriptionHideTimer:Timer Handle Structure))
   > n1 = Set Timer by Event(Delegate=Create Event(), Time=Get ItemDetailsHideTime() -> (ItemDetailsHideTime:float), bLooping=false, bMaxOncePerFrame=true)
   > Set ItemDescriptionHideTimer(ItemDescriptionHideTimer:Timer Handle Structure=@n1) -> (Output_Get:Timer Handle Structure)
   n2 = Get ItemDetails_VerticalBox() -> (ItemDetails_VerticalBox:Vertical Box Object Reference)
   > Branch(Condition="float > float"(A=GetRenderOpacity!(self:Widget Object Reference=@n2) -> (ReturnValue:float), B=0.1))
   > ToggleItemDescription!(execute, self, show:bool=false) -> (then) {is=title K2Node_CallFunction}
  ]}
 SetCompactUIStyle {nodes [
   SetCompactUIStyle() {is=title K2Node_FunctionEntry}
   > Branch(Condition=IsMobileGame())
   n1 = SlotAsGridSlot(Widget=Get ItemDetails_VerticalBox())
   n2 = Get!(self:Grid Slot Object Reference=@n1) -> (Padding_Left:float=~, Padding_Top:float=~, Padding_Right:float=~, Padding_Bottom:float=~)
   > SetPadding!(
    execute
    self:Grid Slot Object Reference=@n1
    InPadding_Left:float=@n2
    InPadding_Top:float=@n2.Padding_Top
    InPadding_Right:float=@n2.Padding_Right
    InPadding_Bottom:float=104
   ) -> (then)
   > Set BottomSlotRow!(execute, BottomSlotRow:int=5, self:WBP UIKit Item Description Base Object Reference=Get WBP_UIKit_DescriptionWithSubtitle()) -> (then, Output_Get:int=0)
   > SetPadding!(execute, self:User Widget Object Reference=Get AccessRestrictionWarning(), InPadding:Margin Structure) -> (then)
  ]}
 ShowItemDescription {nodes [
   ShowItemDescription() {is=title K2Node_FunctionEntry}
   > Clear and Invalidate Timer by Handle(Handle=Get ItemDescriptionHideTimer())
   n1 = Get ItemDetails_VerticalBox()
   > Branch(Condition="float < float"(A=GetRenderOpacity(self=@n1), B=0.1))
   > ToggleItemDescription(show=true)
  ]}
 ToggleShowDescriptionWithSubtitle {nodes [
   n1 = ToggleShowDescriptionWithSubtitle!() -> (then, Show:bool=~) {is=title K2Node_FunctionEntry}
   > SetVisibility(
    self=Get WBP_UIKit_DescriptionWithSubtitle()
    InVisibility=Select!(Option 0:ESlateVisibility Enum=Collapsed, Option 1:ESlateVisibility Enum=SelfHitTestInvisible, Index:bool=@n1) -> (ReturnValue:ESlateVisibility Enum)
   )
  ]}
]
variables [
 ItemDescriptionHideTimer {type=TimerHandle}
 ItemDetailsHideTime {type=number, default=2}
 RestrictionTempo {type=EItemAccessRestriction, default=UnsatisfiedExclusiveItem}
]
widgets {
 info {inheritedWidgetCount=0, namedSlotCount=1, parentClass=&UserWidget, rootWidgetClass=&VerticalBox, widgetCount=3}
 widgets [
  ItemDetails_VerticalBox {widgetClassPath=&VerticalBox, bIsVariable=true, widget=ItemDetails_VerticalBox}
  WBP_UIKit_DescriptionWithSubtitle {
   widgetClassPath=/FortUILibrary/UIKit/Items/WBP_UIKit_DescriptionWithSubtitle
   bIsVariable=true
   parent=ItemDetails_VerticalBox
   widget=WBP_UIKit_DescriptionWithSubtitle
   properties {
    bodyTextRow=3
    bottomSlotRow=5
    compatibilityRow=4
    hideBodyText=false
    titleFont {
     bForceMonospaced=false
     bMaterialIsStencil=false
     fontMaterial=null
     fontObject=&HeadingNow
     letterSpacing=0
     monospacedWidth=1
     size=30.75
     skewAmount=0
     typefaceFontName=86 Bold Italic
     outlineSettings {
      bApplyOutlineToDropShadows=false
      bMiteredCorners=false
      bSeparateFillAlpha=false
      outlineMaterial=null
      outlineSize=0
      outlineColor=<rgba 0 0 0 1>
     }
    }
   }
   slot {class=VerticalBoxSlot}
  }
  AccessRestrictionWarning {
   widgetClassPath=/FortUILibrary/UIKit/InfoBox/WBP_UIKit_InfoBox
   bIsVariable=true
   namedSlotHost=WBP_UIKit_DescriptionWithSubtitle
   widget=AccessRestrictionWarning
   properties {showIcon=false, iconSize=<40 40>}
  }
 ]
}
