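# Locker item details screen (CommonActivatableWidget) showing an item description, a swappable custom-details slot and an item preview with Camera and Back prompts.
# Activation plays ScreenIn, activates and refreshes the preview, and disables text chat; deactivation re-enables chat outside zones.
# Back reverses the animation, then deactivates, also popping the content widget in Campaign.
# SetCustomItemDetails fills the slot from a widget class, and PreConstruct adjusts left-column padding for Arabic.
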WidgetBlueprint /Locker/MPL/WBP_MPLocker_ItemDetails
animations [ScreenIn {
  binding_count=1
  length_seconds=0.5
  object=&self:ScreenIn
  sequence {bindings [VerticalBox_LeftSideInfo {
     id=#1
     possessed_class=VerticalBox
     tracks [
      {
       type=MovieScene2DTransformTrack
       sections [{
         type=MovieScene2DTransformSection
         end_frame=10
         start_frame=0
         channels [
          Translation.X {type=MovieSceneScriptingFloatChannel, keys [[0,100],[10,0]]}
          Translation.Y {type=MovieSceneScriptingFloatChannel, const [0,1]}
          Angle {type=MovieSceneScriptingFloatChannel, const [0,1]}
          Scale.X {type=MovieSceneScriptingFloatChannel, const [1,1]}
          Scale.Y {type=MovieSceneScriptingFloatChannel, const [1,1]}
          Shear.X {type=MovieSceneScriptingFloatChannel, const [0,1]}
          Shear.Y {type=MovieSceneScriptingFloatChannel, const [0,1]}
         ]
        }]
      }
      {
       type=MovieSceneFloatTrack
       display_name=Render Opacity
       sections [{type=MovieSceneFloatSection, end_frame=10, start_frame=0, channels [None {type=MovieSceneScriptingFloatChannel, keys [[0,0],[10,1]]}]}]
      }
     ]
    }]}
 }]
graphs [
 BP_GetDesiredInputConfig {nodes [
   Get Desired Input Config() {is=title K2Node_FunctionEntry}
   > Return Node!(
    execute
    ReturnValue:UIInput Config Structure=Make UIInput Config!(
     bIgnoreMoveInput:bool=False
     bIgnoreLookInput:bool=False
     InputMode:ECommonInputMode Enum=All
     MouseCaptureMode:EMouseCaptureMode Enum=CaptureDuringMouseDown
     MouseLockMode:EMouseLockMode Enum=DoNotLock
     bHideCursorDuringViewportCapture:bool=True
    ) -> (UIInputConfig:UIInput Config Structure) {is=title K2Node_MakeStruct}
   )
  ]}
 EventGraph {nodes [
   "On Clicked (Button_Back)"!() -> (OutputDelegate, then, Button:Common Button Base Object Reference) {is=title K2Node_ComponentBoundEvent}
   n1 = Get ScreenIn() -> (ScreenIn:Widget Animation Object Reference)
   > n2 = Branch(Condition=IsAnimationPlaying(InAnimation=@n1))
   n3 = "Play Animation with Finished event (legacy)"!(
    execute:Exec=@n2.else
    Widget:User Widget Object Reference=Self-Reference()
    InAnimation:Widget Animation Object Reference=@n1
    StartAtTime:float=0
    NumLoopsToPlay:int=1
    PlayMode:EUMGSequencePlayMode Enum=Reverse
    PlaybackSpeed:float=1
   ) -> (then:Exec, Result:UMGSequence Player Object Reference, Finished:Exec) {is=title K2Node_PlayAnimation}
   DeactivateWidget(execute=@n3.Finished)
   > Branch(Condition=IsInSubGame!(SubGame:ESubGame Enum=Campaign) -> (ReturnValue:bool))
   > PopContentWidget!(execute) -> (then)
   Event PreConstruct()
   > SetRTLAlignment()
   Event On Activated()
   > PlayAnimation(InAnimation=Get ScreenIn())
   > ActivateWidget(self=Get FortItemPreviewWidget() -> (FortItemPreviewWidget:WBP Item Preview Object Reference))
   > Broadcast Event!(
    execute
    self:Gameplay Event Router Component Object Reference=GetGlobalEventRouter()
    EventData:BPS Set Text Chat Active State Structure=Make BPS Set Text Chat Active State!(bSetActive_5_#2:bool=False) -> (BPS_SetTextChatActiveState:BPS Set Text Chat Active State Structure) {is=title K2Node_MakeStruct}
    EventContext:Object Reference=GetOwningPlayer()
   ) -> (then, ReturnValue:bool)
   > DelayUntilNextTick!(execute) -> (then)
   > UpdatePreview(self:Fort Item Preview Widget Object Reference=Get FortItemPreviewWidget())
   Event On Deactivated()
   > DeactivateWidget(self=Get FortItemPreviewWidget())
   > n4 = Branch(Condition=IsInZone!(self:Fort Global UIContext Object Reference=Get FortGlobalUIContext() -> (ReturnValue:Fort Global UIContext Object Reference) {is=title K2Node_GetSubsystem}) -> (ReturnValue:bool))
   Broadcast Event(
    execute=@n4.else
    self=GetGlobalEventRouter()
    EventData=Make BPS Set Text Chat Active State(bSetActive_5_#2=true)
    EventContext=GetOwningPlayer()
   )
  ]}
 SetCustomItemDetails {nodes [
   n1 = SetCustomItemDetails!() -> (then, CustomItemDetails:User Widget Class Reference) {is=title K2Node_FunctionEntry}
   > Set CustomItemDetails(CustomItemDetails:User Widget Class Reference=@n1) -> (Output_Get:User Widget Class Reference)
   > ClearChildren(self=Get NamedSlot_CustomDetails() -> (NamedSlot_CustomDetails:Named Slot Object Reference))
   > n2 = Create Widget!(
    execute:Exec
    Class:User Widget Class Reference=Get CustomItemDetails() -> (CustomItemDetails:User Widget Class Reference)
    OwningPlayer:Player Controller Object Reference=GetOwningPlayer()
   ) -> (then:Exec, ReturnValue:User Widget Object Reference)
   > AddChild(self=Get NamedSlot_CustomDetails(), Content=@n2)
  ]}
 SetRTLAlignment {nodes [
   SetRTLAlignment() {is=title K2Node_FunctionEntry}
   n1 = GetCurrentCulture!() -> (ReturnValue:String)
   n2 = Get!(self:Overlay Slot Object Reference=SlotAsOverlaySlot(Widget=Get VerticalBox_LeftSideInfo() -> (VerticalBox_LeftSideInfo:Vertical Box Object Reference))) -> (Padding_Left:float=~, Padding_Top:float=~, Padding_Right:float=~, Padding_Bottom:float=~)
   > Set!(
    execute
    Padding_Left:float=@n2
    Padding_Top:float=@n2.Padding_Top
    Padding_Right:float=SelectFloat(A=1200, B=@n2.Padding_Right, bPickA="Equal Exactly (String)"(A=@n1, B=ar))
    Padding_Bottom:float=@n2.Padding_Bottom
    self:Overlay Slot Object Reference=SlotAsOverlaySlot(Widget=Get VerticalBox_LeftSideInfo())
   ) -> (then, Output_Get:Margin Structure)
  ]}
]
variables [CustomItemDetails {type=Class, default=null}]
widgets {
 info {
  inheritedWidgetCount=0
  namedSlotCount=1
  parentClass=&CommonActivatableWidget
  rootWidgetClass=&SafeZone
  widgetCount=16
 }
 widgets [
  SafeZone_22 {widgetClassPath=&SafeZone, bIsVariable=true, widget=SafeZone_22}
  SizeBox_0 {
   widgetClassPath=&SizeBox
   parent=SafeZone_22
   widget=SizeBox_0
   properties {bOverride_WidthOverride=true, widthOverride=1920}
   slot {class=SafeZoneSlot, properties {hAlign=HAlign_Center}}
  }
  Overlay_27 {widgetClassPath=&Overlay, parent=SizeBox_0, widget=Overlay_27, slot {class=SizeBoxSlot}}
  WBP_CaptureForPostBufferUpdate {
   widgetClassPath=&WBP_CaptureForPostBufferUpdate
   bIsVariable=true
   parent=Overlay_27
   slot {class=OverlaySlot, properties {horizontalAlignment=HAlign_Fill, verticalAlignment=VAlign_Fill}}
  }
  VerticalBox_LeftSideInfo {
   widgetClassPath=&VerticalBox
   bIsVariable=true
   parent=Overlay_27
   widget=VerticalBox_LeftSideInfo
   properties {flowDirectionPreference=Culture}
   slot {class=OverlaySlot, properties {horizontalAlignment=HAlign_Fill, verticalAlignment=VAlign_Fill, padding=<ltrb 64 64 64 32>}}
  }
  SizeBox_Details {
   widgetClassPath=&SizeBox
   bIsVariable=true
   parent=VerticalBox_LeftSideInfo
   widget=SizeBox_Details
   properties {
    bOverride_MinDesiredHeight=true
    bOverride_WidthOverride=true
    heightOverride=308
    maxDesiredWidth=800
    minDesiredHeight=308
    widthOverride=840
   }
   slot {class=VerticalBoxSlot, properties {horizontalAlignment=HAlign_Left, padding=<ltrb 0 0 0 16>}}
  }
  WBP_UIKit_ItemDescription_MPL {
   widgetClassPath=/FortniteCoreUI/Customization/MPL/WBP_UIKit_ItemDescription_MPL
   bIsVariable=true
   parent=SizeBox_Details
   widget=WBP_UIKit_ItemDescription_MPL
   properties {bodyTextRow=3, compatibilityRow=2, flowDirectionPreference=LeftToRight, compatibilityPadding=<ltrb 0 8 0 0>, tagPadding=<ltrb 0 0 0 8>}
   slot {class=SizeBoxSlot}
  }
  NamedSlot_CustomDetails {
   widgetClassPath=&NamedSlot
   bIsVariable=true
   parent=VerticalBox_LeftSideInfo
   widget=NamedSlot_CustomDetails
   slot {class=VerticalBoxSlot, properties {horizontalAlignment=HAlign_Left, padding=<ltrb 0 0 0 50>, size {sizeRule=Fill, value=1}}}
  }
  Border_CameraFrame {
   widgetClassPath=&Border
   bIsVariable=true
   parent=Overlay_27
   widget=Border_CameraFrame
   properties {
    visibility=HitTestInvisible
    background {
     bIsDynamicallyLoaded=false
     drawAs=NoDrawType
     imageType=NoImage
     mirroring=NoMirror
     resourceName=None
     resourceObject=null
     tiling=NoTile
     imageSize=<32 32>
     margin=<ltrb 0 0 0 0>
     outlineSettings {
      bUseBrushTransparency=false
      roundingType=HalfHeightRadius
      width=0
      color {colorUseRule=UseColor_Specified, specifiedColor=<rgba 0 0 0 0>}
      cornerRadii=<0 0 0 0>
     }
     tintColor {colorUseRule=UseColor_Specified, specifiedColor=<rgba 1 1 1 1>}
     uVRegion {bIsValid=false, max=<0 0>, min=<0 0>}
    }
    padding=<ltrb 0 0 0 0>
   }
   slot {class=OverlaySlot, properties {horizontalAlignment=HAlign_Fill, verticalAlignment=VAlign_Fill, padding=<ltrb 850 150 40 100>}}
  }
  Widget_CameraFrame {
   widgetClassPath=&HorizontalBox
   bIsVariable=true
   parent=Border_CameraFrame
   widget=Widget_CameraFrame
   slot {class=BorderSlot, properties {padding=<ltrb 0 0 0 0>}}
  }
  FortItemPreviewWidget {
   widgetClassPath=/FortniteCoreUI/Frontend/ItemPreview/WBP_ItemPreview
   bIsVariable=true
   parent=Widget_CameraFrame
   widget=FortItemPreviewWidget
   properties {
    bPreviewActionsEnabled=false
    bSupportsActivationFocus=false
    button_Camera=&self:WidgetTree.Button_Camera
    itemViewObject=&self:WidgetTree.FortItemPreviewWidget.ItemViewObject
   }
   slot {class=HorizontalBoxSlot, properties {padding=<ltrb 0 0 0 96>, size {sizeRule=Fill, value=1}}}
  }
  WBP_ActionBarContainer {
   widgetClassPath=/Game/UI/Frontend/BottomBar/WBP_ActionBarContainer
   bIsVariable=true
   parent=Overlay_27
   widget=WBP_ActionBarContainer
   slot {class=OverlaySlot, properties {horizontalAlignment=HAlign_Right, verticalAlignment=VAlign_Bottom, padding=<ltrb 0 0 125 32>}}
  }
  HorizontalBox_ButtonPrompts {widgetClassPath=&HorizontalBox, namedSlotHost=WBP_ActionBarContainer, widget=HorizontalBox_ButtonPrompts}
  Button_Camera {
   widgetClassPath=/FortUILibrary/UIKit/Buttons/WBP_UIKit_Button_Quiet
   bIsVariable=true
   parent=HorizontalBox_ButtonPrompts
   widget=Button_Camera
   properties {bIsFocusable=false, text=Camera, triggeringInputAction {dataTable=&InputActionDataTable, rowName=UnifiedRotateItem}}
   slot {class=HorizontalBoxSlot, properties {padding=<ltrb 0 0 24 0>}}
  }
  Button_Back {
   widgetClassPath=&WBP_UIKit_Button_Quiet
   bIsVariable=true
   parent=HorizontalBox_ButtonPrompts
   widget=Button_Back
   properties {bIsFocusable=false, text=Back, triggeringInputAction {dataTable=&InputActionDataTable, rowName=GenericBack}}
   slot {class=HorizontalBoxSlot}
  }
  WBP_PlayerInfo {
   widgetClassPath=/Game/UI/Frontend/ItemShop/MultiProduct/WBP_PlayerInfo
   bIsVariable=true
   parent=Overlay_27
   widget=WBP_PlayerInfo
   slot {class=OverlaySlot, properties {horizontalAlignment=HAlign_Right, padding=<ltrb 80 40 80 24>}}
  }
 ]
}
