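# Locker item selection screen (MPL) where players browse, preview, equip and customize cosmetics.
# WBP_ItemCollectionList events drive bottom-bar buttons and details, while FortItemSelectionVM supplies the selected item and category alongside FortItemPreviewWidget.
# Intro and SubMenu slide-fade animations lock input and gate closing or pushing edit styles and item details.
# Category config supplies plugin customize menus and banners; closing customization refreshes the preview and equips items no longer needing customization.
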WidgetBlueprint /Locker/MPL/WBP_MPL_ItemSelection
animations [
 Intro {
  binding_count=2
  length_seconds=0.25
  object=&self:Intro
  sequence {bindings [
    WBP_ItemSelection_ItemDetails {
     id=#1
     possessed_class=WBP_ItemSelection_ItemDetails_C
     tracks [
      {
       type=MovieSceneFloatTrack
       display_name=Render Opacity
       sections [{type=MovieSceneFloatSection, end_frame=5, start_frame=0, channels [None {type=MovieSceneScriptingFloatChannel, keys [[0,0],[5,1]]}]}]
      }
      {
       type=MovieScene2DTransformTrack
       sections [{
         type=MovieScene2DTransformSection
         end_frame=5
         start_frame=0
         channels [
          Translation.X {type=MovieSceneScriptingFloatChannel, keys [[0,200],[5,0]]}
          Translation.Y {type=MovieSceneScriptingFloatChannel}
          Angle {type=MovieSceneScriptingFloatChannel}
          Scale.X {type=MovieSceneScriptingFloatChannel}
          Scale.Y {type=MovieSceneScriptingFloatChannel}
          Shear.X {type=MovieSceneScriptingFloatChannel}
          Shear.Y {type=MovieSceneScriptingFloatChannel}
         ]
        }]
      }
     ]
    }
    WBP_ItemCollectionList {
     id=#2
     possessed_class=WBP_ItemCollectionList_C
     tracks [
      {
       type=MovieSceneFloatTrack
       display_name=Render Opacity
       sections [{type=MovieSceneFloatSection, end_frame=5, start_frame=0, channels [None {type=MovieSceneScriptingFloatChannel, keys [[0,0],[5,1]]}]}]
      }
      {
       type=MovieScene2DTransformTrack
       sections [{
         type=MovieScene2DTransformSection
         end_frame=5
         start_frame=0
         channels [
          Translation.X {type=MovieSceneScriptingFloatChannel, keys [[0,200],[5,0]]}
          Translation.Y {type=MovieSceneScriptingFloatChannel}
          Angle {type=MovieSceneScriptingFloatChannel}
          Scale.X {type=MovieSceneScriptingFloatChannel}
          Scale.Y {type=MovieSceneScriptingFloatChannel}
          Shear.X {type=MovieSceneScriptingFloatChannel}
          Shear.Y {type=MovieSceneScriptingFloatChannel}
         ]
        }]
      }
     ]
    }
   ]}
 }
 SubMenu {
  binding_count=2
  length_seconds=0.25
  object=&self:SubMenu
  sequence {bindings [
    WBP_ItemSelection_ItemDetails {
     id=#1
     possessed_class=WBP_ItemSelection_ItemDetails_C
     tracks [
      {
       type=MovieSceneFloatTrack
       display_name=Render Opacity
       sections [{type=MovieSceneFloatSection, end_frame=5, start_frame=0, channels [None {type=MovieSceneScriptingFloatChannel, keys [[0,1],[5,0]]}]}]
      }
      {
       type=MovieScene2DTransformTrack
       sections [{
         type=MovieScene2DTransformSection
         end_frame=5
         start_frame=0
         channels [
          Translation.X {type=MovieSceneScriptingFloatChannel, keys [[0,0],[5,-200]]}
          Translation.Y {type=MovieSceneScriptingFloatChannel}
          Angle {type=MovieSceneScriptingFloatChannel}
          Scale.X {type=MovieSceneScriptingFloatChannel}
          Scale.Y {type=MovieSceneScriptingFloatChannel}
          Shear.X {type=MovieSceneScriptingFloatChannel}
          Shear.Y {type=MovieSceneScriptingFloatChannel}
         ]
        }]
      }
     ]
    }
    WBP_ItemCollectionList {
     id=#2
     possessed_class=WBP_ItemCollectionList_C
     tracks [
      {
       type=MovieSceneFloatTrack
       display_name=Render Opacity
       sections [{type=MovieSceneFloatSection, end_frame=5, start_frame=0, channels [None {type=MovieSceneScriptingFloatChannel, keys [[0,1],[5,0]]}]}]
      }
      {
       type=MovieScene2DTransformTrack
       sections [{
         type=MovieScene2DTransformSection
         end_frame=5
         start_frame=0
         channels [
          Translation.X {type=MovieSceneScriptingFloatChannel, keys [[0,0],[5,-200]]}
          Translation.Y {type=MovieSceneScriptingFloatChannel}
          Angle {type=MovieSceneScriptingFloatChannel}
          Scale.X {type=MovieSceneScriptingFloatChannel}
          Scale.Y {type=MovieSceneScriptingFloatChannel}
          Shear.X {type=MovieSceneScriptingFloatChannel}
          Shear.Y {type=MovieSceneScriptingFloatChannel}
         ]
        }]
      }
     ]
    }
   ]}
 }
]
graphs [
 Allow Toggle Preview Override Changed {nodes [
   Allow Toggle Preview Override Changed!() -> (then, Object:Object Reference, Field:Field Notification Id Structure) {is=title K2Node_FunctionEntry}
   n1 = Get FortItemPreviewWidget() -> (FortItemPreviewWidget:WBP Item Preview Object Reference)
   > SetVisibility(
    self=Get TogglePreview() -> (TogglePreview:WBP UIKit Button Quiet Object Reference)
    InVisibility="To Visibility (Boolean)"!(
     bIsVisible:bool=AllowToggleVariantOverrideOnCurrentItem!(self:Fort Item Preview Widget Object Reference=@n1) -> (ReturnValue:bool)
     TrueVisibility:ESlateVisibility Enum=SelfHitTestInvisible
     FalseVisibility:ESlateVisibility Enum=Collapsed
    ) -> (ReturnValue:ESlateVisibility Enum=Visible)
   )
  ]}
 BP_GetDesiredInputConfig {nodes [
   Get Desired Input Config() {is=title K2Node_FunctionEntry}
   > Return Node!(
    execute
    ReturnValue:UIInput Config Structure=Make UIInput Config!(
     bIgnoreMoveInput:bool=False
     bIgnoreLookInput:bool=False
     InputMode:ECommonInputMode Enum=All
     MouseCaptureMode:EMouseCaptureMode Enum=CaptureDuringMouseDown
     MouseLockMode:EMouseLockMode Enum=DoNotLock
     bHideCursorDuringViewportCapture:bool=True
    ) -> (UIInputConfig:UIInput Config Structure) {is=title K2Node_MakeStruct}
   )
  ]}
 EventGraph {nodes [
   n1 = "Update Details Button (WBP_ItemCollectionList)"!() -> (OutputDelegate, then, Show:bool, Enabled:bool) {is=title K2Node_ComponentBoundEvent}
   n2 = Get Button_Details() -> (Button_Details:WBP UIKit Button Quiet Object Reference)
   > UpdateBottomBarButtonIsVisible!(execute, self, Button:Widget Object Reference=@n2, IsVisible:bool=@n1.Show) -> (then)
   > UpdateBottomBarButtonIsEnabled!(execute, self, Button:Widget Object Reference=@n2, Enabled:bool=@n1.Enabled) -> (then)
   n3 = "Trigger Activated Event (WBP_ItemCollectionList)"!() -> (OutputDelegate, then, LastActiveScreen:ELockerScreenList Enum=NewEnumerator0) {is=title K2Node_ComponentBoundEvent}
   > OnWidgetActivatedEvent(self:WBP Item Collection List Object Reference=Get WBP_ItemCollectionList() -> (WBP_ItemCollectionList:WBP Item Collection List Object Reference))
   > n4 = Branch(Condition="Equal (Enum)"(A:ELockerScreenList Enum=@n3.LastActiveScreen, B:ELockerScreenList Enum=NewEnumerator4))
   > PlayAnimationReverse(InAnimation=Get SubMenu() -> (SubMenu:Widget Animation Object Reference))
   PlayAnimationForward(execute=@n4.else, InAnimation=Get Intro())
   n5 = "Update Equip Button Is Visible (WBP_ItemCollectionList)"!() -> (OutputDelegate, then, isVisible:bool) {is=title K2Node_ComponentBoundEvent}
   n6 = Get Button_Equip() -> (Button_Equip:WBP UIKit Button Quiet Object Reference)
   > UpdateBottomBarButtonIsVisible(Button=@n6, IsVisible=@n5.isVisible)
   n7 = "Update Equip Button Is Enabled (WBP_ItemCollectionList)"!() -> (OutputDelegate, then, IsEnabled:bool) {is=title K2Node_ComponentBoundEvent}
   > UpdateBottomBarButtonIsEnabled(Button=@n6, Enabled=@n7.IsEnabled)
   n8 = "Update Edit Button Is Enable (WBP_ItemCollectionList)"!() -> (OutputDelegate, then, IsEnabled:bool) {is=title K2Node_ComponentBoundEvent}
   n9 = Get Button_EditStyles() -> (Button_EditStyles:WBP UIKit Button Quiet Object Reference)
   > UpdateBottomBarButtonIsEnabled(Button=@n9, Enabled=@n8.IsEnabled)
   n10 = "On Clicked (Button_Use)"!() -> (OutputDelegate, then, Button:Common Button Base Object Reference) {is=title K2Node_ComponentBoundEvent}
   > TriggerFlowRiderInteraction!(execute, self, Element:Widget Object Reference=@n10.Button, Value:String=Use Search) -> (then)
   > FocusSearchField!(execute, self:WBP Item Collection List Object Reference=Get WBP_ItemCollectionList()) -> (then, bPassThrough:bool)
   n11 = "Update Options Button Is Visible (WBP_ItemCollectionList)"!() -> (OutputDelegate, then, IsVisible:bool) {is=title K2Node_ComponentBoundEvent}
   n12 = Get WBP_UIKit_Button_Options() -> (WBP_UIKit_Button_Options:WBP UIKit Button Quiet Object Reference)
   > UpdateBottomBarButtonIsVisible(Button=@n12, IsVisible=@n11.IsVisible)
   n13 = "Update Clear Button Is Visible (WBP_ItemCollectionList)"!() -> (OutputDelegate, then, IsVisible:bool) {is=title K2Node_ComponentBoundEvent}
   n14 = Get Button_ClearSearch() -> (Button_ClearSearch:WBP UIKit Button Quiet Object Reference)
   > UpdateBottomBarButtonIsVisible(Button=@n14, IsVisible=@n13.IsVisible)
   n15 = "Update Clear Button Is Enabled (WBP_ItemCollectionList)"!() -> (OutputDelegate, then, IsEnabled:bool) {is=title K2Node_ComponentBoundEvent}
   > UpdateBottomBarButtonIsEnabled(Button=@n14, Enabled=@n15.IsEnabled)
   n16 = "Update Edit Button Is Visible (WBP_ItemCollectionList)"!() -> (OutputDelegate, then, IsVisible:bool) {is=title K2Node_ComponentBoundEvent}
   > UpdateBottomBarButtonIsVisible(Button=@n9, IsVisible=@n16.IsVisible)
   n17 = "Update Use Button Is Enabled (WBP_ItemCollectionList)"!() -> (OutputDelegate, then, IsEnabled:bool) {is=title K2Node_ComponentBoundEvent}
   n18 = Get Button_Use() -> (Button_Use:WBP UIKit Button Quiet Object Reference)
   > UpdateBottomBarButtonIsEnabled(Button=@n18, Enabled=@n17.IsEnabled)
   n19 = "Update Use Button Is Visible (WBP_ItemCollectionList)"!() -> (OutputDelegate, then, IsVisible:bool) {is=title K2Node_ComponentBoundEvent}
   > UpdateBottomBarButtonIsVisible(Button=@n18, IsVisible=@n19.IsVisible)
   n20 = "Sub Category Tab Selected (WBP_ItemCollectionList)"!() -> (OutputDelegate, then, TabNameID:Name=None) {is=title K2Node_ComponentBoundEvent}
   > HandleSubCategoryTabSelected!(execute, self, TabNameID:Name=@n20.TabNameID) -> (then)
   n21 = "Update Options Button Is Enabled (WBP_ItemCollectionList)"!() -> (OutputDelegate, then, IsEnabled:bool) {is=title K2Node_ComponentBoundEvent}
   > UpdateBottomBarButtonIsEnabled(Button=@n12, Enabled=@n21.IsEnabled)
   n22 = "On Clicked (WBP_UIKit_Button_Back)"!() -> (OutputDelegate, then, Button:Common Button Base Object Reference) {is=title K2Node_ComponentBoundEvent}
   > TriggerFlowRiderInteraction(Element=@n22.Button, Value=Back)
   n23 = Get WBP_ItemCollectionList()
   > ResetTextField(self:WBP Item Collection List Object Reference=@n23)
   > SetAllInteractionsEnabled!(execute, self:WBP Item Collection List Object Reference=@n23, IsInteractionEnabled:bool=false, IncludeSortFilter:bool=true) -> (then)
   > CloseMenu()
   n24 = "On Clicked (Button_ClearSearch)"!() -> (OutputDelegate, then, Button:Common Button Base Object Reference) {is=title K2Node_ComponentBoundEvent}
   > TriggerFlowRiderInteraction(Element=@n24.Button, Value=Clear Search)
   > ResetTextField(self=Get WBP_ItemCollectionList())
   "On Preview Item Shown (FortItemPreviewWidget)"!() -> (OutputDelegate, then, ItemDefinition:Fort Account Item Definition Object Reference) {is=title K2Node_ComponentBoundEvent}
   > PreviewActionBinds(self:WBP Item Collection List Object Reference=Get WBP_ItemCollectionList())
   n25 = "Update Description (WBP_ItemCollectionList)"!() -> (OutputDelegate, then, ShowDescription:bool, IsTimed:bool) {is=title K2Node_ComponentBoundEvent}
   > n26 = Branch(Condition=@n25.IsTimed)
   > n27 = Branch(Condition=@n25.ShowDescription)
   n28 = Get WBP_ItemSelection_ItemDetails() -> (WBP_ItemSelection_ItemDetails:WBP Item Selection Item Details Object Reference)
   > ShowItemDescription(self:WBP Item Selection Item Details Object Reference=@n28)
   HideItemDescription(execute=@n27.else, self:WBP Item Selection Item Details Object Reference=Get WBP_ItemSelection_ItemDetails())
   ToggleItemDescription!(execute=@n26.else, self:WBP Item Selection Item Details Object Reference=@n28, show:bool=@n25.ShowDescription) -> (then)
   "On Clicked (Button_EditStyles)"!() -> (OutputDelegate, then, Button:Common Button Base Object Reference) {is=title K2Node_ComponentBoundEvent}
   > n29 = Branch(Condition=IsCustomizable!() -> (Customizable:bool))
   > CustomizeClicked(self:WBP Item Collection List Object Reference=Get WBP_ItemCollectionList())
   EditStylesClicked(execute=@n29.else, self:WBP Item Collection List Object Reference=Get WBP_ItemCollectionList())
   "Trigger Edit Styles (WBP_ItemCollectionList)"!() -> (OutputDelegate, then, LastActiveScreen:ELockerScreenList Enum=NewEnumerator0) {is=title K2Node_ComponentBoundEvent}
   n30 = Get SubMenu()
   > Branch(Condition=NOT Boolean(A=IsAnimationPlaying(InAnimation=@n30)))
   > Set bIsEnabled!(execute, bIsEnabled:bool=false, self:Widget Object Reference=Get WBP_ItemCollectionList()) -> (then, Output_Get:bool)
   > SetAllInteractionsEnabled!(execute, self:Self Object Reference, IsInteractionEnabled=false) -> (then)
   > n31 = Play Animation with Finished event!(
    execute:Exec
    Widget:User Widget Object Reference=Self-Reference()
    InAnimation:Widget Animation Object Reference=@n30
    StartAtTime:float=0
    NumLoopsToPlay:int=1
    PlayMode:EUMGSequencePlayMode Enum
    PlaybackSpeed:float=1
   ) -> (then:Exec, Result:Widget Animation Handle Structure, Finished:Exec) {is=title K2Node_PlayAnimation2}
   SetCurrentPreviewItem!(execute=@n31.Finished, self:Fort Item Preview Widget Object Reference=Get FortItemPreviewWidget(), ItemVM:Fort Item VM Object Reference) -> (then)
   > PushEditStyles(self:WBP Item Collection List Object Reference=Get WBP_ItemCollectionList())
   n32 = "Manual Update Selected Item (WBP_ItemCollectionList)"!() -> (OutputDelegate, then, SelectedItem:Fort Item VM Object Reference) {is=title K2Node_ComponentBoundEvent}
   > ManualSetSelectedItem!(execute, self, Selected Item:Fort Item VM Object Reference=@n32.SelectedItem) -> (then)
   n33 = "Manual Update Hovered Item (WBP_ItemCollectionList)"!() -> (OutputDelegate, then, HoveredItem:Fort Item VM Object Reference) {is=title K2Node_ComponentBoundEvent}
   > ManualSetHoveredItem!(execute, self, Hovered Item:Fort Item VM Object Reference=@n33.HoveredItem) -> (then)
   "On Preview Actions Updated Event (FortItemPreviewWidget)"!() -> (OutputDelegate, then, PreviewActions:Array of Data Table Row Handle Structures) {is=title K2Node_ComponentBoundEvent}
   n34 = GetEntryWidgetByItem!(
    self:WBP Item Collection List Object Reference=Get WBP_ItemCollectionList()
    ItemObject:Object Reference=Get SelectedItem!(self:Fort Item Selection VM Object Reference=Get FortItemSelectionVM() -> (FortItemSelectionVM:Fort Item Selection VM Object Reference)) -> (SelectedItem:Fort Item VM Object Reference)
   ) -> (Successful:bool, LockerEntryWiget:User Widget Object Reference)
   > Branch(Condition=@n34)
   > BindPreviewActions!(
    execute:Exec
    self:Object Reference=@n34.LockerEntryWiget
    Preview Widget:Fort Item Preview Widget Object Reference=Get FortItemPreviewWidget()
   ) -> (then:Exec) {is=title K2Node_Message}
   "Trigger Customize Menu (WBP_ItemCollectionList)"!() -> (OutputDelegate, then, LastActiveScreen:ELockerScreenList Enum=NewEnumerator0) {is=title K2Node_ComponentBoundEvent}
   > n35 = Get!(execute:Exec, self:Fort Item Selection VM Object Reference=Get FortItemSelectionVM()) -> (then:Exec, else:Exec, SelectedItem:Fort Item VM Object Reference)
   > n36 = Set CachedSelectedItem(CachedSelectedItem:Fort Item VM Object Reference=@n35) -> (Output_Get:Fort Item VM Object Reference)
   > Branch(Condition=Get bRequiresCustomization!(self:Fort Item VM Object Reference=@n36) -> (bRequiresCustomization:bool))
   n37 = TryGetSoftCustomActivatableWidgetClass!(
    self:Fort Locker Category Config Object Reference=GetItemCategory!(self:Fort Item VM Object Reference=@n36) -> (ReturnValue:Fort Locker Category Config Object Reference)
    ViewRequest:Gameplay Tag Structure=(TagName="UI.ItemPreview.PluginBridge.Mimosa.CustomizeMenu")
   ) -> (WidgetClass:Common Activatable Widget Soft Class Reference, ReturnValue:bool)
   > n38 = Branch(Condition=@n37.ReturnValue)
   > n39 = PushPrimaryContentSoftClassAsync!(execute:Exec, SoftWidgetClass:Common Activatable Widget Soft Class Reference=@n37) -> (then:Exec, OnPrimaryContentPushed:Exec, ActivatableWidget:Common Activatable Widget Object Reference) {is=title K2Node_AsyncAction}
   n40 = Set CustomizeMenu(execute=@n39.OnPrimaryContentPushed, CustomizeMenu:Common Activatable Widget Object Reference=@n39) -> (Output_Get:Common Activatable Widget Object Reference)
   > SetViewModelByClass!(
    execute
    Widget:User Widget Object Reference=@n40
    ViewModel:Notify Field Value Changed Interface=Get CachedSelectedItem() -> (CachedSelectedItem:Fort Item VM Object Reference)
   ) -> (then)
   > n41 = Bind Event to On Widget Deactivated(self:Common Activatable Widget Object Reference=@n40, Delegate=Create Event()) {is=title K2Node_AddDelegate}
   Set bIsEnabled(execute=[@n41.then, @n38.else], bIsEnabled=true, self=Get WBP_UIKit_Button_Options())
   > SetAllInteractionsEnabled(IsInteractionEnabled=false)
   n42 = "Trigger Item Details (WBP_ItemCollectionList)"!() -> (OutputDelegate, then, Custom ItemDetailsClass:User Widget Class Reference) {is=title K2Node_ComponentBoundEvent}
   > PushItemDetails!(execute, self, CustomItemDetailsClass:User Widget Class Reference=@n42.Custom ItemDetailsClass) -> (then)
   n43 = "On Clicked (WBP_UIKit_Button_Options)"!() -> (OutputDelegate, then, Button:Common Button Base Object Reference) {is=title K2Node_ComponentBoundEvent}
   > TriggerFlowRiderInteraction(Element=@n43.Button, Value=Options)
   n44 = Get WBP_ItemCollectionList()
   > PreviewActionBinds(self=@n44)
   > ToggleSelectedContext!(
    execute:Exec
    self:Object Reference=Get LockerGroupList!(self:WBP Item Collection List Object Reference=@n44) -> (LockerGroupList:User Widget Object Reference)
   ) -> (then:Exec) {is=title K2Node_Message}
   n45 = "On Clicked (TogglePreview)"!() -> (OutputDelegate, then, Button:Common Button Base Object Reference) {is=title K2Node_ComponentBoundEvent}
   > TriggerFlowRiderInteraction(Element=@n45.Button, Value=Toggle Preview)
   > ToggleVariantOverrideOnCurrentItem(self:Fort Item Preview Widget Object Reference=Get FortItemPreviewWidget())
   n46 = "On Item Rotation Changed (FortItemPreviewWidget)"() {is=title K2Node_ComponentBoundEvent}
   n47 = Branch(execute=[@n46.then, @n48.then], Condition=IsUsingTouch(PlayerController=GetOwningPlayer()))
   HideItemDescription(execute=@n47.else, self=Get WBP_ItemSelection_ItemDetails())
   n48 = "On Item Zoom Level Changed (FortItemPreviewWidget)"() {is=title K2Node_ComponentBoundEvent}
   n49 = "On Clicked (Button_Details)"!() -> (OutputDelegate, then, Button:Common Button Base Object Reference) {is=title K2Node_ComponentBoundEvent}
   > TriggerFlowRiderInteraction(Element=@n49.Button, Value=Details)
   n50 = Get WBP_ItemCollectionList()
   n51 = GetCategoryConfig!(self:WBP Item Collection List Object Reference=@n50) -> (Category Config:Fort Locker Category Config Object Reference)
   > Is Valid(InputObject=@n51)
   n52 = Get CustomItemDetails!(self:Fort Locker Category Config Object Reference=@n51) -> (CustomItemDetails:User Widget Soft Class Reference)
   > Branch(Condition=IsValidSoftClassReference!(SoftClassReference:"Object Soft Class Reference (by ref)"=@n52) -> (ReturnValue:bool))
   > n53 = Async Load Class Asset!(execute:Exec, AssetClass:Object Soft Class Reference=@n52) -> (then:Exec, Completed:Exec, Class:Object Class Reference) {is=title K2Node_LoadAssetClass}
   Set bIsEnabled(execute=@n53.Completed, bIsEnabled=false, self=@n50)
   > Set LastActiveScreen!(execute, LastActiveScreen:ELockerScreenList Enum=NewEnumerator4, self:WBP Item Collection List Object Reference=@n50) -> (then, Output_Get:ELockerScreenList Enum=NewEnumerator0)
   > PushItemDetails(CustomItemDetailsClass=Cast To UserWidget Class!(Class:Object Class Reference=@n53) -> (AsUser Widget:User Widget Class Reference, bSuccess:bool=~) {is=title K2Node_ClassDynamicCast})
   n54 = "On Clicked (Button_Equip)"!() -> (OutputDelegate, then, Button:Common Button Base Object Reference) {is=title K2Node_ComponentBoundEvent}
   > n55 = Branch(Condition=IsCustomizable())
   > CustomizeClicked(self=Get WBP_ItemCollectionList())
   TriggerFlowRiderInteraction(execute=@n55.else, Element=@n54.Button, Value=Equip)
   > TryEquipItem(self:WBP Item Collection List Object Reference=Get WBP_ItemCollectionList())
   CloseMenu() {is=title K2Node_CustomEvent}
   > SetAllInteractionsEnabled(IsInteractionEnabled=false)
   n56 = Get Intro()
   > Branch(Condition=NOT Boolean(A=IsAnimationPlaying(InAnimation=@n56)))
   > n57 = Play Animation with Finished event(Widget=Self-Reference(), InAnimation=@n56, StartAtTime=0, NumLoopsToPlay=1, PlayMode=Reverse, PlaybackSpeed=1)
   CloseMenu(execute=@n57.Finished, self:WBP Item Collection List Object Reference=Get WBP_ItemCollectionList()) {is=title K2Node_CallFunction}
   > DeactivateWidget()
   n58 = PushItemDetails!() -> (OutputDelegate, then, CustomItemDetailsClass:User Widget Class Reference) {is=title K2Node_CustomEvent}
   > Set Cached Selected Item(Cached Selected Item:Fort Item VM Object Reference=Get SelectedItem(self=Get FortItemSelectionVM())) -> (Output_Get:Fort Item VM Object Reference)
   n59 = Get SubMenu()
   > Branch(Condition=NOT Boolean(A=IsAnimationPlaying(InAnimation=@n59)))
   > n60 = Play Animation with Finished event(Widget=Self-Reference(), InAnimation=@n59, StartAtTime=0, NumLoopsToPlay=1, PlaybackSpeed=1)
   n61 = PushPrimaryContentClassAsync!(
    execute:Exec=@n60.Finished
    WidgetClass:Common Activatable Widget Class Reference=&/WBP_MPLocker_ItemDetails.WBP_MPLocker_ItemDetails_C
   ) -> (then:Exec, OnPrimaryContentPushed:Exec, ActivatableWidget:Common Activatable Widget Object Reference) {is=title K2Node_AsyncAction}
   n62 = Cast To WBP_MPLocker_ItemDetails!(execute=@n61.OnPrimaryContentPushed, Object=@n61) -> (then, CastFailed, AsWBP MPLocker Item Details:WBP MPLocker Item Details Object Reference) {is=title K2Node_DynamicCast}
   > SetFortItemVM!(
    execute
    self:WBP MPLocker Item Details Object Reference=@n62
    Viewmodel:Fort Item VM Object Reference=Get Cached Selected Item() -> (Cached Selected Item:Fort Item VM Object Reference)
   ) -> (then)
   > SetCustomItemDetails!(execute, self:WBP MPLocker Item Details Object Reference=@n62, CustomItemDetails:User Widget Class Reference=@n58.CustomItemDetailsClass) -> (then)
   n63 = Create Event()
   > Unbind Event from On Widget Deactivated(self:Common Activatable Widget Object Reference=@n62, Delegate=@n63) {is=title K2Node_RemoveDelegate}
   > Bind Event to On Widget Deactivated(self=@n62, Delegate=@n63)
   > Branch(Condition=IsInLegacyStwLocker!() -> (InStwLocker:bool))
   > ActivateWidget(self=@n62)
   > SetFrontEndVisibility!(
    execute
    self:Fort Front End Context Object Reference=Get FortFrontEndContext() -> (ReturnValue:Fort Front End Context Object Reference) {is=title K2Node_GetSubsystem}
    bHideHeader:bool=true
    bHideFooter:bool=true
    bHideChatWidget:bool=true
   ) -> (then)
   n64 = CustomizeMenuClosed() {is=title K2Node_CustomEvent}
   > Unbind Event from On Widget Deactivated(self=Get CustomizeMenu() -> (CustomizeMenu:Common Activatable Widget Object Reference), Delegate=@n64)
   n65 = Get SelectedCategory!(self:Fort Item Selection VM Object Reference=Get FortItemSelectionVM()) -> (SelectedCategory:Fort Locker Category Item VM Object Reference)
   > ShowCustomizationISB!(
    execute
    self
    ShowISB:bool=Get bRequiresCustomization(self=Get EquippedItem!(self:Fort Locker Category Item VM Object Reference=@n65) -> (EquippedItem:Fort Item VM Object Reference))
   ) -> (then)
   > RequestRefreshFocus(self:Common Activatable Widget Object Reference)
   > Set bIsEnabled(bIsEnabled=true, self=Get WBP_UIKit_Button_Options())
   > SetAllInteractionsEnabled(IsInteractionEnabled=true)
   > n66 = Get(self=Get FortItemSelectionVM())
   > SetCurrentPreviewItem(self=Get FortItemPreviewWidget(), ItemVM=@n66)
   > Branch(Condition=NOT Boolean(A=RequiresCustomization!(self:Fort Item VM Object Reference=@n66) -> (ReturnValue:bool)))
   > TryEquipItem(self=Get WBP_ItemCollectionList())
   Event On Deactivated()
   > DeactivateWidget(self=Get WBP_ItemCollectionList())
   Event Construct()
   > SetCompactUIStyle() {is=title K2Node_CallFunction}
   > SetupTransitionsBindings!(execute, self, Bind:bool=true) -> (then)
   n67 = Create Event()
   n68 = Get FortItemPreviewWidget()
   > Add Field Value Changed Delegate!(
    execute
    self:Widget Object Reference=@n68
    FieldId:Field Notification Id Structure=(FieldName="AllowToggleVariantOverrideOnCurrentItem")
    Delegate:Delegate=@n67
   ) -> (then)
   n69 = Create Event()
   n70 = Get CommonInputSubsystem()
   > Bind Event to On Input Method Changed(self=@n70, Delegate=@n69)
   > Handle Input Changed!(execute, self, NewInputType:ECommonInputType Enum=GetCurrentInputType(self=@n70)) -> (then)
   > Set FortItemPreviewWidget!(
    execute
    FortItemPreviewWidget:WBP Item Preview Object Reference=Get FortItemPreviewWidget()
    self:WBP Item Collection List Object Reference=Get WBP_ItemCollectionList()
   ) -> (then, Output_Get:WBP Item Preview Object Reference)
   Event OnInitialized()
   > Branch(Condition=GetConfigDataBoolWithDefault!(EntryName:String=EnableCosmeticDownloadNotificationsInLocker, bDefaultValue:bool=false) -> (ReturnValue:bool))
   > n71 = CreateCosmeticDownloadWidgets!(
    execute
    ContainerWidget:Panel Widget Object Reference=Get NotificationVerticalBox() -> (NotificationVerticalBox:Vertical Box Object Reference)
    HorizontalAlignment:EHorizontalAlignment Enum=HAlign_Left
   ) -> (then, CosmeticDownloadWidget:WBP Cosmetic Download Object Reference, CosmeticDownloadInfoWidget:WBP Cosmetic Download Info Object Reference)
   > Set CosmeticDownloadWidget(CosmeticDownloadWidget:WBP Cosmetic Download Object Reference=@n71) -> (Output_Get:WBP Cosmetic Download Object Reference)
   > Set CosmeticDownloadInfoWidget(CosmeticDownloadInfoWidget:WBP Cosmetic Download Info Object Reference=@n71.CosmeticDownloadInfoWidget) -> (Output_Get:WBP Cosmetic Download Info Object Reference)
   Event On Activated()
   > Fire Event Main View Opened!(execute, View:Widget Object Reference=Self-Reference(), AdditionalInfo:Flow Analytics Additional Info Structure, bIsOverlay:bool=false) -> (then)
   > ActivateWidget(self=Get WBP_ItemCollectionList())
   > SetCollectablePreviewWidgetPosition()
   > n72 = Get!(
    execute
    self:Fort Locker Category Item VM Object Reference=GetPreviewCategory!(self:Fort Item Selection VM Object Reference=Get FortItemSelectionVM()) -> (ReturnValue:Fort Locker Category Item VM Object Reference)
   ) -> (then, else, EquippedItem:Fort Item VM Object Reference)
   > SetCurrentPreviewItem(self=Get FortItemPreviewWidget(), ItemVM=@n72)
   Event Destruct()
   > SetupTransitionsBindings(Bind=false)
   > Remove Field Value Changed Delegate!(execute, self:Widget Object Reference=@n68, FieldId:Field Notification Id Structure, Delegate:Delegate=@n67) -> (then)
   > Unbind Event from On Input Method Changed(self=@n70, Delegate=@n69)
  ]}
 GetItemsAreEquippable {nodes [
   n1 = GetItemsAreEquippable!() -> (then, VM:Fort Item Selection VM Object Reference) {is=title K2Node_FunctionEntry}
   > n2 = Is Valid(InputObject=@n1)
   Return Node!(execute=[@n4.Is Not Valid, @n2.Is Not Valid], bItemsAreEquippable:bool=false)
   n3 = Get SelectedCategoryGroup!(self:Fort Item Selection VM Object Reference=@n1) -> (SelectedCategoryGroup:Fort Locker Category Group VM Object Reference)
   > n4 = Is Valid(InputObject=@n3)
   n5 = GetFirstCategoryItem!(self:Fort Locker Category Group VM Object Reference=@n3) -> (ReturnValue:Fort Locker Category Item VM Object Reference)
   > n6 = Is Valid(InputObject=@n5)
   Return Node(execute=@n6.Is Not Valid, bItemsAreEquippable=false)
   > Return Node(bItemsAreEquippable=GetIsEquippable!(self:Fort Locker Category Item VM Object Reference=@n5) -> (ReturnValue:bool))
  ]}
 GetPrimaryContentSetup {nodes [
   GetPrimaryContentSetup() {is=title K2Node_FunctionEntry}
   > Return Node!(
    execute
    ReturnValue:Primary Content Setup Structure=Make Primary Content Setup!(
     bShowBottomBar:bool=false
     bShowChatWidget:bool=false
     bEnableNewTextChat:bool=False
     bShowTopBar:bool=false
     bTabSupportsInterface:bool=False
    ) -> (PrimaryContentSetup:Primary Content Setup Structure) {is=title K2Node_MakeStruct}
   )
  ]}
 Handle Input Changed {nodes [
   n1 = Handle Input Changed!() -> (then, NewInputType:ECommonInputType Enum) {is=title K2Node_FunctionEntry}
   n2 = "Equal (Enum)"(A:ECommonInputType Enum=@n1, B:ECommonInputType Enum=Gamepad)
   > SetSpecialPreviewActionAllowList!(
    execute
    self:Fort Item Preview Screen Object Reference=Get FortItemPreviewWidget()
    InSpecialPreviewActionAllowList:Set of Names=Make Set!("[0]":Name=Select(Option 0:Name=SpecialPreviewAction_ViewLoadingScreen, Option 1:Name=None, Index=@n2) -> (ReturnValue:Name)) -> (Set:Set of Names) {is=title K2Node_MakeSet}
   ) -> (then)
  ]}
 HandleCategoryGroupChanged {nodes [
   HandleCategoryGroupChanged!() -> (then, Selected Category Group:Fort Locker Category Group VM Object Reference) {is=title K2Node_FunctionEntry}
   > SetCollectablePreviewWidgetPosition()
  ]}
 HandleSubCategoryTabSelected {nodes [
   HandleSubCategoryTabSelected!() -> (then, TabNameID:Name) {is=title K2Node_FunctionEntry}
   > n1 = Get!(execute) -> (then, else, FortItemSelectionVM:Fort Item Selection VM Object Reference)
   > n2 = For Each Loop with Break!(
    Exec:Exec
    Array:Array of Fort Locker Category Item VM Object References=Get CategoryItems!(self:Fort Locker Category Group VM Object Reference=Get SelectedCategoryGroup(self=@n1)) -> (CategoryItems:Array of Fort Locker Category Item VM Object References)
    Break:Exec=@n6.then
   ) -> (LoopBody:Exec, Array Element:Fort Locker Category Item VM Object Reference, Array Index:int, Completed:Exec)
   n3 = GetAnalyticsName!(self:Fort Locker Category Item VM Object Reference=@n2) -> (ReturnValue:String)
   > Branch(Condition="Equal (Name)"(A=String To Name(InString=@n3), B=Get TabNameID() -> (TabNameID:Name=None)))
   n4 = Get FortItemSelectionVM()
   > n5 = Set SelectedCategory!(execute, SelectedCategory:Fort Locker Category Item VM Object Reference=@n2, self:Fort Item Selection VM Object Reference=@n4) -> (then, Output_Get:Fort Locker Category Item VM Object Reference)
   > n6 = Set HoveredCategory!(execute, HoveredCategory:Fort Locker Category Item VM Object Reference=@n5, self:Fort Item Selection VM Object Reference=@n4) -> (then, Output_Get:Fort Locker Category Item VM Object Reference)
  ]}
 IsCustomizable {nodes [
   IsCustomizable() {is=title K2Node_FunctionEntry}
   > n1 = Get()
   > n2 = Get!(execute, self:Fort Item Selection VM Object Reference=@n1) -> (then, else, SelectedItem:Fort Item VM Object Reference)
   > Return Node!(execute, Customizable:bool=RequiresCustomization(self=@n2))
   Return Node(execute=[@n2.else, @n1.else], Customizable=false)
  ]}
 IsInLegacyStwLocker {nodes [
   IsInLegacyStwLocker() {is=title K2Node_FunctionEntry}
   > n1 = GetConsoleVariableBoolValue!(execute, VariableName:String=UIManager.UseNewUIManagerSTW) -> (then, ReturnValue:bool)
   n2 = GetSubGame!(self:Fort Global UIContext Object Reference=Get FortGlobalUIContext() -> (ReturnValue:Fort Global UIContext Object Reference) {is=title K2Node_GetSubsystem}) -> (ReturnValue:ESubGame Enum=Campaign)
   > Return Node!(execute, InStwLocker:bool=AND Boolean(A=NOT Boolean(A=@n1), B="Equal (Enum)"(A:ESubGame Enum=@n2, B:ESubGame Enum=Campaign)))
  ]}
 IsOptionsEnabled {nodes [
   n1 = IsOptionsEnabled!() -> (then, SelectedItem:Fort Item VM Object Reference) {is=title K2Node_FunctionEntry}
   > n2 = Is Valid(InputObject=@n1)
   n3 = Get bIsValid!(self:Fort Item VM Object Reference=@n1) -> (bIsValid:bool)
   > Return Node!(
    execute
    Result:bool=NOR Boolean!(A:bool=Get bIsShuffleTile!(self:Fort Item VM Object Reference=@n1) -> (bIsShuffleTile:bool), B:bool=NOT Boolean(A=@n3)) -> (ReturnValue:bool)
   )
   Return Node(execute=@n2.Is Not Valid, Result=false)
  ]}
 ManualSetHoveredItem {nodes [
   n1 = ManualSetHoveredItem!() -> (then, Hovered Item:Fort Item VM Object Reference) {is=title K2Node_FunctionEntry}
   > n2 = Get!(execute) -> (then, else, FortItemSelectionVM:Fort Item Selection VM Object Reference)
   > Set HoveredItem!(
    execute
    HoveredItem:Fort Item VM Object Reference=Select(Option 0:Fort Item VM Object Reference, Option 1:Fort Item VM Object Reference=@n1, Index=Get bIsValid(self=@n1)) -> (ReturnValue:Fort Item VM Object Reference)
    self:Fort Item Selection VM Object Reference=@n2
   ) -> (then, Output_Get:Fort Item VM Object Reference)
  ]}
 ManualSetSelectedItem {nodes [
   n1 = ManualSetSelectedItem!() -> (then, Selected Item:Fort Item VM Object Reference) {is=title K2Node_FunctionEntry}
   > n2 = Get()
   > Set SelectedItem!(
    execute
    SelectedItem:Fort Item VM Object Reference=Select(Option 1=@n1, Index=Get bIsValid(self=@n1))
    self:Fort Item Selection VM Object Reference=@n2
   ) -> (then, Output_Get:Fort Item VM Object Reference)
  ]}
 OnCloseMenu {nodes [OnCloseMenu() {is=title K2Node_FunctionEntry}]}
 OnTransitionAnimationFinished {nodes [
   OnTransitionAnimationFinished() {is=title K2Node_FunctionEntry}
   > SetAllInteractionsEnabled(IsInteractionEnabled=true)
  ]}
 OnTransitionAnimationStarted {nodes [
   OnTransitionAnimationStarted() {is=title K2Node_FunctionEntry}
   > SetAllInteractionsEnabled(IsInteractionEnabled=false)
  ]}
 SetAllInteractionsEnabled {nodes [
   SetAllInteractionsEnabled!() -> (then, IsInteractionEnabled:bool=~) {is=title K2Node_FunctionEntry}
   > SetIsInteractionEnabled!(
    execute
    self:Common Button Base Object Reference=[
     Get Button_EditStyles()
     Get WBP_UIKit_Button_Options()
     Get Button_Use()
     Get Button_Details()
     Get TogglePreview()
     Get WBP_UIKit_Button_Back() -> (WBP_UIKit_Button_Back:WBP UIKit Button Quiet Object Reference)
    ]
    bInIsInteractionEnabled:bool=Get IsInteractionEnabled() -> (IsInteractionEnabled:bool)
   ) -> (then)
   > n1 = Set IgnoreInputs(IgnoreInputs:bool=Get IsInteractionEnabled()) -> (Output_Get:bool)
   > SetAllInteractionsEnabled!(execute, self:WBP Item Collection List Object Reference=Get WBP_ItemCollectionList(), IsInteractionEnabled=@n1, IncludeSortFilter:bool=true) -> (then) {is=title K2Node_CallFunction}
  ]}
 SetCollectablePreviewWidgetPosition {nodes [
   SetCollectablePreviewWidgetPosition() {is=title K2Node_FunctionEntry}
   n1 = GetItemsAreEquippable!(VM:Fort Item Selection VM Object Reference=Get FortItemSelectionVM()) -> (bItemsAreEquippable:bool) {is=title K2Node_CallFunction}
   > SetGridSlotPadding!(
    execute
    Widget:Widget Object Reference=Get Widget_CameraFrame() -> (Widget_CameraFrame:Size Box Object Reference)
    Padding:Margin Structure=Select(
     Option 0:Margin Structure=Get PreviewNotEquipabblePadding() -> (PreviewNotEquipabblePadding:Margin Structure)
     Option 1:Margin Structure=Make Margin(Left=0, Top=0, Right=0, Bottom=0)
     Index=@n1
    ) -> (ReturnValue:Margin Structure)
   ) -> (then)
   > n2 = Branch(Condition=@n1)
   n3 = Get SelectedCategory(self=Get FortItemSelectionVM())
   > Is Valid(InputObject=@n3)
   > UpdateCategoryPreviewScale!(execute, self, VM:Fort Locker Category Item VM Object Reference=@n3) -> (then)
   SetWidthOverride(
    execute=@n2.else
    self=Get SizeBox_PreviewWidget() -> (SizeBox_PreviewWidget:Size Box Object Reference)
    InWidthOverride=Get PreviewNotEquippableFixedWidth() -> (PreviewNotEquippableFixedWidth:float)
   )
  ]}
 SetCompactUIStyle {nodes [
   SetCompactUIStyle() {is=title K2Node_FunctionEntry}
   > Branch(Condition=IsUsingCompactUI())
   n1 = Get SizeBox() -> (SizeBox:Size Box Object Reference)
   > ClearWidthOverride(self:Size Box Object Reference=@n1)
   > Set HAlign!(
    execute
    HAlign:EHorizontalAlignment Enum=HAlign_Fill
    self:Safe Zone Slot Object Reference=SlotAsSafeBoxSlot!(Widget:Widget Object Reference=@n1) -> (ReturnValue:Safe Zone Slot Object Reference)
   ) -> (then, Output_Get:EHorizontalAlignment Enum=HAlign_Fill)
   > SetMinDesiredWidth!(
    execute
    self:Size Box Object Reference=Get SizeBox_Collection() -> (SizeBox_Collection:Size Box Object Reference)
    InMinDesiredWidth:float=SelectFloat(A=850, B=1024, bPickA=IsTabletAspectRatio!() -> (ReturnValue:bool))
   ) -> (then)
  ]}
 SetupTransitionsBindings {nodes [
   n1 = SetupTransitionsBindings!() -> (then, Bind:bool=~) {is=title K2Node_FunctionEntry}
   > n2 = Branch(Condition=@n1)
   n3 = Create Event()
   n4 = Get Intro()
   > BindToAnimationStarted!(execute, self:User Widget Object Reference, Animation:Widget Animation Object Reference=@n4, Delegate:Delegate=@n3) -> (then)
   n5 = Create Event()
   n6 = Get SubMenu()
   > BindToAnimationStarted(Animation=@n6, Delegate=@n5)
   n7 = Create Event()
   n8 = Get Intro()
   > BindToAnimationFinished(Animation=@n8, Delegate=@n7)
   n9 = Create Event()
   n10 = Get SubMenu()
   > BindToAnimationFinished(Animation=@n10, Delegate=@n9)
   UnbindFromAnimationStarted!(execute=@n2.else, self:User Widget Object Reference, Animation:Widget Animation Object Reference=@n4, Delegate:Delegate=@n3) -> (then)
   > UnbindFromAnimationStarted(Animation=@n6, Delegate=@n5)
   > UnbindFromAnimationFinished!(execute, self:User Widget Object Reference, Animation:Widget Animation Object Reference=@n8, Delegate:Delegate=@n7) -> (then)
   > UnbindFromAnimationFinished(Animation=@n10, Delegate=@n9)
  ]}
 ShowCustomizationISB {nodes [
   n1 = ShowCustomizationISB!() -> (then, ShowISB:bool=~) {is=title K2Node_FunctionEntry}
   > n2 = Branch(Condition=@n1)
   > n3 = Get!(execute) -> (then, else, MimosaISB:Common User Widget Object Reference)
   n4 = Get!(execute=@n3.else, self:Fort Item Selection VM Object Reference=Get FortItemSelectionVM()) -> (then, else, SelectedItem:Fort Item VM Object Reference)
   > n5 = LoadClassAsset_Blocking(AssetClass=TryGetSoftCustomWidgetClass!(
     self:Fort Locker Category Config Object Reference=GetItemCategory(self=@n4)
     ViewRequest:Gameplay Tag Structure=(TagName="UI.ItemPreview.PluginBridge.Mimosa.ShopISB")
    ) -> (WidgetClass:Common User Widget Soft Class Reference, ReturnValue:bool))
   > n6 = Cast To CommonUserWidget Class!(execute:Exec, Class:Object Class Reference=@n5) -> (then:Exec, CastFailed:Exec, AsCommon User Widget:Common User Widget Class Reference) {is=title K2Node_ClassDynamicCast}
   > n7 = Create Widget!(execute:Exec, Class:User Widget Class Reference=@n6, OwningPlayer:Player Controller Object Reference=GetOwningPlayer()) -> (then:Exec, ReturnValue:Common User Widget Object Reference)
   > n8 = Set MimosaISB(MimosaISB:Common User Widget Object Reference=@n7) -> (Output_Get:Common User Widget Object Reference)
   > AddChild(self=Get ItemDescription_Vbox() -> (ItemDescription_Vbox:Vertical Box Object Reference), Content=@n8)
   > n9 = Cast To WBP_CustomBanner_Base!(execute, Object=@n7) -> (then, CastFailed, AsWBP Custom Banner Base:WBP Custom Banner Base Object Reference) {is=title K2Node_DynamicCast}
   > SetContext!(execute, self:WBP Custom Banner Base Object Reference=@n9, Desired Area:E_MultiproductContext Enum=NewEnumerator1) -> (then)
   n10 = Get!(execute=[@n4.else, @n2.else]) -> (then, else, MimosaISB:Common User Widget Object Reference)
   > RemoveWidget!(execute, Widget:Widget Object Reference=@n10) -> (then)
   > Set MimosaISB()
  ]}
 TriggerFlowRiderInteraction {nodes [
   n1 = TriggerFlowRiderInteraction!() -> (then, Element:Widget Object Reference, Value:String) {is=title K2Node_FunctionEntry}
   > Fire Event Element Interacted(
    Element=@n1
    AdditionalInfo="Add String (Non-Localized)"!(
     Info:"Flow Analytics Additional Info Structure (by ref)"=Make Flow Analytics Additional Info!() -> (FlowAnalyticsAdditionalInfo:Flow Analytics Additional Info Structure) {is=title K2Node_MakeStruct}
     Key:String=Button Name
     Value:String=@n1.Value
    ) -> (ReturnValue:Flow Analytics Additional Info Structure)
   )
  ]}
 TriggerItemDetailsClosed {nodes [
   TriggerItemDetailsClosed() {is=title K2Node_FunctionEntry}
   > ItemDetailsClosed(self:WBP Item Collection List Object Reference=Get WBP_ItemCollectionList())
  ]}
 UpdateBottomBarButtonIsEnabled {nodes [
   UpdateBottomBarButtonIsEnabled!() -> (then, Button:Widget Object Reference, Enabled:bool=~) {is=title K2Node_FunctionEntry}
   > SetIsEnabled(self=Get Button() -> (Button:Widget Object Reference), bInIsEnabled=Get Enabled() -> (Enabled:bool))
  ]}
 UpdateBottomBarButtonIsVisible {nodes [
   UpdateBottomBarButtonIsVisible!() -> (then, Button:Widget Object Reference, IsVisible:bool=~) {is=title K2Node_FunctionEntry}
   > SetVisibility(
    self=Get Button()
    InVisibility=Select(Option 0:ESlateVisibility Enum=Collapsed, Option 1:ESlateVisibility Enum=Visible, Index=Get IsVisible() -> (IsVisible:bool)) -> (ReturnValue:ESlateVisibility Enum)
   )
  ]}
 UpdateCategoryPreviewScale {nodes [
   n1 = UpdateCategoryPreviewScale!() -> (then, VM:Fort Locker Category Item VM Object Reference) {is=title K2Node_FunctionEntry}
   > Is Valid(InputObject=@n1)
   n2 = GetSlotInfo!(self:Fort Locker Category Item VM Object Reference=@n1) -> (
    ReturnValue_CategoryName:Text
    ReturnValue_CategoryNamePlural:Text
    ReturnValue_CategoryNameWithPluralForms:Text
    ReturnValue_CategoryDescription:Text
    ReturnValue_DefaultEmptyTexture:Soft Object Reference
    ReturnValue_OverrideShuffleTexture:Texture 2D Soft Object Reference
    ReturnValue_CategoryIcon:Material Instance Object Reference
    ReturnValue_CategorySoundOverrides:Map of Gameplay Tag Structures to Sound Base Soft Object References
    ReturnValue_bShowInCategoryTab:bool=~
    ReturnValue_PreviewWindowScale:float=~
    ReturnValue_PreviewWindowOffset_X:double
    ReturnValue_PreviewWindowOffset_Y:double
    ReturnValue_PreviewWindowOffset_Z:double
    ReturnValue_bUsePluralCategoryInItemList:bool=~
   )
   > SetRenderScale!(
    execute
    self:Widget Object Reference=Get SizeBox_PreviewWidget()
    Scale:Vector 2D Structure=To Vector2D!(InDouble:double=@n2.ReturnValue_PreviewWindowScale) -> (ReturnValue:Vector 2D Structure)
   ) -> (then)
   > Set LastPreviewScale(LastPreviewScale:double=@n2.ReturnValue_PreviewWindowScale) -> (Output_Get:double)
   > SetRenderTranslation!(
    execute
    self=Get Widget_CameraFrame()
    Translation_X:double=@n2.ReturnValue_PreviewWindowOffset_X
    Translation_Y:double=@n2.ReturnValue_PreviewWindowOffset_Y
   ) -> (then)
  ]}
 UpdateEditButtonEnabled {nodes [
   UpdateEditButtonEnabled!() -> (then, HasVariant:bool=~, RequireCustomization:bool=~, IsBulkCosmeticInstalled:bool=~) {is=title K2Node_FunctionEntry}
   n1 = Get IsBulkCosmeticInstalled() -> (IsBulkCosmeticInstalled:bool)
   n2 = Get RequireCustomization() -> (RequireCustomization:bool)
   n3 = NOT Boolean(A=Get RequireCustomization())
   n4 = Get HasVariant() -> (HasVariant:bool)
   > Return Node!(execute, ReturnValue:bool=OR Boolean(A=AND Boolean(A=@n2, B=@n1), B=AND Boolean(A=@n3, B=@n4)))
  ]}
 UpdateEditButtonText {nodes [
   UpdateEditButtonText!() -> (then, RequireCustomize:bool=~) {is=title K2Node_FunctionEntry}
   > Return Node(ReturnValue:Text=Select(
     Option 0:"Text (by ref)"=MakeLiteralText(Value="LOCTEXT(\"Edit\")")
     Option 1:"Text (by ref)"=MakeLiteralText(Value="LOCTEXT(\"Customize\")")
     Index=Get RequireCustomize() -> (RequireCustomize:bool)
    ) -> (ReturnValue:"Text (by ref)"))
  ]}
]
variables [
 Cached Selected Item {type=FortItemVM, default=null}
 CachedSelectedItem {type=FortItemVM, default=null}
 ChatEventWatch {type=AsyncAction_StartListeningToEvent, default=null}
 CosmeticDownloadInfoWidget {type=WBP_CosmeticDownloadInfo_C, default=null}
 CosmeticDownloadWidget {type=WBP_CosmeticDownload_C, default=null}
 CustomizeMenu {type=CommonActivatableWidget, default=null}
 IgnoreInputs {type=boolean}
 LastPreviewScale {type=number}
 MimosaISB {type=CommonUserWidget, default=null}
 OnCloseMenu {}
 PreviewNotEquipabblePadding {type=Margin, default=<ltrb 0 0 25 50>}
 PreviewNotEquippableFixedWidth {type=number, default=1000}
]
widgets {
 info {
  inheritedWidgetCount=0
  namedSlotCount=1
  parentClass=&CommonActivatableWidget
  rootWidgetClass=&SafeZone
  widgetCount=36
 }
 widgets [
  SafeZone {widgetClassPath=&SafeZone, bIsVariable=true, widget=SafeZone, properties {padBottom=false, padTop=false}}
  SizeBox {
   widgetClassPath=&SizeBox
   bIsVariable=true
   parent=SafeZone
   widget=SizeBox
   properties {bOverride_WidthOverride=true, widthOverride=1920}
   slot {class=SafeZoneSlot, properties {hAlign=HAlign_Center}}
  }
  GridPanel_Main {
   widgetClassPath=&GridPanel
   parent=SizeBox
   widget=GridPanel_Main
   properties {columnFill [1,0], rowFill [1,0]}
   slot {class=SizeBoxSlot}
  }
  GridPanel_Locker {
   widgetClassPath=&GridPanel
   parent=GridPanel_Main
   widget=GridPanel_Locker
   properties {columnFill [0,1], rowFill [1,1,1,1]}
   slot {class=GridSlot, properties {padding=<ltrb 80 64 80 32>}}
  }
  SizeBox_Collection {
   widgetClassPath=&SizeBox
   bIsVariable=true
   parent=GridPanel_Locker
   widget=SizeBox_Collection
   properties {bOverride_MaxDesiredWidth=true, bOverride_MinDesiredWidth=true, maxDesiredWidth=866, minDesiredWidth=866}
   slot {class=GridSlot, properties {column=2, rowSpan=4, padding=<ltrb 0 0 26 72>}}
  }
  WBP_ItemCollectionList {
   widgetClassPath=&/WBP_ItemCollectionList
   bIsVariable=true
   parent=SizeBox_Collection
   widget=WBP_ItemCollectionList
   properties {bIsFocusable=true}
   slot {class=SizeBoxSlot}
  }
  ItemDescription_Vbox {
   widgetClassPath=&VerticalBox
   bIsVariable=true
   parent=GridPanel_Locker
   widget=ItemDescription_Vbox
   slot {
    class=GridSlot
    properties {column=1, horizontalAlignment=HAlign_Left, row=2, rowSpan=2, verticalAlignment=VAlign_Bottom, padding=<ltrb 0 0 0 72>}
   }
  }
  WBP_ItemSelection_ItemDetails {
   widgetClassPath=&/WBP_ItemSelection_ItemDetails
   bIsVariable=true
   parent=ItemDescription_Vbox
   widget=WBP_ItemSelection_ItemDetails
   properties {fortLockerEquipUtilityVM=&self:WidgetTree.WBP_ItemSelection_ItemDetails.FortLockerEquipUtilityVM_0}
   slot {class=VerticalBoxSlot}
  }
  WBP_ActionBarContainer {
   widgetClassPath=/Game/UI/Frontend/BottomBar/WBP_ActionBarContainer
   bIsVariable=true
   parent=GridPanel_Locker
   widget=WBP_ActionBarContainer
   slot {class=GridSlot, properties {column=1, columnSpan=2, horizontalAlignment=HAlign_Right, row=3, verticalAlignment=VAlign_Bottom}}
  }
  Temp_ButtonBar {widgetClassPath=&HorizontalBox, namedSlotHost=WBP_ActionBarContainer, widget=Temp_ButtonBar}
  TogglePreview {
   widgetClassPath=/FortUILibrary/UIKit/Buttons/WBP_UIKit_Button_Quiet
   bIsVariable=true
   parent=Temp_ButtonBar
   widget=TogglePreview
   properties {
    text=Toggle Preview
    visibility=Collapsed
    triggeringInputAction {dataTable=&InputActionDataTable, rowName=GenericAccept}
   }
   slot {class=HorizontalBoxSlot, properties {horizontalAlignment=HAlign_Right, verticalAlignment=VAlign_Bottom, padding=<ltrb 0 0 24 0>}}
  }
  VisibilityBorder_Details {
   widgetClassPath=&CommonHardwareVisibilityBorder
   parent=Temp_ButtonBar
   widget=VisibilityBorder_Details
   properties {
    horizontalAlignment=HAlign_Right
    verticalAlignment=VAlign_Bottom
    padding=<ltrb 0 0 24 0>
    visibilityQuery {
     autoDescription=" ANY( Input.Touch, Input.Gamepad )"
     tokenStreamVersion=0
     userDescription=""
     queryTokenStream [0,1,1,2,0,1]
     tagDictionary [{tagName=Input.Touch}, {tagName=Input.Gamepad}]
    }
   }
   slot {class=HorizontalBoxSlot, properties {horizontalAlignment=HAlign_Right, verticalAlignment=VAlign_Bottom}}
  }
  Button_Details {
   widgetClassPath=&WBP_UIKit_Button_Quiet
   bIsVariable=true
   parent=VisibilityBorder_Details
   widget=Button_Details
   properties {text=View Details, visibility=Visible, triggeringInputAction {dataTable=&InputActionDataTable, rowName=ViewDetails}}
   slot {class=BorderSlot, properties {horizontalAlignment=HAlign_Right, verticalAlignment=VAlign_Bottom, padding=<ltrb 0 0 24 0>}}
  }
  VisibilityBorder_Equip_Touch {
   widgetClassPath=&CommonHardwareVisibilityBorder
   parent=Temp_ButtonBar
   widget=VisibilityBorder_Equip_Touch
   properties {
    horizontalAlignment=HAlign_Right
    verticalAlignment=VAlign_Bottom
    padding=<ltrb 0 0 24 0>
    visibilityQuery {
     autoDescription=" ALL( Input.Touch )"
     tokenStreamVersion=0
     userDescription=""
     queryTokenStream [0,1,2,1,0]
     tagDictionary [{tagName=Input.Touch}]
    }
   }
   slot {class=HorizontalBoxSlot, properties {horizontalAlignment=HAlign_Right, verticalAlignment=VAlign_Bottom}}
  }
  Button_Equip {
   widgetClassPath=&WBP_UIKit_Button_Quiet
   bIsVariable=true
   parent=VisibilityBorder_Equip_Touch
   widget=Button_Equip
   properties {text=Equip, visibility=Visible, triggeringInputAction {dataTable=&InputActionDataTable, rowName=GenericAccept}}
   slot {class=BorderSlot, properties {horizontalAlignment=HAlign_Right, verticalAlignment=VAlign_Bottom, padding=<ltrb 0 0 24 0>}}
  }
  VisibilityBorder_Use_Touch {
   widgetClassPath=&CommonHardwareVisibilityBorder
   parent=Temp_ButtonBar
   widget=VisibilityBorder_Use_Touch
   properties {
    horizontalAlignment=HAlign_Right
    verticalAlignment=VAlign_Bottom
    padding=<ltrb 0 0 24 0>
    visibilityQuery {
     autoDescription=" ANY( Input.Gamepad )"
     tokenStreamVersion=0
     userDescription=""
     queryTokenStream [0,1,1,1,0]
     tagDictionary [{tagName=Input.Gamepad}]
    }
   }
   slot {class=HorizontalBoxSlot, properties {horizontalAlignment=HAlign_Right, verticalAlignment=VAlign_Bottom}}
  }
  Button_Use {
   widgetClassPath=&WBP_UIKit_Button_Quiet
   bIsVariable=true
   parent=VisibilityBorder_Use_Touch
   widget=Button_Use
   properties {text=Use, visibility=Visible, triggeringInputAction {dataTable=&InputActionDataTable, rowName=GenericForward}}
   slot {class=BorderSlot, properties {horizontalAlignment=HAlign_Right, verticalAlignment=VAlign_Bottom, padding=<ltrb 0 0 24 0>}}
  }
  Button_PreviewAction1 {
   widgetClassPath=&WBP_UIKit_Button_Quiet
   bIsVariable=true
   parent=Temp_ButtonBar
   widget=Button_PreviewAction1
   properties {
    bAutoUpdateFromInputAction=true
    text="[PH] Preview 1"
    visibility=Collapsed
    triggeringInputAction {dataTable=&InputActionDataTable, rowName=Edit}
   }
   slot {class=HorizontalBoxSlot, properties {horizontalAlignment=HAlign_Right, verticalAlignment=VAlign_Bottom, padding=<ltrb 0 0 24 0>}}
  }
  Button_PreviewAction {
   widgetClassPath=&WBP_UIKit_Button_Quiet
   bIsVariable=true
   parent=Temp_ButtonBar
   widget=Button_PreviewAction
   properties {
    bAutoUpdateFromInputAction=true
    text="[PH] Preview"
    visibility=Collapsed
    triggeringInputAction {dataTable=&InputActionDataTable, rowName=Edit}
   }
   slot {class=HorizontalBoxSlot, properties {horizontalAlignment=HAlign_Right, verticalAlignment=VAlign_Bottom, padding=<ltrb 0 0 24 0>}}
  }
  Button_EditStyles {
   widgetClassPath=&WBP_UIKit_Button_Quiet
   bIsVariable=true
   parent=Temp_ButtonBar
   widget=Button_EditStyles
   properties {text=Edit, visibility=Visible, triggeringInputAction {dataTable=&InputActionDataTable, rowName=Edit}}
   slot {class=HorizontalBoxSlot, properties {horizontalAlignment=HAlign_Right, verticalAlignment=VAlign_Bottom, padding=<ltrb 0 0 24 0>}}
  }
  VisibilityBorder_TouchGamepad {
   widgetClassPath=&CommonHardwareVisibilityBorder
   parent=Temp_ButtonBar
   widget=VisibilityBorder_TouchGamepad
   properties {
    horizontalAlignment=HAlign_Right
    verticalAlignment=VAlign_Bottom
    padding=<ltrb 0 0 24 0>
    visibilityQuery {
     autoDescription=" ANY( Input.Gamepad, Input.Touch )"
     tokenStreamVersion=0
     userDescription=""
     queryTokenStream [0,1,1,2,0,1]
     tagDictionary [{tagName=Input.Gamepad}, {tagName=Input.Touch}]
    }
   }
   slot {class=HorizontalBoxSlot, properties {horizontalAlignment=HAlign_Right, verticalAlignment=VAlign_Bottom}}
  }
  WBP_UIKit_Button_Options {
   widgetClassPath=&WBP_UIKit_Button_Quiet
   bIsVariable=true
   parent=VisibilityBorder_TouchGamepad
   widget=WBP_UIKit_Button_Options
   properties {
    bHideInputActionWithKeyboard=true
    text=Options
    triggeringInputAction {dataTable=&InputActionDataTable, rowName=ContextPopupMenu}
   }
   slot {class=BorderSlot, properties {horizontalAlignment=HAlign_Right, verticalAlignment=VAlign_Bottom, padding=<ltrb 0 0 24 0>}}
  }
  VisibilityBorder_TouchGamepad_ClearSearch {
   widgetClassPath=&CommonHardwareVisibilityBorder
   parent=Temp_ButtonBar
   widget=VisibilityBorder_TouchGamepad_ClearSearch
   properties {
    horizontalAlignment=HAlign_Right
    verticalAlignment=VAlign_Bottom
    padding=<ltrb 0 0 24 0>
    visibilityQuery {
     autoDescription=" ANY( Input.Gamepad, Input.Touch )"
     tokenStreamVersion=0
     userDescription=""
     queryTokenStream [0,1,1,2,0,1]
     tagDictionary [{tagName=Input.Gamepad}, {tagName=Input.Touch}]
    }
   }
   slot {class=HorizontalBoxSlot, properties {horizontalAlignment=HAlign_Right, verticalAlignment=VAlign_Bottom}}
  }
  Button_ClearSearch {
   widgetClassPath=&WBP_UIKit_Button_Quiet
   bIsVariable=true
   parent=VisibilityBorder_TouchGamepad_ClearSearch
   widget=Button_ClearSearch
   properties {
    bHideInputActionWithKeyboard=true
    text=Clear Search
    triggeringInputAction {dataTable=&InputActionDataTable, rowName=DiscoverReduxClearResults}
   }
   slot {class=BorderSlot, properties {horizontalAlignment=HAlign_Right, verticalAlignment=VAlign_Bottom, padding=<ltrb 0 0 24 0>}}
  }
  WBP_AthenaBottomBar_WithNotifications {
   widgetClassPath=/Game/UI/Frontend/BottomBar/WBP_AthenaBottomBar_WithNotifications
   bIsVariable=true
   parent=Temp_ButtonBar
   widget=WBP_AthenaBottomBar_WithNotifications
   properties {visibility=Collapsed}
   slot {class=HorizontalBoxSlot, properties {horizontalAlignment=HAlign_Right, verticalAlignment=VAlign_Bottom, padding=<ltrb 0 0 24 0>}}
  }
  WBP_UIKit_Button_Back {
   widgetClassPath=&WBP_UIKit_Button_Quiet
   bIsVariable=true
   parent=Temp_ButtonBar
   widget=WBP_UIKit_Button_Back
   properties {text=Back, triggeringInputAction {dataTable=&InputActionDataTable, rowName=Back}}
   slot {class=HorizontalBoxSlot, properties {horizontalAlignment=HAlign_Right, verticalAlignment=VAlign_Bottom}}
  }
  Widget_CameraFrame {
   widgetClassPath=&SizeBox
   bIsVariable=true
   parent=GridPanel_Main
   widget=Widget_CameraFrame
   slot {class=GridSlot, properties {layer=-1}}
  }
  HorizontalBox_0 {widgetClassPath=&HorizontalBox, parent=Widget_CameraFrame, widget=HorizontalBox_0, slot {class=SizeBoxSlot}}
  Camera_Spacer_Left {
   widgetClassPath=&Spacer
   bIsVariable=true
   parent=HorizontalBox_0
   widget=Camera_Spacer_Left
   slot {class=HorizontalBoxSlot, properties {size {sizeRule=Fill, value=0}}}
  }
  SizeBox_PreviewWidget {
   widgetClassPath=&SizeBox
   bIsVariable=true
   parent=HorizontalBox_0
   widget=SizeBox_PreviewWidget
   properties {bOverride_HeightOverride=true, bOverride_WidthOverride=true, heightOverride=1024, widthOverride=970}
   slot {class=HorizontalBoxSlot, properties {horizontalAlignment=HAlign_Center, verticalAlignment=VAlign_Center}}
  }
  FortItemPreviewWidget {
   widgetClassPath=/FortniteCoreUI/Frontend/ItemPreview/WBP_ItemPreview
   bIsVariable=true
   parent=SizeBox_PreviewWidget
   widget=FortItemPreviewWidget
   properties {
    bSupportsActivationFocus=false
    infoAlignment_V=VAlign_Bottom
    infoWidget=&self:WidgetTree.Widget_PreviewInfo_External
    itemViewObject=&self:WidgetTree.FortItemPreviewWidget.ItemViewObject
    infoPadding=<ltrb 60 60 150 30>
    previewActionButtons [&self:WidgetTree.Button_PreviewAction, &self:WidgetTree.Button_PreviewAction1]
    screenTag {tagName=UI.ItemPreview.Screen.Locker.ItemSelection}
    specialPreviewActionAllowList [None]
   }
   slot {class=SizeBoxSlot, properties {padding=<ltrb 0 20 0 96>}}
  }
  Camera_Spacer_Right {
   widgetClassPath=&Spacer
   bIsVariable=true
   parent=HorizontalBox_0
   widget=Camera_Spacer_Right
   slot {class=HorizontalBoxSlot, properties {size {sizeRule=Fill, value=1}}}
  }
  WBP_CaptureForPostBufferUpdate {
   widgetClassPath=&WBP_CaptureForPostBufferUpdate
   bIsVariable=true
   parent=GridPanel_Main
   slot {class=GridSlot, properties {columnSpan=2, layer=-20, rowSpan=2}}
  }
  NotificationSizeBox {
   widgetClassPath=&SizeBox
   parent=GridPanel_Main
   widget=NotificationSizeBox
   properties {bOverride_MaxDesiredWidth=true, maxDesiredWidth=450}
   slot {class=GridSlot, properties {horizontalAlignment=HAlign_Left, verticalAlignment=VAlign_Top, padding=<ltrb 50 50 0 0>}}
  }
  NotificationVerticalBox {
   widgetClassPath=&VerticalBox
   bIsVariable=true
   parent=NotificationSizeBox
   widget=NotificationVerticalBox
   slot {class=SizeBoxSlot, properties {horizontalAlignment=HAlign_Left, verticalAlignment=VAlign_Top}}
  }
  Widget_PreviewInfo_External {
   widgetClassPath=/FortniteCoreUI/Frontend/ItemPreview/WBP_ItemPreview_Info
   bIsVariable=true
   parent=NotificationVerticalBox
   widget=Widget_PreviewInfo_External
   slot {class=VerticalBoxSlot, properties {horizontalAlignment=HAlign_Left, verticalAlignment=VAlign_Top, padding=<ltrb 0 0 0 10>}}
  }
 ]
}
