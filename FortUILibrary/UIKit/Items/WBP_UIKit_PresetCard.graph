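# Locker preset card, derived from the UIKit item card base, that displays a modular loadout viewmodel's name and first valid item.
# Helper functions restyle it as a shuffle or save-slot tile.
# Hover, or selection on touch, scrolls text and plays OnFocus to reveal a context menu for valid loadouts; touch hold opens it.
# Menu Rename and Delete, selection, and options-hint changes are relayed through event dispatchers.
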WidgetBlueprint /FortUILibrary/UIKit/Items/WBP_UIKit_PresetCard
animations [OnFocus {
  binding_count=2
  length_seconds=0.35
  object=&self:OnFocus
  sequence {bindings [
    ContextMenu {
     id=#1
     possessed_class=WBP_UIKit_ContextMenuAnchorButton_C
     tracks [
      {
       type=MovieScene2DTransformTrack
       sections [{
         type=MovieScene2DTransformSection
         end_frame=7
         start_frame=0
         channels [
          Translation.X {type=MovieSceneScriptingFloatChannel, const [0,1]}
          Translation.Y {type=MovieSceneScriptingFloatChannel, keys [[0,42],[5,0]]}
          Angle {type=MovieSceneScriptingFloatChannel, const [0,1]}
          Scale.X {type=MovieSceneScriptingFloatChannel, const [1,1]}
          Scale.Y {type=MovieSceneScriptingFloatChannel, const [1,1]}
          Shear.X {type=MovieSceneScriptingFloatChannel, const [0,1]}
          Shear.Y {type=MovieSceneScriptingFloatChannel, const [0,1]}
         ]
        }]
      }
      {
       type=MovieSceneColorTrack
       display_name=Color and Opacity
       sections [{
         type=MovieSceneColorSection
         end_frame=5
         start_frame=0
         channels [
          Color.R {type=MovieSceneScriptingFloatChannel, const [1,1]}
          Color.G {type=MovieSceneScriptingFloatChannel, const [1,1]}
          Color.B {type=MovieSceneScriptingFloatChannel, const [1,1]}
          Color.A {type=MovieSceneScriptingFloatChannel, keys [[0,0],[5,1]]}
         ]
        }]
      }
     ]
    }
    Text_Loadoutname {
     id=#2
     possessed_class=CommonTextBlock
     tracks [{
       type=MovieScene2DTransformTrack
       sections [{
         type=MovieScene2DTransformSection
         end_frame=7
         start_frame=0
         channels [
          Translation.X {type=MovieSceneScriptingFloatChannel, const [0,1]}
          Translation.Y {type=MovieSceneScriptingFloatChannel, keys [[0,0],[5,-42]]}
          Angle {type=MovieSceneScriptingFloatChannel, const [0,1]}
          Scale.X {type=MovieSceneScriptingFloatChannel, const [1,1]}
          Scale.Y {type=MovieSceneScriptingFloatChannel, const [1,1]}
          Shear.X {type=MovieSceneScriptingFloatChannel, const [0,1]}
          Shear.Y {type=MovieSceneScriptingFloatChannel, const [0,1]}
         ]
        }]
      }]
    }
   ]}
 }]
graphs [
 BP_OnHoldTriggered {nodes [
   n1 = On Hold Triggered!() -> (then, CurrentInputType:ECommonInputType Enum) {is=title K2Node_FunctionEntry}
   > n2 = Switch on ECommonInputType!(execute, Selection:ECommonInputType Enum=@n1) -> (MouseAndKeyboard:Exec, Gamepad:Exec, Touch:Exec, Count:Exec) {is=title K2Node_SwitchEnum}
   Branch(
    execute=@n2.Touch
    Condition=Get bIsValid!(self:Fort Typed Modular Loadout VM Object Reference=Get FortTypedModularLoadoutVM() -> (FortTypedModularLoadoutVM:Fort Typed Modular Loadout VM Object Reference)) -> (bIsValid:bool)
   )
   > SetVisibility(self=Get ContextMenu() -> (ContextMenu:WBP UIKit Context Menu Anchor Button Object Reference), InVisibility=Visible)
   > ToggleContextMenu()
   > SetIsSelected!(execute, self:Common Button Base Object Reference, InSelected:bool=true, bGiveClickFeedback:bool=false) -> (then)
   > Return Node!(execute, ReturnValue:Event Reply Structure=Handled())
  ]}
 DisableOptionsHint {nodes [DisableOptionsHint() {is=title K2Node_FunctionEntry}]}
 EnableOptionsHint {nodes [EnableOptionsHint() {is=title K2Node_FunctionEntry}]}
 EventGraph {nodes [
   "On Context Menu Closed (ContextMenu)"() {is=title K2Node_ComponentBoundEvent}
   > Call On Context Menu Closed!(execute:Exec, self:WBP UIKit Preset Card Object Reference) -> (then:Exec) {is=title K2Node_CallDelegate}
   n1 = GetOwningPlayer()
   > n2 = Branch(Condition=OR Boolean(A=IsHovered(), B=IsUsingTouch(PlayerController=@n1)))
   SetVisibility(execute=@n2.else, self=Get ContextMenu(), InVisibility=Collapsed)
   "On Context Menu Opened (ContextMenu)"() {is=title K2Node_ComponentBoundEvent}
   > Call On Context Menu Opened!(execute:Exec, self:WBP UIKit Preset Card Object Reference) -> (then:Exec) {is=title K2Node_CallDelegate}
   n3 = Event On Rename Preset!() -> (OutputDelegate, then, InteractionNameID:Name) {is=title K2Node_CustomEvent}
   > Call On Rename Event Fired!(execute:Exec, self:WBP UIKit Preset Card Object Reference, PresetCard:WBP UIKit Preset Card Object Reference=Self-Reference()) -> (then:Exec) {is=title K2Node_CallDelegate}
   n4 = Event On Delete Preset!() -> (OutputDelegate, then, InteractionNameID:Name) {is=title K2Node_CustomEvent}
   > Call On Delete Event Fired!(execute:Exec, self:WBP UIKit Preset Card Object Reference, PresetCard:WBP UIKit Preset Card Object Reference=Self-Reference()) -> (then:Exec) {is=title K2Node_CallDelegate}
   n5 = Event PreConstruct()
   > Parent: PreConstruct(IsDesignTime=@n5.IsDesignTime)
   > RemoveFromParent(self=[
     Get EquippedState_Outline() -> (EquippedState_Outline:WBP UIKit Equipped State Object Reference)
     Get Compatibility_Icon() -> (Compatibility_Icon:Image Object Reference)
     Get UIKit_Transference() -> (UIKit_Transference:WBP UIKit Transference Object Reference)
    ])
   > SetItemIsEquipped!(execute, self:WBP UIKit Item Card Base Object Reference, Is Equipped:bool=false) -> (then)
   > SetContextMenuPadding()
   > SetVisibility(self=Get ContextMenu(), InVisibility=Collapsed)
   > SetContextMenuActionContext!(execute, self, RenameInInteractionDelegate:"Delegate (by ref)"=@n3, DeleteInInteractionDelegate:"Delegate (by ref)"=@n4) -> (then)
   > Branch(Condition=IsMobileGame())
   > Set Outline Color!(
    In:Exec
    CommonTextWidget:Common Text Object Reference=Get Text_Loadoutname() -> (Text_Loadoutname:Common Text Object Reference)
    Color:Linear Color Structure=(R=0,G=0,B=0,A=0.251)
   ) -> (Out:Exec) {is=title K2Node_MacroInstance}
   n6 = Event OnListItemObjectSet()
   > n7 = Cast To FortTypedModularLoadoutVM!(execute, Object=@n6.ListItemObject) -> (then, CastFailed, AsFort Typed Modular Loadout VM:Fort Typed Modular Loadout VM Object Reference) {is=title K2Node_DynamicCast}
   > InitializeItemCard(self:WBP UIKit Item Card Base Object Reference)
   > SetFortTypedModularLoadoutVM!(execute, self, Viewmodel:Fort Typed Modular Loadout VM Object Reference=@n7) -> (then)
   Event On Unhovered()
   > Parent: On Unhovered() {is=title K2Node_CallParentFunction}
   n8 = AND Boolean(
    A=Get bIsShuffle!(self:Fort Typed Modular Loadout VM Object Reference=Get FortTypedModularLoadoutVM()) -> (bIsShuffle:bool)
    B=IsUsingTouch(PlayerController=GetOwningPlayer())
   )
   > n9 = Branch(Condition=@n8)
   n10 = Get Text_Loadoutname()
   SetScrollingEnabled!(execute=[@n16.then, @n9.then], self:Common Text Object Reference=@n10, bInIsScrollingEnabled:bool=false) -> (then)
   > Set IsItemHovered(IsItemHovered:bool=false) -> (Output_Get:bool)
   n11 = Get ContextMenu()
   > n12 = Branch(Condition=IsOpen!(self:Menu Anchor Object Reference=Get Anchor!(self:UIKit Context Menu Anchor Object Reference=@n11) -> (Anchor:Menu Anchor Object Reference)) -> (ReturnValue:bool))
   n13 = Branch(execute=@n12.else, Condition=Get bIsShuffle(self=Get FortTypedModularLoadoutVM()))
   n14 = Get ContextMenu()
   SetVisibility(execute=[@n30.then, @n15.then, @n13.then], self=@n14, InVisibility=Collapsed)
   > Toggle Quick Actions!(execute, self:WBP UIKit Context Menu Anchor Button Base Object Reference=@n14, Enabled:bool=false) -> (then)
   n15 = PlayAnimationReverse(execute=@n13.else, InAnimation=Get OnFocus() -> (OnFocus:Widget Animation Object Reference))
   n16 = Set TextOverflowPolicy!(execute=@n9.else, TextOverflowPolicy:ETextOverflowPolicy Enum=Ellipsis, self:Text Object Reference=@n10) -> (then, Output_Get:ETextOverflowPolicy Enum=Clip)
   Event On Selected()
   > Parent: On Selected() {is=title K2Node_CallParentFunction}
   > n17 = Branch(Condition=IsUsingTouch(PlayerController=GetOwningPlayer()))
   > n18 = SetScrollingEnabled(self=Get Text_Loadoutname(), bInIsScrollingEnabled=true)
   n19 = Get FortTypedModularLoadoutVM()
   Call Selected Tile VM!(
    execute:Exec=[@n18.then, @n17.else]
    self:WBP UIKit Preset Card Object Reference
    PresetContents:Fort Typed Modular Loadout VM Object Reference=@n19
   ) -> (then:Exec) {is=title K2Node_CallDelegate}
   n20 = Get FortTypedModularLoadoutVM()
   n21 = AND Boolean(A=NOT Boolean(A=Get bIsShuffle(self=@n20)), B=IsUsingTouch(PlayerController=GetOwningPlayer()))
   > Branch(Condition=@n21)
   > SetVisibility(self=Get ContextMenu(), InVisibility=Select(Option 0=Collapsed, Option 1=SelfHitTestInvisible, Index=Get bIsValid(self=@n19)))
   > PlayAnimation(InAnimation=Get OnFocus())
   Event On Deselected()
   > Parent: On Deselected() {is=title K2Node_CallParentFunction}
   > n22 = Branch(Condition=IsUsingTouch(PlayerController=GetOwningPlayer()))
   > n23 = SetScrollingEnabled(self=Get Text_Loadoutname(), bInIsScrollingEnabled=false)
   Branch(execute=[@n23.then, @n22.else], Condition=@n21)
   n24 = Get ContextMenu()
   > n25 = Branch(Condition=IsOpen(self=Get Anchor(self=@n24)))
   SetVisibility(execute=@n25.else, self=Get ContextMenu(), InVisibility=Collapsed)
   > PlayAnimationReverse(InAnimation=Get OnFocus())
   Event On Hovered()
   > Parent: On Hovered() {is=title K2Node_CallParentFunction}
   > n26 = Branch(Condition=@n8)
   SetScrollingEnabled(execute=[@n31.then, @n26.then], self=@n10, bInIsScrollingEnabled=true)
   > Set IsItemHovered(IsItemHovered=true)
   > n27 = Branch(Condition=Get bIsValid(self=Get FortTypedModularLoadoutVM()))
   n28 = GetOwningPlayer()
   n29 = Get ContextMenu()
   > SetVisibility(self=@n29, InVisibility=Select(Option 0=Visible, Option 1=SelfHitTestInvisible, Index=IsUsingTouch(PlayerController=@n28)))
   > Toggle Quick Actions(self=@n29, Enabled=true)
   > Call Enable Options Hint!(execute:Exec, self:WBP UIKit Preset Card Object Reference) -> (then:Exec) {is=title K2Node_CallDelegate}
   > PlayAnimation(InAnimation=Get OnFocus())
   n30 = Call Disable Options Hint!(execute:Exec=@n27.else, self:WBP UIKit Preset Card Object Reference) -> (then:Exec) {is=title K2Node_CallDelegate}
   n31 = Set TextOverflowPolicy(execute=@n26.else, TextOverflowPolicy=Clip, self=@n10) -> (Output_Get=Clip)
  ]}
 GetPreviewImage {nodes [
   n1 = GetPreviewImage!() -> (then, ItemImage:Texture 2D Soft Object Reference, SlotInfo:"Fort Locker Category Slot Info Structure (by ref)") {is=title K2Node_FunctionEntry}
   n2 = Get ItemImage() -> (ItemImage:Texture 2D Soft Object Reference)
   > Return Node!(
    execute
    OutItemImage:Soft Object Reference=Select(
     Option 0:Soft Object Reference=Break Fort Locker Category Slot Info!(FortLockerCategorySlotInfo:"Fort Locker Category Slot Info Structure (by ref)"=@n1.SlotInfo) -> (
      CategoryName:Text
      CategoryNamePlural:Text
      CategoryNameWithPluralForms:Text
      CategoryDescription:Text
      DefaultEmptyTexture:Soft Object Reference
      OverrideShuffleTexture:Texture 2D Soft Object Reference
      CategoryIcon:Material Instance Object Reference
      CategorySoundOverrides:Map of Gameplay Tag Structures to Sound Base Soft Object References
      bShowInCategoryTab:bool=~
      PreviewWindowScale:float=~
      PreviewWindowOffset:Vector 3f
      bUsePluralCategoryInItemList:bool=~
     ) {is=title K2Node_BreakStruct}.DefaultEmptyTexture
     Option 1:Soft Object Reference=@n2
     Index=IsValidSoftObjectReference(SoftObjectReference=@n2)
    ) -> (ReturnValue:Soft Object Reference)
   )
  ]}
 IsSaveNewTile {nodes [
   n1 = IsSaveNewTile!() -> (then, IsSaveNewTile:bool=~) {is=title K2Node_FunctionEntry}
   > Branch(Condition=@n1)
   n2 = Get FortTypedModularLoadoutVM()
   n3 = GetLoadoutIndex!(self:Fort Typed Modular Loadout VM Object Reference=@n2) -> (ReturnValue:int=0)
   > SetText!(
    execute
    self:Self Object Reference
    Text:Text=Select(
     Option 0:Text="LOCTEXT(\"Save New\")"
     Option 1:Text=Format Text!(Format="LOCTEXT(\"Save Slot {0}\")", 0:int=@n3) -> (Result)
     Index=Get bIsEmptySlot!(self:Fort Typed Modular Loadout VM Object Reference=@n2) -> (bIsEmptySlot:bool)
    ) -> (ReturnValue:Text)
   ) -> (then)
  ]}
 IsShuffleActive {nodes [
   IsShuffleActive!() -> (then, Void:bool=~) {is=title K2Node_FunctionEntry}
   > Branch(Condition=Get bIsShuffle(self=Get FortTypedModularLoadoutVM()))
   > SetItemImage!(
    execute
    self:WBP UIKit Item Card Base Object Reference
    Item Image:Soft Object Reference=/Game/UI/Foundation/Textures/Icons/Locker/T_Icon_Random_128
   ) -> (then)
   n1 = Get FortTypedModularLoadoutsVM() -> (FortTypedModularLoadoutsVM:Fort Typed Modular Loadouts VM Object Reference)
   n2 = AND Boolean(
    A=Get bShuffleEnabled!(self:Fort Typed Modular Loadouts VM Object Reference=@n1) -> (bShuffleEnabled:bool)
    B=GetCanEnableShuffle!(self:Fort Typed Modular Loadouts VM Object Reference=@n1) -> (ReturnValue:bool)
   )
   > SetText(Text=SelectText(A="LOCTEXT(\"Random On\")", B="LOCTEXT(\"Random Off\")", bPickA=@n2))
   n3 = Select(Option 0:Linear Color Structure=(R=0,G=0,B=0,A=0.6), Option 1:Linear Color Structure=(R=0.3663,G=0.3663,B=0.3663,A=1), Index=@n2) -> (ReturnValue:Linear Color Structure)
   > Set Default Background!(
    execute
    self:WBP UIKit Item Card Base Object Reference
    Color 1:Linear Color Structure=@n3
    Color 2:Linear Color Structure=@n3
    GradientStart:float=0
    GradientEnd:float=0
    GradientAngle:float=0
   ) -> (then) {is=title K2Node_CallFunction}
   > Branch(Condition=IsMobileGame())
   n4 = Get Text_Loadoutname()
   > SetAutoWrapText!(execute, self:Text Object Reference=@n4, InAutoTextWrap:bool=false) -> (then)
   > SetTextOverflowPolicy!(execute, self:Text Object Reference=@n4, InOverflowPolicy:ETextOverflowPolicy Enum=Clip) -> (then)
   > SetStretch!(execute, self:Scale Box Object Reference=Get ScaleBox() -> (ScaleBox:Scale Box Object Reference), InStretch:EStretch Enum=ScaleToFit) -> (then)
  ]}
 IsShuffleLoadoutSufficient {nodes [
   n1 = IsShuffleLoadoutSufficient!() -> (then, ActiveLoadouts:int, IsShuffleTile:bool=~) {is=title K2Node_FunctionEntry}
   n2 = "integer < integer"(A=@n1, B=2)
   > Return Node!(execute, IsTileEnabled:bool=NOT Boolean(A=AND Boolean(A=@n2, B=@n1.IsShuffleTile)))
  ]}
 IsShuffleTile {nodes [
   IsShuffleTile!() -> (then, IsShuffleTile:bool=~, IsShuffled:bool=~, GetCanEnableShuffle:bool=~) {is=title K2Node_FunctionEntry}
   > Return Node!(execute, Void:bool=false)
  ]}
 OnContextMenuClosed {nodes [OnContextMenuClosed() {is=title K2Node_FunctionEntry}]}
 OnContextMenuOpened {nodes [OnContextMenuOpened() {is=title K2Node_FunctionEntry}]}
 OnDeleteEventFired {nodes [OnDeleteEventFired!() -> (then, PresetCard:WBP UIKit Preset Card Object Reference) {is=title K2Node_FunctionEntry}]}
 OnRenameEventFired {nodes [OnRenameEventFired!() -> (then, PresetCard:WBP UIKit Preset Card Object Reference) {is=title K2Node_FunctionEntry}]}
 SelectedTileVM {nodes [SelectedTileVM!() -> (then, PresetContents:Fort Typed Modular Loadout VM Object Reference) {is=title K2Node_FunctionEntry}]}
 Set Image if Valid {nodes [
   n1 = Set Image if Valid!() -> (then, LoadoutItem:Fort Modular Loadout Item VM Object Reference) {is=title K2Node_FunctionEntry}
   > n2 = Is Valid(InputObject=@n1)
   Set Custom Widget!(execute=@n2.Is Not Valid, self:WBP UIKit Item Card Base Object Reference, ItemVM:Fort Item VM Object Reference) -> (then) {is=title K2Node_CallFunction}
   > SetItemImage()
   > Set Default Background(Color 1=(R=0,G=0,B=0,A=0.6), Color 2=(R=0,G=0,B=0,A=0.6), GradientStart=0, GradientEnd=0, GradientAngle=0)
   Set Custom Widget(execute=@n2.Is Valid, ItemVM=Get Item!(self:Fort Modular Loadout Item VM Object Reference=@n1) -> (Item:Fort Item VM Object Reference))
   > Set Default Background(Color 1=(R=0.3663,G=0.3663,B=0.3663,A=1), Color 2=(R=0.3663,G=0.3663,B=0.3663,A=1), GradientStart=0, GradientEnd=0, GradientAngle=0)
  ]}
 Set Shuffle Tile Enabled Style {nodes [
   n1 = Set Shuffle Tile Enabled Style!() -> (then, IsShuffleEnabled:bool=~) {is=title K2Node_FunctionEntry}
   > SetRenderOpacity(InOpacity=Select(Option 0:float=0.3, Option 1:float=1, Index=@n1) -> (ReturnValue:float))
  ]}
 SetContextMenuActionContext {nodes [
   n1 = SetContextMenuActionContext!() -> (then, RenameInInteractionDelegate:"Delegate (by ref)", DeleteInInteractionDelegate:"Delegate (by ref)") {is=title K2Node_FunctionEntry}
   > n2 = Construct Object from Class(Class=/Script/FortniteUI.UIKitContextMenuActionContext) -> (ReturnValue:UIKit Context Menu Action Context Object Reference)
   > AddInteraction!(
    execute
    self:UIKit Context Menu Action Context Object Reference=@n2
    InActionInfo_DisplayName:Text=Get Rename In Action Info Display Name() -> (Rename In Action Info Display Name:Text)
    InActionInfo_IconTexture:Texture 2D Soft Object Reference
    InActionInfo_IconMaterial:Material Instance Soft Object Reference
    InActionInfo_SortPriority:int=1000
    InActionInfo_bIsAvailable:bool=True
    InActionInfo_bIsEnabled:bool=True
    InActionInfo_bIsSelected:bool=False
    InActionInfo_bAddDivider:bool=False
    InActionInfo_bPositionDividerAfter:bool=False
    InActionInfo_bCloseMenuOnSelect:bool=False
    InActionInfo_InteractionNameID:Name=None
    InActionInfo_Interaction:Data Table Row Handle Structure=Get RenameInput() -> (RenameInput:Data Table Row Handle Structure)
    InActionInfo_PressedDelegate:Delegate="(null).None"
    InActionInfo_ReleasedDelegate:Delegate="(null).None"
    InInteractionDelegate:"Delegate (by ref)"=@n1
   ) -> (then)
   > AddInteraction(
    self=@n2
    InActionInfo_DisplayName=Get Delete In Action Info Display Name() -> (Delete In Action Info Display Name:Text)
    InActionInfo_SortPriority=1000
    InActionInfo_bIsAvailable=True
    InActionInfo_bIsEnabled=True
    InActionInfo_bIsSelected=False
    InActionInfo_bAddDivider=False
    InActionInfo_bPositionDividerAfter=False
    InActionInfo_bCloseMenuOnSelect=False
    InActionInfo_InteractionNameID=None
    InActionInfo_Interaction=Get DeleteInput() -> (DeleteInput:Data Table Row Handle Structure)
    InActionInfo_PressedDelegate="(null).None"
    InActionInfo_ReleasedDelegate="(null).None"
    InInteractionDelegate=@n1.DeleteInInteractionDelegate
   )
   > SetContext!(
    execute
    self:WBP UIKit Context Menu Anchor Button Base Object Reference=Get ContextMenu()
    InteractionContext_Data:UIKit Context Menu Action Context Object Reference=@n2
   ) -> (then)
   n3 = Get ContextMenu()
   > SetRenderOpacity(
    self=Get Prompt!(self:WBP UIKit Button Generic Base Object Reference=Get ContextMenuButton!(self:WBP UIKit Context Menu Anchor Button Base Object Reference=@n3) -> (ContextMenuButton:WBP UIKit Button Generic Base Object Reference)) -> (Prompt:Grid Panel Object Reference)
    InOpacity=0
   )
   > Return Node!(execute)
  ]}
 SetContextMenuPadding {nodes [
   SetContextMenuPadding() {is=title K2Node_FunctionEntry}
   n1 = Get ContextMenu()
   > SetPadding!(
    execute
    self:Grid Slot Object Reference=SlotAsGridSlot(Widget=Get Anchor(self=@n1))
    InPadding_Left:float=0
    InPadding_Top:float=0
    InPadding_Right:float=0
    InPadding_Bottom:float=0
   ) -> (then)
   > Return Node()
  ]}
 SetText {nodes [
   n1 = SetText!() -> (then, Text:Text) {is=title K2Node_FunctionEntry}
   > Set Text(Text=@n1, self=Get Text_Loadoutname())
  ]}
 SetupPresetCard {nodes [
   SetupPresetCard!() -> (then, Items:Array of Fort Modular Loadout Item VM Object References) {is=title K2Node_FunctionEntry}
   n1 = GetFirstValidItemInAnyCategory!(self:Fort Typed Modular Loadout VM Object Reference=Get FortTypedModularLoadoutVM()) -> (ReturnValue:Fort Modular Loadout Item VM Object Reference)
   > Set Image if Valid!(execute, self, LoadoutItem:Fort Modular Loadout Item VM Object Reference=@n1) -> (then) {is=title K2Node_CallFunction}
   > Is Valid(InputObject=Get FortTypedModularLoadoutVM())
   > SetFortModularLoadoutItemVM!(execute, self, Viewmodel:Fort Modular Loadout Item VM Object Reference=@n1) -> (then)
   n2 = Get FortTypedModularLoadoutsVM()
   > Is Valid(InputObject=@n2)
   > IsShuffleActive!(execute, self, Void:bool=Get bShuffleEnabled(self=@n2)) -> (then) {is=title K2Node_CallFunction}
  ]}
 Toggle Context Menu If Hovered {nodes [
   Toggle Context Menu If Hovered() {is=title K2Node_FunctionEntry}
   > n1 = Branch(Condition=Get IsItemHovered() -> (IsItemHovered:bool))
   n2 = Get ContextMenu()
   > n3 = Branch(Condition=IsOpen(self=Get Anchor(self=@n2)))
   > Return Node!(execute, ToggleSucess:bool=true)
   ToggleContextMenu(execute=@n3.else, self:WBP UIKit Context Menu Anchor Button Base Object Reference=@n2)
   > Return Node(ToggleSucess=true)
   Return Node(execute=@n1.else, ToggleSucess=false)
  ]}
 ToggleContextMenu {nodes [
   ToggleContextMenu() {is=title K2Node_FunctionEntry}
   > ToggleContextMenu(self=Get ContextMenu()) {is=title K2Node_CallFunction}
  ]}
]
variables [
 Delete In Action Info Display Name {type=string, default=Delete}
 DisableOptionsHint {}
 EnableOptionsHint {}
 IsItemHovered {type=boolean}
 IsSaveShuffleTile {type=boolean}
 LargePreviewImage {type=Texture2D}
 OnContextMenuClosed {}
 OnContextMenuOpened {}
 OnDeleteEventFired {}
 OnRenameEventFired {}
 Rename In Action Info Display Name {type=string, default=Rename}
 SelectedTileVM {}
]
widgets {
 info {
  inheritedWidgetCount=0
  namedSlotCount=1
  parentClass=/FortUILibrary/Systems/Items/WBP_UIKit_ItemCard_Base.WBP_UIKit_ItemCard_Base_C
  rootWidgetClass=null
  widgetCount=7
 }
 widgets [
  Overlay_76 {widgetClassPath=&Overlay, widget=Overlay_76}
  CommonBorder_67 {
   widgetClassPath=&CommonBorder
   parent=Overlay_76
   widget=CommonBorder_67
   properties {background {
     bIsDynamicallyLoaded=false
     drawAs=Image
     imageType=NoImage
     mirroring=NoMirror
     resourceName=None
     resourceObject=/FortUILibrary/Systems/Items/MaterialInstances/MI_UI_Preset_Overlay
     tiling=NoTile
     imageSize=<32 32>
     margin=<ltrb 0 0 0 0>
     outlineSettings {
      bUseBrushTransparency=false
      roundingType=HalfHeightRadius
      width=0
      color {colorUseRule=UseColor_Specified, specifiedColor=<rgba 0 0 0 0>}
      cornerRadii=<0 0 0 0>
     }
     tintColor {colorUseRule=UseColor_Specified, specifiedColor=<rgba 1 1 1 1>}
     uVRegion {bIsValid=false, max=<0 0>, min=<0 0>}
    }}
   slot {class=OverlaySlot, properties {horizontalAlignment=HAlign_Fill, verticalAlignment=VAlign_Fill}}
  }
  CommonBorder_54 {
   widgetClassPath=&CommonBorder
   parent=Overlay_76
   widget=CommonBorder_54
   properties {verticalAlignment=VAlign_Bottom, padding=<ltrb 0 0 0 0>}
   slot {class=OverlaySlot, properties {horizontalAlignment=HAlign_Fill, verticalAlignment=VAlign_Bottom, padding=<ltrb 8 0 8 4>}}
  }
  ScaleBox {
   widgetClassPath=&ScaleBox
   bIsVariable=true
   parent=CommonBorder_54
   widget=ScaleBox
   properties {stretch=None}
   slot {class=BorderSlot, properties {verticalAlignment=VAlign_Bottom, padding=<ltrb 0 0 0 0>}}
  }
  Text_Loadoutname {
   widgetClassPath=&CommonTextBlock
   bIsVariable=true
   parent=ScaleBox
   widget=Text_Loadoutname
   properties {
    autoWrapText=true
    bAutoCollapseWithEmptyText=true
    bIsScrollingEnabled=false
    scrollStyle=/FortUILibrary/Styles/Typography/Foundation/TextScrollStyle-NoFade.TextScrollStyle-NoFade_C
    text=Text Block
    textOverflowPolicy=Ellipsis
    font {
     bForceMonospaced=false
     bMaterialIsStencil=false
     fontMaterial=null
     fontObject=&HeadingNow
     letterSpacing=0
     monospacedWidth=1
     size=21
     skewAmount=0
     typefaceFontName=75 Medium
     outlineSettings {
      bApplyOutlineToDropShadows=false
      bMiteredCorners=false
      bSeparateFillAlpha=false
      outlineMaterial=null
      outlineSize=2
      outlineColor=<rgba 0 0 0 0>
     }
    }
    shadowOffset=<0 0>
   }
   slot {class=ScaleBoxSlot, properties {horizontalAlignment=HAlign_Fill, verticalAlignment=VAlign_Bottom}}
  }
  FortVisualAttachment_0 {
   widgetClassPath=&FortVisualAttachment
   parent=Overlay_76
   widget=FortVisualAttachment_0
   properties {bOverride_HeightOverride=true, bOverride_WidthOverride=true, heightOverride=40, visibility=Visible, widthOverride=40, contentAnchor=<1 0>}
   slot {class=OverlaySlot, properties {horizontalAlignment=HAlign_Right, verticalAlignment=VAlign_Bottom, padding=<ltrb 0 0 8 48>}}
  }
  ContextMenu {
   widgetClassPath=/FortUILibrary/UIKit/ContextMenu/WBP_UIKit_ContextMenuAnchorButton
   bIsVariable=true
   parent=FortVisualAttachment_0
   widget=ContextMenu
   properties {
    button Background Material=/FortUILibrary/UIKit/Buttons/MaterialInstances/CTAButtons/MI_UIKit_BgQuiet
    buttonClass=/FortUILibrary/UIKit/Buttons/WBP_UIKit_Button_Contextual.WBP_UIKit_Button_Contextual_C
    buttonIcon=/FortUILibrary/Textures/Icons/T_UI_Icon_UIKit_ContextMenu
    buttonText=...
    closeMenuOnContextClicked=true
    button Padding=<ltrb 4 4 4 4>
    button Text Font {
     bForceMonospaced=false
     bMaterialIsStencil=false
     fontMaterial=/FortUILibrary/UIKit/Buttons/MaterialInstances/GenericButtons/MI_UIKit_GenericText_Quiet
     fontObject=&Burbank
     letterSpacing=180
     monospacedWidth=1
     size=26.25
     skewAmount=0
     typefaceFontName=BurbankRegular-Black
     outlineSettings {
      bApplyOutlineToDropShadows=false
      bMiteredCorners=false
      bSeparateFillAlpha=false
      outlineMaterial=null
      outlineSize=0
      outlineColor=<rgba 0 0 0 1>
     }
    }
    triggering Input Action {dataTable=&InputActionDataTable, rowName=ContextPopupMenu}
   }
   slot {class=SizeBoxSlot}
  }
 ]
}
