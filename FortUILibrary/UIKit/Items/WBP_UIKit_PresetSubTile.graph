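# Preset sub-tile item card, derived from WBP_UIKit_ItemCard_Base, that shows an item from a modular loadout slot.
# PreConstruct strips the compatibility icon, block and equipped outlines and transference widget, then sets a flat translucent black background.
# SetTileImage uses the slot's empty texture or a no-color-variant default as fallback, and prefers the item's large preview when UseLargeImage is set, otherwise the passed Image.
# Anim_OnSelected swaps the item material to its selected variant.
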WidgetBlueprint /FortUILibrary/UIKit/Items/WBP_UIKit_PresetSubTile
animations [Anim_OnSelected {
  binding_count=1
  length_seconds=0.0501
  object=&self:Anim_OnSelected
  sequence {bindings [WBP_UIKit_PresetSubTile_C_31 {
     id=#1
     possessed_class=WBP_UIKit_PresetSubTile_C
     tracks [{
       type=MovieSceneObjectPropertyTrack
       display_name=Item Material
       sections [{
         type=MovieSceneObjectPropertySection
         end_frame=1
         start_frame=0
         channels [None {
           type=MovieSceneScriptingObjectPathChannel
           keys [
            [
             0
             "<Object '/FortUILibrary/Systems/Items/MaterialInstances/MI_UIKit_PresetSubItemPreview' (0x466e7d77800) Class 'MaterialInstanceConstant'>"
            ]
            [
             1
             "<Object '/FortUILibrary/Systems/Items/MaterialInstances/MI_UIKit_PresetSubItemPreview_Selected' (0x466e7d79800) Class 'MaterialInstanceConstant'>"
            ]
           ]
          }]
        }]
      }]
    }]}
 }]
graphs [
 EventGraph {nodes [
   n1 = Event PreConstruct()
   > Parent: PreConstruct(IsDesignTime=@n1.IsDesignTime)
   > RemoveFromParent(self=[
     Get Compatibility_Icon() -> (Compatibility_Icon:Image Object Reference)
     Get Block_Outline() -> (Block_Outline:WBP UIKit Block Outline Object Reference)
     Get EquippedState_Outline() -> (EquippedState_Outline:WBP UIKit Equipped State Object Reference)
     Get UIKit_Transference() -> (UIKit_Transference:WBP UIKit Transference Object Reference)
    ])
   > Set Default Background!(
    execute
    self:WBP UIKit Item Card Base Object Reference
    Color 1:Linear Color Structure=(R=0,G=0,B=0,A=0.6)
    Color 2:Linear Color Structure=(R=0,G=0,B=0,A=0.6)
    GradientStart:float=0
    GradientEnd:float=0
    GradientAngle:float=0
   ) -> (then) {is=title K2Node_CallFunction}
  ]}
 SetTileImage {nodes [
   SetTileImage!() -> (then, Image:Texture 2D Soft Object Reference) {is=title K2Node_FunctionEntry}
   n1 = Break Fort Locker Category Slot Info!(FortLockerCategorySlotInfo:"Fort Locker Category Slot Info Structure (by ref)"=GetSlotInfo!(self:Fort Modular Loadout Item VM Object Reference=Get FortModularLoadoutItemVM() -> (FortModularLoadoutItemVM:Fort Modular Loadout Item VM Object Reference)) -> (ReturnValue:"Fort Locker Category Slot Info Structure (by ref)")) -> (
    CategoryName:Text
    CategoryNamePlural:Text
    CategoryNameWithPluralForms:Text
    CategoryDescription:Text
    DefaultEmptyTexture:Soft Object Reference
    OverrideShuffleTexture:Texture 2D Soft Object Reference
    CategoryIcon:Material Instance Object Reference
    CategorySoundOverrides:Map of Gameplay Tag Structures to Sound Base Soft Object References
    bShowInCategoryTab:bool=~
    PreviewWindowScale:float=~
    PreviewWindowOffset:Vector 3f
    bUsePluralCategoryInItemList:bool=~
   ) {is=title K2Node_BreakStruct}
   > Set Fallback Image(Fallback Image:Soft Object Reference=Select(
     Option 0:Soft Object Reference=/FortUILibrary/Systems/Items/Textures/T_UI_NoColorVariant
     Option 1:Soft Object Reference=@n1.DefaultEmptyTexture
     Index=IsValidSoftObjectReference(SoftObjectReference=@n1.DefaultEmptyTexture)
    ) -> (ReturnValue:Soft Object Reference)) -> (Output_Get:Soft Object Reference)
   > n2 = Branch(Condition=Get UseLargeImage() -> (UseLargeImage:bool))
   > n3 = Get!(execute:Exec) -> (then:Exec, else:Exec, FortModularLoadoutItemVM:Fort Modular Loadout Item VM Object Reference)
   n4 = GetLargePreviewImage!(ItemDefinition:Item Definition Base Object Reference=Get ItemDefinition!(self:Fort Item VM Object Reference=Get Item!(self:Fort Modular Loadout Item VM Object Reference=@n3) -> (Item:Fort Item VM Object Reference)) -> (ItemDefinition:Fort Item Definition Object Reference)) -> (ReturnValue:Texture 2D Soft Object Reference)
   > n5 = Branch(Condition=IsValidSoftObjectReference(SoftObjectReference=@n4))
   > SetItemImage!(execute, self:WBP UIKit Item Card Base Object Reference, Item Image:Soft Object Reference=@n4) -> (then)
   SetItemImage(execute=[@n2.else, @n5.else, @n3.else], Item Image=Get Image() -> (Image:Texture 2D Soft Object Reference))
  ]}
]
variables [
 LargePreviewImage {type=Texture2D}
 UseLargeImage {type=boolean}
]
widgets {
 info {
  inheritedWidgetCount=0
  namedSlotCount=1
  parentClass=/FortUILibrary/Systems/Items/WBP_UIKit_ItemCard_Base.WBP_UIKit_ItemCard_Base_C
  rootWidgetClass=null
  widgetCount=1
 }
 widgets [CommonBorder_54 {widgetClassPath=&CommonBorder, widget=CommonBorder_54, properties {verticalAlignment=VAlign_Bottom}}]
}
