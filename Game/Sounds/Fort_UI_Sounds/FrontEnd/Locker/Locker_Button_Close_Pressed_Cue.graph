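# UI sound cue for the close-button press in the frontend Locker, playing a single close whoosh wave.
# A modulator scales the wave player's volume by a fixed amount, with the cue's own multiplier lowering it further.
# It routes to the Fort_UI sound class and overrides concurrency to one instance, stopping the oldest on retrigger.
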SoundCue /Game/Sounds/Fort_UI_Sounds/FrontEnd/Locker/Locker_Button_Close_Pressed_Cue
root=Modulator_0
nodes [
 {id=Modulator_0, children [WavePlayer_2], properties {volumeMax=0.8, volumeMin=0.8}}
 {id=WavePlayer_2, properties {soundWaveAssetPtr=&/UI_Locker_Close_Whoosh}}
]
properties {
 bOverrideConcurrency=true
 maxDistance=2097152
 soundClassObject=&Fort_UI
 volumeMultiplier=0.3
 concurrencyOverrides {maxCount=1, resolutionRule=StopOldest}
}
