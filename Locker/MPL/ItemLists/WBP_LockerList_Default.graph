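# Default Locker list widget, derived from WBP_LockerList_Base, showing items as WBP_UIKit_ItemCard_Locker cards in a tile view.
# It relays card context-menu, modal, details and streaming events through its dispatchers, and binds preview actions to hovered or selected cards.
# Context-menu toggling on touch first scrolls the selection into view via a one-shot callback.
# Impression telemetry gathers displayed, non-shuffle item template IDs, and emote-wheel mode collapses each card's context-menu attachment.
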WidgetBlueprint /Locker/MPL/ItemLists/WBP_LockerList_Default
graphs [
 AddImpressionTelemetry {nodes [
   AddImpressionTelemetry!() -> (then, TabFlowRiderString:String) {is=title K2Node_FunctionEntry}
   > n1 = For Each Loop(Array:Array of User Widget Object References=GetDisplayedEntryWidgets!(self:List View Base Object Reference=Get LockerTiles() -> (LockerTiles:Fort Tile View Object Reference)) -> (ReturnValue:Array of User Widget Object References)) -> (Array Element:User Widget Object Reference)
   n2 = Get SectionContentsArray() -> (SectionContentsArray:Array of Flow Analytics Impression Section Content Structures)
   Branch(
    execute=@n1.Completed
    Condition="integer > integer"(A=Length(TargetArray:Array of Flow Analytics Impression Section Content Structures=@n2))
   )
   > "Fire Event Add Impression (Non-Localized)"!(
    execute
    SectionID:String=Get TabFlowRiderString() -> (TabFlowRiderString:String)
    SectionContents:Array of Flow Analytics Impression Section Content Structures=Get SectionContentsArray()
   ) -> (then)
   n3 = Cast To WBP_UIKit_ItemCard_Locker!(execute=@n1.LoopBody, Object=@n1) -> (then, CastFailed, AsWBP UIKit Item Card Locker:WBP UIKit Item Card Locker Object Reference) {is=title K2Node_DynamicCast}
   > Is Valid(InputObject=@n3)
   n4 = Get FortItemVM!(self:WBP UIKit Item Card Locker Object Reference=@n3) -> (FortItemVM:Fort Item VM Object Reference)
   > Branch(Condition=NOT Boolean(A=Get bIsShuffleTile!(self:Fort Item VM Object Reference=@n4) -> (bIsShuffleTile:bool)))
   n5 = GetItemTemplateId!(self:Fort Item VM Object Reference=@n4) -> (ReturnValue:String)
   > Branch(Condition=NOT Boolean(A=IsEmpty(InString=@n5)))
   > Add(
    TargetArray:Array of Flow Analytics Impression Section Content Structures=Get SectionContentsArray()
    NewItem:"Flow Analytics Impression Section Content Structure (by ref)"=Make Flow Analytics Impression Section Content!(ContentID:Text="To Text (String)"(InString=@n5), OfferID:Text, Row:int=-1, Column:int=-1, Index:int=-1) -> (FlowAnalyticsImpressionSectionContent:Flow Analytics Impression Section Content Structure) {is=title K2Node_MakeStruct}
   ) {is=title K2Node_CallArrayFunction}
  ]}
 BindFunctions {nodes [
   BindFunctions() {is=title K2Node_FunctionEntry}
   > n1 = Sequence!(execute) -> (then_0, then_1, then_2:Exec, then_3:Exec, then_4:Exec, then_5:Exec, then_6:Exec)
   > Bind Event to On Context Menu Opened(
    self:WBP UIKit Item Card Locker Object Reference=Get CachedWBPUIKitItemCardLocker() -> (CachedWBPUIKitItemCardLocker:WBP UIKit Item Card Locker Object Reference)
    Delegate=Create Event()
   ) {is=title K2Node_AddDelegate}
   Bind Event to On Context Menu Closed(execute=@n1.then_1, self:WBP UIKit Item Card Locker Object Reference=Get CachedWBPUIKitItemCardLocker(), Delegate=Create Event()) {is=title K2Node_AddDelegate}
   Bind Event to On Hovered(execute=@n1.then_2, self:Common Button Base Object Reference=Get CachedWBPUIKitItemCardLocker(), Delegate=Create Event()) {is=title K2Node_AddDelegate}
   Bind Event to On Context Menu Edit Styles Clicked(execute=@n1.then_3, self:WBP UIKit Item Card Locker Object Reference=Get CachedWBPUIKitItemCardLocker(), Delegate=Create Event()) {is=title K2Node_AddDelegate}
   Bind Event to Go to Locker Item Details(execute=@n1.then_4, self:WBP UIKit Item Card Locker Object Reference=Get CachedWBPUIKitItemCardLocker(), Delegate=Create Event()) {is=title K2Node_AddDelegate}
   Bind Event to Is Modal Open(execute=@n1.then_5, self:WBP UIKit Item Card Locker Object Reference=Get CachedWBPUIKitItemCardLocker(), Delegate=Create Event()) {is=title K2Node_AddDelegate}
   Bind Event to On Wait for Streaming Content(execute=@n1.then_6, self:WBP UIKit Item Card Locker Object Reference=Get CachedWBPUIKitItemCardLocker(), Delegate=Create Event()) {is=title K2Node_AddDelegate}
  ]}
 EventGraph {nodes [
   n1 = "On Entry Generated (LockerTiles)"!() -> (OutputDelegate, then, Widget:User Widget Object Reference) {is=title K2Node_ComponentBoundEvent}
   > n2 = Cast To WBP_UIKit_ItemCard_Locker(Object=@n1.Widget)
   > Set IsEmoteWheel!(execute, IsEmoteWheel:bool=Get IsEmoteWheel() -> (IsEmoteWheel:bool), self:WBP UIKit Item Card Locker Object Reference=@n2) -> (then, Output_Get:bool)
   > SetVisibility(
    self=Get ContextMenu_Attachment!(self:WBP UIKit Item Card Locker Object Reference=@n2) -> (ContextMenu_Attachment:Fort Visual Attachment Object Reference)
    InVisibility=Select(Option 0=Visible, Option 1=Collapsed, Index=Get IsEmoteWheel())
   )
   n3 = BP_OnItemScrolledIntoView_Event!() -> (OutputDelegate, then, Item:Object Reference, Widget:User Widget Object Reference) {is=title K2Node_CustomEvent}
   > Unbind Event from On Item Scrolled Into View(self:List View Object Reference=Get LockerTiles(), Delegate=@n3) {is=title K2Node_RemoveDelegate}
   > ToggleSelectedContextInternal()
   n4 = Event SetupBinds!() -> (OutputDelegate, then, Widget:Widget Object Reference, PreviewWidget:Fort Item Preview Widget Object Reference) {is=title K2Node_Event}
   > Set CachedPreviewWidget(CachedPreviewWidget:Fort Item Preview Widget Object Reference=@n4.PreviewWidget) -> (Output_Get:Fort Item Preview Widget Object Reference)
   > n5 = Cast To WBP_UIKit_ItemCard_Locker(Object=@n4.Widget)
   > Set CachedWBPUIKitItemCardLocker(CachedWBPUIKitItemCardLocker:WBP UIKit Item Card Locker Object Reference=@n5) -> (Output_Get:WBP UIKit Item Card Locker Object Reference)
   > BindFunctions() {is=title K2Node_CallFunction}
   Event ToggleSelectedContext() {is=title K2Node_Event}
   n6 = Get CommonInputSubsystem()
   > n7 = Branch(Condition="Equal (Enum)"(A=GetCurrentInputType(self=@n6), B=Touch))
   n8 = Get Selected Item!(self:List View Object Reference=Get LockerTiles()) -> (ReturnValue:Object Reference) {is=title K2Node_CallFunction}
   > Is Valid(InputObject=@n8)
   > Unbind Event from On Item Scrolled Into View(self=Get LockerTiles(), Delegate=@n3)
   > Assign On Item Scrolled Into View!(execute:Exec, self:List View Object Reference=Get LockerTiles(), Delegate:"Delegate (by ref)"=@n3) -> (then:Exec) {is=title K2Node_AssignDelegate}
   > Scroll Item Into View!(execute, self:List View Object Reference=Get LockerTiles(), Item:Object Reference=@n8) -> (then)
   ToggleSelectedContextInternal(execute=@n7.else)
   Event SetContextVisibility() {is=title K2Node_Event}
   > n9 = For Each Loop(Array=GetDisplayedEntryWidgets(self=Get LockerTiles()))
   > n10 = Cast To WBP_UIKit_ItemCard_Locker(Object=@n9)
   > SetVisibility(
    self=Get ContextMenu!(self:WBP UIKit Item Card Locker Object Reference=@n10) -> (ContextMenu:WBP UIKit Context Menu Anchor Button Object Reference)
    InVisibility=Hidden
   )
   n11 = Event SetupPreviewActionBinds!() -> (OutputDelegate, then, PreviewWidget:Fort Item Preview Widget Object Reference) {is=title K2Node_Event}
   > n12 = For Each Loop with Break!(Exec:Exec, Array:Array of User Widget Object References=GetDisplayedEntryWidgets(self=Get LockerTiles()), Break:Exec=@n14.then) -> (LoopBody:Exec, Array Element:User Widget Object Reference, Array Index:int, Completed:Exec)
   > n13 = Cast To WBP_UIKit_ItemCard_Locker(Object=@n12)
   > Branch(Condition=GetSelected!(self:Common Button Base Object Reference=@n13) -> (ReturnValue:bool))
   > Set Hovered Item(Hovered Item:WBP UIKit Item Card Locker Object Reference=@n13) -> (Output_Get:WBP UIKit Item Card Locker Object Reference)
   > n14 = PreviewActions!(execute, self:WBP Locker List Base Object Reference, PreviewWidget:Fort Item Preview Widget Object Reference=@n11.PreviewWidget) -> (then)
   n15 = Event SetupTelemetry!() -> (OutputDelegate, then, TabFlowRiderString:String) {is=title K2Node_Event}
   > AddImpressionTelemetry!(execute, self:WBP Locker List Base Object Reference, TabFlowRiderString:String=@n15.TabFlowRiderString) -> (then) {is=title K2Node_CallFunction}
   n16 = Event UpdateIsEmoteWheel!() -> (OutputDelegate, then, IsEmoteWheel:bool) {is=title K2Node_Event}
   > Set IsEmoteWheel!(execute, IsEmoteWheel=@n16.IsEmoteWheel) -> (then, Output_Get)
   n17 = Event BindPreviewActions!() -> (OutputDelegate, then, Preview Widget:Fort Item Preview Widget Object Reference) {is=title K2Node_Event}
   > PreviewActions(PreviewWidget=@n17.Preview Widget)
  ]}
 GetListView {nodes [
   GetListView() {is=title K2Node_FunctionEntry}
   > Return Node!(execute, TileView:Fort Tile View Object Reference=Get LockerTiles())
  ]}
 OnContextEditStylesClickedBind {nodes [
   OnContextEditStylesClickedBind() {is=title K2Node_FunctionEntry}
   > Call Context Edit Styles Clicked!(execute:Exec, self:WBP Locker List Default Object Reference) -> (then:Exec) {is=title K2Node_CallDelegate}
  ]}
 OnGoToLockerItemDetailsBind {nodes [
   n1 = OnGoToLockerItemDetailsBind!() -> (then, CustomItemDetails:User Widget Class Reference) {is=title K2Node_FunctionEntry}
   > Call Go to Locker Item Details!(execute:Exec, self:WBP Locker List Base Object Reference, CustomItemDetails:User Widget Class Reference=@n1) -> (then:Exec) {is=title K2Node_CallDelegate}
  ]}
 OnItemCardHoveredBind {nodes [
   n1 = OnItemCardHoveredBind!() -> (then, Button:Common Button Base Object Reference) {is=title K2Node_FunctionEntry}
   > n2 = Cast To WBP_UIKit_ItemCard_Locker(Object=@n1)
   > Set Hovered Item(Hovered Item=@n2)
   n3 = Get CommonInputSubsystem()
   > Branch(Condition="Not Equal (Enum)"!(A:ECommonInputType Enum=GetCurrentInputType(self=@n3), B:ECommonInputType Enum=Gamepad) -> (ReturnValue:bool))
   > PreviewActions(PreviewWidget=Get CachedPreviewWidget() -> (CachedPreviewWidget:Fort Item Preview Widget Object Reference))
  ]}
 OnModalOpenBind {nodes [
   n1 = OnModalOpenBind!() -> (then, IsOpen:bool=~) {is=title K2Node_FunctionEntry}
   > Call On Modal Changed!(execute:Exec, self:WBP Locker List Base Object Reference, IsModalOpen:bool=@n1) -> (then:Exec) {is=title K2Node_CallDelegate}
  ]}
 OnTileContextMenuClosedBind {nodes [
   OnTileContextMenuClosedBind() {is=title K2Node_FunctionEntry}
   > Call On Tile Context Menu Closed!(execute:Exec, self:WBP Locker List Base Object Reference) -> (then:Exec) {is=title K2Node_CallDelegate}
  ]}
 OnTileContextMenuOpenBind {nodes [
   n1 = OnTileContextMenuOpenBind!() -> (then, ContextMenuWidgetParentEntry:Widget Object Reference) {is=title K2Node_FunctionEntry}
   > Call On Tile Context Menu Open!(execute:Exec, self:WBP Locker List Base Object Reference, Widget:Widget Object Reference=@n1) -> (then:Exec) {is=title K2Node_CallDelegate}
  ]}
 OnWaitForStreamingContentBind {nodes [
   OnWaitForStreamingContentBind() {is=title K2Node_FunctionEntry}
   > Call On Wait for Streaming Content!(execute:Exec, self:WBP Locker List Base Object Reference) -> (then:Exec) {is=title K2Node_CallDelegate}
  ]}
 PreviewActions {nodes [
   n1 = PreviewActions!() -> (then, PreviewWidget:Fort Item Preview Widget Object Reference) {is=title K2Node_FunctionEntry}
   n2 = Get Hovered Item() -> (Hovered Item:WBP UIKit Item Card Locker Object Reference)
   > Is Valid(InputObject=@n2)
   > Bind Preview Actions!(
    execute
    self:WBP UIKit Item Card Base Object Reference=@n2
    Preview Widget:Fort Item Preview Widget Object Reference=@n1
    Context Menu:WBP UIKit Context Menu Anchor Button Base Object Reference=Get ContextMenu(self=@n2)
    Item VM:Fort Item VM Object Reference=Get FortItemVM(self=@n2)
    Include Special Actions:bool=true
    Include Toggle Actions:bool=true
   ) -> (then)
  ]}
 ToggleSelectedContextInternal {nodes [
   ToggleSelectedContextInternal() {is=title K2Node_FunctionEntry}
   > n1 = For Each Loop(Array=GetDisplayedEntryWidgets(self=Get LockerTiles()))
   > n2 = Cast To WBP_UIKit_ItemCard_Locker(Object=@n1)
   > ToggleContextMenuIfSelected(self:WBP UIKit Item Card Locker Object Reference=@n2)
  ]}
]
variables [
 CachedPreviewWidget {type=FortItemPreviewWidget, default=null}
 CachedWBPUIKitItemCardLocker {type=WBP_UIKit_ItemCard_Locker_C, default=null}
 Hovered Item {type=WBP_UIKit_ItemCard_Locker_C, default=null}
 IsEmoteWheel {type=boolean}
]
widgets {
 info {
  inheritedWidgetCount=0
  parentClass=&/WBP_LockerList_Base.WBP_LockerList_Base_C
  rootWidgetClass=&GridPanel
  widgetCount=2
 }
 widgets [
  GridPanel_0 {widgetClassPath=&GridPanel, widget=GridPanel_0, properties {columnFill [1], rowFill [1]}}
  LockerTiles {
   widgetClassPath=/Script/FortniteUI.FortTileView
   bIsVariable=true
   parent=GridPanel_0
   widget=LockerTiles
   properties {
    bConsumeTouchInputWhenScrolling=false
    clipping=Inherit
    entryHeight=152
    entryWidgetClass=/Locker/UIKit/Items/WBP_UIKit_ItemCard_Locker.WBP_UIKit_ItemCard_Locker_C
    entryWidth=120
    horizontalEntrySpacing=16
    navigation=&self:WidgetTree.LockerTiles.WidgetNavigation_0
    numDesignerPreviewEntries=20
    scrollbarDisabledVisibility=Hidden
    tileAlignment=EvenlySize
    verticalEntrySpacing=24
    scrollBarStyle {
     thickness=16
     draggedThumbImage {
      bIsDynamicallyLoaded=false
      drawAs=RoundedBox
      imageType=FullColor
      mirroring=NoMirror
      resourceName=../../../Engine/Content/Slate/Common/Scrollbar_Thumb.png
      resourceObject=null
      tiling=NoTile
      imageSize=<12 12>
      margin=<ltrb 0.25 0.25 0.25 0.25>
      outlineSettings {
       bUseBrushTransparency=false
       roundingType=FixedRadius
       width=2
       color {colorUseRule=UseColor_Specified, specifiedColor=<rgba 0 0 0 0>}
       cornerRadii=<6 6 6 6>
      }
      tintColor {colorUseRule=UseColor_Specified, specifiedColor=<rgba 1 1 1 1>}
      uVRegion {bIsValid=false, max=<0 0>, min=<0 0>}
     }
     horizontalBackgroundImage {
      bIsDynamicallyLoaded=false
      drawAs=NoDrawType
      imageType=NoImage
      mirroring=NoMirror
      resourceName=None
      resourceObject=null
      tiling=NoTile
      imageSize=<0 0>
      margin=<ltrb 0 0 0 0>
      outlineSettings {
       bUseBrushTransparency=false
       roundingType=HalfHeightRadius
       width=0
       color {colorUseRule=UseColor_Specified, specifiedColor=<rgba 0 0 0 0>}
       cornerRadii=<0 0 0 0>
      }
      tintColor {colorUseRule=UseColor_Specified, specifiedColor=<rgba 1 1 1 1>}
      uVRegion {bIsValid=false, max=<0 0>, min=<0 0>}
     }
     horizontalBottomSlotImage {
      bIsDynamicallyLoaded=false
      drawAs=NoDrawType
      imageType=FullColor
      mirroring=NoMirror
      resourceName=../../../Engine/Content/Slate/Common/Scrollbar_Background_Horizontal.png
      resourceObject=null
      tiling=NoTile
      imageSize=<8 8>
      margin=<ltrb 0 0 0 0>
      outlineSettings {
       bUseBrushTransparency=false
       roundingType=HalfHeightRadius
       width=0
       color {colorUseRule=UseColor_Specified, specifiedColor=<rgba 0 0 0 0>}
       cornerRadii=<0 0 0 0>
      }
      tintColor {colorUseRule=UseColor_Specified, specifiedColor=<rgba 1 1 1 1>}
      uVRegion {bIsValid=false, max=<0 0>, min=<0 0>}
     }
     horizontalTopSlotImage {
      bIsDynamicallyLoaded=false
      drawAs=NoDrawType
      imageType=FullColor
      mirroring=NoMirror
      resourceName=../../../Engine/Content/Slate/Common/Scrollbar_Background_Horizontal.png
      resourceObject=null
      tiling=NoTile
      imageSize=<8 8>
      margin=<ltrb 0 0 0 0>
      outlineSettings {
       bUseBrushTransparency=false
       roundingType=HalfHeightRadius
       width=0
       color {colorUseRule=UseColor_Specified, specifiedColor=<rgba 0 0 0 0>}
       cornerRadii=<0 0 0 0>
      }
      tintColor {colorUseRule=UseColor_Specified, specifiedColor=<rgba 1 1 1 1>}
      uVRegion {bIsValid=false, max=<0 0>, min=<0 0>}
     }
     hoveredThumbImage {
      bIsDynamicallyLoaded=false
      drawAs=RoundedBox
      imageType=FullColor
      mirroring=NoMirror
      resourceName=../../../Engine/Content/Slate/Common/Scrollbar_Thumb.png
      resourceObject=null
      tiling=NoTile
      imageSize=<12 12>
      margin=<ltrb 0.25 0.25 0.25 0.25>
      outlineSettings {
       bUseBrushTransparency=false
       roundingType=FixedRadius
       width=2
       color {colorUseRule=UseColor_Specified, specifiedColor=<rgba 0 0 0 0>}
       cornerRadii=<6 6 6 6>
      }
      tintColor {colorUseRule=UseColor_Specified, specifiedColor=<rgba 1 1 1 1>}
      uVRegion {bIsValid=false, max=<0 0>, min=<0 0>}
     }
     normalThumbImage {
      bIsDynamicallyLoaded=false
      drawAs=RoundedBox
      imageType=FullColor
      mirroring=NoMirror
      resourceName=../../../Engine/Content/Slate/Common/Scrollbar_Thumb.png
      resourceObject=null
      tiling=NoTile
      imageSize=<12 12>
      margin=<ltrb 0.25 0.25 0.25 0.25>
      outlineSettings {
       bUseBrushTransparency=false
       roundingType=FixedRadius
       width=2
       color {colorUseRule=UseColor_Specified, specifiedColor=<rgba 0 0 0 0>}
       cornerRadii=<6 6 6 6>
      }
      tintColor {colorUseRule=UseColor_Specified, specifiedColor=<rgba 1 1 1 0.75>}
      uVRegion {bIsValid=false, max=<0 0>, min=<0 0>}
     }
     verticalBackgroundImage {
      bIsDynamicallyLoaded=false
      drawAs=Box
      imageType=NoImage
      mirroring=NoMirror
      resourceName=None
      resourceObject=/FortUILibrary/UIKit/ScrollBar/MaterialInstances/MI_UIKit_ScrollbarBackground
      tiling=NoTile
      imageSize=<12 12>
      margin=<ltrb 0.5 0.5 0.5 0.5>
      outlineSettings {
       bUseBrushTransparency=false
       roundingType=HalfHeightRadius
       width=0
       color {colorUseRule=UseColor_Specified, specifiedColor=<rgba 0 0 0 0>}
       cornerRadii=<0 0 0 0>
      }
      tintColor {colorUseRule=UseColor_Specified, specifiedColor=<rgba 1 1 1 1>}
      uVRegion {bIsValid=false, max=<0 0>, min=<0 0>}
     }
     verticalBottomSlotImage {
      bIsDynamicallyLoaded=false
      drawAs=NoDrawType
      imageType=FullColor
      mirroring=NoMirror
      resourceName=../../../Engine/Content/Slate/Common/Scrollbar_Background_Vertical.png
      resourceObject=null
      tiling=NoTile
      imageSize=<8 8>
      margin=<ltrb 0 0 0 0>
      outlineSettings {
       bUseBrushTransparency=false
       roundingType=HalfHeightRadius
       width=0
       color {colorUseRule=UseColor_Specified, specifiedColor=<rgba 0 0 0 0>}
       cornerRadii=<0 0 0 0>
      }
      tintColor {colorUseRule=UseColor_Specified, specifiedColor=<rgba 1 1 1 1>}
      uVRegion {bIsValid=false, max=<0 0>, min=<0 0>}
     }
     verticalTopSlotImage {
      bIsDynamicallyLoaded=false
      drawAs=NoDrawType
      imageType=FullColor
      mirroring=NoMirror
      resourceName=../../../Engine/Content/Slate/Common/Scrollbar_Background_Vertical.png
      resourceObject=null
      tiling=NoTile
      imageSize=<8 8>
      margin=<ltrb 0 0 0 0>
      outlineSettings {
       bUseBrushTransparency=false
       roundingType=HalfHeightRadius
       width=0
       color {colorUseRule=UseColor_Specified, specifiedColor=<rgba 0 0 0 0>}
       cornerRadii=<0 0 0 0>
      }
      tintColor {colorUseRule=UseColor_Specified, specifiedColor=<rgba 1 1 1 1>}
      uVRegion {bIsValid=false, max=<0 0>, min=<0 0>}
     }
    }
   }
   slot {class=GridSlot}
  }
 ]
}
