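# Activatable locker modal that lets the player choose a sort, a filter and Game Compatibility options.
# Section clicks write pending choices to FortSortFilterSelectionVM and enable Clear.
# Apply commits them, sends telemetry, fires SortFilterAccepted and deactivates; Close reverts and sends telemetry; Clear resets to defaults and reselects the first entries.
# It also plays an intro, adjusts mobile padding and shows an active count on Apply.
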WidgetBlueprint /Locker/MPL/SortFilter/WBP_MPLocker_SortFilter_Modal
animations [Intro {binding_count=0, length_seconds=0.2986, object=&self:Intro, sequence {bindings []}}]
graphs [
 BP_GetDesiredFocusTarget {nodes [
   Get Desired Focus Target()
   n1 = Get DynBox_Entries!(self:WBP MPLocker Sort Filter Section Object Reference=Get Sort_Section() -> (Sort_Section:WBP MPLocker Sort Filter Section Object Reference)) -> (DynBox_Entries:Dynamic Entry Box Object Reference)
   > Return Node!(
    execute
    ReturnValue:Widget Object Reference="Get (a copy)"(Array:Array of User Widget Object References=GetAllEntries!(self:Dynamic Entry Box Base Object Reference=@n1) -> (ReturnValue:Array of User Widget Object References)) -> (Output:User Widget Object Reference)
   )
  ]}
 EventGraph {nodes [
   n1 = "On Item Clicked (Filter_Section)"!() -> (OutputDelegate, then, Item Filter VM:Fort Sort Filter Item VM Object Reference) {is=title K2Node_ComponentBoundEvent}
   > n2 = Cast To FortFilterItemVM!(execute, Object=@n1.Item Filter VM) -> (then, CastFailed, AsFort Filter Item VM:Fort Filter Item VM Object Reference) {is=title K2Node_DynamicCast}
   > SetPendingFilterItem!(
    execute
    self:Fort Sort Filter Selection VM Object Reference=Get FortSortFilterSelectionVM() -> (FortSortFilterSelectionVM:Fort Sort Filter Selection VM Object Reference)
    InFilterItem:Fort Filter Item VM Object Reference=@n2
   ) -> (then)
   n3 = Get Button_Clear() -> (Button_Clear:WBP UIKit Button Regular Object Reference)
   > SetIsEnabled(self=@n3, bInIsEnabled=true)
   n4 = "On Item Clicked (Sort_Section)"!() -> (OutputDelegate, then, Item Filter VM:Fort Sort Filter Item VM Object Reference) {is=title K2Node_ComponentBoundEvent}
   > n5 = Cast To FortSortItemVM!(execute, Object=@n4.Item Filter VM) -> (then, CastFailed, AsFort Sort Item VM:Fort Sort Item VM Object Reference) {is=title K2Node_DynamicCast}
   > SetPendingSortItem!(
    execute
    self:Fort Sort Filter Selection VM Object Reference=Get FortSortFilterSelectionVM()
    InSortItem:Fort Sort Item VM Object Reference=@n5
   ) -> (then)
   > SetIsEnabled(self=@n3, bInIsEnabled=true)
   n6 = "On Clicked (Button_Apply)"!() -> (OutputDelegate, then, Button:Common Button Base Object Reference) {is=title K2Node_ComponentBoundEvent}
   > ApplyPending(self:Fort Sort Filter Selection VM Object Reference=Get FortSortFilterSelectionVM())
   > SendTelemetryList!(execute, self, Button:Common Button Base Object Reference=@n6.Button) -> (then)
   > Call Sort Filter Accepted!(execute:Exec, self:WBP MPLocker Sort Filter Modal Object Reference) -> (then:Exec) {is=title K2Node_CallDelegate}
   > DeactivateWidget()
   n7 = "On Item Clicked (Option_Section)"!() -> (OutputDelegate, then, ItemFilterVM:Fort Sort Filter Item VM Object Reference) {is=title K2Node_ComponentBoundEvent}
   > n8 = Cast To FortOptionItemVM!(execute, Object=@n7.ItemFilterVM) -> (then, CastFailed, AsFort Option Item VM:Fort Option Item VM Object Reference) {is=title K2Node_DynamicCast}
   > TogglePendingOptionItem!(
    execute
    self:Fort Sort Filter Selection VM Object Reference=Get FortSortFilterSelectionVM()
    OptionItem:Fort Option Item VM Object Reference=@n8
   ) -> (then)
   > SetIsEnabled(self=Get Button_Clear(), bInIsEnabled=true)
   n9 = "On Clicked (Button_Clear)"!() -> (OutputDelegate, then, Button:Common Button Base Object Reference) {is=title K2Node_ComponentBoundEvent}
   > ResetPending!(execute, self:Fort Sort Filter Selection VM Object Reference=Get FortSortFilterSelectionVM(), bIsDefault:bool=true) -> (then)
   > Fire Event Element Interacted(Element=@n9.Button)
   > SetIsEnabled(self=@n9.Button, bInIsEnabled=false)
   > ResetToDefault()
   n10 = "On Clicked (Button_Close)"!() -> (OutputDelegate, then, Button:Common Button Base Object Reference) {is=title K2Node_ComponentBoundEvent}
   > ResetPending(self=Get FortSortFilterSelectionVM(), bIsDefault=false)
   > SendTelemetryList(Button=@n10.Button)
   > DeactivateWidget()
   n11 = Event PreConstruct()
   > SetCompactUIStyle() {is=title K2Node_CallFunction}
   > Branch(Condition=@n11.IsDesignTime)
   > n12 = Construct Object from Class(Class=&/DialogVM_MPLocker_SortFilter.DialogVM_MPLocker_SortFilter_C) -> (ReturnValue:Dialog VM MPLocker Sort Filter Object Reference)
   > SetDialogData!(
    execute
    self:WBP UIKit Dialog Internal Object Reference=Get Dialog_Internal() -> (Dialog_Internal:WBP UIKit Dialog Internal Object Reference)
    Data:UI Kit Dialog Viewmodel Object Reference=@n12
   ) -> (then)
   Event On Activated()
   > Parent: On Activated() {is=title K2Node_CallParentFunction}
   > PlayAnimation(InAnimation=Get Intro())
   > n13 = Construct Object from Class(Class=&DialogVM_MPLocker_SortFilter_C)
   > SetDialogData(self=Get Dialog_Internal(), Data=@n13)
  ]}
 ResetToDefault {nodes [
   ResetToDefault() {is=title K2Node_FunctionEntry}
   > SelectButtonAtIndex!(
    execute
    self:Common Button Group Base Object Reference=Get ButtonGroup!(self:WBP MPLocker Sort Filter Section Object Reference=Get Sort_Section()) -> (ButtonGroup:Common Button Group Base Object Reference)
    ButtonIndex:int=0
    bAllowSound:bool=true
   ) -> (then)
   > SelectButtonAtIndex(
    self=Get ButtonGroup(self=Get Filter_Section() -> (Filter_Section:WBP MPLocker Sort Filter Section Object Reference))
    ButtonIndex=0
    bAllowSound=true
   )
   n1 = Get Option_Section() -> (Option_Section:WBP MPLocker Sort Filter Options Section Object Reference)
   > n2 = For Each Loop(Array:Array of User Widget Object References=GetAllEntries(self=Get DynBox_Entries(self:WBP MPLocker Sort Filter Options Section Object Reference=@n1))) -> (Array Element:User Widget Object Reference)
   > n3 = Cast To WBP_MPLocker_SortFilterOptionButton!(execute, Object=@n2) -> (then, CastFailed, AsWBP MPLocker Sort Filter Option Button:WBP MPLocker Sort Filter Option Button Object Reference) {is=title K2Node_DynamicCast}
   > SetIsSelected!(
    execute
    self:Common Button Base Object Reference=Get WBP_UIKit_ButtonWithToggle!(self:WBP MPLocker Sort Filter Option Button Object Reference=@n3) -> (WBP_UIKit_ButtonWithToggle:WBP UIKit Button With Toggle Object Reference)
    InSelected:bool=false
    bGiveClickFeedback:bool=false
   ) -> (then)
  ]}
 SendTelemetryList {nodes [
   SendTelemetryList!() -> (then, Button:Common Button Base Object Reference) {is=title K2Node_FunctionEntry}
   > Get!(execute:Exec, self:Fort Sort Filter Selection VM Object Reference=Get FortSortFilterSelectionVM()) -> (then:Exec, else:Exec, PendingSortItem:Fort Sort Item VM Object Reference)
   > Set AdditionalInfo(AdditionalInfo:Flow Analytics Additional Info Structure="Add String (Non-Localized)"!(
     Info:"Flow Analytics Additional Info Structure (by ref)"=Make Flow Analytics Additional Info!() -> (FlowAnalyticsAdditionalInfo:Flow Analytics Additional Info Structure) {is=title K2Node_MakeStruct}
     Key:String=Interaction
     Value:String=Current Sort and Filter Settings
    ) -> (ReturnValue:Flow Analytics Additional Info Structure)) -> (Output_Get:Flow Analytics Additional Info Structure)
   > n1 = Keys!(
    execute
    TargetMap:Map of Fort Option Item VM Object References to Booleans=Get PendingOptionItems!(self:Fort Sort Filter Selection VM Object Reference=Get FortSortFilterSelectionVM()) -> (PendingOptionItems:Map of Fort Option Item VM Object References to Booleans)
   ) -> (then, Keys:Array of Fort Option Item VM Object References)
   > n2 = For Each Loop(Array:Array of Fort Option Item VM Object References=@n1) -> (Array Element:Fort Option Item VM Object Reference)
   n3 = Get FortSortFilterSelectionVM()
   n4 = Get Source String from Text!(Text:Text=Get Label!(self:Fort Sort Filter Item VM Object Reference=Get PendingSortItem!(self:Fort Sort Filter Selection VM Object Reference=@n3) -> (PendingSortItem:Fort Sort Item VM Object Reference)) -> (Label:Text)) -> (ReturnValue:String) {is=title K2Node_CallFunction}
   n5 = IsEmpty(InString=@n4)
   Set AdditionalInfo(
    execute=@n2.Completed
    AdditionalInfo="Add String (Non-Localized)"(
     Info=Get AdditionalInfo() -> (AdditionalInfo:Flow Analytics Additional Info Structure)
     Key=Sort
     Value=Select(Option 0:String=@n4, Option 1:String=Latest Acquired, Index=@n5) -> (ReturnValue:String)
    )
   )
   n6 = Get FortSortFilterSelectionVM()
   n7 = Get Source String from Text(Text=Get Label(self=Get PendingFilterItem!(self:Fort Sort Filter Selection VM Object Reference=@n6) -> (PendingFilterItem:Fort Filter Item VM Object Reference)))
   n8 = IsEmpty(InString=@n7)
   > Set AdditionalInfo(AdditionalInfo="Add String (Non-Localized)"(Info=Get AdditionalInfo(), Key=Filter, Value=Select(Option 0=@n7, Option 1=All, Index=@n8)))
   > n9 = Cast To FortPlayerController(Object=GetOwningPlayer())
   > "Fire Event Element Interacted (Non-Localized)"!(
    execute
    FortPC:Fort Player Controller Object Reference=@n9
    ElementName:String=GetDisplayName(Object=Get Button() -> (Button:Common Button Base Object Reference))
    AdditionalInfo:Flow Analytics Additional Info Structure=Get AdditionalInfo()
   ) -> (then)
   n10 = Values!(
    execute=@n2.LoopBody
    TargetMap:Map of Fort Option Item VM Object References to Booleans=Get PendingOptionItems(self=Get FortSortFilterSelectionVM())
   ) -> (then, Values:Array of Booleans)
   n11 = "Get (a copy)"(Array:Array of Booleans=@n10, Dimension 1=@n2.Array Index) -> (Output:bool)
   n12 = Get Label(self=@n2)
   > Set AdditionalInfo(AdditionalInfo="Add String (Non-Localized)"(Info=Get AdditionalInfo(), Key=Get Source String from Text(Text=@n12), Value="To String (Boolean)"(InBool=@n11)))
  ]}
 SetCompactUIStyle {nodes [
   SetCompactUIStyle() {is=title K2Node_FunctionEntry}
   > Branch(Condition=IsMobileGame())
   n1 = SlotAsVerticalBoxSlot(Widget=Get Option_Section())
   n2 = Get!(self:Vertical Box Slot Object Reference=@n1) -> (Padding_Left:float=~, Padding_Top:float=~, Padding_Right:float=~, Padding_Bottom:float=~)
   > SetPadding!(
    execute
    self:Vertical Box Slot Object Reference=@n1
    InPadding_Left:float=@n2
    InPadding_Top:float=48
    InPadding_Right:float=@n2.Padding_Right
    InPadding_Bottom:float=48
   ) -> (then)
  ]}
 SortFilterAccepted {nodes [SortFilterAccepted() {is=title K2Node_FunctionEntry}]}
 UpdateActiveCount {nodes [
   n1 = UpdateActiveCount!() -> (then, ActiveItems:int) {is=title K2Node_FunctionEntry}
   n2 = Get StatusIndicator!(self:WBP UIKit Button Generic Base Object Reference=Get Button_Apply() -> (Button_Apply:WBP UIKit Button Generic Object Reference)) -> (StatusIndicator:WBP UIKit Status Indicator Object Reference)
   > SetVisibility(
    self=@n2
    InVisibility=Select(Option 0:ESlateVisibility Enum=Collapsed, Option 1:ESlateVisibility Enum=HitTestInvisible, Index="integer > integer"(A=@n1)) -> (ReturnValue:ESlateVisibility Enum)
   )
   > "SetText (Text)"(
    self=Get Indicator_Text!(self:WBP UIKit Status Indicator Base Object Reference=@n2) -> (Indicator_Text:Common Text Object Reference)
    InText="To Text (Integer)"(Value=@n1)
   )
  ]}
]
variables [
 Filter Items {type="array<FortFilterItemVM>", default []}
 Option Items {type="array<FortOptionItemVM>", default []}
 Saved Active Filter Item {type=FortFilterItemVM, default=null}
 Saved Active Sort Item {type=FortSortItemVM, default=null}
 Sort Items {type="array<FortSortItemVM>", default []}
 SortFilterAccepted {}
]
widgets {
 info {
  inheritedWidgetCount=0
  namedSlotCount=1
  parentClass=&CommonActivatableWidget
  rootWidgetClass=/FortUILibrary/Systems/Dialog/WBP_UIKit_Dialog_Internal.WBP_UIKit_Dialog_Internal_C
  widgetCount=12
 }
 widgets [
  Dialog_Internal {
   widgetClassPath=/FortUILibrary/Systems/Dialog/WBP_UIKit_Dialog_Internal
   bIsVariable=true
   widget=Dialog_Internal
   properties {
    buttons {
     each Button Horizontal Fill=true
     each Button Min Width=0
     each Button Size=Medium
     focuse Button=0
     space Between Buttons=30
     cTA Buttons Types [Regular,Quiet]
     group Placement {
      horizontal Alignment=HAlign_Fill
      vertical Alignment=VAlign_Bottom
      grid {column=1, column Span=0, layer=0, row=2, row Span=0, nudge=<0 0>}
      padding=<ltrb 80 64 80 48>
     }
    }
    dialogSIze {
     height Override {enable=false, value=800}
     max Aspect Ratio {enable=false, value=0}
     max Desired Height {enable=true, value=980}
     max Desired Width {enable=false, value=2000}
     min Aspect Ratio {enable=false, value=0}
     min Desired Height {enable=true, value=400}
     min Desired Width {enable=false, value=800}
     width Override {enable=true, value=1200}
    }
    titleProperties {
     auto Wrap Text=true
     justification=Left
     line Height Percentage=1
     mobile Font Size Multiplier=1.3
     transform Policy=ToUpper
     wrap Text At=0
     color {colorUseRule=UseColor_Specified, specifiedColor=<rgba 1 1 1 1>}
     font {
      bForceMonospaced=false
      bMaterialIsStencil=false
      fontMaterial=null
      fontObject=&HeadingNow
      letterSpacing=1
      monospacedWidth=1
      size=25.5
      skewAmount=0
      typefaceFontName=86 Bold Italic
      outlineSettings {
       bApplyOutlineToDropShadows=false
       bMiteredCorners=false
       bSeparateFillAlpha=false
       outlineMaterial=null
       outlineSize=0
       outlineColor=<rgba 0 0 0 1>
      }
     }
     placement {
      horizontal Alignment=HAlign_Fill
      vertical Alignment=VAlign_Top
      grid {column=1, column Span=0, layer=0, row=0, row Span=0, nudge=<0 0>}
      padding=<ltrb 90 80 80 32>
     }
    }
   }
  }
  MainContent {widgetClassPath=&Overlay, namedSlotHost=Dialog_Internal, widget=MainContent}
  ScrollBox_1 {
   widgetClassPath=&ScrollBox
   parent=MainContent
   widget=ScrollBox_1
   properties {
    bAnimateWheelScrolling=true
    navigationScrollPadding=100
    scrollWhenFocusChanges=AnimatedScroll
    wheelScrollMultiplier=3
    scrollbarPadding=<ltrb 2 0 2 0>
    scrollbarThickness=<10 10>
    widgetBarStyle {
     thickness=10
     draggedThumbImage {
      bIsDynamicallyLoaded=false
      drawAs=Box
      imageType=FullColor
      mirroring=NoMirror
      resourceName=../../../Engine/Content/Slate/Common/Scrollbar_Thumb.png
      resourceObject=/FortUILibrary/Foundation/Art_ToolKit/Tools/M_UI_Circle_White
      tiling=NoTile
      imageSize=<10 10>
      margin=<ltrb 0.5 0.5 0.5 0.5>
      outlineSettings {
       bUseBrushTransparency=false
       roundingType=HalfHeightRadius
       width=0
       color {colorUseRule=UseColor_Specified, specifiedColor=<rgba 0 0 0 0>}
       cornerRadii=<0 0 0 0>
      }
      tintColor {colorUseRule=UseColor_Specified, specifiedColor=<rgba 1 1 1 1>}
      uVRegion {bIsValid=false, max=<0 0>, min=<0 0>}
     }
     horizontalBackgroundImage {
      bIsDynamicallyLoaded=false
      drawAs=NoDrawType
      imageType=NoImage
      mirroring=NoMirror
      resourceName=None
      resourceObject=null
      tiling=NoTile
      imageSize=<0 0>
      margin=<ltrb 0 0 0 0>
      outlineSettings {
       bUseBrushTransparency=false
       roundingType=HalfHeightRadius
       width=0
       color {colorUseRule=UseColor_Specified, specifiedColor=<rgba 0 0 0 0>}
       cornerRadii=<0 0 0 0>
      }
      tintColor {colorUseRule=UseColor_Specified, specifiedColor=<rgba 1 1 1 1>}
      uVRegion {bIsValid=false, max=<0 0>, min=<0 0>}
     }
     horizontalBottomSlotImage {
      bIsDynamicallyLoaded=false
      drawAs=Image
      imageType=FullColor
      mirroring=NoMirror
      resourceName=../../../Engine/Content/Slate/Common/Scrollbar_Background_Horizontal.png
      resourceObject=null
      tiling=NoTile
      imageSize=<8 8>
      margin=<ltrb 0 0 0 0>
      outlineSettings {
       bUseBrushTransparency=false
       roundingType=HalfHeightRadius
       width=0
       color {colorUseRule=UseColor_Specified, specifiedColor=<rgba 0 0 0 0>}
       cornerRadii=<0 0 0 0>
      }
      tintColor {colorUseRule=UseColor_Specified, specifiedColor=<rgba 1 1 1 1>}
      uVRegion {bIsValid=false, max=<0 0>, min=<0 0>}
     }
     horizontalTopSlotImage {
      bIsDynamicallyLoaded=false
      drawAs=Image
      imageType=FullColor
      mirroring=NoMirror
      resourceName=../../../Engine/Content/Slate/Common/Scrollbar_Background_Horizontal.png
      resourceObject=null
      tiling=NoTile
      imageSize=<8 8>
      margin=<ltrb 0 0 0 0>
      outlineSettings {
       bUseBrushTransparency=false
       roundingType=HalfHeightRadius
       width=0
       color {colorUseRule=UseColor_Specified, specifiedColor=<rgba 0 0 0 0>}
       cornerRadii=<0 0 0 0>
      }
      tintColor {colorUseRule=UseColor_Specified, specifiedColor=<rgba 1 1 1 1>}
      uVRegion {bIsValid=false, max=<0 0>, min=<0 0>}
     }
     hoveredThumbImage {
      bIsDynamicallyLoaded=false
      drawAs=Box
      imageType=FullColor
      mirroring=NoMirror
      resourceName=../../../Engine/Content/Slate/Common/Scrollbar_Thumb.png
      resourceObject=&M_UI_Circle_White
      tiling=NoTile
      imageSize=<10 10>
      margin=<ltrb 0.5 0.5 0.5 0.5>
      outlineSettings {
       bUseBrushTransparency=false
       roundingType=HalfHeightRadius
       width=0
       color {colorUseRule=UseColor_Specified, specifiedColor=<rgba 0 0 0 0>}
       cornerRadii=<0 0 0 0>
      }
      tintColor {colorUseRule=UseColor_Specified, specifiedColor=<rgba 1 1 1 1>}
      uVRegion {bIsValid=false, max=<0 0>, min=<0 0>}
     }
     normalThumbImage {
      bIsDynamicallyLoaded=false
      drawAs=Box
      imageType=FullColor
      mirroring=NoMirror
      resourceName=../../../Engine/Content/Slate/Common/Scrollbar_Thumb.png
      resourceObject=&M_UI_Circle_White
      tiling=NoTile
      imageSize=<10 10>
      margin=<ltrb 0.5 0.5 0.5 0.5>
      outlineSettings {
       bUseBrushTransparency=false
       roundingType=HalfHeightRadius
       width=0
       color {colorUseRule=UseColor_Specified, specifiedColor=<rgba 0 0 0 0>}
       cornerRadii=<1 1 1 1>
      }
      tintColor {colorUseRule=UseColor_Specified, specifiedColor=<rgba 1 1 1 1>}
      uVRegion {bIsValid=false, max=<0 0>, min=<0 0>}
     }
     verticalBackgroundImage {
      bIsDynamicallyLoaded=false
      drawAs=Box
      imageType=NoImage
      mirroring=NoMirror
      resourceName=None
      resourceObject=&M_UI_Circle_White
      tiling=NoTile
      imageSize=<10 10>
      margin=<ltrb 0.5 0.5 0.5 0.5>
      outlineSettings {
       bUseBrushTransparency=false
       roundingType=HalfHeightRadius
       width=9
       color {colorUseRule=UseColor_Specified, specifiedColor=<rgba 0 0 0 0>}
       cornerRadii=<1 1 1 1>
      }
      tintColor {colorUseRule=UseColor_Specified, specifiedColor=<rgba 1 1 1 0.2>}
      uVRegion {bIsValid=false, max=<0 0>, min=<0 0>}
     }
     verticalBottomSlotImage {
      bIsDynamicallyLoaded=false
      drawAs=NoDrawType
      imageType=FullColor
      mirroring=NoMirror
      resourceName=../../../Engine/Content/Slate/Common/Scrollbar_Background_Vertical.png
      resourceObject=null
      tiling=NoTile
      imageSize=<10 10>
      margin=<ltrb 0.5 0.5 0.5 0.5>
      outlineSettings {
       bUseBrushTransparency=false
       roundingType=HalfHeightRadius
       width=0
       color {colorUseRule=UseColor_Specified, specifiedColor=<rgba 0 0 0 0>}
       cornerRadii=<0 0 0 0>
      }
      tintColor {colorUseRule=UseColor_Specified, specifiedColor=<rgba 1 1 1 1>}
      uVRegion {bIsValid=false, max=<0 0>, min=<0 0>}
     }
     verticalTopSlotImage {
      bIsDynamicallyLoaded=false
      drawAs=NoDrawType
      imageType=FullColor
      mirroring=NoMirror
      resourceName=../../../Engine/Content/Slate/Common/Scrollbar_Background_Vertical.png
      resourceObject=null
      tiling=NoTile
      imageSize=<8 8>
      margin=<ltrb 0 0 0 0>
      outlineSettings {
       bUseBrushTransparency=false
       roundingType=HalfHeightRadius
       width=9
       color {colorUseRule=UseColor_Specified, specifiedColor=<rgba 0 0 0 0>}
       cornerRadii=<0.5 0.5 0.5 0.5>
      }
      tintColor {colorUseRule=UseColor_Specified, specifiedColor=<rgba 1 1 1 1>}
      uVRegion {bIsValid=false, max=<0 0>, min=<0 0>}
     }
    }
    widgetStyle {
     barThickness=10
     bottomShadowBrush {
      bIsDynamicallyLoaded=false
      drawAs=Box
      imageType=FullColor
      mirroring=NoMirror
      resourceName=../../../Engine/Content/Slate/Common/ScrollBoxShadowBottom.png
      resourceObject=null
      tiling=NoTile
      imageSize=<16 8>
      margin=<ltrb 0 0 0 0>
      outlineSettings {
       bUseBrushTransparency=false
       roundingType=HalfHeightRadius
       width=0
       color {colorUseRule=UseColor_Specified, specifiedColor=<rgba 0 0 0 0>}
       cornerRadii=<0 0 0 0>
      }
      tintColor {colorUseRule=UseColor_Specified, specifiedColor=<rgba 1 1 1 0>}
      uVRegion {bIsValid=false, max=<0 0>, min=<0 0>}
     }
     horizontalScrolledContentPadding=<ltrb 0 0 1 0>
     leftShadowBrush {
      bIsDynamicallyLoaded=false
      drawAs=Box
      imageType=FullColor
      mirroring=NoMirror
      resourceName=../../../Engine/Content/Slate/Common/ScrollBoxShadowLeft.png
      resourceObject=null
      tiling=NoTile
      imageSize=<8 16>
      margin=<ltrb 1 0.5 0 0.5>
      outlineSettings {
       bUseBrushTransparency=false
       roundingType=HalfHeightRadius
       width=0
       color {colorUseRule=UseColor_Specified, specifiedColor=<rgba 0 0 0 0>}
       cornerRadii=<0 0 0 0>
      }
      tintColor {colorUseRule=UseColor_Specified, specifiedColor=<rgba 1 1 1 1>}
      uVRegion {bIsValid=false, max=<0 0>, min=<0 0>}
     }
     rightShadowBrush {
      bIsDynamicallyLoaded=false
      drawAs=Box
      imageType=FullColor
      mirroring=NoMirror
      resourceName=../../../Engine/Content/Slate/Common/ScrollBoxShadowRight.png
      resourceObject=null
      tiling=NoTile
      imageSize=<8 16>
      margin=<ltrb 0 0.5 1 0.5>
      outlineSettings {
       bUseBrushTransparency=false
       roundingType=HalfHeightRadius
       width=0
       color {colorUseRule=UseColor_Specified, specifiedColor=<rgba 0 0 0 0>}
       cornerRadii=<0 0 0 0>
      }
      tintColor {colorUseRule=UseColor_Specified, specifiedColor=<rgba 1 1 1 0>}
      uVRegion {bIsValid=false, max=<0 0>, min=<0 0>}
     }
     topShadowBrush {
      bIsDynamicallyLoaded=false
      drawAs=NoDrawType
      imageType=FullColor
      mirroring=NoMirror
      resourceName=../../../Engine/Content/Slate/Common/ScrollBoxShadowTop.png
      resourceObject=null
      tiling=NoTile
      imageSize=<1 90>
      margin=<ltrb 0 0 0 0>
      outlineSettings {
       bUseBrushTransparency=false
       roundingType=HalfHeightRadius
       width=0
       color {colorUseRule=UseColor_Specified, specifiedColor=<rgba 0 0 0 0>}
       cornerRadii=<0 0 0 0>
      }
      tintColor {colorUseRule=UseColor_Specified, specifiedColor=<rgba 1 1 1 0>}
      uVRegion {bIsValid=false, max=<0 0>, min=<0 0>}
     }
     verticalScrolledContentPadding=<ltrb 0 0 0 1>
    }
   }
   slot {class=OverlaySlot, properties {horizontalAlignment=HAlign_Fill, verticalAlignment=VAlign_Fill, padding=<ltrb 72 148 72 136>}}
  }
  VerticalBox_0 {
   widgetClassPath=&VerticalBox
   parent=ScrollBox_1
   widget=VerticalBox_0
   slot {class=ScrollBoxSlot, properties {padding=<ltrb 8 0 8 0>}}
  }
  Option_Section {
   widgetClassPath=&/WBP_MPLocker_SortFilterOptions_Section
   bIsVariable=true
   parent=VerticalBox_0
   widget=Option_Section
   properties {sectionLabel=Game Compatibility}
   slot {class=VerticalBoxSlot}
  }
  HorizontalBox_0 {widgetClassPath=&HorizontalBox, parent=VerticalBox_0, widget=HorizontalBox_0, slot {class=VerticalBoxSlot}}
  Sort_Section {
   widgetClassPath=&/WBP_MPLocker_SortFilter_Section
   bIsVariable=true
   parent=HorizontalBox_0
   widget=Sort_Section
   properties {bIsFocusable=false, sectionLabel=Sort}
   slot {class=HorizontalBoxSlot, properties {size {sizeRule=Fill, value=1}}}
  }
  Filter_Section {
   widgetClassPath=&WBP_MPLocker_SortFilter_Section
   bIsVariable=true
   parent=HorizontalBox_0
   widget=Filter_Section
   properties {bIsFocusable=false, sectionLabel=Filter}
   slot {class=HorizontalBoxSlot, properties {size {sizeRule=Fill, value=1}}}
  }
  HorizontalBox {
   widgetClassPath=&HorizontalBox
   parent=MainContent
   widget=HorizontalBox
   slot {class=OverlaySlot, properties {horizontalAlignment=HAlign_Fill, verticalAlignment=VAlign_Bottom, padding=<ltrb 80 0 80 40>}}
  }
  Button_Apply {
   widgetClassPath=&WBP_UIKit_Button_Generic
   bIsVariable=true
   parent=HorizontalBox
   widget=Button_Apply
   properties {
    bIsFocusable=false
    backgroundMaterial=/FortUILibrary/UIKit/Buttons/MaterialInstances/CTAButtons/MI_UIKit_BgRegular
    configuration=Number
    enableFocusedSelectedState=false
    mobileMinHeight=80
    statusIndicatorVerticalAlignment=VAlign_Center
    text=Apply
    useDarkInputIcon=true
    useIconBlock=false
    useStatusIndicator=true
    buttonPadding=<ltrb 0 0 0 0>
    font {
     bForceMonospaced=false
     bMaterialIsStencil=false
     fontMaterial=/FortUILibrary/UIKit/Buttons/MaterialInstances/CTAButtons/MI_UIKit_TextQuiet
     fontObject=&HeadingNow
     letterSpacing=48
     monospacedWidth=1
     size=18
     skewAmount=0
     typefaceFontName=76 Bold
     outlineSettings {
      bApplyOutlineToDropShadows=false
      bMiteredCorners=false
      bSeparateFillAlpha=false
      outlineMaterial=null
      outlineSize=0
      outlineColor=<rgba 0 0 0 1>
     }
    }
    promptNudge=<24 0>
    statusIndicatorPadding=<ltrb 0 0 24 0>
    textPadding=<ltrb 96 0 96 0>
    triggeringInputAction {dataTable=&InputActionDataTable, rowName=GenericFaceButtonTop}
   }
   slot {class=HorizontalBoxSlot, properties {padding=<ltrb 0 0 24 0>}}
  }
  Button_Clear {
   widgetClassPath=&WBP_UIKit_Button_Regular
   bIsVariable=true
   parent=HorizontalBox
   widget=Button_Clear
   properties {
    bIsFocusable=false
    buttonType=Quiet
    text=Clear
    triggeringInputAction {dataTable=&InputActionDataTable, rowName=GenericFaceButtonLeft}
   }
   slot {class=HorizontalBoxSlot}
  }
  Button_Close {
   widgetClassPath=&WBP_UIKit_Button_Regular
   bIsVariable=true
   parent=HorizontalBox
   widget=Button_Close
   properties {
    bIsFocusable=false
    buttonType=Quiet
    text=Close
    triggeringInputAction {dataTable=&InputActionDataTable, rowName=GenericBack}
   }
   slot {class=HorizontalBoxSlot, properties {horizontalAlignment=HAlign_Right, padding=<ltrb 24 0 0 0>, size {sizeRule=Fill, value=1}}}
  }
 ]
}
