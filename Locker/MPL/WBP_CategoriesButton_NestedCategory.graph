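# Widget Blueprint for a Locker category or nested subcategory button, built on the UIKit generic modular button.
# Update functions set icon, text and a new-item bang, while archetype functions read the category group view model to show primary, shuffle or empty icons.
# Enabled and disabled events toggle opacity to dim the button.
# On Construct, mobile builds hide the text block, resize the icon and round the tab background corners.
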WidgetBlueprint /Locker/MPL/WBP_CategoriesButton_NestedCategory
animations [ShowText {binding_count=0, length_seconds=0.1787, object=&self:ShowText, sequence {bindings []}}]
graphs [
 EventGraph {nodes [
   n1 = RefreshBang!() -> (OutputDelegate, then, AnyNewItems:bool=~) {is=title K2Node_CustomEvent}
   > SetStatusIndicator!(
    execute
    self:WBP UIKit Button Generic Base Object Reference
    BangType:EFortBangType Enum=Select(Option 0:EFortBangType Enum=Invalid, Option 1:EFortBangType Enum=STWLocker, Index=@n1.AnyNewItems) -> (ReturnValue:EFortBangType Enum)
   ) -> (then)
   n2 = ChangeEnabledStyle!() -> (OutputDelegate, then, Enabled:bool=~) {is=title K2Node_CustomEvent}
   > n3 = Branch(Condition=@n2.Enabled)
   SetColorAndOpacity(execute=[@n7.then, @n3.then], self:User Widget Object Reference, InColorAndOpacity:Linear Color Structure=(R=1,G=1,B=1,A=1))
   SetColorAndOpacity(execute=[@n6.then, @n3.else], InColorAndOpacity=(R=1,G=1,B=1,A=0.4))
   n4 = Event PreConstruct()
   > Parent: PreConstruct(IsDesignTime=@n4.IsDesignTime)
   Event Construct()
   > Parent: Construct()
   > Branch(Condition=IsMobileGame())
   > SetVisibility(self=Get Block_Text() -> (Block_Text:WBP UIKit Block Text Object Reference), InVisibility=Collapsed)
   > SetIconSize!(execute, self:WBP UIKit Button Generic Base Object Reference, Icon:Vector 2D Structure=(X=64,Y=64)) -> (then)
   > n5 = Cast To WBP_UIKit_BlockBackground_Tab_Enhanced!(execute, Object=Get Block_Background() -> (Block_Background:WBP UIKit Block Background Base Object Reference)) -> (then, CastFailed, AsWBP UIKit Block Background Tab Enhanced:WBP UIKit Block Background Tab Enhanced Object Reference) {is=title K2Node_DynamicCast}
   > SetCornerRadiuses!(execute, self:WBP UIKit Block Background Tab Enhanced Object Reference=@n5, CornerRadius:float=20, RadiusInPixels:bool=true) -> (then)
   n6 = Event On Disabled() {is=title K2Node_Event}
   n7 = Event On Enabled() {is=title K2Node_Event}
  ]}
 IsPrimaryArchetype {nodes [
   n1 = IsPrimaryArchetype!() -> (then, IsActive:bool=~) {is=title K2Node_FunctionEntry}
   n2 = Get CategoryGroupCollection!(self:Fort Locker Category Group VM Object Reference=Get FortLockerCategoryGroupVM() -> (FortLockerCategoryGroupVM:Fort Locker Category Group VM Object Reference)) -> (CategoryGroupCollection:Fort Locker Category Group Collection VM Object Reference)
   > Branch(Condition=IsArchetype!(self:Fort Locker Category Group Collection VM Object Reference=@n2) -> (ReturnValue:bool))
   > n3 = Branch(Condition=@n1)
   > SetIcon!(execute, self:WBP UIKit Button Generic Base Object Reference, Icon:Object Reference=/FortUILibrary/Textures/Icons/T_UI_Icon_UIKit_Online) -> (then)
   n4 = Branch(
    execute=@n3.else
    Condition=Get bIsShuffleArchetype!(self:Fort Locker Category Group Collection VM Object Reference=@n2) -> (bIsShuffleArchetype:bool)
   )
   SetIcon(execute=@n4.else, Icon=/Game/Athena/UI/Frontend/Spatial/Art/T_Ui_Empty)
  ]}
 IsShuffleArchetype {nodes [
   n1 = IsShuffleArchetype!() -> (then, IsActive:bool=~) {is=title K2Node_FunctionEntry}
   n2 = Get FortLockerCategoryGroupVM()
   > Branch(Condition=IsArchetype(self=Get CategoryGroupCollection(self=@n2)))
   > n3 = Branch(Condition=@n1)
   > SetIcon(Icon=/FortUILibrary/Textures/Icons/T_UI_Shuffle)
   n4 = Branch(execute=@n3.else, Condition=Get bIsActiveArchetype!(self:Fort Locker Category Group VM Object Reference=@n2) -> (bIsActiveArchetype:bool))
   SetIcon(execute=@n4.else, Icon=&T_Ui_Empty)
  ]}
 UpdateCategoryIcon {nodes [
   n1 = UpdateCategoryIcon!() -> (then, Texture:Texture 2D Object Reference) {is=title K2Node_FunctionEntry}
   > Is Valid(InputObject=@n1)
   > SetIcon(Icon=@n1)
  ]}
 UpdateCategoryNotification {nodes [
   n1 = UpdateCategoryNotification!() -> (then, Bang:bool=~) {is=title K2Node_FunctionEntry}
   > RefreshBang!(execute, self, AnyNewItems:bool=@n1) -> (then) {is=title K2Node_CallFunction}
  ]}
 UpdateCategoryText {nodes [
   n1 = UpdateCategoryText!() -> (then, TestInputText:Text) {is=title K2Node_FunctionEntry}
   > n2 = Branch(Condition=TextIsEmpty(InText=@n1))
   SetText!(execute=@n2.else, self:WBP UIKit Button Generic Base Object Reference=Self-Reference(), Text:Text=@n1) -> (then)
  ]}
 UpdateSubCategoryIcon {nodes [
   n1 = UpdateSubCategoryIcon!() -> (then, Texture:Texture 2D Object Reference) {is=title K2Node_FunctionEntry}
   > Is Valid(InputObject=@n1)
   > SetIcon(Icon=@n1)
  ]}
 UpdateSubCategoryNotification {nodes [
   n1 = UpdateSubCategoryNotification!() -> (then, Bang:bool=~) {is=title K2Node_FunctionEntry}
   > RefreshBang(AnyNewItems=@n1)
  ]}
 UpdateSubCategoryText {nodes [
   n1 = UpdateSubCategoryText!() -> (then, TestInputText:Text) {is=title K2Node_FunctionEntry}
   > n2 = Branch(Condition=TextIsEmpty(InText=@n1))
   SetText(execute=@n2.else, self=Self-Reference(), Text=@n1)
  ]}
]
variables [
 CurrentCatName {type=string, default=""}
 CurrentSubCatName {type=string, default=""}
 LockerVM {type=FortLockerCategoryGroupVM, default=null}
]
widgets {
 info {
  inheritedWidgetCount=0
  parentClass=/FortUILibrary/Systems/ModularButton/WBP_UIKit_ButtonGeneric_Base.WBP_UIKit_ButtonGeneric_Base_C
  rootWidgetClass=null
  widgetCount=0
 }
 widgets []
}
