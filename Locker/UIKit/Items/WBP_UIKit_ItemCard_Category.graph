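# Locker category tile widget: an item card showing a category slot's item, driving hover and selection in the item selection viewmodel.
# A context menu, rebuilt from action interactions, offers rename, favorite, edit styles, apply to all and randomize.
# Menu visibility, selectability and focus branch on mouse, gamepad or touch input.
# Helpers set fallback image, badge count, status indicator and emote wheel icon; delegates and analytics report interactions.
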WidgetBlueprint /Locker/UIKit/Items/WBP_UIKit_ItemCard_Category
animations [Anim_Hovered {
  binding_count=1
  length_seconds=0.2004
  object=&self:Anim_Hovered
  sequence {bindings [WBP_UIKit_ItemCard_Category_C_12 {
     id=#1
     possessed_class=WBP_UIKit_ItemCard_Category_C
     tracks [{
       type=MovieScene2DTransformTrack
       sections [{
         type=MovieScene2DTransformSection
         end_frame=4
         start_frame=0
         channels [
          Translation.X {type=MovieSceneScriptingFloatChannel, const [0,1]}
          Translation.Y {type=MovieSceneScriptingFloatChannel, const [0,1]}
          Scale.X {type=MovieSceneScriptingFloatChannel, keys [[0,1],[4,1.012]]}
          Scale.Y {type=MovieSceneScriptingFloatChannel, keys [[0,1],[4,1.012]]}
          Shear.X {type=MovieSceneScriptingFloatChannel, const [0,1]}
          Shear.Y {type=MovieSceneScriptingFloatChannel, const [0,1]}
         ]
        }]
      }]
    }]}
 }]
graphs [
 AdjustContextMenuTouch {nodes [
   AdjustContextMenuTouch() {is=title K2Node_FunctionEntry}
   n1 = IsUsingTouch(PlayerController=GetOwningPlayer())
   n2 = AND Boolean(
    A=Get bIsShuffleTile!(self:Fort Item VM Object Reference=Get FortItemVM() -> (FortItemVM:Fort Item VM Object Reference)) -> (bIsShuffleTile:bool)
    B=@n1
   )
   > SetPadding!(
    execute
    self:Grid Slot Object Reference=SlotAsGridSlot(Widget=Get FortVisualAttachment() -> (FortVisualAttachment:Fort Visual Attachment Object Reference))
    InPadding_Left:float=0
    InPadding_Top:float=0
    InPadding_Right:float=SelectFloat(A=-8, B=0, bPickA=@n2)
    InPadding_Bottom:float=SelectFloat(A=16, B=0, bPickA=@n2)
   ) -> (then)
   > SetVisibility(
    self=Get ContextMenuButton!(self:WBP UIKit Context Menu Anchor Button Base Object Reference=Get ContextMenu() -> (ContextMenu:WBP UIKit Context Menu Anchor Button Object Reference)) -> (ContextMenuButton:WBP UIKit Button Generic Base Object Reference)
    InVisibility=Select(Option 0=Visible, Option 1=HitTestInvisible, Index=@n1)
   )
   n3 = GetSelected!(self:Common Button Base Object Reference) -> (ReturnValue:bool)
   > SetVisibility(
    self=Get UIKit_Transference() -> (UIKit_Transference:WBP UIKit Transference Object Reference)
    InVisibility=Select(Option 0=SelfHitTestInvisible, Option 1=Hidden, Index=AND Boolean(A=@n1, B=@n3))
   )
  ]}
 AnalyticsCategoryInteracted {nodes [
   n1 = AnalyticsCategoryInteracted!() -> (then, Context:String) {is=title K2Node_FunctionEntry}
   n2 = GetAnalyticsName!(self:Fort Locker Category Item VM Object Reference=Get FortLockerCategoryItemVM() -> (FortLockerCategoryItemVM:Fort Locker Category Item VM Object Reference)) -> (ReturnValue:String)
   > FireEventElementInteractedByName!(
    execute
    Element Name:String="To String (Text)"(InText=Format Text!(Format="INVTEXT(\"{0} CategoryTile {1}\")", 0:String=@n2, 1:String=@n1) -> (Result))
    WidgetCallingFrom:Widget Object Reference=Self-Reference()
   ) -> (then, Success:bool)
  ]}
 BP_OnHoldTriggered {nodes [
   n1 = On Hold Triggered!() -> (then, CurrentInputType:ECommonInputType Enum) {is=title K2Node_FunctionEntry}
   > n2 = Branch(Condition="Equal (Enum)"(A=@n1, B=Touch))
   > n3 = Set SelectedCategory!(
    execute
    SelectedCategory:Fort Locker Category Item VM Object Reference=Get FortLockerCategoryItemVM()
    self:Fort Item Selection VM Object Reference=Get FortItemSelectionVM() -> (FortItemSelectionVM:Fort Item Selection VM Object Reference)
   ) -> (then, Output_Get:Fort Locker Category Item VM Object Reference)
   > Set HoveredCategory!(
    execute
    HoveredCategory:Fort Locker Category Item VM Object Reference=@n3
    self:Fort Item Selection VM Object Reference=Get FortItemSelectionVM()
   ) -> (then, Output_Get:Fort Locker Category Item VM Object Reference)
   > SetIsSelected!(execute, self:Common Button Base Object Reference, InSelected:bool=true, bGiveClickFeedback:bool=false) -> (then)
   n4 = Get Anchor!(self:UIKit Context Menu Anchor Object Reference=Get ContextMenu()) -> (Anchor:Menu Anchor Object Reference)
   > n5 = Branch(Condition=IsOpen!(self:Menu Anchor Object Reference=@n4) -> (ReturnValue:bool))
   Call Category Card Selected!(
    execute:Exec=[@n6.then, @n5.then]
    self:WBP UIKit Item Card Category Object Reference
    ItemCardCategory:WBP UIKit Item Card Category Object Reference=Self-Reference()
   ) -> (then:Exec) {is=title K2Node_CallDelegate}
   > Return Node!(execute, ReturnValue:Event Reply Structure=Handled())
   OnSetContextMenuActionContext(execute=@n5.else)
   > n6 = Open!(execute, self:Menu Anchor Object Reference=@n4, bFocusMenu:bool=false) -> (then)
   Return Node(execute=@n2.else, ReturnValue=Unhandled())
  ]}
 CategoryCardHovered {nodes [CategoryCardHovered!() -> (then, IsCategoryCardHovered:bool=~) {is=title K2Node_FunctionEntry}]}
 CategoryCardSelected {nodes [CategoryCardSelected!() -> (then, ItemCardCategory:WBP UIKit Item Card Category Object Reference) {is=title K2Node_FunctionEntry}]}
 CloseContextMenu {nodes [
   CloseContextMenu() {is=title K2Node_FunctionEntry}
   > Close(self:Menu Anchor Object Reference=Get Anchor(self=Get ContextMenu()))
  ]}
 EventGraph {nodes [
   OnRightClickDown() {is=title K2Node_CustomEvent}
   n1 = Get Slot_Main() -> (Slot_Main:Named Slot Object Reference)
   > Branch(Condition="Not Equal (Enum)"!(
     A:ESlateVisibility Enum=GetVisibility!(self:Widget Object Reference=@n1) -> (ReturnValue:ESlateVisibility Enum=Visible)
     B:ESlateVisibility Enum=Collapsed
    ) -> (ReturnValue:bool))
   > Open(self=Get Anchor(self=Get ContextMenu()), bFocusMenu=false)
   > PlaySound2D(Sound=Get Right Clicked Sound() -> (Right Clicked Sound:Sound Base Object Reference))
   ToggleContextMenuIfSelected() {is=title K2Node_CustomEvent}
   n2 = GetCurrentInputType(self=Get CommonInputSubsystem())
   > Branch(Condition=AND Boolean(A=Get Is Item Card Category Hovered() -> (Is Item Card Category Hovered:bool), B="Equal (Enum)"(A=@n2, B=Gamepad)))
   > Set SelectedCategory(SelectedCategory=Get FortLockerCategoryItemVM(), self=Get FortItemSelectionVM())
   > SetVisibility(self=Get ContextMenu(), InVisibility=Visible)
   > ToggleContextMenu(self:WBP UIKit Context Menu Anchor Button Base Object Reference=Get ContextMenu())
   > DelayUntilNextTick!(execute) -> (then)
   > RefreshFocusIfLeafmostDescendant!(execute) -> (then, ReturnValue:bool)
   n3 = OnSetContextMenuActionContext() {is=title K2Node_CustomEvent}
   HandleContextActions(execute=[@n3.then, @n8.else])
   > SetCachedIndicatorConfig!(
    execute
    self
    CachedIndicatorTypeToFind:E_UIKit_ItemCardStatusIndicator Enum=NewEnumerator5
    NewIndicatorConfig:E_UI_StatusIndicator_Configuration Enum=NewEnumerator3
   ) -> (then, Success:bool)
   n4 = Event PreConstruct()
   > Parent: PreConstruct(IsDesignTime=@n4.IsDesignTime)
   > SetGridAlignment!(
    execute
    self:WBP UIKit Item Card Base Object Reference
    WIdget:Widget Object Reference=Get Slot_Main()
    Horizontal:EHorizontalAlignment Enum=HAlign_Right
    Vertical:EVerticalAlignment Enum=VAlign_Bottom
   ) -> (then)
   > SetContextMenuAnchorPadding()
   > SetVisibility(self=Get ContextMenu(), InVisibility=Hidden)
   > RemoveFromParent(self=Get EquippedState_Outline() -> (EquippedState_Outline:WBP UIKit Equipped State Object Reference))
   Event On Hovered()
   > Set HoveredCategory(HoveredCategory=Get FortLockerCategoryItemVM(), self=Get FortItemSelectionVM())
   > Set Is Item Card Category Hovered(Is Item Card Category Hovered:bool=true) -> (Output_Get:bool)
   > Call Category Card Hovered!(execute:Exec, self:WBP UIKit Item Card Category Object Reference, IsCategoryCardHovered:bool=true) -> (then:Exec) {is=title K2Node_CallDelegate}
   n5 = "Equal (Enum)"(A=GetCurrentInputType(self=Get CommonInputSubsystem()), B=Touch)
   > n6 = Branch(Condition=@n5)
   Parent: On Hovered(execute=[@n12.then, @n6.then]) {is=title K2Node_CallParentFunction}
   > PlayAnimationForward(InAnimation=Get Anim_Hovered() -> (Anim_Hovered:Widget Animation Object Reference))
   n7 = Get ContextMenu()
   > n8 = Branch(Condition=IsOpen(self=Get Anchor(self=@n7)))
   n9 = Branch(execute=@n6.else, Condition=Get AwaitingContextMenuClose() -> (AwaitingContextMenuClose:bool))
   n10 = Get ContextMenu()
   SetVisibility(execute=@n9.else, self=[@n10, Get ContextMenuButton(self=@n10)], InVisibility=Visible)
   n11 = Get CommonInputSubsystem()
   > n12 = Toggle Quick Actions!(
    execute
    self:WBP UIKit Context Menu Anchor Button Base Object Reference=@n10
    Enabled:bool="Not Equal (Enum)"(A:ECommonInputType Enum=GetCurrentInputType(self=@n11), B:ECommonInputType Enum=Gamepad)
   ) -> (then)
   Event On Deselected()
   > Parent: On Deselected() {is=title K2Node_CallParentFunction}
   > PlayAnimationReverse(InAnimation=Get Anim_Hovered())
   > SetCachedIndicatorConfig(CachedIndicatorTypeToFind=NewEnumerator5, NewIndicatorConfig=NewEnumerator2)
   > Branch(Condition=IsUsingTouch(PlayerController=GetOwningPlayer()))
   n13 = Get Anchor(self=Get ContextMenu())
   > n14 = Branch(Condition=IsOpen(self=@n13))
   > n15 = Close(self=@n13)
   n16 = Get ContextMenu()
   SetVisibility(execute=[@n15.then, @n14.else], self=[@n16, Get ContextMenuButton(self=@n16)], InVisibility=Hidden)
   > SetVisibility(self=Get UIKit_Transference(), InVisibility=SelfHitTestInvisible)
   > RemoveAllInteractions(self:UIKit Context Menu Action Context Object Reference=Get ContextMenuActionContext() -> (ContextMenuActionContext:UIKit Context Menu Action Context Object Reference))
   Event On Selected()
   > Parent: On Selected() {is=title K2Node_CallParentFunction}
   > AdjustContextMenuTouch() {is=title K2Node_CallFunction}
   > SetVisibility(self=Get ContextMenu(), InVisibility=Visible)
   > PlayAnimationForward(InAnimation=Get Anim_Hovered())
   > SetCachedIndicatorConfig(CachedIndicatorTypeToFind=NewEnumerator5, NewIndicatorConfig=NewEnumerator3)
   Event On Clicked()
   > Parent: On Clicked() {is=title K2Node_CallParentFunction}
   > n17 = Set SelectedCategory(SelectedCategory=Get FortLockerCategoryItemVM(), self=Get FortItemSelectionVM())
   > Set HoveredCategory(HoveredCategory=@n17, self=Get FortItemSelectionVM())
   > Call On Category Clicked!(execute:Exec, self:WBP UIKit Item Card Category Object Reference) -> (then:Exec) {is=title K2Node_CallDelegate}
   > AnalyticsCategoryInteracted!(execute, self, Context:String="Clicked ") -> (then) {is=title K2Node_CallFunction}
   Event On Unhovered()
   n18 = GetOwningPlayer()
   > Branch(Condition=NOT Boolean(A=IsUsingTouch(PlayerController=@n18)))
   > Set Is Item Card Category Hovered(Is Item Card Category Hovered=false)
   > Call Category Card Hovered()
   > n19 = Branch(Condition=@n5)
   Parent: On Unhovered(execute=[@n25.then, @n19.then]) {is=title K2Node_CallParentFunction}
   > PlayAnimationReverse(InAnimation=Get Anim_Hovered())
   n20 = Get ContextMenu()
   > n21 = Branch(Condition=IsOpen(self=Get Anchor(self=@n20)))
   RemoveAllInteractions(execute=@n21.else, self=Get ContextMenuActionContext())
   > SetCachedIndicatorConfig(CachedIndicatorTypeToFind=NewEnumerator5, NewIndicatorConfig=NewEnumerator2)
   n22 = Get ContextMenu()
   n23 = Branch(execute=@n19.else, Condition=IsOpen(self=Get Anchor(self=@n22)))
   > Set AwaitingContextMenuClose(AwaitingContextMenuClose:bool=true) -> (Output_Get:bool)
   n24 = Get ContextMenu()
   SetVisibility(execute=@n23.else, self=[Get ContextMenuButton(self=@n24), @n24], InVisibility=Hidden)
   > n25 = Toggle Quick Actions(self=@n24, Enabled=false)
   Event Construct()
   > Parent: Construct()
   n26 = Create Event()
   n27 = Get Anchor(self=Get ContextMenu())
   > Bind Event to On Menu Open Changed(self:Menu Anchor Object Reference=@n27, Delegate=@n26) {is=title K2Node_AddDelegate}
   > SetIsSelectable!(execute, self:Common Button Base Object Reference, bInIsSelectable:bool=IsUsingTouch(PlayerController=GetOwningPlayer())) -> (then)
   > n28 = Is Valid(InputObject=Get ContextMenuActionContext())
   n29 = Construct Object from Class(execute=@n28.Is Not Valid, Class=/Script/FortniteUI.UIKitContextMenuActionContext) -> (ReturnValue:UIKit Context Menu Action Context Object Reference)
   > Set ContextMenuActionContext(ContextMenuActionContext:UIKit Context Menu Action Context Object Reference=@n29) -> (Output_Get:UIKit Context Menu Action Context Object Reference)
   n30 = Event On Input Method Changed!() -> (OutputDelegate, then, CurrentInputType:ECommonInputType Enum) {is=title K2Node_Event}
   > n31 = Sequence()
   > Set AwaitingContextMenuClose(AwaitingContextMenuClose=false)
   n32 = Get ContextMenu()
   > n33 = Branch(Condition=IsOpen(self=Get Anchor(self=@n32)))
   > Branch(Condition=IsMobileGame())
   > CloseContextMenu() {is=title K2Node_CallFunction}
   n34 = Get ContextMenu()
   SetVisibility(execute=@n33.else, self=[@n34, Get ContextMenuButton(self=@n34)], InVisibility=Hidden)
   n35 = "Equal (Enum)"(A=@n30.CurrentInputType, B=Touch)
   > n36 = Branch(Condition=@n35)
   RemoveAllInteractions(execute=@n36.else, self=Get ContextMenuActionContext())
   SetIsSelectable(execute=@n31.then_1, bInIsSelectable=@n35)
   > Branch(Condition=@n35)
   > SelectIfPreviewed()
   n37 = Event OnAddedToFocusPath!() -> (OutputDelegate, then, InFocusEvent:Focus Event Structure) {is=title K2Node_Event}
   > Parent: OnAddedToFocusPath!(execute, InFocusEvent:Focus Event Structure=@n37.InFocusEvent) -> (then) {is=title K2Node_CallParentFunction}
   > Branch(Condition=IsUsingTouch(PlayerController=GetOwningPlayer()))
   > SelectIfPreviewed()
   > AdjustContextMenuTouch()
   Event Destruct()
   > Parent: Destruct()
   > Unbind Event from On Menu Open Changed(self:Menu Anchor Object Reference=@n27, Delegate=@n26) {is=title K2Node_RemoveDelegate}
  ]}
 Get Is Favorite State {nodes [
   n1 = Get Is Favorite State!() -> (then, Is Favorite:bool=~, Is New:bool=~) {is=title K2Node_FunctionEntry}
   > Return Node!(execute, Result:bool=AND Boolean(A=@n1, B=NOT Boolean(A=@n1.Is New)))
  ]}
 GetCurrentCategory {nodes [
   GetCurrentCategory() {is=title K2Node_FunctionEntry}
   n1 = Get FortItemVM()
   > Return Node!(
    execute
    CanApplyToAll:bool=Get bCanApplyToAll!(self:Fort Locker Category Config Object Reference=GetItemCategory!(self:Fort Item VM Object Reference=@n1) -> (ReturnValue:Fort Locker Category Config Object Reference)) -> (bCanApplyToAll:bool)
    CanFavorite:bool=Get bCanFavorite!(self:Fort Item VM Object Reference=@n1) -> (bCanFavorite:bool)
   )
  ]}
 GetShuffleVisibility {nodes [
   n1 = GetShuffleVisibility!() -> (then, IsShuffleEnabled:bool=~, IsCommonShuffleEnabled:bool=~, IsCommonShuffleValid:bool=~) {is=title K2Node_FunctionEntry}
   > Return Node!(execute, is Shuffled:bool=OR Boolean(A=@n1, B=AND Boolean(A=@n1.IsCommonShuffleEnabled, B=@n1.IsCommonShuffleValid)))
  ]}
 HandleContextMenuOpenChanged {nodes [
   n1 = HandleContextMenuOpenChanged!() -> (then, IsOpen:bool=~) {is=title K2Node_FunctionEntry}
   > n2 = Branch(Condition=@n1)
   > Call On Context Menu Opened!(execute:Exec, self:WBP UIKit Item Card Category Object Reference, ContextMenuWidgetParentEntry:Widget Object Reference=Self-Reference()) -> (then:Exec) {is=title K2Node_CallDelegate}
   Set AwaitingContextMenuClose(execute=@n2.else, AwaitingContextMenuClose=false)
   > n3 = Switch on ECommonInputType!(execute, Selection:ECommonInputType Enum=GetCurrentInputType(self=Get CommonInputSubsystem())) -> (MouseAndKeyboard:Exec, Gamepad:Exec, Touch:Exec, Count:Exec) {is=title K2Node_SwitchEnum}
   Call On Context Menu Closed!(execute:Exec=[@n3.Count, @n3.Gamepad], self:WBP UIKit Item Card Category Object Reference) -> (then:Exec) {is=title K2Node_CallDelegate}
   n4 = Get ContextMenu()
   > SetVisibility(self=[@n4, Get ContextMenuButton(self=@n4)], InVisibility=Hidden)
   n5 = Branch(execute=@n3.MouseAndKeyboard, Condition=IsHovered())
   Call On Context Menu Closed(execute=[@n7.then, @n5.then])
   SetFocusState!(execute=[@n8.else, @n5.else], self:WBP UIKit Item Card Base Object Reference, IsFocused:bool=false) -> (then)
   n6 = Get ContextMenu()
   > SetVisibility(self=[@n6, Get ContextMenuButton(self=@n6)], InVisibility=Hidden)
   > PlayAnimationReverse(InAnimation=Get Anim_Hovered())
   > n7 = RemoveAllInteractions(self=Get ContextMenuActionContext())
   n8 = Branch(execute=@n3.Touch, Condition=GetSelected())
   > RemoveAllInteractions(self=Get ContextMenuActionContext())
   > Call On Context Menu Closed()
  ]}
 Initialize {nodes [
   n1 = Initialize!() -> (then, Category Item:Fort Locker Category Item VM Object Reference, CategoryGroup:Fort Locker Category Group VM Object Reference) {is=title K2Node_FunctionEntry}
   > SetFortLockerCategoryItemVM!(execute, self, Viewmodel:Fort Locker Category Item VM Object Reference=@n1) -> (then)
   > SetFortLockerCategoryGroupVM!(execute, self, Viewmodel:Fort Locker Category Group VM Object Reference=@n1.CategoryGroup) -> (then)
   n2 = Break Fort Locker Category Slot Info!(FortLockerCategorySlotInfo:"Fort Locker Category Slot Info Structure (by ref)"=GetSlotInfo!(self:Fort Locker Category Item VM Object Reference=Get FortLockerCategoryItemVM()) -> (ReturnValue:"Fort Locker Category Slot Info Structure (by ref)")) -> (
    CategoryName:Text
    CategoryNamePlural:Text
    CategoryNameWithPluralForms:Text
    CategoryDescription:Text
    DefaultEmptyTexture:Soft Object Reference
    OverrideShuffleTexture:Texture 2D Soft Object Reference
    CategoryIcon:Material Instance Object Reference
    CategorySoundOverrides:Map of Gameplay Tag Structures to Sound Base Soft Object References
    bShowInCategoryTab:bool=~
    PreviewWindowScale:float=~
    PreviewWindowOffset:Vector 3f
    bUsePluralCategoryInItemList:bool=~
   ) {is=title K2Node_BreakStruct}
   > Set Fallback Image(Fallback Image:Soft Object Reference=Select(
     Option 0:Soft Object Reference=/FortUILibrary/Systems/Items/Textures/T_UI_NoColorVariant
     Option 1:Soft Object Reference=@n2.DefaultEmptyTexture
     Index=IsValidSoftObjectReference(SoftObjectReference=@n2.DefaultEmptyTexture)
    ) -> (ReturnValue:Soft Object Reference)) -> (Output_Get:Soft Object Reference)
   > InitializeItemCard(self:WBP UIKit Item Card Base Object Reference)
   > SetCachedIndicatorConfig(CachedIndicatorTypeToFind=NewEnumerator5, NewIndicatorConfig=NewEnumerator2)
  ]}
 IsCustomizable {nodes [
   IsCustomizable() {is=title K2Node_FunctionEntry}
   > n1 = Get!(execute:Exec) -> (then:Exec, else:Exec, FortItemSelectionVM:Fort Item Selection VM Object Reference)
   > n2 = Get!(execute, self:Fort Item Selection VM Object Reference=@n1) -> (then, else, SelectedItem:Fort Item VM Object Reference)
   > Return Node!(execute, Customizable:bool=RequiresCustomization!(self:Fort Item VM Object Reference=@n2) -> (ReturnValue:bool))
   Return Node(execute=[@n2.else, @n1.else], Customizable=false)
  ]}
 IsModalOpen {nodes [IsModalOpen!() -> (then, IsOpen:bool=~) {is=title K2Node_FunctionEntry}]}
 On Secondary Action Button Clicked {nodes [
   n1 = On Secondary Action Button Clicked!() -> (then, NameID:Name=None) {is=title K2Node_FunctionEntry}
   > Parent: On Secondary Action Button Clicked!(execute, NameID:Name=@n1) -> (then) {is=title K2Node_CallParentFunction}
   n2 = Get Preview Actions() -> (Preview Actions:Array of Names)
   > AnalyticsCategoryInteracted(Context="To String (Name)"(InName="Get (a ref)"(Array:Array of Names=@n2, Dimension 1=1) -> (Output:"Name (by ref)")))
  ]}
 OnCategoryClicked {nodes [OnCategoryClicked() {is=title K2Node_FunctionEntry}]}
 OnContextMenuCardEditStylesClicked {nodes [OnContextMenuCardEditStylesClicked() {is=title K2Node_FunctionEntry}]}
 OnContextMenuClosed {nodes [OnContextMenuClosed() {is=title K2Node_FunctionEntry}]}
 OnContextMenuOpened {nodes [OnContextMenuOpened!() -> (then, ContextMenuWidgetParentEntry:Widget Object Reference) {is=title K2Node_FunctionEntry}]}
 OnFortItemVM {nodes [OnFortItemVM() {is=title K2Node_FunctionEntry}]}
 OnFortItemVMChanged {nodes [
   n1 = OnFortItemVMChanged!() -> (then, InFortVM:Fort Item VM Object Reference) {is=title K2Node_FunctionEntry}
   > Set Custom Widget!(execute, self:WBP UIKit Item Card Base Object Reference, ItemVM:Fort Item VM Object Reference=@n1) -> (then) {is=title K2Node_CallFunction}
   > Call On Fort Item VM!(execute:Exec, self:WBP UIKit Item Card Category Object Reference) -> (then:Exec) {is=title K2Node_CallDelegate}
  ]}
 OnMouseButtonDown {nodes [
   n1 = OnMouseButtonDown!() -> (then, MyGeometry:Geometry Structure, MouseEvent:"Pointer Event Structure (by ref)") {is=title K2Node_FunctionEntry}
   n2 = IsOpen(self=Get Anchor(self=Get ContextMenu()))
   > n3 = Branch(Condition=AND Boolean(
     A=Is Mouse Button Down!(Input:"Pointer Event Structure (by ref)"=@n1.MouseEvent, MouseButton:Key Structure=RightMouseButton) -> (ReturnValue:bool)
     B=NOT Boolean(A=@n2)
    ))
   > Set Timer for Next Tick by Event!(execute, Delegate:Delegate=Create Event()) -> (then, ReturnValue:Timer Handle Structure) {is=title K2Node_CallFunction}
   > Return Node!(execute, ReturnValue:Event Reply Structure=Handled())
   Return Node(execute=@n3.else, ReturnValue=Handled())
  ]}
 OnRandomizeTimer {nodes [OnRandomizeTimer() {is=title K2Node_FunctionEntry}]}
 SelectIfPreviewed {nodes [
   SelectIfPreviewed() {is=title K2Node_FunctionEntry}
   n1 = Get FortLockerCategoryItemVM()
   n2 = Get FortItemSelectionVM()
   n3 = Get HoveredCategory!(self:Fort Item Selection VM Object Reference=@n2) -> (HoveredCategory:Fort Locker Category Item VM Object Reference)
   > SetIsSelected(
    InSelected=AND Boolean!(A=IsValid(Object=@n2), B=IsValid(Object=@n3), C:bool=IsValid(Object=@n1), D:bool="Equal (Object)"(A=@n3, B=@n1)) -> (ReturnValue)
    bGiveClickFeedback=false
   )
   > Call Category Card Selected(ItemCardCategory=Self-Reference())
  ]}
 Set Context Menu Action Context {nodes [
   Set Context Menu Action Context() {is=title K2Node_FunctionEntry}
   > RemoveAllInteractions(self=Get ContextMenuActionContext())
   n1 = IsInZone!(self:Fort Global UIContext Object Reference=Get FortGlobalUIContext() -> (ReturnValue:Fort Global UIContext Object Reference) {is=title K2Node_GetSubsystem}) -> (ReturnValue:bool)
   n2 = IsCustomizable!() -> (Customizable:bool) {is=title K2Node_CallFunction}
   n3 = GetItemType!(ItemDefinition:Item Definition Base Object Reference=Get ItemDefinition!(self:Fort Item VM Object Reference=Get FortItemVM()) -> (ItemDefinition:Fort Item Definition Object Reference)) -> (ReturnValue:EFortItemType Enum=WorldItem)
   > n4 = Branch(Condition=AND Boolean!(A="Equal (Enum)"(A:EFortItemType Enum=@n3, B:EFortItemType Enum=CosmeticMimosa), B=NOT Boolean(A=@n2), C=NOT Boolean(A=@n1)) -> (ReturnValue))
   > n5 = IsNameable!(
    execute:Exec
    self:Object Reference=Get ItemDetails!(self:Fort Item VM Object Reference=Get FortItemVM()) -> (ItemDetails:Fort Item Details VM Object Reference)
   ) -> (then:Exec, ReturnValue:bool) {is=title K2Node_Message}
   > n6 = AddInteraction!(
    execute
    self:UIKit Context Menu Action Context Object Reference=Get ContextMenuActionContext()
    InActionInfo_DisplayName:Text="LOCTEXT(\"Rename\")"
    InActionInfo_IconTexture:Texture 2D Soft Object Reference
    InActionInfo_IconMaterial:Material Instance Soft Object Reference
    InActionInfo_SortPriority:int=980
    InActionInfo_bIsAvailable:bool=@n5
    InActionInfo_bIsEnabled:bool=True
    InActionInfo_bIsSelected:bool=False
    InActionInfo_bAddDivider:bool=False
    InActionInfo_bPositionDividerAfter:bool=False
    InActionInfo_bCloseMenuOnSelect:bool=False
    InActionInfo_InteractionNameID:Name=None
    InActionInfo_Interaction:Data Table Row Handle Structure=Get RenameModalInput() -> (RenameModalInput:Data Table Row Handle Structure)
    InActionInfo_PressedDelegate:Delegate="(null).None"
    InActionInfo_ReleasedDelegate:Delegate="(null).None"
    InInteractionDelegate:"Delegate (by ref)"=Create Event()
   ) -> (then)
   n7 = Get FortItemVM()
   AddInteraction(
    execute=[@n6.then, @n4.else]
    self=Get ContextMenuActionContext()
    InActionInfo_DisplayName=SelectText(
     A="LOCTEXT(\"Unfavorite Item\")"
     B="LOCTEXT(\"Favorite Item\")"
     bPickA=Get bIsFavorite!(self:Fort Item VM Object Reference=@n7) -> (bIsFavorite:bool)
    )
    InActionInfo_SortPriority=1200
    InActionInfo_bIsAvailable=Get bCanFavorite(self=@n7)
    InActionInfo_bIsEnabled=True
    InActionInfo_bIsSelected=False
    InActionInfo_bAddDivider=False
    InActionInfo_bPositionDividerAfter=False
    InActionInfo_bCloseMenuOnSelect=False
    InActionInfo_InteractionNameID=None
    InActionInfo_Interaction=Get FavoriteInput() -> (FavoriteInput:Data Table Row Handle Structure)
    InActionInfo_PressedDelegate="(null).None"
    InActionInfo_ReleasedDelegate="(null).None"
    InInteractionDelegate=Create Event()
   )
   > AddInteraction(
    self=Get ContextMenuActionContext()
    InActionInfo_DisplayName="LOCTEXT(\"Edit Styles\")"
    InActionInfo_SortPriority=1100
    InActionInfo_bIsAvailable=Get bHasVariants!(self:Fort Item VM Object Reference=Get FortItemVM()) -> (bHasVariants:bool)
    InActionInfo_bIsEnabled=True
    InActionInfo_bIsSelected=False
    InActionInfo_bAddDivider=False
    InActionInfo_bPositionDividerAfter=False
    InActionInfo_bCloseMenuOnSelect=False
    InActionInfo_InteractionNameID=None
    InActionInfo_Interaction=Get EditStylesInput() -> (EditStylesInput:Data Table Row Handle Structure)
    InActionInfo_PressedDelegate="(null).None"
    InActionInfo_ReleasedDelegate="(null).None"
    InInteractionDelegate=Create Event()
   )
   n8 = Get FortItemVM()
   > AddInteraction(
    self=Get ContextMenuActionContext()
    InActionInfo_DisplayName="LOCTEXT(\"Apply To All\")"
    InActionInfo_SortPriority=1000
    InActionInfo_bIsAvailable=Get bCanApplyToAll(self=GetItemCategory(self=@n8))
    InActionInfo_bIsEnabled=True
    InActionInfo_bIsSelected=False
    InActionInfo_bAddDivider=False
    InActionInfo_bPositionDividerAfter=False
    InActionInfo_bCloseMenuOnSelect=False
    InActionInfo_InteractionNameID=None
    InActionInfo_Interaction=Get ApplytoAllInput() -> (ApplytoAllInput:Data Table Row Handle Structure)
    InActionInfo_PressedDelegate="(null).None"
    InActionInfo_ReleasedDelegate="(null).None"
    InInteractionDelegate=Create Event()
   )
   > AddInteraction(
    self=Get ContextMenuActionContext()
    InActionInfo_DisplayName="LOCTEXT(\"Randomize Item\")"
    InActionInfo_SortPriority=1300
    InActionInfo_bIsAvailable=True
    InActionInfo_bIsEnabled=True
    InActionInfo_bIsSelected=False
    InActionInfo_bAddDivider=False
    InActionInfo_bPositionDividerAfter=False
    InActionInfo_bCloseMenuOnSelect=False
    InActionInfo_InteractionNameID=None
    InActionInfo_Interaction=Get RandomizeItemInput() -> (RandomizeItemInput:Data Table Row Handle Structure)
    InActionInfo_PressedDelegate="(null).None"
    InActionInfo_ReleasedDelegate="(null).None"
    InInteractionDelegate=Create Event()
   )
   n9 = GetAnalyticsName(self=Get FortLockerCategoryItemVM())
   > SetParentIDForContextMenuTelemetry!(
    execute
    self:WBP UIKit Context Menu Anchor Button Base Object Reference=Get ContextMenu()
    ParentID:String=Get Source String from Text!(Text:Text=Format Text!(Format="LOCTEXT(\"{CategoryName} CategoryTile\")", CategoryName:String=@n9) -> (Result)) -> (ReturnValue:String) {is=title K2Node_CallFunction}
    IsParentIdEGPI:bool=false
   ) -> (then)
   > SetContext!(
    execute
    self:WBP UIKit Context Menu Anchor Button Base Object Reference=Get ContextMenu()
    InteractionContext_Data:UIKit Context Menu Action Context Object Reference=Get ContextMenuActionContext()
   ) -> (then)
  ]}
 Set Context Menu Preview Actions {nodes [
   n1 = Set Context Menu Preview Actions!() -> (then, PreviewActionName:Text, InInPreviewActionIteractionDelegate:"Delegate (by ref)") {is=title K2Node_FunctionEntry}
   n2 = String To Name(InString="To String (Text)"(InText=@n1))
   > RemoveInteraction!(execute, self:UIKit Context Menu Action Context Object Reference=Get ContextMenuActionContext(), InteractionNameID:Name=@n2) -> (then)
   > AddInteraction(
    self=Get ContextMenuActionContext()
    InActionInfo_DisplayName=@n1
    InActionInfo_SortPriority=1000
    InActionInfo_bIsAvailable=True
    InActionInfo_bIsEnabled=True
    InActionInfo_bIsSelected=False
    InActionInfo_bAddDivider=False
    InActionInfo_bPositionDividerAfter=False
    InActionInfo_bCloseMenuOnSelect=False
    InActionInfo_InteractionNameID=@n2
    InActionInfo_Interaction=(DataTable=None,RowName="")
    InActionInfo_PressedDelegate="(null).None"
    InActionInfo_ReleasedDelegate="(null).None"
    InInteractionDelegate=@n1.InInPreviewActionIteractionDelegate
   )
   > SetContext(self=Get ContextMenu(), InteractionContext_Data=Get ContextMenuActionContext())
  ]}
 Set Tile Image {nodes [
   n1 = Set Tile Image!() -> (then, Image:Texture 2D Soft Object Reference) {is=title K2Node_FunctionEntry}
   > SetItemImage!(execute, self:WBP UIKit Item Card Base Object Reference, Item Image:Soft Object Reference=@n1) -> (then)
  ]}
 SetBangCount {nodes [
   n1 = SetBangCount!() -> (then, NewBangCount:int) {is=title K2Node_FunctionEntry}
   > OverrideStatusIndicatorBangCount!(execute, self:WBP UIKit Item Card Base Object Reference, Indicator:E_UIKit_ItemCardStatusIndicator Enum=NewEnumerator5, Value:int=@n1) -> (then)
   > SetNewIcon!(execute, self, Item Count:int=@n1) -> (then)
  ]}
 SetCachedIndicatorConfig {nodes [
   SetCachedIndicatorConfig!() -> (
    then
    CachedIndicatorTypeToFind:E_UIKit_ItemCardStatusIndicator Enum=NewEnumerator5
    NewIndicatorConfig:E_UI_StatusIndicator_Configuration Enum=NewEnumerator2
   ) {is=title K2Node_FunctionEntry}
   > n1 = FindCachedStatusIndicator!(
    execute
    self:WBP UIKit Item Card Base Object Reference
    IndicatorType:E_UIKit_ItemCardStatusIndicator Enum=Get CachedIndicatorTypeToFind() -> (CachedIndicatorTypeToFind:E_UIKit_ItemCardStatusIndicator Enum=NewEnumerator10)
   ) -> (then, Status Indicator:WBP UIKit Status Indicator Base Object Reference, Valid:bool)
   > n2 = Branch(Condition=@n1.Valid)
   > SetIndicatorConfiguration!(
    execute
    self:WBP UIKit Status Indicator Base Object Reference=@n1
    InCount:int=0
    IndicatorConfig:E_UI_StatusIndicator_Configuration Enum=Get NewIndicatorConfig() -> (NewIndicatorConfig:E_UI_StatusIndicator_Configuration Enum=NewEnumerator0)
   ) -> (then)
   > Return Node!(execute, Success:bool=true)
   Return Node(execute=@n2.else, Success=false)
  ]}
 SetContextMenuActions {nodes [
   HandleContextActions() {is=title K2Node_CustomEvent}
   > Set Context Menu Action Context() {is=title K2Node_CallFunction}
   Rename Modal!() -> (OutputDelegate, then, InteractionNameID:Name) {is=title K2Node_CustomEvent}
   n1 = Get FortItemVM()
   > n2 = LoadClassAsset_Blocking(AssetClass=TryGetSoftCustomActivatableWidgetClass!(
     self:Fort Locker Category Config Object Reference=GetItemCategory(self=@n1)
     ViewRequest:Gameplay Tag Structure=(TagName="UI.ItemPreview.PluginBridge.Mimosa.RenameModal")
    ) -> (WidgetClass:Common Activatable Widget Soft Class Reference, ReturnValue:bool))
   > n3 = Cast To CommonActivatableWidget Class!(execute:Exec, Class:Object Class Reference=@n2) -> (then:Exec, CastFailed:Exec, AsCommon Activatable Widget:Common Activatable Widget Class Reference) {is=title K2Node_ClassDynamicCast}
   > n4 = Create Widget!(execute:Exec, Class:User Widget Class Reference=@n3, OwningPlayer:Player Controller Object Reference=GetOwningPlayer()) -> (then:Exec, ReturnValue:Common Activatable Widget Object Reference)
   > n5 = Set RenameModal(RenameModal:Common Activatable Widget Object Reference=@n4) -> (Output_Get:Common Activatable Widget Object Reference)
   > n6 = TryPushWidgetAsModalPanel!(execute, PushWidget:User Widget Object Reference=@n5) -> (then)
   n7 = ModalDeactivated() {is=title K2Node_CustomEvent}
   Bind Event to On Widget Deactivated(execute=@n6.then, self:Common Activatable Widget Object Reference=@n5, Delegate=@n7) {is=title K2Node_AddDelegate}
   > n8 = Call Is Modal Open!(execute:Exec, self:WBP UIKit Item Card Category Object Reference, IsOpen:bool=true) -> (then:Exec) {is=title K2Node_CallDelegate}
   n9 = CloseContextMenu(execute=[@n12.then, @n8.then, @n17.then, @n16.then, @n14.then])
   Set Context Menu Action Context(execute=[@n20.then, @n9.then])
   Event On Edit Styles!() -> (OutputDelegate, then, InteractionNameID:Name) {is=title K2Node_CustomEvent}
   n10 = Get FortItemSelectionVM()
   > n11 = Set SelectedCategory(SelectedCategory=Get FortLockerCategoryItemVM(), self=@n10)
   > Set SelectedItem!(
    execute
    SelectedItem:Fort Item VM Object Reference=Get EquippedItem!(self:Fort Locker Category Item VM Object Reference=@n11) -> (EquippedItem:Fort Item VM Object Reference)
    self:Fort Item Selection VM Object Reference=@n10
   ) -> (then, Output_Get:Fort Item VM Object Reference)
   > n12 = Call On Context Menu Card Edit Styles Clicked!(execute:Exec, self:WBP UIKit Item Card Category Object Reference) -> (then:Exec) {is=title K2Node_CallDelegate}
   Event On Apply To All!() -> (OutputDelegate, then, InteractionNameID:Name) {is=title K2Node_CustomEvent}
   > n13 = SetFIVMItemAndSlot!(execute, self) -> (then, Selected Item:Fort Item VM Object Reference, Slot Index:int=0)
   > n14 = EquipItem!(
    execute:Exec
    Item:Fort Item VM Object Reference=@n13
    SlotIndex:int=@n13.Slot Index
    Variants:Array of Mcp Variant Channel Info Structures=Make Array!() -> (Array:Array of Mcp Variant Channel Info Structures)
    bApplyToAllSlots:bool=true
    bAllowMocking:bool=true
   ) -> (then:Exec, OnOperationCompleted:Exec, Success:bool=~) {is=title K2Node_AsyncAction}
   Event On Favorite Item!() -> (OutputDelegate, then, InteractionNameID:Name) {is=title K2Node_CustomEvent}
   n15 = Get FortItemVM()
   > ToggleItemFavorite!(
    execute
    self:Fort Locker Category VM Object Reference=Get Category!(self:Fort Locker Category Item VM Object Reference=Get FortLockerCategoryItemVM()) -> (Category:Fort Locker Category VM Object Reference)
    Item:Fort Item VM Object Reference=@n15
   ) -> (then)
   > n16 = PlaySound2D(Sound=Select(
     Option 0:Sound Base Object Reference=Get Unfavorite Sound() -> (Unfavorite Sound:Sound Base Object Reference)
     Option 1:Sound Base Object Reference=Get Favorite Sound() -> (Favorite Sound:Sound Base Object Reference)
     Index=Get bIsFavorite(self=@n15)
    ) -> (ReturnValue:Sound Base Object Reference))
   Event On Clear All Badges!() -> (OutputDelegate, then, InteractionNameID:Name) {is=title K2Node_CustomEvent}
   > n17 = ClearAllNewItemFlags(self:Fort Locker Category Group VM Object Reference=Get SelectedCategoryGroup!(self:Fort Item Selection VM Object Reference=Get FortItemSelectionVM()) -> (SelectedCategoryGroup:Fort Locker Category Group VM Object Reference))
   Event On Randomize Item!() -> (OutputDelegate, then, InteractionNameID:Name) {is=title K2Node_CustomEvent}
   > RandomizeSlot(self:Fort Locker Category Item VM Object Reference=Get FortLockerCategoryItemVM())
   > CloseContextMenu()
   > n18 = Set Timer by Event(Delegate=Create Event(), Time=0.5, bLooping=false)
   > Set RandomizeTimerHandle(RandomizeTimerHandle:Timer Handle Structure=@n18) -> (Output_Get:Timer Handle Structure)
   OnFortItemVM_Event() {is=title K2Node_CustomEvent}
   n19 = Get RandomizeTimerHandle() -> (RandomizeTimerHandle:Timer Handle Structure)
   > Branch(Condition=Is Timer Active by Handle!(Handle:Timer Handle Structure=@n19) -> (ReturnValue:bool))
   > n20 = Clear and Invalidate Timer by Handle(Handle=@n19)
   Call Is Modal Open(execute=@n21.then)
   n21 = Unbind Event from On Widget Deactivated(
    execute=@n7.then
    self:Common Activatable Widget Object Reference=Get RenameModal() -> (RenameModal:Common Activatable Widget Object Reference)
    Delegate=@n7
   ) {is=title K2Node_RemoveDelegate}
  ]}
 SetContextMenuAnchorPadding {nodes [
   SetContextMenuAnchorPadding() {is=title K2Node_FunctionEntry}
   n1 = Get ContextMenu()
   > SetPadding(self=SlotAsGridSlot(Widget=Get Anchor(self=@n1)), InPadding_Left=0, InPadding_Top=0, InPadding_Right=0, InPadding_Bottom=0)
  ]}
 SetEmoteIndicatorVisibility {nodes [
   SetEmoteIndicatorVisibility!() -> (then, EquippedItemValid:bool=~) {is=title K2Node_FunctionEntry}
   > n1 = GetConsoleVariableBoolValue!(execute, VariableName:String=Fort.Rollback.FortAllowStabilizedEmoteWheel) -> (then, ReturnValue:bool)
   n2 = Get EquippedItemValid() -> (EquippedItemValid:bool)
   n3 = Get FortLockerCategoryItemVM()
   n4 = Get bShowSmallCategoryIcon!(self:Fort Locker Category Config Object Reference=GetConfig!(self:Fort Locker Category VM Object Reference=Get Category(self=@n3)) -> (ReturnValue:Fort Locker Category Config Object Reference)) -> (bShowSmallCategoryIcon:bool)
   > SetVisibility(
    self=Get EmoteWheelPositionImage() -> (EmoteWheelPositionImage:Fort Lazy Image Object Reference)
    InVisibility=Select(Option 0:ESlateVisibility Enum=Collapsed, Option 1:ESlateVisibility Enum=SelfHitTestInvisible, Index=AND Boolean(A=@n1, B=@n2, C=@n4)) -> (ReturnValue:ESlateVisibility Enum)
   )
  ]}
 SetEmoteSlotIndicatorIcon {nodes [
   SetEmoteSlotIndicatorIcon!() -> (then, FortLockerCategorySlotInfo:Fort Locker Category Slot Info Structure) {is=title K2Node_FunctionEntry}
   > n1 = GetConsoleVariableBoolValue(VariableName=Fort.Rollback.FortAllowStabilizedEmoteWheel)
   > Branch(Condition=@n1)
   n2 = Break Fort Locker Category Slot Info(FortLockerCategorySlotInfo=Get FortLockerCategorySlotInfo() -> (FortLockerCategorySlotInfo:Fort Locker Category Slot Info Structure)) -> (bShowInCategoryTab=~, PreviewWindowScale=~, bUsePluralCategoryInItemList=~)
   n3 = Get EmoteWheelPositionImage()
   > SetBrushFromLazyDisplayAsset!(
    execute
    self:Common Lazy Image Object Reference=@n3
    LazyObject:"Soft Object Reference (by ref)"=@n2.DefaultEmptyTexture
    bMatchTextureSize:bool=false
   ) -> (then)
   > n4 = Branch(Condition=IsAssetNull!(Asset:Soft Object Reference=@n2.DefaultEmptyTexture) -> (ReturnValue:bool))
   n5 = GetDynamicMaterial(execute=@n4.else, self=@n3)
   n6 = MakeLiteralName(Value=OuterGlowColor)
   > n7 = Get Vector Parameter Value!(execute, self:Material Instance Dynamic Object Reference=@n5, ParameterName:Name=@n6) -> (then, ReturnValue:Linear Color Structure) {is=title K2Node_CallFunction}
   n8 = BreakColor(InColor=@n7)
   > SetVectorParameterValue!(
    execute
    self=@n5
    ParameterName=@n6
    Value_R:float=@n8
    Value_G:float=@n8.G
    Value_B:float=@n8.B
    Value_A:float=float + float(A=@n8.A, B=0.3)
   ) -> (then)
  ]}
 SetFIVMItemAndSlot {nodes [
   SetFIVMItemAndSlot() {is=title K2Node_FunctionEntry}
   n1 = Get FortItemSelectionVM()
   > Set SelectedItem(SelectedItem=Get FortItemVM(), self=@n1)
   > Return Node!(
    execute
    Selected Item:Fort Item VM Object Reference=Get SelectedItem!(self:Fort Item Selection VM Object Reference=@n1) -> (SelectedItem:Fort Item VM Object Reference)
    Slot Index:int=Get SlotIndex!(self:Fort Locker Category Item VM Object Reference=Get SelectedCategory!(self:Fort Item Selection VM Object Reference=@n1) -> (SelectedCategory:Fort Locker Category Item VM Object Reference)) -> (SlotIndex:int=0)
   )
  ]}
 SetImageSize {nodes [
   n1 = SetImageSize!() -> (then, IsSmall:bool=~, SmallSize:double=~, DefaultSize:double=1) {is=title K2Node_FunctionEntry}
   > n2 = Branch(Condition=@n1)
   > SetItemAnimationScale!(execute, self:WBP UIKit Item Card Base Object Reference, Item Scale:"Float (double-precision) (by ref)"=@n1.SmallSize) -> (then)
   SetItemAnimationScale(execute=@n2.else, Item Scale=@n1.DefaultSize)
  ]}
 SetNewIcon {nodes [
   n1 = SetNewIcon!() -> (then, Item Count:int) {is=title K2Node_FunctionEntry}
   > Set with Broadcast IsNewItem!(execute, IsNewItem:bool="integer > integer"(A=@n1)) -> (then, Output_Get:bool)
  ]}
 SetTileSize {nodes [
   n1 = SetTileSize!() -> (then, TileSize:ELockerTileSize Enum=NewEnumerator0) {is=title K2Node_FunctionEntry}
   > SetItemCardSize!(
    execute
    self:WBP UIKit Item Card Base Object Reference
    ItemSize:Vector 2D Structure=Select!(
     NewEnumerator0:Vector 2D Structure=(X=128,Y=128)
     NewEnumerator1:Vector 2D Structure=(X=256,Y=128)
     NewEnumerator3:Vector 2D Structure=(X=128,Y=256)
     NewEnumerator2:Vector 2D Structure=(X=256,Y=256)
     NewEnumerator4:Vector 2D Structure=(X=128,Y=210)
     NewEnumerator5:Vector 2D Structure=(X=90,Y=120)
     Index:ELockerTileSize Enum=@n1
    ) -> (ReturnValue:Vector 2D Structure)
   ) -> (then)
  ]}
 UpdateAgeGated {nodes [
   UpdateAgeGated!() -> (then, RestrictionType:EItemAccessRestriction Enum=UnsatisfiedExclusiveItem) {is=title K2Node_FunctionEntry}
   n1 = Get FortItemVM()
   > SetItemIsAgeGated!(
    execute
    self:WBP UIKit Item Card Base Object Reference
    IsAgeGated:bool=GetIsItemAgeGatedFortCurrentExperience!(self:WBP UIKit Item Card Base Object Reference, Item Def:Fort Item Definition Object Reference=Get ItemDefinition(self=@n1)) -> (Is Item Gated:bool)
   ) -> (then)
  ]}
]
variables [
 AwaitingContextMenuClose {type=boolean}
 CategoryCardHovered {}
 CategoryCardSelected {}
 ContextMenuActionContext {type=UIKitContextMenuActionContext, default=null}
 Favorite Sound {type=SoundBase, default=/Game/Sounds/Fort_UI_Sounds/FrontEnd/MultiProduct/Common/Buttons/MSS_UI_MP_Button_Click_Favorite}
 Is Item Card Category Hovered {type=boolean}
 Is Shuffle Enabled {type=boolean}
 IsModalOpen {}
 IsNewItem {type=boolean}
 IsWrap {type=boolean}
 List Item Object {type=Object, default=null}
 OnCategoryClicked {}
 OnContextMenuCardEditStylesClicked {}
 OnContextMenuClosed {}
 OnContextMenuOpened {}
 OnFortItemVM {}
 OnRandomizeTimer {}
 RandomizeTimerHandle {type=TimerHandle}
 RenameModal {type=CommonActivatableWidget, default=null}
 Right Clicked Sound {type=SoundBase, default=/Game/Sounds/Fort_UI_Sounds/FrontEnd/DefaultSounds/PressedButtons/MSS_UI_Button_Pressed_Default}
 Unfavorite Sound {type=SoundBase, default=/Game/Sounds/Fort_UI_Sounds/FrontEnd/MultiProduct/Common/Buttons/MSS_UI_MP_Button_Click_Unfavorite}
]
widgets {
 info {
  inheritedWidgetCount=0
  namedSlotCount=1
  parentClass=/FortUILibrary/Systems/Items/WBP_UIKit_ItemCard_Base.WBP_UIKit_ItemCard_Base_C
  rootWidgetClass=null
  widgetCount=5
 }
 widgets [
  GridPanel_36 {widgetClassPath=&GridPanel, widget=GridPanel_36}
  FortVisualAttachment {
   widgetClassPath=&FortVisualAttachment
   bIsVariable=true
   parent=GridPanel_36
   widget=FortVisualAttachment
   properties {bOverride_HeightOverride=true, bOverride_WidthOverride=true, heightOverride=40, visibility=Visible, widthOverride=40, contentAnchor=<1 1>}
   slot {class=GridSlot, properties {horizontalAlignment=HAlign_Right, verticalAlignment=VAlign_Bottom}}
  }
  ContextMenu {
   widgetClassPath=/FortUILibrary/UIKit/ContextMenu/WBP_UIKit_ContextMenuAnchorButton
   bIsVariable=true
   parent=FortVisualAttachment
   widget=ContextMenu
   properties {
    button Background Material=/FortUILibrary/UIKit/Buttons/MaterialInstances/CTAButtons/MI_UIKit_BgQuiet
    buttonClass=/FortUILibrary/UIKit/Buttons/WBP_UIKit_Button_Contextual.WBP_UIKit_Button_Contextual_C
    buttonIcon=/FortUILibrary/Textures/Icons/T_UI_Icon_UIKit_ContextMenu
    buttonText=""
    button Padding=<ltrb 4 4 4 4>
    button Text Font {
     bForceMonospaced=false
     bMaterialIsStencil=false
     fontMaterial=/FortUILibrary/UIKit/Buttons/MaterialInstances/GenericButtons/MI_UIKit_GenericText_Quiet
     fontObject=&Burbank
     letterSpacing=180
     monospacedWidth=1
     size=26.25
     skewAmount=0
     typefaceFontName=BurbankRegular-Black
     outlineSettings {
      bApplyOutlineToDropShadows=false
      bMiteredCorners=false
      bSeparateFillAlpha=false
      outlineMaterial=null
      outlineSize=0
      outlineColor=<rgba 0 0 0 1>
     }
    }
   }
   slot {class=SizeBoxSlot, properties {horizontalAlignment=HAlign_Right, verticalAlignment=VAlign_Bottom, padding=<ltrb -8 -8 8 8>}}
  }
  Spacer_57 {widgetClassPath=&Spacer, parent=GridPanel_36, widget=Spacer_57, properties {size=<128 128>}, slot {class=GridSlot}}
  EmoteWheelPositionImage {
   widgetClassPath=&FortLazyImage
   bIsVariable=true
   parent=GridPanel_36
   widget=EmoteWheelPositionImage
   properties {brush {
     bIsDynamicallyLoaded=false
     drawAs=Image
     imageType=FullColor
     mirroring=NoMirror
     resourceName=None
     resourceObject=/Game/UI/Foundation/Textures/Icons/Locker/T_UI_Global_EmoteWheel_Pos1
     tiling=NoTile
     imageSize=<30 30>
     margin=<ltrb 0 0 0 0>
     outlineSettings {
      bUseBrushTransparency=false
      roundingType=HalfHeightRadius
      width=0
      color {colorUseRule=UseColor_Specified, specifiedColor=<rgba 0 0 0 0>}
      cornerRadii=<0 0 0 0>
     }
     tintColor {colorUseRule=UseColor_Specified, specifiedColor=<rgba 1 1 1 1>}
     uVRegion {bIsValid=false, max=<0 0>, min=<0 0>}
    }}
   slot {class=GridSlot, properties {horizontalAlignment=HAlign_Left, verticalAlignment=VAlign_Bottom, padding=<ltrb 4 4 4 4>}}
  }
 ]
}
