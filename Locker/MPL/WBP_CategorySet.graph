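# Locker category set widget that shows a group's visible category items as item cards in two entry boxes.
# SetupCategoryRows filters by equipped item visibility, splitting 4–8 cards evenly across rows with a spacer, otherwise using only the wrapping bottom box.
# Cards join a button group, with the first selected on touch, and their click, hover and context-menu events are re-broadcast as dispatchers.
# ResetItems unbinds cards and resets both boxes without deleting widgets.
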WidgetBlueprint /Locker/MPL/WBP_CategorySet
graphs [
 EventGraph {nodes [
   Event PreConstruct()
   > SetRowDistance!(execute, self, EnableSpacer:bool=true) -> (then)
   > n1 = Construct Object from Class!(execute, Class=/Script/CommonUI.CommonButtonGroupBase, self, bSelectionRequired:bool=False) -> (then, ReturnValue:Common Button Group Base Object Reference)
   > Set ButtonGroup(ButtonGroup:Common Button Group Base Object Reference=@n1) -> (Output_Get:Common Button Group Base Object Reference)
  ]}
 Handle Context Menu Opened {nodes [
   n1 = Handle Context Menu Opened!() -> (then, ContextMenuParent:Widget Object Reference) {is=title K2Node_FunctionEntry}
   > Call On Category Context Menu Opened!(execute:Exec, self:WBP Category Set Object Reference, ContextMenuParent:Widget Object Reference=@n1) -> (then:Exec) {is=title K2Node_CallDelegate}
  ]}
 HandleCategoryCardHovered {nodes [
   n1 = HandleCategoryCardHovered!() -> (then, IsCategoryCardHovered:bool=~) {is=title K2Node_FunctionEntry}
   > Call On Card Hovered!(execute:Exec, self:WBP Category Set Object Reference, IsHovered:bool=@n1) -> (then:Exec) {is=title K2Node_CallDelegate}
  ]}
 HandleCategoryClicked {nodes [
   HandleCategoryClicked() {is=title K2Node_FunctionEntry}
   > Call On Category Clicked!(execute:Exec, self:WBP Category Set Object Reference) -> (then:Exec) {is=title K2Node_CallDelegate}
  ]}
 HandleCategoryEditStyles {nodes [
   HandleCategoryEditStyles() {is=title K2Node_FunctionEntry}
   > Call On Category Edit Styles!(execute:Exec, self:WBP Category Set Object Reference) -> (then:Exec) {is=title K2Node_CallDelegate}
  ]}
 HandleContextMenuClosed {nodes [
   HandleContextMenuClosed() {is=title K2Node_FunctionEntry}
   > Call On Category Context Menu Closed!(execute:Exec, self:WBP Category Set Object Reference) -> (then:Exec) {is=title K2Node_CallDelegate}
  ]}
 OnCardHovered {nodes [OnCardHovered!() -> (then, IsHovered:bool=~) {is=title K2Node_FunctionEntry}]}
 OnCategoryClicked {nodes [OnCategoryClicked() {is=title K2Node_FunctionEntry}]}
 OnCategoryContextMenuClosed {nodes [OnCategoryContextMenuClosed() {is=title K2Node_FunctionEntry}]}
 OnCategoryContextMenuOpened {nodes [OnCategoryContextMenuOpened!() -> (then, ContextMenuParent:Widget Object Reference) {is=title K2Node_FunctionEntry}]}
 OnCategoryEditStyles {nodes [OnCategoryEditStyles() {is=title K2Node_FunctionEntry}]}
 OnCategoryItemSetup {nodes [OnCategoryItemSetup!() -> (then, CategoryItemEntryObject:User Widget Object Reference) {is=title K2Node_FunctionEntry}]}
 ResetItems {nodes [
   ResetItems() {is=title K2Node_FunctionEntry}
   n1 = Get ButtonGroup() -> (ButtonGroup:Common Button Group Base Object Reference)
   > DeselectAll(self:Common Button Group Base Object Reference=@n1)
   > RemoveAll(self:Common Widget Group Base Object Reference=@n1)
   > n2 = For Each Loop(Array:Array of User Widget Object References=GetAllEntries!(self:Dynamic Entry Box Base Object Reference=Get CategoryTopRow() -> (CategoryTopRow:Fort Dynamic Entry Box Object Reference)) -> (ReturnValue:Array of User Widget Object References)) -> (Array Element:User Widget Object Reference)
   Reset!(execute=@n2.Completed, self:Dynamic Entry Box Object Reference=Get CategoryTopRow(), bDeleteWidgets:bool=false) -> (then)
   > n3 = For Each Loop(Array=GetAllEntries(self=Get CategoryBottom() -> (CategoryBottom:Fort Dynamic Entry Box Object Reference)))
   Reset(execute=@n3.Completed, self=Get CategoryBottom(), bDeleteWidgets=false)
   n4 = Cast To WBP_UIKit_ItemCard_Category!(execute=@n3.LoopBody, Object=@n3) -> (then, CastFailed, AsWBP UIKit Item Card Category:WBP UIKit Item Card Category Object Reference) {is=title K2Node_DynamicCast}
   > Unbind Event from On Category Clicked(self:WBP UIKit Item Card Category Object Reference=@n4, Delegate=Create Event()) {is=title K2Node_RemoveDelegate}
   > Unbind Event from On Context Menu Closed(self:WBP UIKit Item Card Category Object Reference=@n4, Delegate=Create Event()) {is=title K2Node_RemoveDelegate}
   > Unbind Event from On Context Menu Opened(self:WBP UIKit Item Card Category Object Reference=@n4, Delegate=Create Event()) {is=title K2Node_RemoveDelegate}
   > Unbind Event from On Context Menu Card Edit Styles Clicked(self:WBP UIKit Item Card Category Object Reference=@n4, Delegate=Create Event()) {is=title K2Node_RemoveDelegate}
   > Unbind Event from Category Card Hovered(self:WBP UIKit Item Card Category Object Reference=@n4, Delegate=Create Event()) {is=title K2Node_RemoveDelegate}
   n5 = Cast To WBP_UIKit_ItemCard_Category(execute=@n2.LoopBody, Object=@n2)
   > Unbind Event from On Category Clicked(self=@n5, Delegate=Create Event())
   > Unbind Event from On Context Menu Closed(self=@n5, Delegate=Create Event())
   > Unbind Event from On Context Menu Opened(self=@n5, Delegate=Create Event())
   > Unbind Event from On Context Menu Card Edit Styles Clicked(self=@n5, Delegate=Create Event())
   > Unbind Event from Category Card Hovered(self=@n5, Delegate=Create Event())
  ]}
 SetRowDistance {nodes [
   n1 = SetRowDistance!() -> (then, EnableSpacer:bool=~) {is=title K2Node_FunctionEntry}
   n2 = Get CategoryTopRow()
   > SetSize!(
    execute
    self:Spacer Object Reference=Get Spacer_1() -> (Spacer_1:Spacer Object Reference)
    InSize_X:double=0
    InSize_Y:double=SelectFloat(A=Get!(self:Dynamic Entry Box Base Object Reference=@n2) -> (EntrySpacing_X:double, EntrySpacing_Y:double).EntrySpacing_Y, B=0, bPickA=@n1)
   ) -> (then)
  ]}
 SetupCategoryItem {nodes [
   n1 = SetupCategoryItem!() -> (then, Category:Fort Locker Category Item VM Object Reference, IsSecondRow:bool=~) {is=title K2Node_FunctionEntry}
   > n2 = Create Entry of Class!(
    execute
    self:Dynamic Entry Box Object Reference=Select(
     Option 0:Dynamic Entry Box Object Reference=Get CategoryTopRow()
     Option 1:Dynamic Entry Box Object Reference=Get CategoryBottom()
     Index=@n1.IsSecondRow
    ) -> (ReturnValue:Dynamic Entry Box Object Reference)
    EntryClass:User Widget Class Reference=/Locker/UIKit/Items/WBP_UIKit_ItemCard_Category.WBP_UIKit_ItemCard_Category_C
   ) -> (then, ReturnValue:WBP UIKit Item Card Category Object Reference)
   > Call On Category Item Setup!(execute:Exec, self:WBP Category Set Object Reference, CategoryItemEntryObject:User Widget Object Reference=@n2) -> (then:Exec) {is=title K2Node_CallDelegate}
   > Initialize!(
    execute
    self:WBP UIKit Item Card Category Object Reference=@n2
    Category Item:Fort Locker Category Item VM Object Reference=@n1
    CategoryGroup:Fort Locker Category Group VM Object Reference=Get FortLockerCategoryGroupVM() -> (FortLockerCategoryGroupVM:Fort Locker Category Group VM Object Reference)
   ) -> (then)
   > AddWidget!(execute, self:Common Widget Group Base Object Reference=Get ButtonGroup(), InWidget:Widget Object Reference=@n2) -> (then)
   > Bind Event to On Category Clicked(self:WBP UIKit Item Card Category Object Reference=@n2, Delegate=Create Event()) {is=title K2Node_AddDelegate}
   > Bind Event to On Context Menu Closed(self:WBP UIKit Item Card Category Object Reference=@n2, Delegate=Create Event()) {is=title K2Node_AddDelegate}
   > Bind Event to On Context Menu Opened(self:WBP UIKit Item Card Category Object Reference=@n2, Delegate=Create Event()) {is=title K2Node_AddDelegate}
   > Bind Event to On Context Menu Card Edit Styles Clicked(self:WBP UIKit Item Card Category Object Reference=@n2, Delegate=Create Event()) {is=title K2Node_AddDelegate}
   > Bind Event to Category Card Hovered(self:WBP UIKit Item Card Category Object Reference=@n2, Delegate=Create Event()) {is=title K2Node_AddDelegate}
   Select(
    Option 0:ELockerTileSize Enum=NewEnumerator0
    Option 1:ELockerTileSize Enum=Select(Option 0=NewEnumerator3, Option 1=NewEnumerator2, Index="Equal (Enum)"(A:Wildcard, B:Wildcard))
    Index="Not Equal (Enum)"!(A:Wildcard, B:Wildcard) -> (ReturnValue:bool)
   ) -> (ReturnValue:ELockerTileSize Enum)
   Get!(self:Fort Locker Category Config Object Reference=GetItemCategory!(self:Fort Item VM Object Reference=Get EquippedItem!(self:Fort Locker Category Item VM Object Reference) -> (EquippedItem:Fort Item VM Object Reference)) -> (ReturnValue:Fort Locker Category Config Object Reference)) -> (
    SlotInfo_CategoryName:Text
    SlotInfo_CategoryNamePlural:Text
    SlotInfo_CategoryNameWithPluralForms:Text
    SlotInfo_CategoryDescription:Text
    SlotInfo_DefaultEmptyTexture:Soft Object Reference
    SlotInfo_OverrideShuffleTexture:Texture 2D Soft Object Reference
    SlotInfo_CategoryIcon:Material Instance Object Reference
    SlotInfo_CategorySoundOverrides:Map of Gameplay Tag Structures to Sound Base Soft Object References
    SlotInfo_bShowInCategoryTab:bool=~
    SlotInfo_PreviewWindowScale:float=~
    SlotInfo_PreviewWindowOffset:Vector 3f
    SlotInfo_bUsePluralCategoryInItemList:bool=~
   )
  ]}
 SetupCategoryRows {nodes [
   n1 = SetupCategoryRows!() -> (then, CategoryItem:Array of Fort Locker Category Item VM Object References) {is=title K2Node_FunctionEntry}
   > Clear(TargetArray:Array of Fort Locker Category Item VM Object References=Get VisibleItems() -> (VisibleItems:Array of Fort Locker Category Item VM Object References))
   > ResetItems() {is=title K2Node_CallFunction}
   > n2 = For Each Loop(Array:Array of Fort Locker Category Item VM Object References=@n1) -> (Array Element:Fort Locker Category Item VM Object Reference)
   n3 = Last Index!(TargetArray:Array of Fort Locker Category Item VM Object References=Get VisibleItems()) -> (ReturnValue=0)
   n4 = OR Boolean(A="integer > integer"(A=@n3, B=7), B="integer < integer"(A=@n3, B=3))
   SetRowDistance(execute=@n2.Completed, EnableSpacer=NOT Boolean(A=@n4)) {is=title K2Node_CallFunction}
   > n5 = For Each Loop(Array=Get VisibleItems())
   Branch(execute=@n5.Completed, Condition=IsUsingTouch(PlayerController=GetOwningPlayer()))
   > SelectButtonAtIndex!(execute, self:Common Button Group Base Object Reference=Get ButtonGroup(), ButtonIndex:int=0, bAllowSound:bool=false) -> (then)
   n6 = "To Float (Integer)"(InInt=int + int(A=@n3))
   n7 = int - int(A=Ceil(A=float * float(A=@n6, B=0.5) {is=title K2Node_PromotableOperator}) {is=title K2Node_CallFunction})
   SetupCategoryItem!(
    execute=@n5.LoopBody
    self
    Category:Fort Locker Category Item VM Object Reference=@n5
    IsSecondRow:bool=OR Boolean(A="integer > integer"(A=@n5.Array Index, B=@n7), B=@n4)
   ) -> (then) {is=title K2Node_CallFunction}
   Branch(execute=@n2.LoopBody, Condition=IsItemVisible!(self:Fort Item VM Object Reference=Get EquippedItem(self=@n2)) -> (ReturnValue:bool))
   > Add(
    TargetArray:Array of Fort Locker Category Item VM Object References=Get VisibleItems()
    NewItem:"Fort Locker Category Item VM Object Reference (by ref)"=@n2
   ) {is=title K2Node_CallArrayFunction}
  ]}
]
variables [
 ButtonGroup {type=CommonButtonGroupBase, default=null}
 FirstTile {type=Widget, default=null}
 OnCardHovered {}
 OnCategoryClicked {}
 OnCategoryContextMenuClosed {}
 OnCategoryContextMenuOpened {}
 OnCategoryEditStyles {}
 OnCategoryItemSetup {}
 Tile Size {type=ELockerTileSize, default=Large Block 2x2}
 VisibleItems {type="array<FortLockerCategoryItemVM>", default []}
]
widgets {
 info {inheritedWidgetCount=0, parentClass=&UserWidget, rootWidgetClass=&VerticalBox, widgetCount=4}
 widgets [
  VerticalBox_0 {widgetClassPath=&VerticalBox, widget=VerticalBox_0}
  CategoryTopRow {
   widgetClassPath=/Script/FortniteUI.FortDynamicEntryBox
   bIsVariable=true
   parent=VerticalBox_0
   widget=CategoryTopRow
   properties {
    entryVerticalAlignment=VAlign_Top
    entryWidgetClass=&WBP_UIKit_ItemCard_Category_C
    numDesignerPreviewEntries=4
    entrySpacing=<16 32>
   }
   slot {class=VerticalBoxSlot, properties {verticalAlignment=VAlign_Top}}
  }
  Spacer_1 {
   widgetClassPath=&Spacer
   bIsVariable=true
   parent=VerticalBox_0
   widget=Spacer_1
   properties {size=<0 32>}
   slot {class=VerticalBoxSlot}
  }
  CategoryBottom {
   widgetClassPath=&FortDynamicEntryBox
   bIsVariable=true
   parent=VerticalBox_0
   widget=CategoryBottom
   properties {
    entryBoxType=Wrap
    entryVerticalAlignment=VAlign_Top
    entryWidgetClass=&WBP_UIKit_ItemCard_Category_C
    entrySpacing=<16 16>
   }
   slot {class=VerticalBoxSlot}
  }
 ]
}
