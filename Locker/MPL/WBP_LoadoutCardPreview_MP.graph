WidgetBlueprint /Locker/MPL/WBP_LoadoutCardPreview_MP
graphs [
 Create Item Entry {nodes [
   n1 = Create Item Entry!() -> (then, Dynamic Box:Dynamic Entry Box Object Reference, Viewmodel:Fort Locker Category Item VM Object Reference) {is=title K2Node_FunctionEntry}
   > SetVisibility(self=@n1, InVisibility=HitTestInvisible)
   > n2 = Create Entry!(execute, self:Dynamic Entry Box Object Reference=@n1) -> (then, ReturnValue:User Widget Object Reference)
   > n3 = Cast To WBP_UIKit_PresetSubPreviewTile!(execute, Object=@n2) -> (then, CastFailed, AsWBP UIKit Preset Sub Preview Tile:WBP UIKit Preset Sub Preview Tile Object Reference) {is=title K2Node_DynamicCast}
   > SetFortLockerCategoryItemVM!(execute, self:WBP UIKit Preset Sub Preview Tile Object Reference=@n3, Viewmodel:Fort Locker Category Item VM Object Reference=@n1.Viewmodel) -> (then)
   > SetItemCardSize!(execute, self:WBP UIKit Item Card Base Object Reference=@n3, ItemSize:Vector 2D Structure=(X=108,Y=108)) -> (then)
   > Set Default Background!(
    execute
    self:WBP UIKit Item Card Base Object Reference=@n3
    Color 1:Linear Color Structure=(R=0.016,G=0.0782,B=0.1779,A=1)
    Color 2:Linear Color Structure=(R=0.016,G=0.0782,B=0.1779,A=1)
    GradientStart:float=0
    GradientEnd:float=0
    GradientAngle:float=0
   ) -> (then) {is=title K2Node_CallFunction}
  ]}
 EventGraph {nodes [
   SetVisibility(
    self=Get SizeBox_Header() -> (SizeBox_Header:Size Box Object Reference)
    InVisibility="To Visibility (Boolean)"!(
     bIsVisible:bool=Get ShowHeader() -> (ShowHeader:bool)
     TrueVisibility:ESlateVisibility Enum=Visible
     FalseVisibility:ESlateVisibility Enum=Collapsed
    ) -> (ReturnValue:ESlateVisibility Enum=Visible)
   )
   Get PresetName() -> (PresetName:Common Text Object Reference)
  ]}
 Loadout Changed {nodes [
   n1 = Loadout Changed!() -> (then, CategoryItems:Array of Fort Locker Category Item VM Object References) {is=title K2Node_FunctionEntry}
   > SetVisibility(
    self=[Get SecondRow() -> (SecondRow:Dynamic Entry Box Object Reference), Get ThirdRow() -> (ThirdRow:Dynamic Entry Box Object Reference)]
    InVisibility=Collapsed
   )
   > Reset!(
    execute
    self:Dynamic Entry Box Object Reference=[Get FIrstRow() -> (FIrstRow:Dynamic Entry Box Object Reference), Get SecondRow(), Get ThirdRow()]
    bDeleteWidgets:bool=false
   ) -> (then)
   > n2 = For Each Loop(Array:Array of Fort Locker Category Item VM Object References=@n1) -> (Array Element:Fort Locker Category Item VM Object Reference)
   n3 = Last Index!(TargetArray:Array of Fort Locker Category Item VM Object References=@n1) -> (ReturnValue=0)
   n4 = SelectInt(
    A=99
    B=Select(Option 0:int=3, Option 1:int=4, Index="integer >= integer"(A=@n3, B=6)) -> (ReturnValue:int)
    bPickA=Get SingleRow() -> (SingleRow:bool)
   )
   > Create Item Entry!(
    execute
    self
    Dynamic Box:Dynamic Entry Box Object Reference=Select!(
     Option 0:Dynamic Entry Box Object Reference=Get FIrstRow()
     Option 1:Dynamic Entry Box Object Reference=Get SecondRow()
     Option 2:Dynamic Entry Box Object Reference=Get ThirdRow()
     Index:int=int / int(A=@n2.Array Index, B=@n4)
    ) -> (ReturnValue:Dynamic Entry Box Object Reference)
    Viewmodel:Fort Locker Category Item VM Object Reference=@n2
   ) -> (then) {is=title K2Node_CallFunction}
  ]}
 OnLoadoutChanged {nodes [OnLoadoutChanged() {is=title K2Node_FunctionEntry}]}
]
variables [
 HeaderShowGroupName {type=boolean}
 OnLoadoutChanged {}
 ShowHeader {type=boolean}
 SingleRow {type=boolean}
]
widgets {
 info {inheritedWidgetCount=0, parentClass=&UserWidget, rootWidgetClass=&VerticalBox, widgetCount=6}
 widgets [
  VBMainCard {widgetClassPath=&VerticalBox, widget=VBMainCard}
  SizeBox_Header {
   widgetClassPath=&SizeBox
   bIsVariable=true
   parent=VBMainCard
   widget=SizeBox_Header
   properties {bOverride_HeightOverride=true, heightOverride=48, visibility=Collapsed}
   slot {class=VerticalBoxSlot}
  }
  PresetName {
   widgetClassPath=&CommonTextBlock
   bIsVariable=true
   parent=SizeBox_Header
   widget=PresetName
   properties {
    justification=Center
    scrollStyle=/FortUILibrary/Styles/Typography/Foundation/TextScrollStyle-NoFade.TextScrollStyle-NoFade_C
    text="[PH] Preset Name"
    font {
     bForceMonospaced=false
     bMaterialIsStencil=false
     fontMaterial=null
     fontObject=&HeadingNow
     letterSpacing=0
     monospacedWidth=1
     size=18
     skewAmount=0
     typefaceFontName=86 Bold Italic
     outlineSettings {
      bApplyOutlineToDropShadows=false
      bMiteredCorners=false
      bSeparateFillAlpha=false
      outlineMaterial=null
      outlineSize=0
      outlineColor=<rgba 0 0 0 1>
     }
    }
    shadowOffset=<0 0>
   }
   slot {class=SizeBoxSlot, properties {horizontalAlignment=HAlign_Left, verticalAlignment=VAlign_Bottom}}
  }
  FIrstRow {
   widgetClassPath=&DynamicEntryBox
   bIsVariable=true
   parent=VBMainCard
   widget=FIrstRow
   properties {entryWidgetClass=/FortUILibrary/UIKit/Items/WBP_UIKit_PresetSubPreviewTile.WBP_UIKit_PresetSubPreviewTile_C, entrySpacing=<4 4>}
   slot {class=VerticalBoxSlot, properties {horizontalAlignment=HAlign_Center, verticalAlignment=VAlign_Center}}
  }
  SecondRow {
   widgetClassPath=&DynamicEntryBox
   bIsVariable=true
   parent=VBMainCard
   widget=SecondRow
   properties {
    entryWidgetClass=&WBP_UIKit_PresetSubPreviewTile_C
    visibility=Collapsed
    entrySpacing=<4 4>
   }
   slot {class=VerticalBoxSlot, properties {horizontalAlignment=HAlign_Center, verticalAlignment=VAlign_Center, padding=<ltrb 0 4 0 0>}}
  }
  ThirdRow {
   widgetClassPath=&DynamicEntryBox
   bIsVariable=true
   parent=VBMainCard
   widget=ThirdRow
   properties {
    entryWidgetClass=&WBP_UIKit_PresetSubPreviewTile_C
    visibility=Collapsed
    entrySpacing=<4 4>
   }
   slot {class=VerticalBoxSlot, properties {horizontalAlignment=HAlign_Left, verticalAlignment=VAlign_Center, padding=<ltrb 0 4 0 0>}}
  }
 ]
}
