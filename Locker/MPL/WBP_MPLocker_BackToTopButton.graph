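# MP Locker Back To Top widget that scrolls a target ListView to TargetIndex on click or controller action.
# SetList or Construct binds scroll and input-method events, which Destruct unbinds.
# Scrolling shows the touch or non-touch button once the offset reaches ScrollOffset, and collapses it below.
# Hardware visibility borders separate the two variants, input changes adjust height and scale, and compact UI tightens the touch icon padding.
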WidgetBlueprint /Locker/MPL/WBP_MPLocker_BackToTopButton
graphs [
 BackToTopInputAction {nodes [
   BackToTopInputAction() {is=title K2Node_FunctionEntry}
   > OnControllerClickEvent!(
    execute
    self
    Button:Common Button Base Object Reference=Get WBP_UIKit_Button_BackToTop() -> (WBP_UIKit_Button_BackToTop:WBP UIKit Button Generic Object Reference)
   ) -> (then)
   > Return Node!(execute, bPassThrough:bool=false)
  ]}
 EventGraph {nodes [
   K2Node_ComponentBoundEvent_0 = OnControllerClickEvent!() -> (OutputDelegate, then, Button:Common Button Base Object Reference) {is=title K2Node_CustomEvent}
   > n1 = Fire Event Element Interacted(Element=@K2Node_ComponentBoundEvent_0.Button)
   NavigateToIndex!(
    execute=[@n5.then, @n3.then, @n1.then]
    self:List View Object Reference=Get TargetList() -> (TargetList:List View Object Reference)
    Index:int=Get TargetIndex() -> (TargetIndex:int=0)
   ) -> (then)
   n2 = "On Clicked (WBP_UIKit_Button_BackToTop_Touch)"!() -> (OutputDelegate, then, Button:Common Button Base Object Reference) {is=title K2Node_ComponentBoundEvent}
   > n3 = Fire Event Element Interacted(Element=@n2.Button)
   n4 = "On Clicked (WBP_UIKit_Button_BackToTop)"!() -> (OutputDelegate, then, Button:Common Button Base Object Reference) {is=title K2Node_ComponentBoundEvent}
   > n5 = Fire Event Element Interacted(Element=@n4.Button)
   ListScrolled!() -> (OutputDelegate, then, ItemOffset:float=~, DistanceRemaining:float=~) {is=title K2Node_CustomEvent}
   > n6 = Branch(Condition=IsUsingTouch(PlayerController=GetPlayerController!(PlayerIndex:int=0) -> (ReturnValue:Player Controller Object Reference)))
   n7 = Get TargetList()
   n8 = Select(
    Option 0=Collapsed
    Option 1=Visible
    Index="float >= float"(A=GetScrollOffset!(self:List View Base Object Reference=@n7) -> (ReturnValue:float), B=Get ScrollOffset() -> (ScrollOffset:double))
   )
   > SetVisibility(self=Get WBP_UIKit_Button_BackToTop_Touch() -> (WBP_UIKit_Button_BackToTop_Touch:WBP UIKit Button Quiet Object Reference), InVisibility=@n8)
   SetVisibility(execute=@n6.else, self=Get WBP_UIKit_Button_BackToTop(), InVisibility=@n8)
   n9 = InputChanged!() -> (OutputDelegate, then, NewInputType:ECommonInputType Enum) {is=title K2Node_CustomEvent}
   n10 = Make Literal Float(Value=36)
   > SetMinDesiredHeight!(
    execute
    self:Size Box Object Reference=Get SizeBox_BackToTop() -> (SizeBox_BackToTop:Size Box Object Reference)
    InMinDesiredHeight:float=Select!(MouseAndKeyboard:float=@n10, Gamepad:float=78, Touch:float=80, Count:float=@n10, Index:ECommonInputType Enum=@n9.NewInputType) -> (ReturnValue:float)
   ) -> (then)
   > SetDesiredSizeScale!(
    execute
    self:Border Object Reference=Get VisibilityBorder_BackToTop_NonTouch() -> (VisibilityBorder_BackToTop_NonTouch:Common Hardware Visibility Border Object Reference)
    InScale:Vector 2D Structure=Select(
     MouseAndKeyboard:Vector 2D Structure=(X=1,Y=1)
     Gamepad:Vector 2D Structure=(X=1,Y=1.5)
     Touch:Vector 2D Structure=(X=1,Y=1)
     Count:Vector 2D Structure=(X=1,Y=1)
     Index=@n9.NewInputType
    ) -> (ReturnValue:Vector 2D Structure)
   ) -> (then)
   n11 = SetList!() -> (OutputDelegate, then, List:List View Object Reference) {is=title K2Node_CustomEvent}
   > n12 = Set TargetList(TargetList:List View Object Reference=@n11.List) -> (Output_Get:List View Object Reference)
   Get!(execute:Exec=[@n16.then, @n12.then]) -> (then:Exec, else:Exec, TargetList:List View Object Reference)
   > SetCompactUIStyle() {is=title K2Node_CallFunction}
   > ListScrolled!(execute, self, ItemOffset:float=0, DistanceRemaining:float=0) -> (then) {is=title K2Node_CallFunction}
   n13 = Create Event()
   > Bind Event to On List View Scrolled(self:List View Object Reference=Get TargetList(), Delegate=@n13) {is=title K2Node_AddDelegate}
   n14 = Get CommonInputSubsystem()
   > InputChanged!(execute, self, NewInputType:ECommonInputType Enum=GetCurrentInputType(self=@n14)) -> (then) {is=title K2Node_CallFunction}
   n15 = Create Event()
   > Bind Event to On Input Method Changed(self=@n14, Delegate=@n15)
   n16 = Event Construct()
   Event Destruct()
   n17 = Get TargetList()
   > n18 = Is Valid(InputObject=@n17)
   > n19 = Unbind Event from On List View Scrolled(self:List View Object Reference=@n17, Delegate=@n13) {is=title K2Node_RemoveDelegate}
   Unbind Event from On Input Method Changed(execute=[@n19.then, @n18.Is Not Valid], self=@n14, Delegate=@n15)
  ]}
 SetCompactUIStyle {nodes [
   SetCompactUIStyle() {is=title K2Node_FunctionEntry}
   > Branch(Condition=IsUsingCompactUI())
   n1 = Make Margin(Left=0, Top=-20, Right=0, Bottom=-20)
   n2 = Get Block_Icon!(self:WBP UIKit Button CTA Base Object Reference=Get WBP_UIKit_Button_BackToTop_Touch()) -> (Block_Icon:WBP UIKit Block Image Object Reference)
   > SetGridSlotPadding!(execute, Widget:Widget Object Reference=@n2, Padding:Margin Structure=@n1) -> (then)
   > SetPadding!(execute, self:User Widget Object Reference=@n2, InPadding:Margin Structure=@n1) -> (then)
  ]}
]
variables [
 Gamepad {type=number, default=78}
 In Scale {type=Vector2D, default=<0 0>}
 ScrollOffset {type=number, default=4}
 Target {type=WBP_MPLocker_BackToTopButton_C, default=null}
 TargetIndex {type=integer}
 TargetList {type=ListView, default=null}
]
widgets {
 info {inheritedWidgetCount=0, parentClass=&UserWidget, rootWidgetClass=&SizeBox, widgetCount=6}
 widgets [
  SizeBox_BackToTop {
   widgetClassPath=&SizeBox
   bIsVariable=true
   widget=SizeBox_BackToTop
   properties {bOverride_MinDesiredHeight=true, minDesiredHeight=38}
  }
  Overlay_0 {
   widgetClassPath=&Overlay
   parent=SizeBox_BackToTop
   widget=Overlay_0
   slot {class=SizeBoxSlot, properties {horizontalAlignment=HAlign_Center, verticalAlignment=VAlign_Bottom}}
  }
  VisibilityBorder_BackToTop_NonTouch {
   widgetClassPath=&CommonHardwareVisibilityBorder
   bIsVariable=true
   parent=Overlay_0
   widget=VisibilityBorder_BackToTop_NonTouch
   properties {visibilityQuery {
     autoDescription=" NONE( Input.Touch )"
     tokenStreamVersion=0
     userDescription=""
     queryTokenStream [0,1,3,1,0]
     tagDictionary [{tagName=Input.Touch}]
    }}
   slot {class=OverlaySlot, properties {horizontalAlignment=HAlign_Fill, verticalAlignment=VAlign_Fill}}
  }
  WBP_UIKit_Button_BackToTop {
   widgetClassPath=&WBP_UIKit_Button_Generic
   bIsVariable=true
   parent=VisibilityBorder_BackToTop_NonTouch
   widget=WBP_UIKit_Button_BackToTop
   properties {
    contentVerticalAlignment=VAlign_Fill
    icon=/Game/UI/CodeReferenced/T_UI_DirectionArrow
    iconMaterial=/FortUILibrary/UIKit/Buttons/MaterialInstances/GenericButtons/MI_GenericIconDeepBlue
    iconVerticalAlignment=VAlign_Top
    minHeight=0
    promptHorizontalAlignment=HAlign_Center
    text=""
    useDarkInputIcon=true
    visibility=Hidden
    buttonPadding=<ltrb 0 12 0 0>
    iconPadding=<ltrb 0 0 0 4>
    iconSize=<24 28>
    promptNudge=<0 14>
    triggeringInputAction {dataTable=&InputActionDataTable, rowName=GenericLeftStickClick}
   }
   slot {class=BorderSlot, properties {padding=<ltrb 0 0 0 0>}}
  }
  VisibilityBorder_BackToTop_Touch {
   widgetClassPath=&CommonHardwareVisibilityBorder
   parent=Overlay_0
   widget=VisibilityBorder_BackToTop_Touch
   properties {visibilityQuery {
     autoDescription=" ANY( Input.Touch )"
     tokenStreamVersion=0
     userDescription=""
     queryTokenStream [0,1,1,1,0]
     tagDictionary [{tagName=Input.Touch}]
    }}
   slot {class=OverlaySlot, properties {horizontalAlignment=HAlign_Center, verticalAlignment=VAlign_Bottom}}
  }
  WBP_UIKit_Button_BackToTop_Touch {
   widgetClassPath=/FortUILibrary/UIKit/Buttons/WBP_UIKit_Button_Quiet
   bIsVariable=true
   parent=VisibilityBorder_BackToTop_Touch
   widget=WBP_UIKit_Button_BackToTop_Touch
   properties {
    bHideInputActionWithKeyboard=true
    minHeight=48
    mobileMinHeight=80
    mobileMinWidth=56
    text=Back To Top
    visibility=Hidden
    triggeringInputAction {dataTable=&InputActionDataTable, rowName=GenericLeftStickClick}
   }
   slot {class=BorderSlot, properties {padding=<ltrb 0 0 0 0>}}
  }
 ]
}
