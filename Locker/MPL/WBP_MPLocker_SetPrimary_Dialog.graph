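# MP Locker dialog for setting the active loadout archetype or choosing Random shuffle.
# UpdateOptionsArray lists activatable category groups plus a Random entry and preselects the selected group.
# Arrow or shoulder buttons wrap the broadcast CurrentOptionIndex, with update functions setting the label, choice and preview card or shuffle icon.
# Confirm calls SetActiveArchetype or SetShuffleArchetype, then deactivates; activation loads dialog data, wraps navigation and hides front-end chrome after a short delay.
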WidgetBlueprint /Locker/MPL/WBP_MPLocker_SetPrimary_Dialog
graphs [
 BP_GetDesiredFocusTarget {nodes [
   Get Desired Focus Target()
   > n1 = Branch(Condition=IsMobileGame())
   > Return Node!(execute, ReturnValue:Widget Object Reference=Get WBP_UIKit_Backplate() -> (WBP_UIKit_Backplate:WBP UIKit Backplate Object Reference))
   n2 = Parent: Get Desired Focus Target!(execute=@n1.else) -> (then, ReturnValue:Widget Object Reference) {is=title K2Node_CallParentFunction}
   > Return Node(ReturnValue=@n2)
  ]}
 BP_GetDesiredInputConfig {nodes [
   Get Desired Input Config() {is=title K2Node_FunctionEntry}
   > Return Node(ReturnValue:UIInput Config Structure=Make UIInput Config!(
     bIgnoreMoveInput:bool=true
     bIgnoreLookInput:bool=true
     InputMode:ECommonInputMode Enum=Menu
     MouseCaptureMode:EMouseCaptureMode Enum=NoCapture
     MouseLockMode:EMouseLockMode Enum=DoNotLock
     bHideCursorDuringViewportCapture:bool=false
    ) -> (UIInputConfig:UIInput Config Structure) {is=title K2Node_MakeStruct})
  ]}
 EventGraph {nodes [
   n1 = "Dialog Button Clicked Dispatcher (Dialog_Internal)"!() -> (OutputDelegate, then, ButtonIndex:int=0) {is=title K2Node_ComponentBoundEvent}
   > n2 = Switch on Int!(execute:Exec, Selection:int=@n1.ButtonIndex) -> (0:Exec, 1:Exec) {bHasDefaultPin=false}
   > n3 = Branch(Condition=Get ShuffleSelected() -> (ShuffleSelected:bool))
   > n4 = SetShuffleArchetype(self:Fort Locker Category Group Collection VM Object Reference=Get CategoryGroupCollection!(self:Fort Locker Category Group VM Object Reference=Get FortLockerCategoryGroupVM() -> (FortLockerCategoryGroupVM:Fort Locker Category Group VM Object Reference)) -> (CategoryGroupCollection:Fort Locker Category Group Collection VM Object Reference))
   DeactivateWidget(execute=[@n5.then, @n4.then, @n2.1])
   n5 = SetActiveArchetype(
    execute=@n3.else
    self:Fort Locker Category Group VM Object Reference=Get CurrentChoiceVM() -> (CurrentChoiceVM:Fort Locker Category Group VM Object Reference)
   )
   "On Clicked (ToggleLeft)"!() -> (OutputDelegate, then, Button:Common Button Base Object Reference) {is=title K2Node_ComponentBoundEvent}
   > IncrementOption!(execute, self, IncrementUp:bool=false) -> (then)
   "On Clicked (ToggleRight)"!() -> (OutputDelegate, then, Button:Common Button Base Object Reference) {is=title K2Node_ComponentBoundEvent}
   > IncrementOption(IncrementUp=true)
   Event PreConstruct()
   n6 = Get ToggleLeft() -> (ToggleLeft:WBP UIKit Button Generic Object Reference)
   > SetScalarParameterValue(
    self=Get ImageMID!(self:WBP UIKit Block Image Object Reference=Get Block_Icon!(self:WBP UIKit Button Generic Base Object Reference=@n6) -> (Block_Icon:WBP UIKit Block Image Object Reference)) -> (ImageMID:Material Instance Dynamic Object Reference)
    ParameterName=Rotation
    Value=0.5
   )
   n7 = Get Dialog_Internal() -> (Dialog_Internal:WBP UIKit Dialog Internal Object Reference)
   > n8 = For Each Loop(Array:Array of WBP UIKit Button CTA Base Object References=Get ButtonWidgets!(self:WBP UIKit Dialog Internal Object Reference=@n7) -> (ButtonWidgets:Array of WBP UIKit Button CTA Base Object References)) -> (Array Element:WBP UIKit Button CTA Base Object Reference)
   > SetMinDimensions!(
    execute
    self:Common Button Base Object Reference=@n8
    InMinWidth:int=Break S UI Dialog Buttons!(S_UI_DialogButtons:"S UI Dialog Buttons Structure (by ref)"=Get Buttons!(self:WBP UIKit Dialog Internal Object Reference=@n7) -> (Buttons:S UI Dialog Buttons Structure)) -> (EachButtonMinWidth_77_#1:int) {is=title K2Node_BreakStruct}
    InMinHeight:int=Get Min Height() -> (Min Height:int=0)
   ) -> (then)
   Event Construct()
   > SetCompactUIStyle() {is=title K2Node_CallFunction}
   Event On Activated()
   > n9 = Construct Object from Class(Class=/Locker/MPL/Data/DialogVM_SetPrimary.DialogVM_SetPrimary_C) -> (ReturnValue:Dialog VM Set Primary Object Reference)
   > SetDialogData!(execute, self:WBP UIKit Dialog Internal Object Reference=Get Dialog_Internal(), Data:UI Kit Dialog Viewmodel Object Reference=@n9) -> (then)
   > SetNavigationRulesForSTW()
   > Delay(Duration=0.2)
   > SetFrontEndVisibility!(
    execute
    self:Fort Front End Context Object Reference=Get FortFrontEndContext() -> (ReturnValue:Fort Front End Context Object Reference) {is=title K2Node_GetSubsystem}
    bHideHeader:bool=true
    bHideFooter:bool=true
    bHideChatWidget:bool=true
   ) -> (then)
  ]}
 IncrementOption {nodes [
   n1 = IncrementOption!() -> (then, IncrementUp:bool=~) {is=title K2Node_FunctionEntry}
   n2 = Last Index!(TargetArray:Array of Texts=Get TextOptionsArray() -> (TextOptionsArray:Array of Texts)) -> (ReturnValue=0)
   n3 = int + int(A=Get CurrentOptionIndex() -> (CurrentOptionIndex:int=0), B=Select(Option 0:int=-1, Option 1:int=1, Index=@n1) -> (ReturnValue:int))
   > Set with Broadcast CurrentOptionIndex!(
    execute
    CurrentOptionIndex:int=Select(
     Option 0=Select(Option 0=@n2, Option 1=0, Index=@n1)
     Option 1=@n3
     Index="In Range (Integer)"!(Value:int=@n3, Min:int=0, Max:int=@n2, InclusiveMin:bool=true, InclusiveMax:bool=true) -> (ReturnValue:bool)
    )
   ) -> (then, Output_Get:int=0)
  ]}
 SetCompactUIStyle {nodes [
   SetCompactUIStyle() {is=title K2Node_FunctionEntry}
   > Branch(Condition=IsMobileGame())
   > Set WidthOverride!(execute, WidthOverride:float=80, self:Size Box Object Reference=Get SizeBox_LeftArrow() -> (SizeBox_LeftArrow:Size Box Object Reference)) -> (then, Output_Get:float)
   > Set HeightOverride!(execute, HeightOverride:float=80, self:Size Box Object Reference=Get SizeBox_LeftArrow()) -> (then, Output_Get:float)
   > Set WidthOverride(WidthOverride=80, self=Get SizeBox_RightArrow() -> (SizeBox_RightArrow:Size Box Object Reference))
   > Set HeightOverride(HeightOverride=80, self=Get SizeBox_RightArrow())
  ]}
 SetNavigationRulesForSTW {nodes [
   SetNavigationRulesForSTW() {is=title K2Node_FunctionEntry}
   > SetNavigationRuleBase!(execute, self:Widget Object Reference, Direction:EUINavigation Enum=Left, Rule:EUINavigationRule Enum=Wrap) -> (then)
   > SetNavigationRuleBase(Direction=Right, Rule=Wrap)
   > SetNavigationRuleBase(Direction=Up, Rule=Wrap)
   > SetNavigationRuleBase(Direction=Down, Rule=Wrap)
   > SetNavigationRuleBase(Direction=Next, Rule=Wrap)
   > SetNavigationRuleBase(Direction=Previous, Rule=Wrap)
  ]}
 SetRandom {nodes [
   n1 = SetRandom!() -> (then, IsRandom:bool=~) {is=title K2Node_FunctionEntry}
   > SetActiveWidgetIndex(
    self:Common Visibility Switcher Object Reference=Get LoadoutSwitcher() -> (LoadoutSwitcher:Common Visibility Switcher Object Reference)
    Index=SelectInt(A=1, B=0, bPickA=@n1)
   )
  ]}
 UpdateCategoryGroupVM {nodes [
   n1 = UpdateCategoryGroupVM!() -> (then, CategoryGroup:Fort Locker Category Group VM Object Reference) {is=title K2Node_FunctionEntry}
   > SetFortLockerCategoryGroupVM!(execute, self, Viewmodel:Fort Locker Category Group VM Object Reference=@n1) -> (then)
  ]}
 UpdateChoiceVM {nodes [
   n1 = UpdateChoiceVM!() -> (then, Index:int) {is=title K2Node_FunctionEntry}
   n2 = Get OptionsVMArray() -> (OptionsVMArray:Array of Fort Locker Category Group VM Object References)
   > n3 = Set CurrentChoiceVM(CurrentChoiceVM:Fort Locker Category Group VM Object Reference="Get (a copy)"(Array:Array of Fort Locker Category Group VM Object References=@n2, Dimension 1=@n1) -> (Output:Fort Locker Category Group VM Object Reference)) -> (Output_Get:Fort Locker Category Group VM Object Reference)
   n4 = Length(TargetArray:Array of Fort Locker Category Group VM Object References=@n2)
   > Set ShuffleSelected(ShuffleSelected:bool="integer > integer"(A=@n1, B=int - int(A=@n4))) -> (Output_Get:bool)
   > SetFortLockerCategoryGroupVM(
    self:WBP Loadout Card Preview MP Object Reference=Get Card_PresetPreview() -> (Card_PresetPreview:WBP Loadout Card Preview MP Object Reference)
    Viewmodel=@n3
   )
  ]}
 UpdateIndexInfo {nodes [
   n1 = UpdateIndexInfo!() -> (then, Index:int) {is=title K2Node_FunctionEntry}
   n2 = Get TextOptionsArray()
   > "SetText (Text)"(
    self=Get SelectorText() -> (SelectorText:Common Text Object Reference)
    InText="Get (a copy)"(Array:Array of Texts=@n2, Dimension 1=@n1) -> (Output:Text)
   )
   > SetRandom!(execute, self, IsRandom:bool="integer >= integer"(A=@n1, B=Last Index(TargetArray=@n2) -> (ReturnValue=0))) -> (then) {is=title K2Node_CallFunction}
  ]}
 UpdateOptionsArray {nodes [
   n1 = UpdateOptionsArray!() -> (then, GroupCollection:Fort Locker Category Group Collection VM Object Reference) {is=title K2Node_FunctionEntry}
   > Set SelectedGroupFound(SelectedGroupFound:bool=false) -> (Output_Get:bool)
   n2 = Get OptionsVMArray()
   > Clear(TargetArray:Array of Fort Locker Category Group VM Object References=@n2)
   n3 = Get TextOptionsArray()
   > Clear(TargetArray:Array of Texts=@n3)
   > n4 = For Each Loop(Array:Array of Fort Locker Category Group VM Object References=Get CategoryGroups!(self:Fort Locker Category Group Collection VM Object Reference=@n1) -> (CategoryGroups:Array of Fort Locker Category Group VM Object References)) -> (Array Element:Fort Locker Category Group VM Object Reference)
   > Branch(Condition=Get bIsArchetypeActivatable!(self:Fort Locker Category Group VM Object Reference=@n4) -> (bIsArchetypeActivatable:bool))
   > Add(
    TargetArray:Array of Texts=@n3
    NewItem:"Text (by ref)"=GetDisplayName!(self:Fort Locker Category Group VM Object Reference=@n4) -> (ReturnValue:"Text (by ref)")
   ) {is=title K2Node_CallArrayFunction}
   > n5 = Add(
    TargetArray:Array of Fort Locker Category Group VM Object References=@n2
    NewItem:"Fort Locker Category Group VM Object Reference (by ref)"=@n4
   )
   n6 = Get FortItemSelectionVM() -> (FortItemSelectionVM:Fort Item Selection VM Object Reference)
   > Branch(Condition="Equal (Object)"(
     A=@n4
     B=Get SelectedCategoryGroup!(self:Fort Item Selection VM Object Reference=@n6) -> (SelectedCategoryGroup:Fort Locker Category Group VM Object Reference)
    ))
   > Set with Broadcast CurrentOptionIndex(CurrentOptionIndex=@n5) -> (Output_Get=0)
   > Set SelectedGroupFound(SelectedGroupFound=true)
   Add(execute=@n4.Completed, TargetArray:Array of Texts=@n3, NewItem:"Text (by ref)"="LOCTEXT(\"Random\")")
   > n7 = Branch(Condition=Get SelectedGroupFound() -> (SelectedGroupFound:bool))
   Set with Broadcast CurrentOptionIndex(execute=@n7.else, CurrentOptionIndex=0) -> (Output_Get=0)
  ]}
]
variables [
 CurrentChoiceVM {type=FortLockerCategoryGroupVM, default=null}
 CurrentOptionIndex {type=integer}
 CurrentSelectorText {type=string, default=""}
 Min Height {type=integer, default=56}
 OptionsVMArray {type="array<FortLockerCategoryGroupVM>", default []}
 SelectedGroupFound {type=boolean}
 ShuffleSelected {type=boolean}
 TextOptionsArray {type="array<string>", default []}
]
widgets {
 info {
  inheritedWidgetCount=0
  namedSlotCount=2
  parentClass=&CommonActivatableWidget
  rootWidgetClass=/FortUILibrary/Systems/Dialog/WBP_UIKit_Dialog_Internal.WBP_UIKit_Dialog_Internal_C
  widgetCount=15
 }
 widgets [
  Dialog_Internal {
   widgetClassPath=/FortUILibrary/Systems/Dialog/WBP_UIKit_Dialog_Internal
   bIsVariable=true
   widget=Dialog_Internal
   properties {
    designTimePreview_DialogVM=&DialogVM_SetPrimary_C
    useImage=true
    useTitleImage=true
    buttons {
     each Button Horizontal Fill=false
     each Button Min Width=200
     each Button Size=Medium
     focuse Button=0
     space Between Buttons=24
     cTA Buttons Types [Regular,Quiet]
     group Placement {
      horizontal Alignment=HAlign_Fill
      vertical Alignment=VAlign_Bottom
      grid {column=1, column Span=0, layer=0, row=2, row Span=0, nudge=<0 0>}
      padding=<ltrb 80 64 80 48>
     }
    }
    dialogSIze {
     height Override {enable=false, value=0}
     max Aspect Ratio {enable=false, value=0}
     max Desired Height {enable=true, value=800}
     max Desired Width {enable=false, value=0}
     min Aspect Ratio {enable=false, value=0}
     min Desired Height {enable=true, value=520}
     min Desired Width {enable=false, value=0}
     width Override {enable=true, value=1424}
    }
    imageProperties {
     material=/FortUILibrary/Systems/Dialog/Materials/MI_UI_BackgroundBlue
     image Size=<512 0>
     placement {
      horizontal Alignment=HAlign_Fill
      vertical Alignment=VAlign_Fill
      grid {column=1, column Span=0, layer=20, row=0, row Span=1, nudge=<0 0>}
      padding=<ltrb 0 0 0 0>
     }
    }
    textProperties {
     auto Wrap Text=true
     justification=Left
     line Height Percentage=1.35
     mobile Font Size Multiplier=1.3
     transform Policy=None
     wrap Text At=0
     color {colorUseRule=UseColor_Specified, specifiedColor=<rgba 1 1 1 1>}
     font {
      bForceMonospaced=false
      bMaterialIsStencil=false
      fontMaterial=null
      fontObject=&HeadingNow
      letterSpacing=0
      monospacedWidth=1
      size=15.75
      skewAmount=0
      typefaceFontName=74 Regular
      outlineSettings {
       bApplyOutlineToDropShadows=false
       bMiteredCorners=false
       bSeparateFillAlpha=false
       outlineMaterial=null
       outlineSize=0
       outlineColor=<rgba 0 0 0 1>
      }
     }
     placement {
      horizontal Alignment=HAlign_Fill
      vertical Alignment=VAlign_Fill
      grid {column=1, column Span=0, layer=0, row=1, row Span=0, nudge=<0 0>}
      padding=<ltrb 80 0 80 75>
     }
    }
    titleImagePadding=<ltrb 0 0 8 0>
   }
  }
  Overlay_0 {widgetClassPath=&Overlay, namedSlotHost=Dialog_Internal, widget=Overlay_0}
  SizeBox_1 {
   widgetClassPath=&SizeBox
   parent=Overlay_0
   widget=SizeBox_1
   properties {bOverride_WidthOverride=true, widthOverride=660}
   slot {class=OverlaySlot, properties {horizontalAlignment=HAlign_Fill, verticalAlignment=VAlign_Bottom, padding=<ltrb 84 0 84 180>}}
  }
  Overlay_1 {widgetClassPath=&Overlay, parent=SizeBox_1, widget=Overlay_1, slot {class=SizeBoxSlot}}
  WBP_UIKit_Backplate {
   widgetClassPath=&WBP_UIKit_Backplate
   bIsVariable=true
   parent=Overlay_1
   widget=WBP_UIKit_Backplate
   properties {
    bIsFocusable=true
    brightness=Dark
    cornerRadius=Interactive Component
    navigation=&self:WidgetTree.WBP_UIKit_Backplate.WidgetNavigation_0
   }
   slot {class=OverlaySlot, properties {horizontalAlignment=HAlign_Fill, verticalAlignment=VAlign_Fill, padding=<ltrb -7 -7 -7 -7>}}
  }
  HorizontalBox_24 {
   widgetClassPath=&HorizontalBox
   parent=Overlay_1
   widget=HorizontalBox_24
   slot {class=OverlaySlot, properties {horizontalAlignment=HAlign_Fill}}
  }
  SizeBox_LeftArrow {
   widgetClassPath=&SizeBox
   bIsVariable=true
   parent=HorizontalBox_24
   widget=SizeBox_LeftArrow
   properties {bOverride_HeightOverride=true, bOverride_WidthOverride=true, heightOverride=56, widthOverride=56}
   slot {class=HorizontalBoxSlot, properties {horizontalAlignment=HAlign_Left, verticalAlignment=VAlign_Top}}
  }
  ToggleLeft {
   widgetClassPath=&WBP_UIKit_Button_Generic
   bIsVariable=true
   parent=SizeBox_LeftArrow
   widget=ToggleLeft
   properties {
    iconHorizontalAlignment=HAlign_Fill
    iconMaterial=/Locker/MPL/Materials/MI_MPLocker_Chevron
    iconVerticalAlignment=VAlign_Fill
    mobileMinHeight=80
    useTextBlock=false
    iconPadding=<ltrb -20 -20 -20 -20>
    promptNudge=<-48 0>
    triggeringInputAction {dataTable=&InputActionDataTable, rowName=GenericLeftShoulder}
   }
   slot {class=SizeBoxSlot}
  }
  SelectorText {
   widgetClassPath=&CommonTextBlock
   bIsVariable=true
   parent=HorizontalBox_24
   widget=SelectorText
   properties {
    justification=Center
    text=Text Block
    font {
     bForceMonospaced=false
     bMaterialIsStencil=false
     fontMaterial=null
     fontObject=&HeadingNow
     letterSpacing=0
     monospacedWidth=1
     size=20.25
     skewAmount=0
     typefaceFontName=75 Medium
     outlineSettings {
      bApplyOutlineToDropShadows=false
      bMiteredCorners=false
      bSeparateFillAlpha=false
      outlineMaterial=null
      outlineSize=0
      outlineColor=<rgba 0 0 0 1>
     }
    }
    shadowOffset=<0 0>
   }
   slot {class=HorizontalBoxSlot, properties {verticalAlignment=VAlign_Center, size {sizeRule=Fill, value=1}}}
  }
  SizeBox_RightArrow {
   widgetClassPath=&SizeBox
   bIsVariable=true
   parent=HorizontalBox_24
   widget=SizeBox_RightArrow
   properties {bOverride_HeightOverride=true, bOverride_WidthOverride=true, heightOverride=56, widthOverride=56}
   slot {class=HorizontalBoxSlot, properties {horizontalAlignment=HAlign_Right, verticalAlignment=VAlign_Top}}
  }
  ToggleRight {
   widgetClassPath=&WBP_UIKit_Button_Generic
   bIsVariable=true
   parent=SizeBox_RightArrow
   widget=ToggleRight
   properties {
    iconHorizontalAlignment=HAlign_Fill
    iconMaterial=&MI_MPLocker_Chevron
    iconVerticalAlignment=VAlign_Fill
    mobileMinHeight=80
    promptHorizontalAlignment=HAlign_Right
    useTextBlock=false
    iconPadding=<ltrb -20 -20 -20 -20>
    promptNudge=<48 0>
    triggeringInputAction {dataTable=&InputActionDataTable, rowName=GenericRightShoulder}
   }
   slot {class=SizeBoxSlot}
  }
  Overlay_CardsPreset {widgetClassPath=&Overlay, bIsVariable=true, namedSlotHost=Dialog_Internal, widget=Overlay_CardsPreset}
  LoadoutSwitcher {
   widgetClassPath=&CommonVisibilitySwitcher
   bIsVariable=true
   parent=Overlay_CardsPreset
   widget=LoadoutSwitcher
   slot {class=OverlaySlot, properties {horizontalAlignment=HAlign_Fill, verticalAlignment=VAlign_Fill}}
  }
  Card_PresetPreview {
   widgetClassPath=&/WBP_LoadoutCardPreview_MP
   bIsVariable=true
   parent=LoadoutSwitcher
   widget=Card_PresetPreview
   properties {renderTransform {angle=0, scale=<1.08 1.08>, shear=<0 0>, translation=<0 0>}}
   slot {
    class=CommonVisibilitySwitcherSlot
    properties {horizontalAlignment=HAlign_Center, verticalAlignment=VAlign_Center, padding=<ltrb 36 0 36 0>}
   }
  }
  ShuffleIcon {
   widgetClassPath=&Image
   bIsVariable=true
   parent=LoadoutSwitcher
   widget=ShuffleIcon
   properties {
    brush {
     bIsDynamicallyLoaded=false
     drawAs=Image
     imageType=FullColor
     mirroring=NoMirror
     resourceName=None
     resourceObject=/Game/UI/Foundation/Textures/Icons/Locker/T_UI_Random_256
     tiling=NoTile
     imageSize=<300 300>
     margin=<ltrb 0 0 0 0>
     outlineSettings {
      bUseBrushTransparency=false
      roundingType=HalfHeightRadius
      width=0
      color {colorUseRule=UseColor_Specified, specifiedColor=<rgba 0 0 0 0>}
      cornerRadii=<0 0 0 0>
     }
     tintColor {colorUseRule=UseColor_Specified, specifiedColor=<rgba 1 1 1 1>}
     uVRegion {bIsValid=false, max=<0 0>, min=<0 0>}
    }
    colorAndOpacity=<rgba 1 1 1 0.9>
   }
   slot {class=CommonVisibilitySwitcherSlot, properties {horizontalAlignment=HAlign_Center, verticalAlignment=VAlign_Center, padding=<ltrb 0 0 36 0>}}
  }
 ]
}
