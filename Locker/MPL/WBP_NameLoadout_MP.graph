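# Locker dialog popup (FortSaveLoadoutPopup) for saving, overwriting, renaming, clearing or deleting a modular loadout preset.
# Initialize stores the action and loadouts, then configures preview cards and name field, showing both cards for overwrites.
# Accept confirms names after asynchronous sanitization, runs async deletion, or fires the ClearLoadout dispatcher.
# Focus targets the name field or Accept by action and subgame, and commit focus-clearing follows input method.
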WidgetBlueprint /Locker/MPL/WBP_NameLoadout_MP
graphs [
 BP_GetDesiredFocusTarget {nodes [
   Get Desired Focus Target()
   > n1 = Branch(Condition=NOT Boolean(A=IsInSubGame!(SubGame:ESubGame Enum=Campaign) -> (ReturnValue:bool)))
   > n2 = Switch on EFortLoadoutActionType!(execute, Selection:EFortLoadoutActionType Enum=Get Loadout Action() -> (Loadout Action:EFortLoadoutActionType Enum=Rename)) -> (Rename:Exec, Overwrite:Exec, SaveNew:Exec, Clear:Exec, Delete:Exec) {is=title K2Node_SwitchEnum}
   n3 = Cast To WBP_UIKit_FilteredInputField!(
    execute=[@n2.Overwrite, @n2.Rename, @n2.SaveNew]
    Object=Get Text_EditableLoadoutName() -> (Text_EditableLoadoutName:Fort Editable Filtered Counted Text Box Object Reference)
   ) -> (then, CastFailed, AsWBP UIKit Filtered Input Field:WBP UIKit Filtered Input Field Object Reference) {is=title K2Node_DynamicCast}
   n4 = Get InputField!(self:WBP UIKit Filtered Input Field Object Reference=@n3) -> (InputField:WBP UIKit Input Field Object Reference)
   > Return Node!(
    execute
    ReturnValue:Widget Object Reference=Get EditableText!(self:WBP UIKit Block Editable Text Object Reference=Get Block_EditableText!(self:WBP UIKit Input Field Base Object Reference=@n4) -> (Block_EditableText:WBP UIKit Block Editable Text Object Reference)) -> (EditableText:UIKit Editable Text Object Reference)
   )
   Return Node(
    execute=[@n3.CastFailed, @n2.Clear, @n2.Delete, @n1.else]
    ReturnValue=Get Button_Accept() -> (Button_Accept:WBP UIKit Button Regular Object Reference)
   )
  ]}
 ClearLoadout {nodes [ClearLoadout() {is=title K2Node_FunctionEntry}]}
 EventGraph {nodes [
   "On Clicked (Button_Accept)"!() -> (OutputDelegate, then, Button:Common Button Base Object Reference) {is=title K2Node_ComponentBoundEvent}
   > n1 = Switch on EFortLoadoutActionType(Selection=Get Loadout Action() -> (Loadout Action=Rename))
   n2 = Call Clear Loadout!(execute:Exec=@n1.Clear, self:WBP Name Loadout MP Object Reference) -> (then:Exec) {is=title K2Node_CallDelegate}
   Set RetainFocus(execute=[@n2.then, @n4.then, @n5.then, @n10.then], RetainFocus:bool=false) -> (Output_Get:bool)
   > TryPopActivatablePanel!(execute, Panel:Common Activatable Panel Legacy Object Reference=Self-Reference()) -> (then)
   n3 = DeleteModularLoadout!(
    execute:Exec=@n1.Delete
    ModularLoadout:Fort Typed Modular Loadout VM Object Reference=Get Target Loadout() -> (Target Loadout:Fort Typed Modular Loadout VM Object Reference)
   ) -> (then:Exec, OnOperationCompleted:Exec, Success:bool=~) {is=title K2Node_AsyncAction}
   n4 = HandleActionConfirmed!(execute=@n3.OnOperationCompleted, self:Fort Save Loadout Popup Object Reference, OptionalNewName:String) -> (then)
   SanitizeText(execute=[@n1.Overwrite, @n1.Rename, @n1.SaveNew], self:Fort Save Loadout Popup Object Reference)
   "On Clicked (Button_Cancel)"!() -> (OutputDelegate, then, Button:Common Button Base Object Reference) {is=title K2Node_ComponentBoundEvent}
   > n5 = HandleActionDeclined(self:Fort Save Loadout Popup Object Reference)
   "On Text Changed (Text_EditableLoadoutName)"!() -> (OutputDelegate, then, Text:Text) {is=title K2Node_ComponentBoundEvent}
   > Set bIsEnabled!(execute, bIsEnabled:bool=true, self:Widget Object Reference=Get Button_Accept()) -> (then, Output_Get:bool)
   n6 = "On Text Committed (Text_EditableLoadoutName)"!() -> (OutputDelegate, then, CommitMethod:ETextCommit Enum=Default) {is=title K2Node_ComponentBoundEvent}
   > n7 = Switch on ETextCommit!(execute, Selection:ETextCommit Enum=@n6.CommitMethod) -> (Default:Exec, OnEnter:Exec, OnUserMovedFocus:Exec, OnCleared:Exec) {is=title K2Node_SwitchEnum}
   RequestRefreshFocus(execute=@n7.OnEnter, self:Common Activatable Widget Object Reference)
   OnInputMethodChanged!() -> (OutputDelegate, then, bNewInputType:ECommonInputType Enum) {is=title K2Node_CustomEvent}
   > UpdateShouldClearKeyboardFocusOnCommit()
   Event On Activated()
   > SetNavigationRules()
   > Branch(Condition=IsInSubGame(SubGame=Campaign))
   > Set RetainFocus(RetainFocus=true)
   > Delay(Duration=0.2)
   > SetFocus()
   > SetFrontEndVisibility!(
    execute
    self:Fort Front End Context Object Reference=Get FortFrontEndContext() -> (ReturnValue:Fort Front End Context Object Reference) {is=title K2Node_GetSubsystem}
    bHideHeader:bool=true
    bHideFooter:bool=true
    bHideChatWidget:bool=true
   ) -> (then)
   Event OnInitialized()
   > SetCompactUIStyle() {is=title K2Node_CallFunction}
   > Bind Event to On Input Method Changed(self=Get CommonInputSubsystem(), Delegate=Create Event())
   > UpdateShouldClearKeyboardFocusOnCommit()
   n8 = Event OnSanitizationReady!() -> (OutputDelegate, then, bSuccess:bool, Sanitizedtext:String) {is=title K2Node_Event}
   > n9 = Branch(Condition=@n8.bSuccess)
   > n10 = HandleActionConfirmed(OptionalNewName=@n8.Sanitizedtext)
   ForceSetText!(
    execute=@n9.else
    self:Fort Editable Filtered Counted Text Box Object Reference=Get Text_EditableLoadoutName()
    Text:"Text (by ref)"="To Text (String)"(InString=@n8.Sanitizedtext)
   ) -> (then)
   > Set bIsEnabled(bIsEnabled=false, self=Get Button_Accept())
   Event Construct()
   > n11 = Get!(execute:Exec) -> (then:Exec, else:Exec, DialogVM:UI Kit Dialog Viewmodel Object Reference)
   > SetDialogData!(
    execute
    self:WBP UIKit Dialog Internal Object Reference=Get Dialog_Internal() -> (Dialog_Internal:WBP UIKit Dialog Internal Object Reference)
    Data:UI Kit Dialog Viewmodel Object Reference=@n11
   ) -> (then)
   Event On Deactivated()
   > Set RetainFocus(RetainFocus=false)
  ]}
 Initialize {nodes [
   n1 = Initialize!() -> (
    then
    Loadout Action:EFortLoadoutActionType Enum=Rename
    Target Loadout:Fort Typed Modular Loadout VM Object Reference
    TypedModularLoadouts:Fort Typed Modular Loadouts VM Object Reference
   ) {is=title K2Node_FunctionEntry}
   > Set Target Loadout(Target Loadout:Fort Typed Modular Loadout VM Object Reference=@n1.Target Loadout) -> (Output_Get:Fort Typed Modular Loadout VM Object Reference)
   > Set CurrentLoadout(CurrentLoadout:Fort Typed Modular Loadout VM Object Reference=Get ActiveLoadout!(self:Fort Typed Modular Loadouts VM Object Reference=@n1.TypedModularLoadouts) -> (ActiveLoadout:Fort Typed Modular Loadout VM Object Reference)) -> (Output_Get:Fort Typed Modular Loadout VM Object Reference)
   > Set Loadout Action(Loadout Action:EFortLoadoutActionType Enum=@n1) -> (Output_Get:EFortLoadoutActionType Enum=Rename)
   > n2 = Switch on EFortLoadoutActionType(Selection=@n1)
   n3 = Get Target Loadout()
   SetupForClear!(execute=@n2.Clear, self, TargetLoadout:Fort Typed Modular Loadout VM Object Reference=@n3) -> (then)
   SetupForDelete!(execute=@n2.Delete, self, TargetLoadout:Fort Typed Modular Loadout VM Object Reference=@n3) -> (then)
   SetupForSave!(
    execute=[@n2.Overwrite, @n2.SaveNew]
    self
    Target Loadout:Fort Typed Modular Loadout VM Object Reference=@n3
    Override:bool="Equal (Enum)"(A:EFortLoadoutActionType Enum=@n1, B:EFortLoadoutActionType Enum=Overwrite)
   ) -> (then)
   SetupForRename!(execute=@n2.Rename, self, Target Loadout:Fort Typed Modular Loadout VM Object Reference=@n3) -> (then)
  ]}
 SetCompactUIStyle {nodes [
   SetCompactUIStyle() {is=title K2Node_FunctionEntry}
   > Branch(Condition=IsUsingCompactUI())
   > SetPadding!(
    execute
    self:Size Box Slot Object Reference=SlotAsSizeBoxSlot!(Widget:Widget Object Reference=Get Overlay_CardsPreset() -> (Overlay_CardsPreset:Overlay Object Reference)) -> (ReturnValue:Size Box Slot Object Reference)
    InPadding:Margin Structure=Make Margin(Left=915, Top=0, Right=0, Bottom=0)
   ) -> (then)
   > SetPadding(
    self:Vertical Box Slot Object Reference=SlotAsVerticalBoxSlot(Widget=Get HB_Buttons() -> (HB_Buttons:Horizontal Box Object Reference))
    InPadding=Make Margin(Left=0, Top=0, Right=0, Bottom=0)
   )
   > SetPadding!(
    execute
    self=SlotAsVerticalBoxSlot(Widget=Get Spacer() -> (Spacer:Spacer Object Reference))
    InPadding_Left:float=0
    InPadding_Top:float=0
    InPadding_Right:float=0
    InPadding_Bottom:float=45
   ) -> (then)
  ]}
 SetNavigationRules {nodes [
   SetNavigationRules() {is=title K2Node_FunctionEntry}
   > SetNavigationRuleBase!(execute, self:Widget Object Reference, Direction:EUINavigation Enum=Left, Rule:EUINavigationRule Enum=Wrap) -> (then)
   > SetNavigationRuleBase(Direction=Right, Rule=Wrap)
   > SetNavigationRuleBase(Direction=Up, Rule=Wrap)
   > SetNavigationRuleBase(Direction=Down, Rule=Wrap)
   > SetNavigationRuleBase(Direction=Next, Rule=Wrap)
   > SetNavigationRuleBase(Direction=Previous, Rule=Wrap)
  ]}
 SetupForClear {nodes [
   n1 = SetupForClear!() -> (then, TargetLoadout:Fort Typed Modular Loadout VM Object Reference) {is=title K2Node_FunctionEntry}
   > SetFortTypedModularLoadoutVM!(
    execute
    self:WBP Loadout Card MP Object Reference=Get Card_NewPreset() -> (Card_NewPreset:WBP Loadout Card MP Object Reference)
    Viewmodel:Fort Typed Modular Loadout VM Object Reference=@n1
   ) -> (then)
   > Use2PresetLayout!(execute, self, Use 2:bool=false) -> (then)
   > SetVisibility(self=Get Text_EditableLoadoutName(), InVisibility=Collapsed)
   > SetText!(execute, self:Fort CTAButton Object Reference=Get Button_Accept(), ButtonText:"Text (by ref)"=Get AcceptLabel() -> (AcceptLabel:Text)) -> (then)
   > SetText(self=Get Button_Cancel() -> (Button_Cancel:WBP UIKit Button Regular Object Reference), ButtonText=Get CancelLabel() -> (CancelLabel:Text))
  ]}
 SetupForDelete {nodes [
   n1 = SetupForDelete!() -> (then, TargetLoadout:Fort Typed Modular Loadout VM Object Reference) {is=title K2Node_FunctionEntry}
   > SetFortTypedModularLoadoutVM(self=Get Card_NewPreset(), Viewmodel=@n1)
   > Use2PresetLayout(Use 2=false)
   > SetVisibility(self=Get Text_EditableLoadoutName(), InVisibility=Collapsed)
   > SetText(self=Get Button_Accept(), ButtonText=Get AcceptLabel())
   > SetText(self=Get Button_Cancel(), ButtonText=Get CancelLabel())
  ]}
 SetupForRename {nodes [
   n1 = SetupForRename!() -> (then, Target Loadout:Fort Typed Modular Loadout VM Object Reference) {is=title K2Node_FunctionEntry}
   > SetFortTypedModularLoadoutVM(self=Get Card_NewPreset(), Viewmodel=@n1)
   > ForceSetText(
    self=Get Text_EditableLoadoutName()
    Text="To Text (String)"(InString=Get Name!(self:Fort Typed Modular Loadout VM Object Reference=@n1) -> (Name:String))
   )
   > Use2PresetLayout(Use 2=false)
  ]}
 SetupForSave {nodes [
   n1 = SetupForSave!() -> (then, Target Loadout:Fort Typed Modular Loadout VM Object Reference, Override:bool=~) {is=title K2Node_FunctionEntry}
   > SetFortTypedModularLoadoutVM(self=Get Card_NewPreset(), Viewmodel=Get CurrentLoadout() -> (CurrentLoadout:Fort Typed Modular Loadout VM Object Reference))
   > n2 = Branch(Condition=Get Override() -> (Override:bool))
   n3 = Get Target Loadout()
   > SetFortTypedModularLoadoutVM(self=Get Card_ExistingPreset() -> (Card_ExistingPreset:WBP Loadout Card MP Object Reference), Viewmodel=@n3)
   n4 = Get Text_EditableLoadoutName()
   > n5 = ForceSetText(self=@n4, Text="To Text (String)"(InString=Get Name(self=@n3)))
   Use2PresetLayout(execute=[@n5.then, @n6.then], Use 2=Get Override())
   n6 = ForceSetText(execute=@n2.else, self=@n4, Text=GetDefaultName!(self:Fort Typed Modular Loadout VM Object Reference=@n3) -> (ReturnValue:Text))
   Reroute Node(InputPin:bool=@n1.Override) -> (OutputPin:bool)
  ]}
 UpdateShouldClearKeyboardFocusOnCommit {nodes [
   UpdateShouldClearKeyboardFocusOnCommit() {is=title K2Node_FunctionEntry}
   > n1 = Cast To WBP_UIKit_FilteredInputField(Object=Get Text_EditableLoadoutName())
   > Set ShouldClearKeyboardFocusOnCommit!(
    execute
    ShouldClearKeyboardFocusOnCommit:bool=IsInputType!(InputType:ECommonInputType Enum=Gamepad, IsNot:bool=true) -> (Result:bool)
    self:WBP UIKit Input Field Base Object Reference=Get InputField(self=@n1)
   ) -> (then, Output_Get:bool)
  ]}
 Use2PresetLayout {nodes [
   n1 = Use2PresetLayout!() -> (then, Use 2:bool=~) {is=title K2Node_FunctionEntry}
   > Set Visibility(
    Visibility=Select(Option 0=Collapsed, Option 1=HitTestInvisible, Index=@n1)
    self=Get ExisitingPresetGroup() -> (ExisitingPresetGroup:Overlay Object Reference)
   )
  ]}
]
variables [
 ClearLoadout {}
 CurrentLoadout {type=FortTypedModularLoadoutVM, default=null}
 DialogVM {type=UIKitDialogViewModel, default=null}
 Loadout Action {type=EFortLoadoutActionType, default=Rename}
 RetainFocus {type=boolean}
 Target Loadout {type=FortTypedModularLoadoutVM, default=null}
]
widgets {
 info {
  inheritedWidgetCount=1
  namedSlotCount=2
  parentClass=/Script/FortniteUI.FortSaveLoadoutPopup
  rootWidgetClass=/FortUILibrary/Systems/Dialog/WBP_UIKit_Dialog_Internal.WBP_UIKit_Dialog_Internal_C
  widgetCount=14
 }
 widgets [
  Dialog_Internal {
   widgetClassPath=/FortUILibrary/Systems/Dialog/WBP_UIKit_Dialog_Internal
   bIsVariable=true
   widget=Dialog_Internal
   properties {
    designTimePreview_DialogVM=/Locker/MPL/Data/DialogVM_LockerRename.DialogVM_LockerRename_C
    useImage=true
    backgroundProperties {
     material=/FortUILibrary/UIKit/Dialogs/MaterialInstances/MI_UI_DialogBackground
     image Size=<32 32>
     placement {
      horizontal Alignment=HAlign_Fill
      vertical Alignment=VAlign_Fill
      grid {column=0, column Span=2, layer=1, row=0, row Span=0, nudge=<0 0>}
      padding=<ltrb 0 0 0 0>
     }
    }
    buttons {
     each Button Horizontal Fill=true
     each Button Min Width=200
     each Button Size=Medium
     focuse Button=0
     space Between Buttons=30
     cTA Buttons Types []
     group Placement {
      horizontal Alignment=HAlign_Fill
      vertical Alignment=VAlign_Bottom
      grid {column=1, column Span=0, layer=0, row=2, row Span=0, nudge=<0 0>}
      padding=<ltrb 80 64 80 48>
     }
    }
    dialogSIze {
     height Override {enable=false, value=800}
     max Aspect Ratio {enable=false, value=0}
     max Desired Height {enable=true, value=800}
     max Desired Width {enable=false, value=0}
     min Aspect Ratio {enable=false, value=0}
     min Desired Height {enable=true, value=512}
     min Desired Width {enable=false, value=0}
     width Override {enable=true, value=1540}
    }
    imageProperties {
     material=/FortUILibrary/Systems/Dialog/Materials/MI_UI_BackgroundBlue
     image Size=<556 0>
     placement {
      horizontal Alignment=HAlign_Fill
      vertical Alignment=VAlign_Fill
      grid {column=1, column Span=0, layer=20, row=0, row Span=1, nudge=<0 0>}
      padding=<ltrb 0 0 0 0>
     }
    }
    textProperties {
     auto Wrap Text=true
     justification=Left
     line Height Percentage=1.5
     mobile Font Size Multiplier=1.3
     transform Policy=None
     wrap Text At=0
     color {colorUseRule=UseColor_Specified, specifiedColor=<rgba 1 1 1 0.75>}
     font {
      bForceMonospaced=false
      bMaterialIsStencil=false
      fontMaterial=null
      fontObject=&HeadingNow
      letterSpacing=0
      monospacedWidth=1
      size=18
      skewAmount=0
      typefaceFontName=74 Regular
      outlineSettings {
       bApplyOutlineToDropShadows=false
       bMiteredCorners=false
       bSeparateFillAlpha=false
       outlineMaterial=null
       outlineSize=0
       outlineColor=<rgba 0 0 0 1>
      }
     }
     placement {
      horizontal Alignment=HAlign_Fill
      vertical Alignment=VAlign_Top
      grid {column=1, column Span=0, layer=0, row=1, row Span=0, nudge=<0 0>}
      padding=<ltrb 80 0 80 0>
     }
    }
    titleProperties {
     auto Wrap Text=true
     justification=Left
     line Height Percentage=1
     mobile Font Size Multiplier=1.3
     transform Policy=ToUpper
     wrap Text At=0
     color {colorUseRule=UseColor_Specified, specifiedColor=<rgba 1 1 1 1>}
     font {
      bForceMonospaced=false
      bMaterialIsStencil=false
      fontMaterial=null
      fontObject=&HeadingNow
      letterSpacing=0
      monospacedWidth=1
      size=27
      skewAmount=0
      typefaceFontName=86 Bold Italic
      outlineSettings {
       bApplyOutlineToDropShadows=false
       bMiteredCorners=false
       bSeparateFillAlpha=false
       outlineMaterial=null
       outlineSize=0
       outlineColor=<rgba 0 0 0 1>
      }
     }
     placement {
      horizontal Alignment=HAlign_Fill
      vertical Alignment=VAlign_Top
      grid {column=1, column Span=0, layer=0, row=0, row Span=0, nudge=<0 0>}
      padding=<ltrb 84 80 80 64>
     }
    }
   }
  }
  Overlay_1 {widgetClassPath=&Overlay, namedSlotHost=Dialog_Internal, widget=Overlay_1}
  VerticalBox_0 {
   widgetClassPath=&VerticalBox
   parent=Overlay_1
   widget=VerticalBox_0
   slot {class=OverlaySlot, properties {horizontalAlignment=HAlign_Fill, verticalAlignment=VAlign_Fill, padding=<ltrb 76 0 152 -60>}}
  }
  Text_EditableLoadoutName {
   widgetClassPath=/FortUILibrary/UIKit/InputField/WBP_UIKit_FilteredInputField
   bInherited=true
   bIsVariable=true
   parent=VerticalBox_0
   widget=Text_EditableLoadoutName
   properties {bUseChatSanitizer=true, entryCharLimit=16, priority=1}
   slot {class=VerticalBoxSlot}
  }
  Spacer {
   widgetClassPath=&Spacer
   bIsVariable=true
   parent=VerticalBox_0
   widget=Spacer
   properties {size=<1 148>}
   slot {class=VerticalBoxSlot}
  }
  HB_Buttons {
   widgetClassPath=&HorizontalBox
   bIsVariable=true
   parent=VerticalBox_0
   widget=HB_Buttons
   slot {class=VerticalBoxSlot, properties {horizontalAlignment=HAlign_Left, verticalAlignment=VAlign_Top, padding=<ltrb 12 0 0 0>}}
  }
  Button_Accept {
   widgetClassPath=&WBP_UIKit_Button_Regular
   bIsVariable=true
   parent=HB_Buttons
   widget=Button_Accept
   properties {
    bHideInputActionWithKeyboard=true
    inputPriority=1
    minHeight=56
    minWidth=220
    text=Save
    touchMethod=PreciseTap
    hoveredSlateSoundOverride {resourceObject=/Game/Sounds/Fort_UI_Sounds/FrontEnd/Locker/Locker_Button_Hover_Default_Cue}
    pressedSlateSoundOverride {resourceObject=/Game/Sounds/Fort_UI_Sounds/FrontEnd/Locker/Locker_Button_Close_Pressed_Cue}
    triggeringInputAction {dataTable=&InputActionDataTable, rowName=Accept}
   }
   slot {class=HorizontalBoxSlot, properties {horizontalAlignment=HAlign_Left, verticalAlignment=VAlign_Top, padding=<ltrb 0 0 24 0>}}
  }
  Button_Cancel {
   widgetClassPath=&WBP_UIKit_Button_Regular
   bIsVariable=true
   parent=HB_Buttons
   widget=Button_Cancel
   properties {
    bHideInputActionWithKeyboard=true
    buttonType=Quiet
    inputPriority=1
    minHeight=56
    minWidth=220
    text=Close
    touchMethod=PreciseTap
    hoveredSlateSoundOverride {resourceObject=&Locker_Button_Hover_Default_Cue}
    pressedSlateSoundOverride {resourceObject=&Locker_Button_Close_Pressed_Cue}
    triggeringInputAction {dataTable=&InputActionDataTable, rowName=GenericBack}
   }
   slot {class=HorizontalBoxSlot, properties {horizontalAlignment=HAlign_Left, verticalAlignment=VAlign_Top}}
  }
  Overlay_CardsPreset {widgetClassPath=&Overlay, bIsVariable=true, namedSlotHost=Dialog_Internal, widget=Overlay_CardsPreset}
  VerticalBox_1 {
   widgetClassPath=&VerticalBox
   parent=Overlay_CardsPreset
   widget=VerticalBox_1
   slot {class=OverlaySlot, properties {horizontalAlignment=HAlign_Center, verticalAlignment=VAlign_Center}}
  }
  ExisitingPresetGroup {
   widgetClassPath=&Overlay
   bIsVariable=true
   parent=VerticalBox_1
   widget=ExisitingPresetGroup
   properties {renderTransform {angle=0, scale=<0.8 0.8>, shear=<0 0>, translation=<0 0>}}
   slot {class=VerticalBoxSlot, properties {verticalAlignment=VAlign_Center, padding=<ltrb 0 -16 0 0>}}
  }
  Card_ExistingPreset {
   widgetClassPath=&/WBP_LoadoutCard_MP
   bIsVariable=true
   parent=ExisitingPresetGroup
   widget=Card_ExistingPreset
   properties {renderOpacity=0.3}
   slot {class=OverlaySlot, properties {horizontalAlignment=HAlign_Center}}
  }
  Image_Chevron {
   widgetClassPath=&Image
   bIsVariable=true
   parent=ExisitingPresetGroup
   widget=Image_Chevron
   properties {
    brush {
     bIsDynamicallyLoaded=false
     drawAs=Image
     imageType=NoImage
     mirroring=NoMirror
     resourceName=None
     resourceObject=/FortUILibrary/Textures/Icons/ToBeDeprecated/T_UI_Icon_Global_Chevron
     tiling=NoTile
     imageSize=<64 64>
     margin=<ltrb 0 0 0 0>
     outlineSettings {
      bUseBrushTransparency=false
      roundingType=HalfHeightRadius
      width=0
      color {colorUseRule=UseColor_Specified, specifiedColor=<rgba 0 0 0 0>}
      cornerRadii=<0 0 0 0>
     }
     tintColor {colorUseRule=UseColor_Specified, specifiedColor=<rgba 1 1 1 1>}
     uVRegion {bIsValid=false, max=<0 0>, min=<0 0>}
    }
    colorAndOpacity=<rgba 0.7011 0.7913 0.9647 1>
    renderTransform {angle=180, scale=<1 1>, shear=<0 0>, translation=<0 0>}
   }
   slot {class=OverlaySlot, properties {horizontalAlignment=HAlign_Center, verticalAlignment=VAlign_Center}}
  }
  Card_NewPreset {
   widgetClassPath=&WBP_LoadoutCard_MP
   bIsVariable=true
   parent=VerticalBox_1
   widget=Card_NewPreset
   properties {renderTransform {angle=0, scale=<1.08 1.08>, shear=<0 0>, translation=<0 0>}}
   slot {class=VerticalBoxSlot, properties {horizontalAlignment=HAlign_Center, verticalAlignment=VAlign_Center}}
  }
 ]
}
