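# Locker categories panel showing a loadout's category set, save and options buttons, item description and shuffle warning.
# CategoryGroupChanged passes the viewmodel to the category set and rebuilds a context menu for clearing badges, randomizing or clearing the loadout, the latter via modal confirmation.
# Save turns off shuffle if active, otherwise broadcasts OnCharacterSaved; category and menu events are relayed through dispatchers.
# Focus targets the selected category card, and mobile gets adjusted layout.
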WidgetBlueprint /Locker/MPL/WBP_MPLocker_Categories
graphs [
 CategoryGroupChanged {nodes [
   n1 = CategoryGroupChanged!() -> (then, Viewmodel:Fort Locker Category Group VM Object Reference) {is=title K2Node_FunctionEntry}
   > SetFortLockerCategoryGroupVM!(execute, self, Viewmodel:Fort Locker Category Group VM Object Reference=@n1) -> (then)
   > SetFortLockerCategoryGroupVM(self:WBP Category Set Object Reference=Get WBP_CategorySet() -> (WBP_CategorySet:WBP Category Set Object Reference), Viewmodel=@n1)
   > SetContextMenuActionContext!(
    execute
    self
    ClearAllBangsInInteractionDelegate:"Delegate (by ref)"=Create Event()
    RandomizeLoadoutInInteractionDelegate:"Delegate (by ref)"=Create Event()
    ClearLoadoutInInteractionDelegate:"Delegate (by ref)"=Create Event()
   ) -> (then)
  ]}
 CloseContextMenu {nodes [
   CloseContextMenu() {is=title K2Node_FunctionEntry}
   > Close(self:Menu Anchor Object Reference=Get Anchor!(self:UIKit Context Menu Anchor Object Reference=Get Button_Options() -> (Button_Options:WBP UIKit Context Menu Anchor Button Object Reference)) -> (Anchor:Menu Anchor Object Reference))
  ]}
 EventGraph {nodes [
   n1 = "On Category Item Setup (WBP_CategorySet)"!() -> (OutputDelegate, then, CategoryItemEntryObject:User Widget Object Reference) {is=title K2Node_ComponentBoundEvent}
   > Call On Category Item Setup!(
    execute:Exec
    self:WBP MPLocker Categories Object Reference
    CategoryItemEntryObject:User Widget Object Reference=@n1.CategoryItemEntryObject
   ) -> (then:Exec) {is=title K2Node_CallDelegate}
   "On Clicked (Button_SavePresets)"!() -> (OutputDelegate, then, Button:Common Button Base Object Reference) {is=title K2Node_ComponentBoundEvent}
   > n2 = Cast To FortPlayerController(Object=GetOwningPlayer())
   n3 = Get Button_SavePresets() -> (Button_SavePresets:WBP Locker Button Object Reference)
   > "Fire Event Element Interacted (Non-Localized)"!(
    execute
    FortPC:Fort Player Controller Object Reference=@n2
    ElementName:String=Get Source String from Text!(Text:Text=Get Text(self:WBP UIKit Button Generic Base Object Reference=@n3)) -> (ReturnValue:String) {is=title K2Node_CallFunction}
    AdditionalInfo:Flow Analytics Additional Info Structure
   ) -> (then)
   n4 = Get FortTypedModularLoadoutsVM() -> (FortTypedModularLoadoutsVM:Fort Typed Modular Loadouts VM Object Reference)
   > n5 = Branch(Condition=Get bShuffleEnabled!(self:Fort Typed Modular Loadouts VM Object Reference=@n4) -> (bShuffleEnabled:bool))
   > Set bShuffleEnabled!(execute, bShuffleEnabled:bool=false, self:Fort Typed Modular Loadouts VM Object Reference=@n4) -> (then, Output_Get:bool)
   Call On Character Saved!(execute:Exec=@n5.else, self:WBP MPLocker Categories Object Reference) -> (then:Exec) {is=title K2Node_CallDelegate}
   n6 = "On Category Context Menu Opened (WBP_CategorySet)"!() -> (OutputDelegate, then, ContextMenuParent:Widget Object Reference) {is=title K2Node_ComponentBoundEvent}
   Call On Option Context Menu Opened!(execute:Exec=[@n18.then, @n6.then], self:WBP MPLocker Categories Object Reference) -> (then:Exec) {is=title K2Node_CallDelegate}
   n7 = "On Category Context Menu Closed (WBP_CategorySet)"() {is=title K2Node_ComponentBoundEvent}
   Call On Option Context Menu Closed!(execute:Exec=[@n7.then, @n13.then], self:WBP MPLocker Categories Object Reference) -> (then:Exec) {is=title K2Node_CallDelegate}
   n8 = "On Clicked (Button_SetPrimary)"!() -> (OutputDelegate, then, Button:Common Button Base Object Reference) {is=title K2Node_ComponentBoundEvent}
   > Fire Event Element Interacted(Element=@n8.Button)
   > n9 = Create Widget!(
    execute:Exec
    Class:User Widget Class Reference=&/WBP_MPLocker_SetPrimary_Dialog.WBP_MPLocker_SetPrimary_Dialog_C
    OwningPlayer:Player Controller Object Reference=GetOwningPlayer()
    FortLockerCategoryGroupVM:Fort Locker Category Group VM Object Reference
   ) -> (then:Exec, ReturnValue:WBP MPLocker Set Primary Dialog Object Reference)
   > TryPushWidgetAsModalPanel!(execute, PushWidget:User Widget Object Reference=@n9) -> (then)
   > Call Is Modal Active!(execute:Exec, self:WBP MPLocker Categories Object Reference, IsModalActive:bool=true) -> (then:Exec) {is=title K2Node_CallDelegate}
   "On Category Edit Styles (WBP_CategorySet)"() {is=title K2Node_ComponentBoundEvent}
   > Call On Category Edit Styles!(execute:Exec, self:WBP MPLocker Categories Object Reference) -> (then:Exec) {is=title K2Node_CallDelegate}
   "On Category Clicked (WBP_CategorySet)"() {is=title K2Node_ComponentBoundEvent}
   > Call On Category Button Clicked!(execute:Exec, self:WBP MPLocker Categories Object Reference) -> (then:Exec) {is=title K2Node_CallDelegate}
   "On Hovered (Button_Options)"() {is=title K2Node_ComponentBoundEvent}
   n10 = Get Button_Options()
   > Toggle Quick Actions!(execute, self:WBP UIKit Context Menu Anchor Button Base Object Reference=@n10, Enabled:bool=true) -> (then)
   "On Unhovered (Button_Options)"() {is=title K2Node_ComponentBoundEvent}
   > Toggle Quick Actions(self=@n10, Enabled=false)
   n11 = "On Card Hovered (WBP_CategorySet)"!() -> (OutputDelegate, then, IsHovered:bool) {is=title K2Node_ComponentBoundEvent}
   > Call Is Hovered Over Tile!(execute:Exec, self:WBP MPLocker Categories Object Reference, IsTileHovered:bool=@n11.IsHovered) -> (then:Exec) {is=title K2Node_CallDelegate}
   OnContextMenuClosedBind() {is=title K2Node_CustomEvent}
   n12 = Get ContextMenuButton!(self:WBP UIKit Context Menu Anchor Button Base Object Reference=Get Button_Options()) -> (ContextMenuButton:WBP UIKit Button Generic Base Object Reference)
   > n13 = Set bIsEnabled!(execute, bIsEnabled:bool=true, self:Widget Object Reference=@n12) -> (then, Output_Get:bool)
   n14 = ClearLoadoutVM() {is=title K2Node_CustomEvent}
   > ClearAllSlots(self:Fort Locker Category Group VM Object Reference=Get FortLockerCategoryGroupVM() -> (FortLockerCategoryGroupVM:Fort Locker Category Group VM Object Reference))
   > FireEventElementInteractedByName!(execute, Element Name:String=Options ClearLoadout Success, WidgetCallingFrom:Widget Object Reference=Self-Reference()) -> (then, Success:bool)
   ClearAllBangs!() -> (OutputDelegate, then, NameID:Name) {is=title K2Node_CustomEvent}
   > Call On Clear All Bangs!(execute:Exec, self:WBP MPLocker Categories Object Reference) -> (then:Exec) {is=title K2Node_CallDelegate}
   > FireEventElementInteractedByName(Element Name=Options ClearAllBadges, WidgetCallingFrom=Self-Reference())
   OnContextMenuOpenedBind() {is=title K2Node_CustomEvent}
   n15 = Get Button_Options()
   > Branch(Condition=Get bIsEnabled!(self:Widget Object Reference=Get ContextMenuButton(self=@n15)) -> (bIsEnabled:bool))
   > Set bIsEnabled(bIsEnabled=false, self=@n12)
   > n16 = Cast To FortPlayerController(Object=GetOwningPlayer())
   n17 = Get Button_Options()
   > n18 = "Fire Event Element Interacted (Non-Localized)"(
    FortPC=@n16
    ElementName=Get Source String from Text(Text=Get ButtonText!(self:WBP UIKit Context Menu Anchor Button Base Object Reference=@n17) -> (ButtonText:Text))
   )
   RandomizeLoadout!() -> (OutputDelegate, then, NameID:Name) {is=title K2Node_CustomEvent}
   > RandomizeAllSlots(self:Fort Locker Category Group VM Object Reference=Get FortLockerCategoryGroupVM())
   > Set bShuffleEnabled(bShuffleEnabled=false, self=Get FortTypedModularLoadoutsVM())
   > FireEventElementInteractedByName(Element Name=Options RandomizeLoadout, WidgetCallingFrom=Self-Reference())
   ClearLoadout!() -> (OutputDelegate, then, NameID:Name) {is=title K2Node_CustomEvent}
   > n19 = Construct Object from Class(Class=/Locker/MPL/Data/DialogVM_LockerClearLoadout.DialogVM_LockerClearLoadout_C) -> (ReturnValue:Dialog VM Locker Clear Loadout Object Reference)
   > n20 = Create Widget!(execute, Class=&/WBP_NameLoadout_MP.WBP_NameLoadout_MP_C, OwningPlayer, DialogVM:UI Kit Dialog Viewmodel Object Reference=@n19) -> (then, ReturnValue:WBP Name Loadout MP Object Reference)
   > TryPushWidgetAsModalPanel(PushWidget=@n20)
   n21 = Get FortTypedModularLoadoutsVM()
   > Initialize!(
    execute
    self:WBP Name Loadout MP Object Reference=@n20
    Loadout Action:EFortLoadoutActionType Enum=Clear
    Target Loadout:Fort Typed Modular Loadout VM Object Reference=Get ActiveLoadout!(self:Fort Typed Modular Loadouts VM Object Reference=@n21) -> (ActiveLoadout:Fort Typed Modular Loadout VM Object Reference)
    TypedModularLoadouts:Fort Typed Modular Loadouts VM Object Reference=@n21
   ) -> (then)
   > Bind Event to Clear Loadout(self:WBP Name Loadout MP Object Reference=@n20, Delegate=@n14) {is=title K2Node_AddDelegate}
   > CloseContextMenu() {is=title K2Node_CallFunction}
   > FireEventElementInteractedByName(Element Name=Options ClearLoadout, WidgetCallingFrom=Self-Reference())
   Event Construct()
   > SetTitleText!(
    execute
    self:WBP UIKit Item Description Base Object Reference=Get WPB_ItemDescription() -> (WPB_ItemDescription:WBP UIKit Item Description MPL Object Reference)
    Text:Text
   ) -> (then)
   Event OnAddedToFocusPath!() -> (OutputDelegate, then, InFocusEvent:Focus Event Structure) {is=title K2Node_Event}
   > Call Is Modal Active()
   Event OnInitialized()
   > n22 = Construct Object from Class(Class=/Script/FortniteUI.UIKitContextMenuActionContext) -> (ReturnValue:UIKit Context Menu Action Context Object Reference)
   > Set MenuActionContext(MenuActionContext:UIKit Context Menu Action Context Object Reference=@n22) -> (Output_Get:UIKit Context Menu Action Context Object Reference)
   > SetMobileLayout()
   Event PreConstruct()
  ]}
 Get Save Preset Text {nodes [
   n1 = Get Save Preset Text!() -> (then, SelectedCategoryGroup:Fort Locker Category Group VM Object Reference, Shuffle Enabled:bool=~, CanEnableShuffle:bool=~) {is=title K2Node_FunctionEntry}
   > Is Valid(InputObject=@n1)
   n2 = GetDisplayName!(self:Fort Locker Category Group VM Object Reference=@n1) -> (ReturnValue:"Text (by ref)")
   > Return Node!(
    execute
    Preset Text:Text=SelectText(
     A="LOCTEXT(\"Stop Random\")"
     B=Format Text!(Format="LOCTEXT(\"Save {Category}\")", Category:"Text (by ref)"=@n2) -> (Result)
     bPickA=AND Boolean(A=@n1.Shuffle Enabled, B=@n1.CanEnableShuffle)
    )
   )
  ]}
 Get Shuffle Message Text {nodes [
   n1 = Get Shuffle Message Text!() -> (then, Category Name:Text) {is=title K2Node_FunctionEntry}
   > n2 = Set Category Name(Category Name:Text=@n1) -> (Output_Get:Text)
   > UpdateShuffleMessageText!(execute, self, Category:Text=@n2) -> (then)
  ]}
 GetFocusTarget {nodes [
   GetFocusTarget() {is=title K2Node_FunctionEntry}
   n1 = Get WBP_CategorySet()
   > n2 = For Each Loop(Array:Array of User Widget Object References=GetAllEntries!(self:Dynamic Entry Box Base Object Reference=Get CategoryTopRow!(self:WBP Category Set Object Reference=@n1) -> (CategoryTopRow:Fort Dynamic Entry Box Object Reference)) -> (ReturnValue:Array of User Widget Object References)) -> (Array Element:User Widget Object Reference)
   n3 = Get WBP_CategorySet()
   n4 = For Each Loop(
    Exec=@n2.Completed
    Array=GetAllEntries(self=Get CategoryBottom!(self:WBP Category Set Object Reference=@n3) -> (CategoryBottom:Fort Dynamic Entry Box Object Reference))
   )
   n5 = Get CategoryTopRow(self=Get WBP_CategorySet())
   Return Node!(
    execute=@n4.Completed
    ReturnValue:Widget Object Reference="Get (a copy)"(Array:Array of User Widget Object References=GetAllEntries(self=@n5)) -> (Output:User Widget Object Reference)
   )
   > n6 = Cast To WBP_UIKit_ItemCard_Category!(execute, Object=@n4) -> (then, CastFailed, AsWBP UIKit Item Card Category:WBP UIKit Item Card Category Object Reference) {is=title K2Node_DynamicCast}
   n7 = Get FortItemSelectionVM() -> (FortItemSelectionVM:Fort Item Selection VM Object Reference)
   > Branch(Condition="Equal (Object)"(
     A=Get FortLockerCategoryItemVM!(self:WBP UIKit Item Card Category Object Reference=@n6) -> (FortLockerCategoryItemVM:Fort Locker Category Item VM Object Reference)
     B=Get SelectedCategory!(self:Fort Item Selection VM Object Reference=@n7) -> (SelectedCategory:Fort Locker Category Item VM Object Reference)
    ))
   > Return Node(ReturnValue=@n4)
   n8 = Cast To WBP_UIKit_ItemCard_Category(execute=@n2.LoopBody, Object=@n2)
   n9 = Get FortItemSelectionVM()
   > Branch(Condition="Equal (Object)"(A=Get FortLockerCategoryItemVM(self=@n8), B=Get SelectedCategory(self=@n9)))
   > Return Node(ReturnValue=@n2)
  ]}
 Go to Save {nodes [
   Go to Save!() -> (then, Navigation:EUINavigation Enum) {is=title K2Node_FunctionEntry}
   > Call Is Hovered Over Tile()
   > Return Node(ReturnValue=Get Button_SavePresets())
  ]}
 IsHoveredOverTile {nodes [IsHoveredOverTile!() -> (then, IsTileHovered:bool=~) {is=title K2Node_FunctionEntry}]}
 IsModalActive {nodes [IsModalActive!() -> (then, IsModalActive:bool=~) {is=title K2Node_FunctionEntry}]}
 OnCategoryButtonClicked {nodes [OnCategoryButtonClicked() {is=title K2Node_FunctionEntry}]}
 OnCategoryEditStyles {nodes [OnCategoryEditStyles() {is=title K2Node_FunctionEntry}]}
 OnCategoryItemSetup {nodes [OnCategoryItemSetup!() -> (then, CategoryItemEntryObject:User Widget Object Reference) {is=title K2Node_FunctionEntry}]}
 OnCharacterSaved {nodes [OnCharacterSaved() {is=title K2Node_FunctionEntry}]}
 OnClearAllBangs {nodes [OnClearAllBangs() {is=title K2Node_FunctionEntry}]}
 OnOptionContextMenuClosed {nodes [OnOptionContextMenuClosed() {is=title K2Node_FunctionEntry}]}
 OnOptionContextMenuOpened {nodes [OnOptionContextMenuOpened() {is=title K2Node_FunctionEntry}]}
 Preset Random Shuffle Active {nodes [
   n1 = Preset Random Shuffle Active!() -> (then, IsShuffle:bool=~) {is=title K2Node_FunctionEntry}
   > n2 = Branch(Condition=@n1)
   > SetVisibility(self=Get ShuffleWarningBox() -> (ShuffleWarningBox:Overlay Object Reference), InVisibility=HitTestInvisible)
   > UpdateShuffleMessageText(Category=Get Category Name() -> (Category Name:Text))
   SetVisibility(execute=@n2.else, self=Get ShuffleWarningBox(), InVisibility=Collapsed)
  ]}
 SetContextMenuActionContext {nodes [
   n1 = SetContextMenuActionContext!() -> (
    then
    ClearAllBangsInInteractionDelegate:"Delegate (by ref)"
    RandomizeLoadoutInInteractionDelegate:"Delegate (by ref)"
    ClearLoadoutInInteractionDelegate:"Delegate (by ref)"
   ) {is=title K2Node_FunctionEntry}
   n2 = Get MenuActionContext() -> (MenuActionContext:UIKit Context Menu Action Context Object Reference)
   > RemoveAllInteractions(self:UIKit Context Menu Action Context Object Reference=@n2)
   > AddInteraction!(
    execute
    self:UIKit Context Menu Action Context Object Reference=@n2
    InActionInfo_DisplayName:Text=Get Clear All Bangs Action Info Display Name() -> (Clear All Bangs Action Info Display Name:Text)
    InActionInfo_IconTexture:Texture 2D Soft Object Reference
    InActionInfo_IconMaterial:Material Instance Soft Object Reference
    InActionInfo_SortPriority:int=1000
    InActionInfo_bIsAvailable:bool=True
    InActionInfo_bIsEnabled:bool=True
    InActionInfo_bIsSelected:bool=False
    InActionInfo_bAddDivider:bool=False
    InActionInfo_bPositionDividerAfter:bool=False
    InActionInfo_bCloseMenuOnSelect:bool=False
    InActionInfo_InteractionNameID:Name=None
    InActionInfo_Interaction:Data Table Row Handle Structure=Get ClearAllBangsInput() -> (ClearAllBangsInput:Data Table Row Handle Structure)
    InActionInfo_PressedDelegate:Delegate="(null).None"
    InActionInfo_ReleasedDelegate:Delegate="(null).None"
    InInteractionDelegate:"Delegate (by ref)"=@n1
   ) -> (then)
   > AddInteraction(
    self=@n2
    InActionInfo_DisplayName=Get Randomize Loadout In Action Info Display Name() -> (Randomize Loadout In Action Info Display Name:Text)
    InActionInfo_SortPriority=1000
    InActionInfo_bIsAvailable=True
    InActionInfo_bIsEnabled=True
    InActionInfo_bIsSelected=False
    InActionInfo_bAddDivider=False
    InActionInfo_bPositionDividerAfter=False
    InActionInfo_bCloseMenuOnSelect=False
    InActionInfo_InteractionNameID=None
    InActionInfo_Interaction=Get RandomizeLoadoutInput() -> (RandomizeLoadoutInput:Data Table Row Handle Structure)
    InActionInfo_PressedDelegate="(null).None"
    InActionInfo_ReleasedDelegate="(null).None"
    InInteractionDelegate=@n1.RandomizeLoadoutInInteractionDelegate
   )
   > AddInteraction(
    self=@n2
    InActionInfo_DisplayName=Get Clear Loadout In Action Info Display Name() -> (Clear Loadout In Action Info Display Name:Text)
    InActionInfo_SortPriority=1000
    InActionInfo_bIsAvailable=True
    InActionInfo_bIsEnabled=True
    InActionInfo_bIsSelected=False
    InActionInfo_bAddDivider=False
    InActionInfo_bPositionDividerAfter=False
    InActionInfo_bCloseMenuOnSelect=False
    InActionInfo_InteractionNameID=None
    InActionInfo_Interaction=Get ClearLoadoutInput() -> (ClearLoadoutInput:Data Table Row Handle Structure)
    InActionInfo_PressedDelegate="(null).None"
    InActionInfo_ReleasedDelegate="(null).None"
    InInteractionDelegate=@n1.ClearLoadoutInInteractionDelegate
   )
   n3 = Get Button_Options()
   > SetContext!(
    execute
    self:WBP UIKit Context Menu Anchor Button Base Object Reference=@n3
    InteractionContext_Data:UIKit Context Menu Action Context Object Reference=@n2
   ) -> (then)
   > Unbind all Events from On Context Menu Opened!(execute:Exec, self:WBP UIKit Context Menu Anchor Button Base Object Reference=@n3) -> (then:Exec) {is=title K2Node_ClearDelegate}
   > Bind Event to On Context Menu Opened(self:WBP UIKit Context Menu Anchor Button Base Object Reference=@n3, Delegate=Create Event()) {is=title K2Node_AddDelegate}
   > Unbind all Events from On Context Menu Closed!(execute:Exec, self:WBP UIKit Context Menu Anchor Button Base Object Reference=@n3) -> (then:Exec) {is=title K2Node_ClearDelegate}
   > Bind Event to On Context Menu Closed(self:WBP UIKit Context Menu Anchor Button Base Object Reference=@n3, Delegate=Create Event()) {is=title K2Node_AddDelegate}
   > Return Node!(execute)
  ]}
 SetMainHeader {nodes [SetMainHeader() {is=title K2Node_FunctionEntry}]}
 SetMobileLayout {nodes [
   SetMobileLayout() {is=title K2Node_FunctionEntry}
   > Branch(Condition=IsMobileGame())
   > Set MobileButtonPadding!(
    execute
    MobileButtonPadding:Margin Structure=Make Margin(Left=50, Top=0, Right=50, Bottom=0)
    self:WBP UIKit Button Generic Base Object Reference=Get Button_SavePresets()
   ) -> (then, Output_Get:Margin Structure)
   n1 = SlotAsGridSlot(Widget=Get VerticalBox_0() -> (VerticalBox_0:Vertical Box Object Reference))
   > SetVerticalAlignment(self:Grid Slot Object Reference=@n1, InVerticalAlignment=VAlign_Top)
   > SetPadding!(execute, self:Grid Slot Object Reference=@n1, InPadding:Margin Structure=Make Margin(Left=0, Top=160, Right=0, Bottom=0)) -> (then)
   > Set Min Height!(execute, Min Height:int=80, self:WBP UIKit Context Menu Anchor Button Base Object Reference=Get Button_Options()) -> (then, Output_Get:int=0)
  ]}
 UpdateShuffleMessageText {nodes [
   n1 = UpdateShuffleMessageText!() -> (then, Category:Text) {is=title K2Node_FunctionEntry}
   > SetText(
    self=Get Shuffle_Text() -> (Shuffle_Text:Common Rich Text Block Object Reference)
    InText=SelectText(
     A=Format Text!(Format="LOCTEXT(\"Your saved presets are cycling in a random order. Any changes will turn it off.\")") -> (Result)
     B=Format Text!(Format="LOCTEXT(\"Your \\'Saved {Category}\\' presets are cycling in a random order. Any changes will turn it off.\")", Category:Text=@n1) -> (Result)
     bPickA=TextIsEmpty(InText=@n1)
    )
   )
  ]}
]
variables [
 Category Name {type=string, default=""}
 Clear All Bangs Action Info Display Name {type=string, default=Clear All Badges}
 Clear Loadout In Action Info Display Name {type=string, default=Clear Loadout}
 CurrentCategoryVM {type=FortLockerCategoryGroupVM, default=null}
 IsHoveredOverTile {}
 IsModalActive {}
 MenuActionContext {type=UIKitContextMenuActionContext, default=null}
 OnCategoryButtonClicked {}
 OnCategoryEditStyles {}
 OnCategoryItemSetup {}
 OnCharacterSaved {}
 OnClearAllBangs {}
 OnOptionContextMenuClosed {}
 OnOptionContextMenuOpened {}
 Randomize Loadout In Action Info Display Name {type=string, default=Randomize Loadout}
 RenameLoadoutWidgetClass {type=Class, default=&WBP_NameLoadout_MP_C}
]
widgets {
 info {inheritedWidgetCount=0, parentClass=&CommonUserWidget, rootWidgetClass=&GridPanel, widgetCount=13}
 widgets [
  Categories {widgetClassPath=&GridPanel, widget=Categories, properties {columnFill [1,1], rowFill [0,0,1,0]}}
  VerticalBox_0 {
   widgetClassPath=&VerticalBox
   bIsVariable=true
   parent=Categories
   widget=VerticalBox_0
   slot {class=GridSlot, properties {row=2, rowSpan=3, verticalAlignment=VAlign_Center}}
  }
  WBP_MPLocker_Heading {
   widgetClassPath=/FortniteCoreUI/Customization/MPL/WBP_MPLocker_Heading
   bIsVariable=true
   parent=VerticalBox_0
   widget=WBP_MPLocker_Heading
   slot {class=VerticalBoxSlot, properties {verticalAlignment=VAlign_Top, padding=<ltrb 0 0 0 40>}}
  }
  WBP_CategorySet {
   widgetClassPath=&/WBP_CategorySet
   bIsVariable=true
   parent=VerticalBox_0
   widget=WBP_CategorySet
   properties {navigation=&self:WidgetTree.WBP_CategorySet.WidgetNavigation_0}
   slot {class=VerticalBoxSlot, properties {padding=<ltrb 0 0 0 32>, size {sizeRule=Fill, value=1}}}
  }
  Button_SavePresets {
   widgetClassPath=&/WBP_Locker_Button
   bIsVariable=true
   parent=VerticalBox_0
   widget=Button_SavePresets
   properties {text=SAVE CHARACTER}
   slot {class=VerticalBoxSlot, properties {horizontalAlignment=HAlign_Left, padding=<ltrb 0 0 0 16>}}
  }
  Button_SetPrimary {
   widgetClassPath=&WBP_Locker_Button
   bIsVariable=true
   parent=VerticalBox_0
   widget=Button_SetPrimary
   properties {text=Set Primary Vehicle, visibility=Collapsed}
   slot {class=VerticalBoxSlot, properties {horizontalAlignment=HAlign_Left, padding=<ltrb 0 0 0 16>}}
  }
  Button_Options {
   widgetClassPath=/FortUILibrary/UIKit/ContextMenu/WBP_UIKit_ContextMenuAnchorButton
   bIsVariable=true
   parent=VerticalBox_0
   widget=Button_Options
   properties {
    button Outline Material=/Locker/Art/MaterialInstances/MI_UI_Button_Options_Outline
    buttonClass=&/WBP_Locker_Button.WBP_Locker_Button_C
    buttonIcon=null
    buttonText=Loadout Options
    closeMenuOnContextClicked=true
    min Height=56
    button Padding=<ltrb 50 0 50 0>
    button Text Font {
     bForceMonospaced=false
     bMaterialIsStencil=false
     fontMaterial=/FortUILibrary/UIKit/Buttons/MaterialInstances/CTAButtons/MI_UIKit_TextRegular
     fontObject=&HeadingNow
     letterSpacing=10
     monospacedWidth=1
     size=15.75
     skewAmount=0
     typefaceFontName=76 Bold
     outlineSettings {
      bApplyOutlineToDropShadows=false
      bMiteredCorners=false
      bSeparateFillAlpha=false
      outlineMaterial=null
      outlineSize=0
      outlineColor=<rgba 0 0 0 1>
     }
    }
   }
   slot {class=VerticalBoxSlot, properties {horizontalAlignment=HAlign_Left}}
  }
  VerticalBox_2 {
   widgetClassPath=&VerticalBox
   parent=Categories
   widget=VerticalBox_2
   slot {
    class=GridSlot
    properties {column=1, horizontalAlignment=HAlign_Right, row=3, rowSpan=2, verticalAlignment=VAlign_Bottom, nudge=<0 44>, padding=<ltrb 0 0 64 0>}
   }
  }
  WPB_ItemDescription {
   widgetClassPath=/FortniteCoreUI/Customization/MPL/WBP_UIKit_ItemDescription_MPL
   bIsVariable=true
   parent=VerticalBox_2
   widget=WPB_ItemDescription
   properties {
    hideBodyText=true
    bodyTextPadding=<ltrb 0 8 0 0>
    compatibilityPadding=<ltrb 0 8 0 0>
    tagPadding=<ltrb 0 0 0 6>
    titleFont {
     bForceMonospaced=false
     bMaterialIsStencil=false
     fontMaterial=null
     fontObject=&HeadingNow
     letterSpacing=0
     monospacedWidth=1
     size=30
     skewAmount=0
     typefaceFontName=86 Bold Italic
     outlineSettings {
      bApplyOutlineToDropShadows=false
      bMiteredCorners=false
      bSeparateFillAlpha=false
      outlineMaterial=null
      outlineSize=0
      outlineColor=<rgba 0 0 0 1>
     }
    }
   }
   slot {class=VerticalBoxSlot}
  }
  ShuffleWarningBox {
   widgetClassPath=&Overlay
   bIsVariable=true
   parent=VerticalBox_2
   widget=ShuffleWarningBox
   properties {visibility=Collapsed}
   slot {class=VerticalBoxSlot, properties {verticalAlignment=VAlign_Bottom, padding=<ltrb 0 16 0 0>}}
  }
  WBP_UIKit_Backplate {
   widgetClassPath=&WBP_UIKit_Backplate
   bIsVariable=true
   parent=ShuffleWarningBox
   widget=WBP_UIKit_Backplate
   properties {brightness=Dark}
   slot {class=OverlaySlot, properties {horizontalAlignment=HAlign_Fill, verticalAlignment=VAlign_Fill}}
  }
  SizeBox_ShuffleText {
   widgetClassPath=&SizeBox
   bIsVariable=true
   parent=ShuffleWarningBox
   widget=SizeBox_ShuffleText
   slot {class=OverlaySlot, properties {horizontalAlignment=HAlign_Fill, verticalAlignment=VAlign_Fill, padding=<ltrb 16 16 16 16>}}
  }
  Shuffle_Text {
   widgetClassPath=&CommonRichTextBlock
   bIsVariable=true
   parent=SizeBox_ShuffleText
   widget=Shuffle_Text
   properties {
    autoWrapText=true
    bOverrideDefaultStyle=true
    text="[PH] Random is on"
    decoratorClasses [/Game/Athena/UI/Frontend/CosmeticGating/Data/RichTextDecorator_CosmeticGate.RichTextDecorator_CosmeticGate_C]
    defaultTextStyleOverride {
     overflowPolicy=Clip
     transformPolicy=None
     colorAndOpacity {colorUseRule=UseColor_Specified, specifiedColor=<rgba 1 1 1 0.75>}
     font {
      bForceMonospaced=false
      bMaterialIsStencil=false
      fontMaterial=null
      fontObject=&HeadingNow
      letterSpacing=10
      monospacedWidth=1
      size=14.25
      skewAmount=0
      typefaceFontName=74 Regular
      outlineSettings {
       bApplyOutlineToDropShadows=false
       bMiteredCorners=false
       bSeparateFillAlpha=false
       outlineMaterial=null
       outlineSize=0
       outlineColor=<rgba 0 0 0 1>
      }
     }
     highlightColor {colorUseRule=UseColor_Specified, specifiedColor=<rgba 0 0 0 1>}
     highlightShape {
      bIsDynamicallyLoaded=false
      drawAs=Image
      imageType=NoImage
      mirroring=NoMirror
      resourceName=None
      resourceObject=null
      tiling=NoTile
      imageSize=<32 32>
      margin=<ltrb 0 0 0 0>
      outlineSettings {
       bUseBrushTransparency=false
       roundingType=HalfHeightRadius
       width=0
       color {colorUseRule=UseColor_Specified, specifiedColor=<rgba 0 0 0 0>}
       cornerRadii=<0 0 0 0>
      }
      tintColor {colorUseRule=UseColor_Specified, specifiedColor=<rgba 1 1 1 1>}
      uVRegion {bIsValid=false, max=<0 0>, min=<0 0>}
     }
     selectedBackgroundColor {colorUseRule=UseColor_Foreground, specifiedColor=<rgba 1 0 1 1>}
     shadowColorAndOpacity=<rgba 0 0 0 1>
     shadowOffset=<0 0>
     strikeBrush {
      bIsDynamicallyLoaded=false
      drawAs=Image
      imageType=NoImage
      mirroring=NoMirror
      resourceName=None
      resourceObject=null
      tiling=NoTile
      imageSize=<32 32>
      margin=<ltrb 0 0 0 0>
      outlineSettings {
       bUseBrushTransparency=false
       roundingType=HalfHeightRadius
       width=0
       color {colorUseRule=UseColor_Specified, specifiedColor=<rgba 0 0 0 0>}
       cornerRadii=<0 0 0 0>
      }
      tintColor {colorUseRule=UseColor_Specified, specifiedColor=<rgba 1 1 1 1>}
      uVRegion {bIsValid=false, max=<0 0>, min=<0 0>}
     }
     underlineBrush {
      bIsDynamicallyLoaded=false
      drawAs=Image
      imageType=NoImage
      mirroring=NoMirror
      resourceName=None
      resourceObject=null
      tiling=NoTile
      imageSize=<32 32>
      margin=<ltrb 0 0 0 0>
      outlineSettings {
       bUseBrushTransparency=false
       roundingType=HalfHeightRadius
       width=0
       color {colorUseRule=UseColor_Specified, specifiedColor=<rgba 0 0 0 0>}
       cornerRadii=<0 0 0 0>
      }
      tintColor {colorUseRule=UseColor_Specified, specifiedColor=<rgba 1 1 1 1>}
      uVRegion {bIsValid=false, max=<0 0>, min=<0 0>}
     }
    }
   }
   slot {class=SizeBoxSlot}
  }
 ]
}
