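# WBP_MPLocker_Items is the Locker screen for browsing, searching, previewing and equipping items within a category.
# Subcategory tabs and the search field drive Fort view models, which feed the tile list, item preview, button states and empty-list warning.
# Equip and Edit Styles use async EquipItem; editing then pushes the variant selector, while customizable items open a customize menu.
# Intro and SubMenu animations handle transitions, with layout adjustments for mobile and emote wheel.
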WidgetBlueprint /Locker/MPL/WBP_MPLocker_Items
animations [
 Intro {
  binding_count=7
  length_seconds=0.25
  object=&self:Intro
  sequence {bindings [
    Image_Fade_Bottom {
     id=#1
     possessed_class=Image
     tracks [
      {
       type=MovieSceneFloatTrack
       display_name=Render Opacity
       sections [{type=MovieSceneFloatSection, end_frame=5, start_frame=0, channels [None {type=MovieSceneScriptingFloatChannel, keys [[0,0],[5,1]]}]}]
      }
      {
       type=MovieScene2DTransformTrack
       sections [{
         type=MovieScene2DTransformSection
         end_frame=5
         start_frame=0
         channels [
          Translation.X {type=MovieSceneScriptingFloatChannel, keys [[0,200],[5,0]]}
          Translation.Y {type=MovieSceneScriptingFloatChannel}
          Angle {type=MovieSceneScriptingFloatChannel}
          Scale.X {type=MovieSceneScriptingFloatChannel}
          Scale.Y {type=MovieSceneScriptingFloatChannel}
          Shear.X {type=MovieSceneScriptingFloatChannel}
          Shear.Y {type=MovieSceneScriptingFloatChannel}
         ]
        }]
      }
     ]
    }
    WBP_UIKit_DescriptionWithSubtitle {
     id=#2
     possessed_class=WBP_UIKit_DescriptionWithSubtitle_C
     tracks [
      {
       type=MovieSceneFloatTrack
       display_name=Render Opacity
       sections [{type=MovieSceneFloatSection, end_frame=5, start_frame=0, channels [None {type=MovieSceneScriptingFloatChannel, keys [[0,0],[5,1]]}]}]
      }
      {
       type=MovieScene2DTransformTrack
       sections [{
         type=MovieScene2DTransformSection
         end_frame=5
         start_frame=0
         channels [Translation.X {type=MovieSceneScriptingFloatChannel, keys [[0,200],[5,0]]}]
        }]
      }
     ]
    }
    LockerList_Container {
     id=#3
     possessed_class=NamedSlot
     tracks [
      {
       type=MovieSceneFloatTrack
       display_name=Render Opacity
       sections [{type=MovieSceneFloatSection, end_frame=5, start_frame=0, channels [None {type=MovieSceneScriptingFloatChannel, keys [[0,0],[5,1]]}]}]
      }
      {
       type=MovieScene2DTransformTrack
       sections [{
         type=MovieScene2DTransformSection
         end_frame=5
         start_frame=0
         channels [Translation.X {type=MovieSceneScriptingFloatChannel, keys [[0,200],[5,0]]}]
        }]
      }
     ]
    }
    Image_Fade_Top {
     id=#4
     possessed_class=Image
     tracks [
      {
       type=MovieSceneFloatTrack
       display_name=Render Opacity
       sections [{type=MovieSceneFloatSection, end_frame=5, start_frame=0, channels [None {type=MovieSceneScriptingFloatChannel, keys [[0,0],[5.0003,1]]}]}]
      }
      {
       type=MovieScene2DTransformTrack
       sections [{
         type=MovieScene2DTransformSection
         end_frame=5
         start_frame=0
         channels [
          Translation.X {type=MovieSceneScriptingFloatChannel, keys [[0,200],[5,0]]}
          Translation.Y {type=MovieSceneScriptingFloatChannel}
          Angle {type=MovieSceneScriptingFloatChannel}
          Scale.X {type=MovieSceneScriptingFloatChannel}
          Scale.Y {type=MovieSceneScriptingFloatChannel}
          Shear.X {type=MovieSceneScriptingFloatChannel}
          Shear.Y {type=MovieSceneScriptingFloatChannel}
         ]
        }]
      }
     ]
    }
    SortFilterBar {
     id=#5
     possessed_class=HorizontalBox
     tracks [
      {
       type=MovieSceneFloatTrack
       display_name=Render Opacity
       sections [{type=MovieSceneFloatSection, end_frame=5, start_frame=0, channels [None {type=MovieSceneScriptingFloatChannel, keys [[0,0],[5,1]]}]}]
      }
      {
       type=MovieScene2DTransformTrack
       sections [{
         type=MovieScene2DTransformSection
         end_frame=5
         start_frame=0
         channels [Translation.X {type=MovieSceneScriptingFloatChannel, keys [[0,200],[5,0]]}]
        }]
      }
     ]
    }
    SubcategoryCarousel {
     id=#6
     possessed_class=WBP_SubcategoryCarousel_C
     tracks [
      {
       type=MovieSceneFloatTrack
       display_name=Render Opacity
       sections [{type=MovieSceneFloatSection, end_frame=5, start_frame=0, channels [None {type=MovieSceneScriptingFloatChannel, keys [[0,0],[5,1]]}]}]
      }
      {
       type=MovieScene2DTransformTrack
       sections [{
         type=MovieScene2DTransformSection
         end_frame=5
         start_frame=0
         channels [Translation.X {type=MovieSceneScriptingFloatChannel, keys [[0,200],[5,0]]}]
        }]
      }
     ]
    }
    WBP_ActionBarContainer {
     id=#7
     possessed_class=WBP_ActionBarContainer_C
     tracks [{
       type=MovieSceneFloatTrack
       display_name=Render Opacity
       sections [{type=MovieSceneFloatSection, end_frame=5, start_frame=0, channels [None {type=MovieSceneScriptingFloatChannel, keys [[0,0],[5,1]]}]}]
      }]
    }
   ]}
 }
 SubMenu {
  binding_count=7
  length_seconds=0.25
  object=&self:SubMenu
  sequence {bindings [
    LockerList_Container {
     id=#8
     possessed_class=NamedSlot
     tracks [
      {
       type=MovieSceneFloatTrack
       display_name=Render Opacity
       sections [{type=MovieSceneFloatSection, end_frame=5, start_frame=0, channels [None {type=MovieSceneScriptingFloatChannel, keys [[0,1],[5,0]]}]}]
      }
      {
       type=MovieScene2DTransformTrack
       sections [{
         type=MovieScene2DTransformSection
         end_frame=5
         start_frame=0
         channels [Translation.X {type=MovieSceneScriptingFloatChannel, keys [[0,0],[5,-200]]}]
        }]
      }
     ]
    }
    SortFilterBar {
     id=#9
     possessed_class=HorizontalBox
     tracks [
      {
       type=MovieSceneFloatTrack
       display_name=Render Opacity
       sections [{type=MovieSceneFloatSection, end_frame=5, start_frame=0, channels [None {type=MovieSceneScriptingFloatChannel, keys [[0,1],[5,0]]}]}]
      }
      {
       type=MovieScene2DTransformTrack
       sections [{
         type=MovieScene2DTransformSection
         end_frame=5
         start_frame=0
         channels [Translation.X {type=MovieSceneScriptingFloatChannel, keys [[0,0],[5,-200]]}]
        }]
      }
     ]
    }
    Image_Fade_Bottom {
     id=#10
     possessed_class=Image
     tracks [
      {
       type=MovieSceneFloatTrack
       display_name=Render Opacity
       sections [{type=MovieSceneFloatSection, end_frame=5, start_frame=0, channels [None {type=MovieSceneScriptingFloatChannel, keys [[0,1],[5,0]]}]}]
      }
      {
       type=MovieScene2DTransformTrack
       sections [{
         type=MovieScene2DTransformSection
         end_frame=5
         start_frame=0
         channels [
          Translation.X {type=MovieSceneScriptingFloatChannel, keys [[0,0],[5,-200]]}
          Translation.Y {type=MovieSceneScriptingFloatChannel}
          Angle {type=MovieSceneScriptingFloatChannel}
          Scale.X {type=MovieSceneScriptingFloatChannel}
          Scale.Y {type=MovieSceneScriptingFloatChannel}
          Shear.X {type=MovieSceneScriptingFloatChannel}
          Shear.Y {type=MovieSceneScriptingFloatChannel}
         ]
        }]
      }
     ]
    }
    WBP_UIKit_DescriptionWithSubtitle {
     id=#11
     possessed_class=WBP_UIKit_DescriptionWithSubtitle_C
     tracks [
      {
       type=MovieScene2DTransformTrack
       sections [{
         type=MovieScene2DTransformSection
         end_frame=5
         start_frame=0
         channels [Translation.X {type=MovieSceneScriptingFloatChannel, keys [[0,0],[5,-200]]}]
        }]
      }
      {
       type=MovieSceneFloatTrack
       display_name=Render Opacity
       sections [{type=MovieSceneFloatSection, end_frame=5, start_frame=0, channels [None {type=MovieSceneScriptingFloatChannel, keys [[0,1],[5,0]]}]}]
      }
     ]
    }
    Image_Fade_Top {
     id=#12
     possessed_class=Image
     tracks [
      {
       type=MovieScene2DTransformTrack
       sections [{
         type=MovieScene2DTransformSection
         end_frame=5
         start_frame=0
         channels [
          Translation.X {type=MovieSceneScriptingFloatChannel, keys [[0,0],[5,-200]]}
          Translation.Y {type=MovieSceneScriptingFloatChannel}
          Angle {type=MovieSceneScriptingFloatChannel}
          Scale.X {type=MovieSceneScriptingFloatChannel}
          Scale.Y {type=MovieSceneScriptingFloatChannel}
          Shear.X {type=MovieSceneScriptingFloatChannel}
          Shear.Y {type=MovieSceneScriptingFloatChannel}
         ]
        }]
      }
      {
       type=MovieSceneFloatTrack
       display_name=Render Opacity
       sections [{type=MovieSceneFloatSection, end_frame=5, start_frame=0, channels [None {type=MovieSceneScriptingFloatChannel, keys [[0,1],[5,0]]}]}]
      }
     ]
    }
    WBP_ActionBarContainer {
     id=#13
     possessed_class=WBP_ActionBarContainer_C
     tracks [{
       type=MovieSceneFloatTrack
       display_name=Render Opacity
       sections [{type=MovieSceneFloatSection, end_frame=5, start_frame=0, channels [None {type=MovieSceneScriptingFloatChannel, keys [[0,1],[5,0]]}]}]
      }]
    }
    SubcategoryCarousel {
     id=#14
     possessed_class=WBP_SubcategoryCarousel_C
     tracks [
      {
       type=MovieSceneFloatTrack
       display_name=Render Opacity
       sections [{type=MovieSceneFloatSection, end_frame=5, start_frame=0, channels [None {type=MovieSceneScriptingFloatChannel, keys [[0,1],[5,0]]}]}]
      }
      {
       type=MovieScene2DTransformTrack
       sections [{
         type=MovieScene2DTransformSection
         end_frame=5
         start_frame=0
         channels [Translation.X {type=MovieSceneScriptingFloatChannel, keys [[0,0],[5,-200]]}]
        }]
      }
     ]
    }
   ]}
 }
 Anim_FadeItemDetails {
  binding_count=1
  length_seconds=0.3
  object=&self:Anim_FadeItemDetails
  sequence {bindings [ItemDetails_VerticalBox {
     id=#15
     possessed_class=VerticalBox
     tracks [{
       type=MovieSceneFloatTrack
       display_name=Render Opacity
       sections [{type=MovieSceneFloatSection, end_frame=6, start_frame=0, channels [None {type=MovieSceneScriptingFloatChannel, keys [[0,1],[6,0]]}]}]
      }]
    }]}
 }
 Anim_HideItemDetails {
  binding_count=1
  length_seconds=0.35
  object=&self:Anim_HideItemDetails
  sequence {bindings [ItemDetails_VerticalBox {
     id=#16
     possessed_class=VerticalBox
     tracks [
      {
       type=MovieSceneFloatTrack
       display_name=Render Opacity
       sections [{type=MovieSceneFloatSection, end_frame=7, start_frame=0, channels [None {type=MovieSceneScriptingFloatChannel, keys [[0,1],[7,0]]}]}]
      }
      {
       type=MovieScene2DTransformTrack
       sections [{
         type=MovieScene2DTransformSection
         end_frame=7
         start_frame=0
         channels [
          Translation.X {type=MovieSceneScriptingFloatChannel, keys [[0,0],[7,-32]]}
          Translation.Y {type=MovieSceneScriptingFloatChannel, keys [[0,0],[7,64]]}
          Scale.X {type=MovieSceneScriptingFloatChannel, keys [[0,1],[7,1.2]]}
          Scale.Y {type=MovieSceneScriptingFloatChannel, keys [[0,1],[7,1.2]]}
         ]
        }]
      }
     ]
    }]}
 }
 Anim_ShowItemDetails {
  binding_count=1
  length_seconds=0.35
  object=&self:Anim_ShowItemDetails
  sequence {bindings [ItemDetails_VerticalBox {
     id=#17
     possessed_class=VerticalBox
     tracks [
      {
       type=MovieSceneFloatTrack
       display_name=Render Opacity
       sections [{type=MovieSceneFloatSection, end_frame=7, start_frame=0, channels [None {type=MovieSceneScriptingFloatChannel, keys [[0,0],[7,1]]}]}]
      }
      {
       type=MovieScene2DTransformTrack
       sections [{
         type=MovieScene2DTransformSection
         end_frame=7
         start_frame=0
         channels [
          Translation.X {type=MovieSceneScriptingFloatChannel, const [0,2]}
          Translation.Y {type=MovieSceneScriptingFloatChannel, const [0,2]}
          Scale.X {type=MovieSceneScriptingFloatChannel, const [1,2]}
          Scale.Y {type=MovieSceneScriptingFloatChannel, const [1,2]}
         ]
        }]
      }
     ]
    }]}
 }
]
graphs [
 Access Restriction Changed {nodes [
   n1 = Access Restriction Changed!() -> (then, Restriction:EItemAccessRestriction Enum=UnsatisfiedExclusiveItem) {is=title K2Node_FunctionEntry}
   n2 = "Equal (Enum)"(A:EItemAccessRestriction Enum=@n1, B:EItemAccessRestriction Enum=Unknown)
   n3 = NOT Boolean(A=@n2)
   n4 = AND Boolean(
    A=Get bIsEquippedInOtherCategory!(self:Fort Locker Equip Utility VM Object Reference=Get FortLockerEquipUtilityVM() -> (FortLockerEquipUtilityVM:Fort Locker Equip Utility VM Object Reference)) -> (bIsEquippedInOtherCategory:bool)
    B=@n2
   )
   > Set Visibility(
    Visibility="To Visibility (Boolean)"!(
     bIsVisible:bool=OR Boolean(A=@n4, B=@n3)
     TrueVisibility:ESlateVisibility Enum=HitTestInvisible
     FalseVisibility:ESlateVisibility Enum=Collapsed
    ) -> (ReturnValue:ESlateVisibility Enum=Visible)
    self=Get AcessRestrictionWarning() -> (AcessRestrictionWarning:WBP UIKit Info Box Object Reference)
   )
   n5 = Get FortLockerEquipUtilityVM()
   > n6 = Branch(Condition=Select!(
     UnsatisfiedExclusiveItem:bool=false
     LockedEmote:bool=false
     CosmeticRestrictions:bool=false
     AllArchived:bool=false
     RatingGated:bool=true
     UnsupportedProduct:bool=false
     AdditionalInfo:bool=false
     AdditionalInfoWarning:bool=true
     Unknown:bool=Get bIsEquippedInOtherCategory(self=@n5)
     Index:EItemAccessRestriction Enum=Get Restriction() -> (Restriction:EItemAccessRestriction Enum=UnsatisfiedExclusiveItem)
    ) -> (ReturnValue:bool))
   n7 = Get AcessRestrictionWarning()
   > SetShowIcon!(execute, self:WBP UIKit Info Box Object Reference=@n7, Show Icon:bool=true) -> (then)
   SetShowIcon(execute=@n6.else, self=@n7, Show Icon=false)
  ]}
 AdjustVisualsForEmoteWheel {nodes [
   AdjustVisualsForEmoteWheel() {is=title K2Node_FunctionEntry}
   n1 = Get IsLoadedAsEmoteWheel() -> (IsLoadedAsEmoteWheel:bool)
   > SetVisibility(
    self=Get SubcategoryCarousel() -> (SubcategoryCarousel:WBP Subcategory Carousel Object Reference)
    InVisibility=Select!(Option 0:ESlateVisibility Enum=SelfHitTestInvisible, Option 1:ESlateVisibility Enum=Collapsed, Index:bool=@n1) -> (ReturnValue:ESlateVisibility Enum)
   )
   > Set Padding!(
    execute
    Padding:Margin Structure=Make Margin(Left=0, Top=SelectFloat(A=80, B=16, bPickA=@n1), Right=0, Bottom=0)
    self:Grid Slot Object Reference=SlotAsGridSlot(Widget=Get SortFilterBar_Hitbox() -> (SortFilterBar_Hitbox:Button Object Reference))
   ) -> (then, Output_Get:Margin Structure)
  ]}
 BP_GetDesiredFocusTarget {nodes [
   Get Desired Focus Target()
   n1 = Get FilteredInputField() -> (FilteredInputField:WBP UIKit Filtered Input Field Object Reference)
   n2 = Get EditableText!(self:WBP UIKit Block Editable Text Object Reference=Get Block_EditableText!(self:WBP UIKit Input Field Base Object Reference=Get InputField!(self:WBP UIKit Filtered Input Field Object Reference=@n1) -> (InputField:WBP UIKit Input Field Object Reference)) -> (Block_EditableText:WBP UIKit Block Editable Text Object Reference)) -> (EditableText:UIKit Editable Text Object Reference)
   > Return Node!(
    execute
    ReturnValue:Widget Object Reference=Select(
     Option 0:Widget Object Reference=Get LockerTileList() -> (LockerTileList:Fort Tile View Object Reference)
     Option 1:Widget Object Reference=@n2
     Index=HasAnyUserFocus!(self:Widget Object Reference=@n2) -> (ReturnValue:bool)
    ) -> (ReturnValue:Widget Object Reference)
   )
  ]}
 BP_GetDesiredInputConfig {nodes [
   Get Desired Input Config() {is=title K2Node_FunctionEntry}
   > Return Node(ReturnValue:UIInput Config Structure=Make UIInput Config!(
     bIgnoreMoveInput:bool=False
     bIgnoreLookInput:bool=False
     InputMode:ECommonInputMode Enum=All
     MouseCaptureMode:EMouseCaptureMode Enum=CaptureDuringMouseDown
     MouseLockMode:EMouseLockMode Enum=DoNotLock
     bHideCursorDuringViewportCapture:bool=True
    ) -> (UIInputConfig:UIInput Config Structure) {is=title K2Node_MakeStruct})
  ]}
 Bind Preview Actions {nodes [
   Bind Preview Actions() {is=title K2Node_FunctionEntry}
   n1 = Get Hovered Item() -> (Hovered Item:WBP UIKit Item Card Locker Object Reference)
   > Is Valid(InputObject=@n1)
   > Bind Preview Actions!(
    execute
    self:WBP UIKit Item Card Base Object Reference=@n1
    Preview Widget:Fort Item Preview Widget Object Reference=Get FortItemPreviewWidget() -> (FortItemPreviewWidget:WBP Item Preview Object Reference)
    Context Menu:WBP UIKit Context Menu Anchor Button Base Object Reference=Get ContextMenu!(self:WBP UIKit Item Card Locker Object Reference=@n1) -> (ContextMenu:WBP UIKit Context Menu Anchor Button Object Reference)
    Item VM:Fort Item VM Object Reference=Get FortItemVM!(self:WBP UIKit Item Card Locker Object Reference=@n1) -> (FortItemVM:Fort Item VM Object Reference)
    Include Special Actions:bool=true
    Include Toggle Actions:bool=true
   ) -> (then) {is=title K2Node_CallFunction}
  ]}
 CompactUI_PostListLoad {nodes [
   CompactUI_PostListLoad() {is=title K2Node_FunctionEntry}
   > Branch(Condition=IsMobileGame())
   n1 = Get DefaultEntryWidth() -> (DefaultEntryWidth:double)
   > n2 = Branch(Condition="Not Equal (Float)"(A=@n1))
   n3 = Get LockerTileList()
   > SetEntryWidth!(execute, self:Tile View Object Reference=@n3, NewWidth:float=Get CompactUIOverrideEntryWidth() -> (CompactUIOverrideEntryWidth:double)) -> (then)
   n4 = Get DefaultEntryHeight() -> (DefaultEntryHeight:double)
   > n5 = SetEntryHeight!(
    execute
    self:Tile View Object Reference=@n3
    NewHeight:float=float * float(A=float / float(A=@n4, B=@n1), B=Get DefaultEntryHeight()) {is=title K2Node_PromotableOperator}
   ) -> (then)
   SetGridSlotPadding!(
    execute=[@n5.then, @n2.else]
    Widget:Widget Object Reference=Get LockerTileList()
    Padding:Margin Structure=Make Margin(Left=8, Top=-4, Right=16, Bottom=-4)
   ) -> (then)
  ]}
 Create Empty Warning when No Filtered Itens {nodes [
   Create Empty Warning when No Filtered Itens() {is=title K2Node_FunctionEntry}
   > n1 = Get!(execute:Exec) -> (then:Exec, else:Exec, EmptyItemsWarning:WBP Locker Empty Warning Object Reference)
   > n2 = RemoveFromParent(self=@n1)
   n3 = Is Valid(exec=[@n2.then, @n1.else], InputObject=Get FortItemSelectionVM() -> (FortItemSelectionVM:Fort Item Selection VM Object Reference))
   n4 = Get FilteredItems!(self:Fort Locker Category Filter VM Object Reference=Get FortLockerCategoryFilterVM() -> (FortLockerCategoryFilterVM:Fort Locker Category Filter VM Object Reference)) -> (FilteredItems:Array of Fort Item VM Object References)
   > n5 = Branch(Condition="integer <= integer"(A=Length(TargetArray:Array of Fort Item VM Object References=@n4)))
   > SetVisibility(
    self=Get WBP_UIKit_DescriptionWithSubtitle() -> (WBP_UIKit_DescriptionWithSubtitle:WBP UIKit Description With Subtitle Object Reference)
    InVisibility=Collapsed
   )
   n6 = Get SelectedCategory!(self:Fort Item Selection VM Object Reference=Get FortItemSelectionVM()) -> (SelectedCategory:Fort Locker Category Item VM Object Reference)
   n7 = Get SortFilter!(self:Fort Locker Category Filter VM Object Reference=Get FortLockerCategoryFilterVM()) -> (SortFilter:Fort Sort Filter VM Object Reference)
   > n8 = Create Widget!(
    execute:Exec
    Class:User Widget Class Reference=&/WBP_LockerEmptyWarning.WBP_LockerEmptyWarning_C
    OwningPlayer:Player Controller Object Reference
    SelectedFilterLabel:Text=Get Label!(self:Fort Sort Filter Item VM Object Reference=Get ActiveFilterItem!(self:Fort Sort Filter VM Object Reference=@n7) -> (ActiveFilterItem:Fort Filter Item VM Object Reference)) -> (Label:Text)
    SelectedCategoryLabel:Text=TextToUpper!(InText:"Text (by ref)"=GetCategoryDisplayName!(self:Fort Locker Category Item VM Object Reference=@n6) -> (ReturnValue:Text)) -> (ReturnValue:Text)
   ) -> (then:Exec, ReturnValue:WBP Locker Empty Warning Object Reference)
   > n9 = Set EmptyItemsWarning(EmptyItemsWarning:WBP Locker Empty Warning Object Reference=@n8) -> (Output_Get:WBP Locker Empty Warning Object Reference)
   > n10 = AddChildToGrid!(
    execute
    self:Grid Panel Object Reference=Get Locker() -> (Locker:Grid Panel Object Reference)
    Content:Widget Object Reference=@n9
    InRow:int=2
    InColumn:int=2
   ) -> (then, ReturnValue:Grid Slot Object Reference)
   > Set HorizontalAlignment!(execute, HorizontalAlignment:EHorizontalAlignment Enum=HAlign_Center, self:Grid Slot Object Reference=@n10) -> (then, Output_Get:EHorizontalAlignment Enum=HAlign_Fill)
   > Set VerticalAlignment!(execute, VerticalAlignment:EVerticalAlignment Enum=VAlign_Top, self:Grid Slot Object Reference=@n10) -> (then, Output_Get:EVerticalAlignment Enum=VAlign_Fill)
   > Set!(execute, Padding_Left:float=0, Padding_Top:float=100, Padding_Right:float=0, Padding_Bottom:float=0, self:Grid Slot Object Reference=@n10) -> (then, Output_Get:Margin Structure)
   > Set SelectedItem!(execute, SelectedItem:Fort Item VM Object Reference, self:Fort Item Selection VM Object Reference=Get FortItemSelectionVM()) -> (then, Output_Get:Fort Item VM Object Reference)
   SetVisibility(execute=@n5.else, self=Get WBP_UIKit_DescriptionWithSubtitle(), InVisibility=SelfHitTestInvisible)
   SetVisibility(execute=@n3.Is Not Valid, self=Get WBP_UIKit_DescriptionWithSubtitle(), InVisibility=Collapsed)
   > n11 = Create Widget(Class=&WBP_LockerEmptyWarning_C)
   > n12 = Set EmptyItemsWarning(EmptyItemsWarning=@n11)
   > n13 = AddChildToGrid(self=Get Locker(), Content=@n12, InRow=2, InColumn=2)
   > Set HorizontalAlignment(HorizontalAlignment=HAlign_Center, self=@n13) -> (Output_Get=HAlign_Fill)
   > Set VerticalAlignment(VerticalAlignment=VAlign_Top, self=@n13) -> (Output_Get=VAlign_Fill)
   > Set(Padding_Left=0, Padding_Top=100, Padding_Right=0, Padding_Bottom=0, self=@n13)
  ]}
 EquippedItemChanged {nodes [
   n1 = EquippedItemChanged!() -> (then, EquippedItem:Fort Item VM Object Reference) {is=title K2Node_FunctionEntry}
   n2 = GetOwningPlayer()
   n3 = Get FilteredInputField()
   n4 = Get EditableText(self=Get Block_EditableText(self=Get InputField(self=@n3)))
   n5 = Get LastActiveScreen() -> (LastActiveScreen:ELockerScreenList Enum=NewEnumerator0)
   > n6 = Branch(Condition=OR Boolean(
     A=HasUserFocus!(self:Widget Object Reference=@n4, PlayerController:Player Controller Object Reference=@n2) -> (ReturnValue:bool)
     B="Equal (Enum)"(A:ELockerScreenList Enum=@n5, B:ELockerScreenList Enum=NewEnumerator5)
    ))
   n7 = Get LockerTileList()
   > n8 = Set Selected Item!(execute, self:List View Object Reference=@n7, Item:Object Reference=@n1) -> (then) {is=title K2Node_CallFunction}
   n9 = Get FortItemSelectionVM()
   Is Valid(exec=[@n10.then, @n8.then], InputObject=@n9)
   > Set SelectedItem(SelectedItem=@n1, self=@n9)
   > Set HoveredItem!(execute, HoveredItem:Fort Item VM Object Reference=@n1, self:Fort Item Selection VM Object Reference=@n9) -> (then, Output_Get:Fort Item VM Object Reference)
   Set Selected Item(execute=@n6.else, self=@n7, Item=@n1)
   > n10 = Navigate To Item!(execute, self:List View Object Reference=@n7, Item:Object Reference=@n1) -> (then)
  ]}
 EventGraph {nodes [
   n1 = SetGridSlotPadding!(
    execute
    Widget=Get WBP_UIKit_DescriptionWithSubtitle()
    Padding_Left:float=0
    Padding_Top:float=0
    Padding_Right:float=0
    Padding_Bottom:float=SelectFloat(A=0, B=42, bPickA=IsMobileGame())
   ) -> (then)
   ClearMaxDesiredHeight(
    execute=[@n1.then, @n178.then]
    self:Size Box Object Reference=Get SizeBoxTitle!(self:WBP UIKit Item Description Base Object Reference=Get WBP_UIKit_DescriptionWithSubtitle()) -> (SizeBoxTitle:Size Box Object Reference)
   )
   > n2 = StartListeningToEvent!(
    execute:Exec
    Target:Gameplay Event Router Component Object Reference=GetGlobalEventRouter()
    EventType:Script Struct Object Reference=/FortniteCoreUI/UI/Social/Chat/TextChat/BPS_TextChatActiveStateChange
    Context:Object Reference=GetOwningPlayer()
    EventBubblingRule:EEventBubblingRule Enum
   ) -> (
    then:Exec
    AsyncTaskProxy:Async Action Start Listening To Event Object Reference
    OnEventReceived:Exec
    Payload:BPS Text Chat Active State Change Structure
   )
   > Set ChatEventWatch(ChatEventWatch:Async Action Start Listening To Event Object Reference=@n2) -> (Output_Get:Async Action Start Listening To Event Object Reference)
   SetVisibility(
    execute=@n2.OnEventReceived
    self=Get WBP_ActionBarContainer() -> (WBP_ActionBarContainer:WBP Action Bar Container Object Reference)
    InVisibility=Select(
     Option 0:ESlateVisibility Enum=SelfHitTestInvisible
     Option 1:ESlateVisibility Enum=Collapsed
     Index=Break BPS Text Chat Active State Change!(BPS_TextChatActiveStateChange:"BPS Text Chat Active State Change Structure (by ref)"=@n2.Payload) -> (bActive_4_#18:bool=~) {is=title K2Node_BreakStruct}
    ) -> (ReturnValue:ESlateVisibility Enum)
   )
   n3 = Get FortItemPreviewWidget()
   n4 = ClearPreviewItem(self:Fort Item Preview Widget Object Reference=@n3)
   DeactivateWidget(execute=[@n4.then, @n167.then, @n166.else], self=@n3)
   > Sequence()
   > n5 = Branch(Condition=IsInZone!(self:Fort Global UIContext Object Reference=Get FortGlobalUIContext() -> (ReturnValue:Fort Global UIContext Object Reference) {is=title K2Node_GetSubsystem}) -> (ReturnValue:bool))
   Broadcast Event!(
    execute=@n5.else
    self:Gameplay Event Router Component Object Reference=GetGlobalEventRouter()
    EventData:BPS Set Text Chat Active State Structure=Make BPS Set Text Chat Active State!(bSetActive_5_#18:bool=true) -> (BPS_SetTextChatActiveState:BPS Set Text Chat Active State Structure) {is=title K2Node_MakeStruct}
    EventContext:Object Reference=GetOwningPlayer()
   ) -> (then, ReturnValue:bool)
   n6 = Get FortItemPreviewWidget()
   n7 = SetCurrentPreviewItem!(
    execute
    self:Fort Item Preview Widget Object Reference=@n6
    ItemVM:Fort Item VM Object Reference=Get SelectedItem!(self:Fort Item Selection VM Object Reference=Get FortItemSelectionVM()) -> (SelectedItem:Fort Item VM Object Reference)
   ) -> (then)
   n8 = Get SelectedCategory(self=Get FortItemSelectionVM())
   SelectTabByID!(
    execute=[@n155.then, @n7.then]
    self:Common Tab List Widget Base Object Reference=Get Tabs_Subcategory!(self:WBP Subcategory Carousel Object Reference=Get SubcategoryCarousel()) -> (Tabs_Subcategory:WBP UIKit Tab Group Object Reference)
    TabNameID:Name=String To Name(InString=GetAnalyticsName!(self:Fort Locker Category Item VM Object Reference=@n8) -> (ReturnValue:String))
    bSuppressClickFeedback:bool=true
   ) -> (then, ReturnValue:bool)
   > UpdateInputPadding!(execute, self, Selection:ECommonInputType Enum=GetCurrentInputType(self=Get CommonInputSubsystem())) -> (then)
   > n9 = Branch(Condition="Equal (Enum)"(A=Get LastActiveScreen() -> (LastActiveScreen=NewEnumerator0), B=NewEnumerator4))
   > n10 = PlayAnimationReverse(InAnimation=Get SubMenu() -> (SubMenu:Widget Animation Object Reference))
   Set LastActiveScreen(execute=[@n10.then, @n11.then], LastActiveScreen:ELockerScreenList Enum=NewEnumerator3) -> (Output_Get:ELockerScreenList Enum=NewEnumerator0)
   n11 = PlayAnimationForward(execute=@n9.else, InAnimation=Get Intro())
   n12 = "On Clicked (WBP_UIKit_Button_Back)"!() -> (OutputDelegate, then, Button:Common Button Base Object Reference) {is=title K2Node_ComponentBoundEvent}
   n13 = MakeLiteralText()
   > n14 = Set SearchText!(
    execute
    SearchText:Text=@n13
    self:Fort Sort Filter Search Text VM Object Reference=Get SearchText!(self:Fort Sort Filter VM Object Reference=Get FortSortFilterVM() -> (FortSortFilterVM:Fort Sort Filter VM Object Reference)) -> (SearchText:Fort Sort Filter Search Text VM Object Reference)
   ) -> (then, Output_Get:Text)
   n15 = Get InputField(self=Get FilteredInputField())
   > SetText!(execute, self:WBP UIKit Block Editable Text Object Reference=Get Block_EditableText(self=@n15), Text:Text=@n14) -> (then)
   > OnTextChanged_Event!(execute, self:WBP UIKit Input Field Base Object Reference=@n15, Text:"Text (by ref)"=@n13) -> (then)
   > Fire Event Element Interacted(Element=@n12.Button)
   > n16 = SetAllInteractionsEnabled!(execute, self, IsInteractionEnabled:bool=false) -> (then)
   n17 = Get Intro()
   Branch(execute=[@n16.then, @n153.then], Condition=NOT Boolean(A=IsAnimationPlaying(InAnimation=@n17)))
   > n18 = "Play Animation with Finished event (legacy)"!(
    execute:Exec
    Widget:User Widget Object Reference=Self-Reference()
    InAnimation:Widget Animation Object Reference=@n17
    StartAtTime:float=0
    NumLoopsToPlay:int=1
    PlayMode:EUMGSequencePlayMode Enum=Reverse
    PlaybackSpeed:float=1
   ) -> (then:Exec, Result:UMGSequence Player Object Reference, Finished:Exec) {is=title K2Node_PlayAnimation}
   DeactivateWidget(execute=@n18.Finished)
   n19 = "Tab Select (SubcategoryCarousel)"!() -> (OutputDelegate, then, TabID:Name=None) {is=title K2Node_ComponentBoundEvent}
   > Branch(Condition=NOT Boolean(A=Get IgnoreInputs() -> (IgnoreInputs:bool)))
   > SetAllInteractionsEnabled(IsInteractionEnabled=false)
   > SetInteractionsEnabled!(execute, self:WBP Subcategory Carousel Object Reference=Get SubcategoryCarousel(), EnableInteractions:bool=false) -> (then)
   n20 = "To String (Name)"(InName=@n19.TabID)
   > Set TabFlowRiderString(TabFlowRiderString:String=Append(A="Tab Selected / ", B=@n20)) -> (Output_Get:String)
   > UpdateFlowriderSubView!(execute, self, SubViewString:String=@n20) -> (then)
   > HandleSubcategoryTabSelected!(execute, self, TabNameID:Name=@n19.TabID) -> (then)
   > UpdateDetailsButtonState!(execute, self, Selected item:Object Reference=Get SelectedItem(self=Get FortItemSelectionVM())) -> (then)
   n21 = Get FortLockerCategoryFilterVM()
   > n22 = UpdateEquipButtonState!(
    execute
    self
    SelectedItem:Object Reference=Get EquippedItem!(self:Fort Locker Category Item VM Object Reference=Get CategoryItem!(self:Fort Locker Category Filter VM Object Reference=@n21) -> (CategoryItem:Fort Locker Category Item VM Object Reference)) -> (EquippedItem:Fort Item VM Object Reference)
   ) -> (then)
   Delay(
    execute=[@n22.then, @n24.then]
    Duration=GetConfigDataFloatWithDefault!(EntryName:String=LockerItemNavigationDelayInTabChange, DefaultValue:float=0.1) -> (ReturnValue:float)
   )
   > n23 = Is Valid(InputObject=Get LockerTileList())
   n24 = LogString!(
    execute=@n23.Is Not Valid
    Instance Log Level:BlueprintLogLevel Enum=NewEnumerator3
    LogString:String=Locker AsyncWidget creation race condition - waiting for completion
    PrintToScreen:bool=false
    Instigator:Object Reference
   ) -> (then)
   Navigate To Item(
    execute=@n23.Is Valid
    self=Get LockerTileList()
    Item=Get FirstValidOrEquippedItem!(self:Fort Locker Category Filter VM Object Reference=@n21) -> (FirstValidOrEquippedItem:Fort Item VM Object Reference)
   )
   > SetAllInteractionsEnabled(IsInteractionEnabled=true)
   > SetInteractionsEnabled(self=Get SubcategoryCarousel(), EnableInteractions=true)
   n25 = "On Hovered (FilteredInputField)"() {is=title K2Node_ComponentBoundEvent}
   ToggleBottomBarButtonsForFilteredInputField!(execute=[@n25.then, @n27.then], self, IsSearchHovered:bool=true) -> (then)
   n26 = "Not Equal (Enum)"!(A:ECommonInputType Enum=GetCurrentInputType(self=Get CommonInputSubsystem()), B:ECommonInputType Enum=MouseAndKeyboard) -> (ReturnValue:bool)
   > Branch(Condition=@n26)
   > UpdateFilteredInputFieldControllerVisuals()
   "On Unhovered (FilteredInputField)"() {is=title K2Node_ComponentBoundEvent}
   > Branch(Condition=@n26)
   > UpdateFilteredInputFieldControllerVisuals()
   n27 = "On Hovered (Button_SortFilter)"!() -> (OutputDelegate, then, Button:Common Button Base Object Reference) {is=title K2Node_ComponentBoundEvent}
   "On Clicked (TogglePreview)"!() -> (OutputDelegate, then, Button:Common Button Base Object Reference) {is=title K2Node_ComponentBoundEvent}
   > ToggleVariantOverrideOnCurrentItem(self:Fort Item Preview Widget Object Reference=Get FortItemPreviewWidget())
   "On Preview Item Shown (FortItemPreviewWidget)"!() -> (OutputDelegate, then, ItemDefinition:Fort Account Item Definition Object Reference) {is=title K2Node_ComponentBoundEvent}
   > n28 = Branch(Condition=Get NeedToSetupPreviewActions() -> (NeedToSetupPreviewActions:bool))
   > Set NeedToSetupPreviewActions(NeedToSetupPreviewActions:bool=false) -> (Output_Get:bool)
   > SetupPreviewActionBinds!(
    execute:Exec
    self:Object Reference=Get LockerTileWidget() -> (LockerTileWidget:User Widget Object Reference)
    PreviewWidget:Fort Item Preview Widget Object Reference=Get FortItemPreviewWidget()
   ) -> (then:Exec) {is=title K2Node_Message}
   BindPreviewActions!(
    execute:Exec=@n28.else
    self:Object Reference=Get LockerTileWidget()
    Preview Widget:Fort Item Preview Widget Object Reference=Get FortItemPreviewWidget()
   ) -> (then:Exec) {is=title K2Node_Message}
   "On Clicked (Button_Equip)"!() -> (OutputDelegate, then, Button:Common Button Base Object Reference) {is=title K2Node_ComponentBoundEvent}
   > n29 = Branch(Condition=NOT Boolean(A=IsCustomizable!() -> (Customizable:bool)))
   > n30 = Cast To FortItemVM!(
    execute
    Object=Get Selected Item!(self:List View Object Reference=Get LockerTileList()) -> (ReturnValue:Object Reference) {is=title K2Node_CallFunction}
   ) -> (then, CastFailed, AsFort Item VM:Fort Item VM Object Reference) {is=title K2Node_DynamicCast}
   n31 = Get FortItemSelectionVM()
   > n32 = EquipItem!(
    execute:Exec
    Item:Fort Item VM Object Reference=@n30
    SlotIndex:int=Get SlotIndex!(self:Fort Locker Category Item VM Object Reference=Get SelectedCategory(self=@n31)) -> (SlotIndex:int=0)
    Variants:Array of Mcp Variant Channel Info Structures
    bApplyToAllSlots:bool=false
    bAllowMocking:bool=true
   ) -> (then:Exec, OnOperationCompleted:Exec, Success:bool=~) {is=title K2Node_AsyncAction}
   Branch(execute=@n32.OnOperationCompleted, Condition=@n32)
   > UpdateEquipButtonState(SelectedItem=@n30)
   > Set Selected Item(self=Get LockerTileList(), Item=@n30)
   OpenCustomizeMenu(execute=@n29.else)
   n33 = "On Text Changed (FilteredInputField)"!() -> (OutputDelegate, then, Text:Text) {is=title K2Node_ComponentBoundEvent}
   > n34 = Set SearchText(SearchText=@n33.Text, self=Get SearchText(self=Get FortSortFilterVM()))
   > Set CurrentSearchText(CurrentSearchText:Text=@n34) -> (Output_Get:Text)
   > UpdateFilteredInputFieldControllerVisuals()
   n35 = Get LockerTileList()
   > Is Valid(InputObject=@n35)
   > Clear Selection(self:List View Object Reference=@n35)
   "On Clicked (Button_Details)"!() -> (OutputDelegate, then, Button:Common Button Base Object Reference) {is=title K2Node_ComponentBoundEvent}
   n36 = GetCategoryConfig!() -> (Category Config:Fort Locker Category Config Object Reference)
   > Is Valid(InputObject=@n36)
   n37 = Get CustomItemDetails!(self:Fort Locker Category Config Object Reference=@n36) -> (CustomItemDetails:User Widget Soft Class Reference)
   > Branch(Condition=IsValidSoftClassReference!(SoftClassReference:"Object Soft Class Reference (by ref)"=@n37) -> (ReturnValue:bool))
   > n38 = Async Load Class Asset!(execute:Exec, AssetClass:Object Soft Class Reference=@n37) -> (then:Exec, Completed:Exec, Class:Object Class Reference) {is=title K2Node_LoadAssetClass}
   ShowItemDetails!(
    execute=@n38.Completed
    self
    CustomItemDetails:User Widget Class Reference=Cast To UserWidget Class!(Class:Object Class Reference=@n38) -> (AsUser Widget:User Widget Class Reference, bSuccess:bool=~) {is=title K2Node_ClassDynamicCast}
   ) -> (then)
   "On Clicked (Button_Use)"!() -> (OutputDelegate, then, Button:Common Button Base Object Reference) {is=title K2Node_ComponentBoundEvent}
   > FocusSearchField!(execute, self) -> (then, bPassThrough:bool)
   "On Unhovered (SortFilterBar_Hitbox)"() {is=title K2Node_ComponentBoundEvent}
   > ToggleBottomBarButtonsForFilteredInputField(IsSearchHovered=false)
   n39 = Get FilteredItems(self=Get FortLockerCategoryFilterVM())
   > Branch(Condition="integer > integer"(A=Length(TargetArray=@n39)))
   n40 = Get FortItemSelectionVM()
   n41 = Get EquippedItem(self=Get SelectedCategory(self=@n40))
   n42 = Get SelectedItem(self=@n40)
   > UpdateEditButtonEnabled!(
    execute
    self
    IsEnabled:bool=Is Edit Enabled!(
     self
     SelectedHasVariants:bool=Get bHasVariants!(self:Fort Item VM Object Reference=@n42) -> (bHasVariants:bool)
     EquippedHasVariants:bool=Get bHasVariants(self=@n41)
    ) -> (Result:bool)
   ) -> (then)
   > Set bIsEnabled!(
    execute
    bIsEnabled:bool=true
    self:Widget Object Reference=Get WBP_UIKit_Button_Options() -> (WBP_UIKit_Button_Options:WBP UIKit Button Quiet Object Reference)
   ) -> (then, Output_Get:bool)
   "On Clicked (Button_EditStyles)"!() -> (OutputDelegate, then, Button:Common Button Base Object Reference) {is=title K2Node_ComponentBoundEvent}
   > SetAllInteractionsEnabled(IsInteractionEnabled=false)
   > Set bIsEnabled!(execute, bIsEnabled=false) -> (then, Output_Get)
   n43 = Get SelectedItem(self=Get FortItemSelectionVM())
   n44 = IsCustomizable()
   > n45 = FireEventElementInteractedByName!(
    execute
    Element Name:String=Append(
     A=Select(Option 0:String=Edit Styles, Option 1:String=Customize, Index=@n44) -> (ReturnValue:String)
     B=GetItemTemplateId!(self:Fort Item VM Object Reference=@n43) -> (ReturnValue:String)
    )
    WidgetCallingFrom:Widget Object Reference=Self-Reference()
   ) -> (then, Success:bool)
   n46 = Branch(execute=[@n151.then, @n45.then], Condition=IsCustomizable())
   > OpenCustomizeMenu()
   n47 = Get FortItemSelectionVM()
   n48 = EquipItem(
    execute=@n46.else
    Item=Get SelectedItem(self=@n47)
    SlotIndex=Get SlotIndex(self=Get SelectedCategory(self=@n47)) -> (SlotIndex=0)
    Variants=Make Array!() -> (Array:Array of Mcp Variant Channel Info Structures)
    bApplyToAllSlots=false
    bAllowMocking=true
   ) -> (Success=~)
   Set LastActiveScreen(execute=@n48.OnOperationCompleted, LastActiveScreen=NewEnumerator4) -> (Output_Get=NewEnumerator0)
   n49 = Get SubMenu()
   > Branch(Condition=NOT Boolean(A=IsAnimationPlaying(InAnimation=@n49)))
   > n50 = "Play Animation with Finished event (legacy)"(Widget=Self-Reference(), InAnimation=@n49, StartAtTime=0, NumLoopsToPlay=1, PlaybackSpeed=1)
   Set bIsEnabled(execute=@n50.Finished, bIsEnabled=true)
   > n51 = Sequence()
   > SetFocus(self=GetDesiredFocusTarget!(self:Common Activatable Widget Object Reference) -> (ReturnValue:Widget Object Reference))
   n52 = PushPrimaryContentClassAsync!(
    execute:Exec=@n51.then_1
    WidgetClass:Common Activatable Widget Class Reference=/Locker/MPL/EditStyles/WBP_MPL_VariantSelector.WBP_MPL_VariantSelector_C
   ) -> (then:Exec, OnPrimaryContentPushed:Exec, ActivatableWidget:Common Activatable Widget Object Reference) {is=title K2Node_AsyncAction}
   SetAllInteractionsEnabled(execute=@n52.OnPrimaryContentPushed, IsInteractionEnabled=true)
   > n53 = Cast To WBP_MPL_VariantSelector!(execute, Object=@n52) -> (then, CastFailed, AsWBP MPL Variant Selector:WBP MPL Variant Selector Object Reference) {is=title K2Node_DynamicCast}
   n54 = Get FortVariantSelectorVM!(self:WBP MPL Variant Selector Object Reference=@n53) -> (FortVariantSelectorVM:Fort Variant Selector VM Object Reference)
   DismissVariantChannels(execute=[@n154.then, @n53.then], self:Fort Variant Selector VM Object Reference=@n54)
   > UpdateVariantChannels(self:Fort Variant Selector VM Object Reference=@n54)
   n55 = "On Item Rotation Changed (FortItemPreviewWidget)"() {is=title K2Node_ComponentBoundEvent}
   n56 = Branch(execute=[@n55.then, @n57.then], Condition=IsUsingTouch(PlayerController=GetOwningPlayer()))
   HideItemDescription(execute=@n56.else)
   n57 = "On Item Zoom Level Changed (FortItemPreviewWidget)"() {is=title K2Node_ComponentBoundEvent}
   n58 = "On Clicked (Button_SortFilter)"!() -> (OutputDelegate, then, Button:Common Button Base Object Reference) {is=title K2Node_ComponentBoundEvent}
   > n59 = Branch(Condition="Equal (Enum)"(A=Get LastActiveScreen() -> (LastActiveScreen=NewEnumerator0), B=NewEnumerator5))
   Fire Event Element Interacted(execute=@n59.else, Element=@n58.Button)
   > Is Valid(InputObject=Get FortSortFilterSelectionVM() -> (FortSortFilterSelectionVM:Fort Sort Filter Selection VM Object Reference))
   n60 = Get SortFilter(self=Get FortLockerCategoryFilterVM())
   > Is Valid(InputObject=@n60)
   > Initialize!(
    execute
    self:Fort Sort Filter Selection Base VM Object Reference=Get FortSortFilterSelectionVM()
    InSortFilterVM:Fort Sort Filter VM Object Reference=@n60
   ) -> (then)
   > n61 = Create Widget!(
    execute
    Class=/Locker/MPL/SortFilter/WBP_MPLocker_SortFilter_Modal.WBP_MPLocker_SortFilter_Modal_C
    OwningPlayer=GetOwningPlayer()
    FortSortFilterVM:Fort Sort Filter VM Object Reference=Get SortFilter(self=Get FortLockerCategoryFilterVM())
   ) -> (then, ReturnValue:WBP MPLocker Sort Filter Modal Object Reference)
   > n62 = Set SortFilter Modal(SortFilter Modal:WBP MPLocker Sort Filter Modal Object Reference=@n61) -> (Output_Get:WBP MPLocker Sort Filter Modal Object Reference)
   n63 = SortFilterAccepted() {is=title K2Node_CustomEvent}
   Bind Event to Sort Filter Accepted(execute=@n62.then, self:WBP MPLocker Sort Filter Modal Object Reference=@n62, Delegate=@n63) {is=title K2Node_AddDelegate}
   > TryPushWidgetAsModalPanel!(execute, PushWidget:User Widget Object Reference=@n62) -> (then)
   > ActivateWidget(self=@n62)
   > n64 = Set LastActiveScreen(LastActiveScreen=NewEnumerator5) -> (Output_Get=NewEnumerator0)
   n65 = RestoreEnabled() {is=title K2Node_CustomEvent}
   Bind Event to On Widget Deactivated(execute=@n64.then, self:Common Activatable Widget Object Reference=@n62, Delegate=@n65) {is=title K2Node_AddDelegate}
   > Set bIsEnabled(bIsEnabled=false)
   > Call On Sort Filter Modal Open!(execute:Exec, self:WBP MPLocker Items Object Reference) -> (then:Exec) {is=title K2Node_CallDelegate}
   "On Clicked (WBP_UIKit_Button_Options)"!() -> (OutputDelegate, then, Button:Common Button Base Object Reference) {is=title K2Node_ComponentBoundEvent}
   > ToggleSelectedContext!(execute:Exec, self:Object Reference=Get LockerTileWidget()) -> (then:Exec) {is=title K2Node_Message}
   "On Unhovered (TileView_Hitbox)"() {is=title K2Node_ComponentBoundEvent}
   > Branch(Condition=NOT Boolean(A=Get IgnoreInputs()))
   n66 = Get FilteredItems(self=Get FortLockerCategoryFilterVM())
   > Branch(Condition="integer > integer"(A=Length(TargetArray=@n66)))
   n67 = GetOwningPlayer()
   n68 = GetIsEquippable!(self:Fort Locker Category Item VM Object Reference=Get SelectedCategory(self=Get FortItemSelectionVM())) -> (ReturnValue:bool)
   > n69 = Branch(Condition=OR Boolean(A=IsUsingTouch(PlayerController=@n67), B=NOT Boolean(A=@n68)))
   n70 = Get EquippedItem(self=Get SelectedCategory(self=Get FortItemSelectionVM()))
   n71 = Get SelectedItem(self=Get FortItemSelectionVM())
   n72 = Branch(
    execute=@n69.else
    Condition="Equal (Object)"(
     A=Get ItemDefinition!(self:Fort Item VM Object Reference=@n70) -> (ItemDefinition:Fort Item Definition Object Reference)
     B=Get ItemDefinition(self=@n71)
    )
   )
   Set SelectedItem(execute=@n72.else, SelectedItem=@n70, self=Get FortItemSelectionVM())
   "On Clicked (Button_ClearSearch)"!() -> (OutputDelegate, then, Button:Common Button Base Object Reference) {is=title K2Node_ComponentBoundEvent}
   n73 = Get InputField(self=Get FilteredInputField())
   > SetText(self=Get Block_EditableText(self=@n73))
   > OnTextChanged_Event(self=@n73, Text=MakeLiteralText())
   n74 = ContextMenuOpen!() -> (OutputDelegate, then, Widget:Widget Object Reference) {is=title K2Node_CustomEvent}
   > SetOpenedContextMenuParentEntryOpacity!(execute, self, ContextMenuParentEntryWidget:Widget Object Reference=@n74.Widget) -> (then)
   n75 = Get LockerTileList()
   > SetIsPointerScrollingEnabled!(execute, self:List View Base Object Reference=@n75, bInIsPointerScrollingEnabled:bool=false) -> (then)
   > SetAllInteractionsEnabled(IsInteractionEnabled=false)
   > Set IsContextMenuOpen(IsContextMenuOpen:bool=true) -> (Output_Get:bool)
   > n76 = Cast To WBP_UIKit_ItemCard_Locker!(execute, Object=@n74.Widget) -> (then, CastFailed, AsWBP UIKit Item Card Locker:WBP UIKit Item Card Locker Object Reference) {is=title K2Node_DynamicCast}
   n77 = Get FortItemVM(self=@n76)
   > Set Selected Item(self=Get LockerTileList(), Item=@n77)
   > n78 = GetConsoleVariableBoolValue!(execute, VariableName:String=fort.rollback.LockerSearchShowToolTip) -> (then, ReturnValue:bool)
   > Branch(Condition=@n78)
   > ShowMatchedTags!(execute, self, ItemVM:Fort Item VM Object Reference=@n77, ItemCard:WBP UIKit Item Card Locker Object Reference=@n76) -> (then)
   ContextMenuClosed() {is=title K2Node_CustomEvent}
   > SetClosedContextMenuParentEntryOpacity()
   > SetIsPointerScrollingEnabled(self=@n75, bInIsPointerScrollingEnabled=true)
   > SetAllInteractionsEnabled(IsInteractionEnabled=true)
   > Set IsContextMenuOpen(IsContextMenuOpen=false)
   n79 = GetCurrentInputType(self=Get CommonInputSubsystem())
   > n80 = Branch(Condition="Equal (Enum)"(A:ECommonInputType Enum=@n79, B:ECommonInputType Enum=MouseAndKeyboard))
   Handle Input Changed!(execute=@n80.else, self, NewInputType:ECommonInputType Enum=@n79) -> (then)
   HideItemDescription() {is=title K2Node_CustomEvent}
   > n81 = Clear and Invalidate Timer by Handle(Handle=Get ItemDescriptionHideTimer() -> (ItemDescriptionHideTimer:Timer Handle Structure))
   n82 = ShowItemDescription() {is=title K2Node_CustomEvent}
   n83 = Set Timer by Event(execute=@n81.then, Delegate=@n82, Time=Get ItemDetailsHideTime() -> (ItemDetailsHideTime:float), bLooping=false, bMaxOncePerFrame=true)
   > Set ItemDescriptionHideTimer(ItemDescriptionHideTimer:Timer Handle Structure=@n83) -> (Output_Get:Timer Handle Structure)
   > Branch(Condition=IsVisible(self=Get ItemDetails_VerticalBox() -> (ItemDetails_VerticalBox:Vertical Box Object Reference)))
   > ToggleItemDescription!(execute, self, show:bool=false) -> (then)
   n84 = ToggleItemDescription!() -> (OutputDelegate, then, show:bool=~) {is=title K2Node_CustomEvent}
   > n85 = Branch(Condition=@n84.show)
   n86 = Get Anim_ShowItemDetails() -> (Anim_ShowItemDetails:Widget Animation Object Reference)
   > n87 = IsAnimationPlayingForward!(execute, self:User Widget Object Reference, InAnimation:Widget Animation Object Reference=@n86) -> (then, ReturnValue:bool)
   n88 = GetOwningPlayer()
   > Branch(Condition=OR Boolean(A=@n87, B=IsUsingTouch(PlayerController=@n88)))
   > StopAnimation(InAnimation=@n86)
   > SetVisibility(self=Get ItemDetails_VerticalBox(), InVisibility=SelfHitTestInvisible)
   > PlayAnimationForward(InAnimation=@n86)
   n89 = Get Anim_HideItemDetails() -> (Anim_HideItemDetails:Widget Animation Object Reference)
   n90 = Branch(execute=@n85.else, Condition=IsAnimationPlaying(InAnimation=@n89))
   > n91 = IsAnimationPlayingForward(InAnimation=@n89)
   > n92 = Branch(Condition=@n91)
   n93 = StopAnimation(execute=@n92.else, InAnimation=@n89)
   n94 = "Play Animation with Finished event (legacy)"(execute=[@n93.then, @n90.else], Widget=Self-Reference(), InAnimation=@n89, StartAtTime=0, NumLoopsToPlay=1, PlaybackSpeed=1)
   SetVisibility(execute=@n94.Finished, self=Get ItemDetails_VerticalBox(), InVisibility=Hidden)
   n95 = ItemDetailsClosed() {is=title K2Node_CustomEvent}
   > RestoreEnabled() {is=title K2Node_CallFunction}
   > UpdatePreview(self:Fort Item Preview Widget Object Reference=Get FortItemPreviewWidget())
   OpenCustomizeMenu() {is=title K2Node_CustomEvent}
   n96 = Get SelectedItem(self=Get FortItemSelectionVM())
   > Set ItemToCustomize(ItemToCustomize:Fort Item VM Object Reference=@n96) -> (Output_Get:Fort Item VM Object Reference)
   n97 = IsInZone(self=Get FortGlobalUIContext())
   > n98 = Branch(Condition=AND Boolean(A=NOT Boolean(A=@n97), B=Get bRequiresCustomization!(self:Fort Item VM Object Reference=@n96) -> (bRequiresCustomization:bool)))
   n99 = TryGetSoftCustomActivatableWidgetClass!(
    self:Fort Locker Category Config Object Reference=GetItemCategory!(self:Fort Item VM Object Reference=@n96) -> (ReturnValue:Fort Locker Category Config Object Reference)
    ViewRequest:Gameplay Tag Structure=(TagName="UI.ItemPreview.PluginBridge.Mimosa.CustomizeMenu")
   ) -> (WidgetClass:Common Activatable Widget Soft Class Reference, ReturnValue:bool)
   > n100 = Branch(Condition=@n99.ReturnValue)
   > n101 = PushPrimaryContentSoftClassAsync!(execute:Exec, SoftWidgetClass:Common Activatable Widget Soft Class Reference=@n99) -> (then:Exec, OnPrimaryContentPushed:Exec, ActivatableWidget:Common Activatable Widget Object Reference) {is=title K2Node_AsyncAction}
   n102 = Set CustomizeMenu(execute=@n101.OnPrimaryContentPushed, CustomizeMenu:Common Activatable Widget Object Reference=@n101) -> (Output_Get:Common Activatable Widget Object Reference)
   > n103 = SetViewModelByClass!(
    execute
    Widget:User Widget Object Reference=@n102
    ViewModel:Notify Field Value Changed Interface=Get ItemToCustomize() -> (ItemToCustomize:Fort Item VM Object Reference)
   ) -> (then)
   n104 = Refresh Items() {is=title K2Node_CustomEvent}
   n105 = Bind Event to On Widget Deactivated(execute=@n103.then, self=@n102, Delegate=@n104)
   Set bIsEnabled(execute=[@n105.then, @n100.else, @n98.else], bIsEnabled=true)
   > Set bIsEnabled!(execute, bIsEnabled=true, self:Widget Object Reference=Get WBP_UIKit_Button_Options()) -> (then, Output_Get)
   > SetVisibility(self=Get TileView_Hitbox() -> (TileView_Hitbox:Button Object Reference), InVisibility=SelfHitTestInvisible)
   > SetAllInteractionsEnabled(IsInteractionEnabled=true)
   n106 = ModalStatus!() -> (OutputDelegate, then, IsModalOpen:bool=~) {is=title K2Node_CustomEvent}
   > Set LastActiveScreen(LastActiveScreen=Select(Option 0:ELockerScreenList Enum=NewEnumerator3, Option 1:ELockerScreenList Enum=NewEnumerator5, Index=@n106.IsModalOpen) -> (ReturnValue:ELockerScreenList Enum)) -> (Output_Get=NewEnumerator0)
   n107 = NOT Boolean(A=@n106.IsModalOpen)
   > SetAllInteractionsEnabled(IsInteractionEnabled=@n107)
   > SetIsEnabled(self=Get FilteredInputField(), bInIsEnabled=@n107)
   > n108 = Branch(Condition=@n106.IsModalOpen)
   Refresh Items(execute=@n108.else) {is=title K2Node_CallFunction}
   On ClickedSearchField!() -> (OutputDelegate, then, Button:Common Button Base Object Reference) {is=title K2Node_CustomEvent}
   n109 = Get FilteredInputField()
   > SetTextFocus(self:WBP UIKit Filtered Input Field Object Reference=@n109)
   n110 = Get InputField(self=@n109)
   > ToggleVirtualKeyboard!(execute, self:Editable Text Object Reference=Get EditableText(self=Get Block_EditableText(self=@n110)), bShow:bool=true) -> (then)
   n111 = OnEntryReleased!() -> (OutputDelegate, then, Widget:User Widget Object Reference) {is=title K2Node_CustomEvent}
   > SetCastedItemCardUnbindings!(execute, self=Self-Reference(), EntryWidgetObjectReference:Widget Object Reference=@n111.Widget) -> (then)
   > n112 = Cast To WBP_UIKit_ItemCard_Base!(execute, Object=@n111.Widget) -> (then, CastFailed, AsWBP UIKit Item Card Base:WBP UIKit Item Card Base Object Reference) {is=title K2Node_DynamicCast}
   > Unbind Event from OnModularButtonHoldTriggered(self:UIKit Holdable Modular Button Object Reference=@n112, Delegate=Create Event()) {is=title K2Node_RemoveDelegate}
   n113 = OnEntryHeld!() -> (OutputDelegate, then, Button:Common Button Base Object Reference) {is=title K2Node_CustomEvent}
   > n114 = Cast To WBP_UIKit_ItemCard_Base(Object=@n113.Button)
   > Set NeedToSetupPreviewActions(NeedToSetupPreviewActions=true)
   > SelectCard(self:WBP UIKit Item Card Base Object Reference=@n114)
   > ToggleSelectedContext(self=Get LockerTileWidget())
   n115 = OnEntryInitialized!() -> (OutputDelegate, then, Item:Object Reference, Widget:User Widget Object Reference) {is=title K2Node_CustomEvent}
   > SetupBinds!(
    execute:Exec
    self:Object Reference=Get LockerTileWidget()
    Widget:Widget Object Reference=@n115.Widget
    PreviewWidget:Fort Item Preview Widget Object Reference=Get FortItemPreviewWidget()
   ) -> (then:Exec) {is=title K2Node_Message}
   n116 = OnEntryGenerated!() -> (OutputDelegate, then, Widget:User Widget Object Reference) {is=title K2Node_CustomEvent}
   > Set Back To Top Options() {is=title K2Node_CallFunction}
   > n117 = Cast To WBP_UIKit_ItemCard_Base(Object=@n116.Widget)
   > ResetAnimationState(self:WBP UIKit Item Card Base Object Reference=@n117)
   > Bind Event to OnModularButtonHoldTriggered(self:UIKit Holdable Modular Button Object Reference=@n117, Delegate=Create Event()) {is=title K2Node_AddDelegate}
   n118 = GetOwningPlayer()
   > n119 = Branch(Condition=AND Boolean(A=IsUsingTouch(PlayerController=@n118), B=GetSelected!(self:Common Button Base Object Reference=@n117) -> (ReturnValue:bool)))
   > n120 = SetFocusState!(execute, self:WBP UIKit Item Card Base Object Reference=@n117, IsFocused:bool=true) -> (then)
   Branch(execute=[@n120.then, @n119.else], Condition=Get IsImpressionAvailable() -> (IsImpressionAvailable:bool))
   > Set IsImpressionAvailable(IsImpressionAvailable:bool=false) -> (Output_Get:bool)
   > Delay(Duration=1)
   > SetupTelemetry!(
    execute:Exec
    self:Object Reference=Get LockerTileWidget()
    TabFlowRiderString:String=Get TabFlowRiderString() -> (TabFlowRiderString:String)
   ) -> (then:Exec) {is=title K2Node_Message}
   > Delay(Duration=1)
   > Set IsImpressionAvailable(IsImpressionAvailable=true)
   Allow Toggle Preview Override Changed!() -> (OutputDelegate, then, Object:Object Reference, Field:Field Notification Id Structure) {is=title K2Node_CustomEvent}
   n121 = Get FortItemPreviewWidget()
   > SetVisibility(
    self=Get TogglePreview() -> (TogglePreview:WBP UIKit Button Quiet Object Reference)
    InVisibility="To Visibility (Boolean)"(
     bIsVisible=AllowToggleVariantOverrideOnCurrentItem!(self:Fort Item Preview Widget Object Reference=@n121) -> (ReturnValue:bool)
     TrueVisibility=SelfHitTestInvisible
     FalseVisibility=Collapsed
    ) -> (ReturnValue=Visible)
   )
   n122 = OnVariantSelectorDeactivated() {is=title K2Node_CustomEvent}
   > DeactivateWidget(self=Get FortItemPreviewWidget())
   > OnWidgetActivatedEvent()
   > RefreshPreviewItem(self:Fort Item Selection VM Object Reference=Get FortItemSelectionVM())
   n123 = OnOptionButtonClicked_Touch!() -> (OutputDelegate, then, Button:Common Button Base Object Reference) {is=title K2Node_CustomEvent}
   > SetupPreviewActionBinds(self=Get LockerTileWidget(), PreviewWidget=Get FortItemPreviewWidget())
   n124 = SetListOfItems!() -> (OutputDelegate, then, Items:Array of Fort Item VM Object References) {is=title K2Node_CustomEvent}
   n125 = Get FortItemSelectionVM()
   > n126 = Is Valid(InputObject=@n125)
   n127 = Get SelectedCategory(self=@n125)
   n128 = Get CustomItem!(self:Fort Locker Category Config Object Reference=GetConfig!(self:Fort Locker Category VM Object Reference=Get Category!(self:Fort Locker Category Item VM Object Reference=@n127) -> (Category:Fort Locker Category VM Object Reference)) -> (ReturnValue:Fort Locker Category Config Object Reference)) -> (CustomItem:User Widget Soft Class Reference)
   > n129 = Branch(Condition=IsValidSoftClassReference(SoftClassReference=@n128))
   > Set ClassToCreate(ClassToCreate:User Widget Soft Class Reference=@n128) -> (Output_Get:User Widget Soft Class Reference)
   n130 = Get LockerTileWidget()
   n131 = Get Class(Object=@n130) -> (ReturnValue:User Widget Class Reference)
   n132 = Get ClassToCreate() -> (ClassToCreate:User Widget Soft Class Reference)
   > n133 = Branch(Condition=AND Boolean(
     A=IsValid(Object=@n130)
     B="Equal (SoftClassReference)"(A:User Widget Class Reference=@n131, B:"Object Soft Class Reference (by ref)"=@n132) -> (ReturnValue:bool=~) {is=title K2Node_PromotableOperator}
    ))
   Set LastItemIndex(
    execute=[@n149.then, @n133.then]
    LastItemIndex:int=Last Index!(TargetArray:Array of Fort Item VM Object References=@n124.Items) -> (ReturnValue=0)
   ) -> (Output_Get:int=0)
   n134 = Get LockerTileList()
   > Set List Items!(execute, self:List View Object Reference=@n134, InListItems:Array of Object References=@n124.Items) -> (then) {is=title K2Node_CallFunction}
   > Clear Selection(self=@n134)
   > n135 = Is Valid(InputObject=Get Category(self:Fort Locker Category Filter VM Object Reference=Get FortLockerCategoryFilterVM()))
   n136 = UpdateNavigationIfEmpty!(execute=[@n141.then, @n135.Is Not Valid, @n140.then], self) -> (then, IsTileListEmpty:bool)
   > n137 = Branch(Condition=@n136)
   > n138 = UpdateEditButtonEnabled(IsEnabled=false)
   Branch(
    execute=[@n138.then, @n137.else]
    Condition="Equal (SoftClassReference)"(
     A:"Object Soft Class Reference (by ref)"=Get Default Class to Create() -> (Default Class to Create:User Widget Soft Class Reference)
     B=Get ClassToCreate()
    ) -> (ReturnValue=~)
   )
   > CompactUI_PostListLoad() {is=title K2Node_CallFunction}
   Create Empty Warning when No Filtered Itens(execute=@n135.Is Valid) {is=title K2Node_CallFunction}
   n139 = Get Block_EditableText(self=Get InputField(self=Get FilteredInputField()))
   > n140 = Branch(Condition=HasAnyUserFocus(self=Get EditableText(self=@n139)))
   n141 = Navigate To Item(execute=@n140.else, self=Get LockerTileList(), Item=Get FirstValidOrEquippedItem(self=Get FortLockerCategoryFilterVM()))
   ListUnbinds(execute=[@n133.else, @n150.then])
   > ClearChildren(self=Get LockerList_Container() -> (LockerList_Container:Named Slot Object Reference))
   > n142 = LoadClassAsset_Blocking(AssetClass=Get ClassToCreate())
   > n143 = Cast To UserWidget Class!(execute:Exec, Class=@n142) -> (then:Exec, CastFailed:Exec, AsUser Widget)
   > n144 = Create Widget!(execute, Class=@n143, OwningPlayer=GetOwningPlayer()) -> (then, ReturnValue:User Widget Object Reference)
   > n145 = Set LockerTileWidget(LockerTileWidget:User Widget Object Reference=@n144) -> (Output_Get:User Widget Object Reference)
   > AddChild(self=Get LockerList_Container(), Content=@n145)
   > n146 = Cast To BPI_LockerList_Type!(execute, Object=@n145) -> (then, CastFailed, AsBPI Locker List Type:BPI Locker List Type Interface) {is=title K2Node_DynamicCast}
   > n147 = GetListView!(execute, self:BPI Locker List Type Interface=@n146) -> (then, TileView:Fort Tile View Object Reference)
   > n148 = Set LockerTileList(LockerTileList:Fort Tile View Object Reference=@n147) -> (Output_Get:Fort Tile View Object Reference)
   > SetList!(
    execute
    self:WBP MPLocker Back To Top Button Object Reference=Get WBP_MPLocker_BackToTopButton() -> (WBP_MPLocker_BackToTopButton:WBP MPLocker Back To Top Button Object Reference)
    List:List View Object Reference=@n148
   ) -> (then)
   > SetList(
    self=Get WBP_MPLocker_BackToTopButton_Touch() -> (WBP_MPLocker_BackToTopButton_Touch:WBP MPLocker Back To Top Button Object Reference)
    List=@n148
   )
   > Set NewListBackToTop(NewListBackToTop:bool=true) -> (Output_Get:bool)
   > ListBinds()
   > n149 = UpdateIsEmoteWheel!(execute:Exec, self:Object Reference=Get LockerTileList(), IsEmoteWheel:bool=Get IsLoadedAsEmoteWheel()) -> (then:Exec) {is=title K2Node_Message}
   n150 = Set ClassToCreate(execute=[@n129.else, @n126.Is Not Valid], ClassToCreate=Get Default Class to Create())
   OnStreamingContent() {is=title K2Node_CustomEvent}
   > CreateCosmeticDownloadPopup!(
    execute
    ContainerWidget:User Widget Object Reference=Self-Reference()
    FortDownloadSystemVM:Fort Download System VM Object Reference=Get FortDownloadSystemVM() -> (FortDownloadSystemVM:Fort Download System VM Object Reference)
   ) -> (then)
   OnCMEditStylesClicked() {is=title K2Node_CustomEvent}
   > SetAllInteractionsEnabled(IsInteractionEnabled=false)
   > Set bIsEnabled!(execute, bIsEnabled=false) -> (then, Output_Get)
   > n151 = SetVisibility(self=Get TileView_Hitbox(), InVisibility=HitTestInvisible)
   CloseItemSelection() {is=title K2Node_CustomEvent}
   n152 = Get FortItemSelectionVM()
   > n153 = Branch(Condition=GetIsEquippable(self=Get SelectedCategory(self=@n152)))
   OnContextMenuCardEditStylesClicked_Event() {is=title K2Node_CustomEvent}
   > ActivateWidget(self=@n53)
   > n154 = Assign On Widget Deactivated!(execute:Exec, self:Common Activatable Widget Object Reference=@n53, Delegate:"Delegate (by ref)"=@n122) -> (then:Exec) {is=title K2Node_AssignDelegate}
   OnWidgetActivatedEvent() {is=title K2Node_CustomEvent}
   > Set bIsEnabled(bIsEnabled=true)
   > n155 = ActivateWidget(self=@n6)
   n156 = ShowItemDetails!() -> (OutputDelegate, then, CustomItemDetails:User Widget Class Reference) {is=title K2Node_CustomEvent}
   > Set CustomItemDetailsClass(CustomItemDetailsClass:User Widget Class Reference=@n156.CustomItemDetails) -> (Output_Get:User Widget Class Reference)
   > Set bIsEnabled(bIsEnabled=false)
   > Set LastActiveScreen(LastActiveScreen=NewEnumerator4) -> (Output_Get=NewEnumerator0)
   n157 = Get SubMenu()
   > Branch(Condition=NOT Boolean(A=IsAnimationPlaying(InAnimation=@n157)))
   > n158 = "Play Animation with Finished event (legacy)"(Widget=Self-Reference(), InAnimation=@n157, StartAtTime=0, NumLoopsToPlay=1, PlaybackSpeed=1)
   n159 = PushPrimaryContentClassAsync(execute=@n158.Finished, WidgetClass=&/WBP_MPLocker_ItemDetails.WBP_MPLocker_ItemDetails_C)
   n160 = Cast To WBP_MPLocker_ItemDetails!(execute=@n159.OnPrimaryContentPushed, Object=@n159) -> (then, CastFailed, AsWBP MPLocker Item Details:WBP MPLocker Item Details Object Reference) {is=title K2Node_DynamicCast}
   > SetCustomItemDetails!(
    execute
    self:WBP MPLocker Item Details Object Reference=@n160
    CustomItemDetails:User Widget Class Reference=Get CustomItemDetailsClass() -> (CustomItemDetailsClass:User Widget Class Reference)
   ) -> (then)
   > Unbind Event from On Widget Deactivated(self:Common Activatable Widget Object Reference=@n160, Delegate=@n95) {is=title K2Node_RemoveDelegate}
   > Bind Event to On Widget Deactivated(self=@n160, Delegate=@n95)
   n161 = UpdateInputPadding!() -> (OutputDelegate, then, Selection:ECommonInputType Enum) {is=title K2Node_CustomEvent}
   > n162 = Switch on ECommonInputType!(execute, Selection:ECommonInputType Enum=@n161.Selection) -> (MouseAndKeyboard:Exec, Gamepad:Exec, Touch:Exec, Count:Exec) {is=title K2Node_SwitchEnum}
   n163 = Get EditableTextPaddingOnKeyboard() -> (EditableTextPaddingOnKeyboard:Margin Structure)
   n164 = Get FilteredInputField()
   SetGridSlotPadding!(execute=@n162.Gamepad, Widget=Get Block_EditableText(self=Get InputField(self=@n164)), Padding:Margin Structure=@n163) -> (then)
   n165 = Get FilteredInputField()
   SetGridSlotPadding(execute=@n162.MouseAndKeyboard, Widget=Get Block_EditableText(self=Get InputField(self=@n165)), Padding=@n163)
   Event On Deactivated()
   > Fire Event Flush Impressions!(execute, FortPC:Fort Player Controller Object Reference) -> (then)
   > n166 = Branch(Condition=Get IsContextMenuOpen() -> (IsContextMenuOpen:bool))
   > n167 = ContextMenuClosed() {is=title K2Node_CallFunction}
   Event OnInitialized()
   > Bind Event to On Clicked(self:Common Button Base Object Reference=Get WBP_UIKit_Button_Options(), Delegate=@n123) {is=title K2Node_AddDelegate}
   > SetCompactUIStyle() {is=title K2Node_CallFunction}
   n168 = Get InputField(self=Get FilteredInputField())
   > Set ShouldClearKeyboardFocusOnCommit!(execute, ShouldClearKeyboardFocusOnCommit:bool=false, self:WBP UIKit Input Field Base Object Reference=@n168) -> (then, Output_Get:bool)
   > Bind Event to On Clicked(self=@n168, Delegate=Create Event())
   > Set bPreviewActionsEnabled!(execute, bPreviewActionsEnabled:bool=true, self:Fort Item Preview Screen Object Reference=Get FortItemPreviewWidget()) -> (then, Output_Get:bool)
   > Branch(Condition=GetConfigDataBoolWithDefault!(EntryName:String=EnableCosmeticDownloadNotificationsInLocker, bDefaultValue:bool=false) -> (ReturnValue:bool))
   > n169 = CreateCosmeticDownloadWidgets!(
    execute
    ContainerWidget:Panel Widget Object Reference=Get NotificationVerticalBox() -> (NotificationVerticalBox:Vertical Box Object Reference)
    HorizontalAlignment:EHorizontalAlignment Enum=HAlign_Left
   ) -> (then, CosmeticDownloadWidget:WBP Cosmetic Download Object Reference, CosmeticDownloadInfoWidget:WBP Cosmetic Download Info Object Reference)
   > Set CosmeticDownloadWidget(CosmeticDownloadWidget:WBP Cosmetic Download Object Reference=@n169) -> (Output_Get:WBP Cosmetic Download Object Reference)
   > Set CosmeticDownloadInfoWidget(CosmeticDownloadInfoWidget:WBP Cosmetic Download Info Object Reference=@n169.CosmeticDownloadInfoWidget) -> (Output_Get:WBP Cosmetic Download Info Object Reference)
   Event Construct()
   > Parent: Construct()
   > ToggleItemDescription(show=true) {is=title K2Node_CallFunction}
   n170 = Create Event()
   n171 = Get CommonInputSubsystem()
   > Bind Event to On Input Method Changed(self=@n171, Delegate=@n170)
   > Handle Input Changed(NewInputType=GetCurrentInputType(self=@n171))
   n172 = Create Event()
   n173 = Get FortItemPreviewWidget()
   > Add Field Value Changed Delegate!(
    execute
    self:Widget Object Reference=@n173
    FieldId:Field Notification Id Structure=(FieldName="AllowToggleVariantOverrideOnCurrentItem")
    Delegate:Delegate=@n172
   ) -> (then)
   n174 = Create Event()
   n175 = Get LockerTileList()
   > Bind Event to On Item Clicked(self:List View Object Reference=@n175, Delegate=@n174) {is=title K2Node_AddDelegate}
   n176 = Create Event()
   n177 = Get FilteredInputField()
   > Bind Event to On Search Added to Focus Path(self:WBP UIKit Filtered Input Field Object Reference=@n177, Delegate=@n176) {is=title K2Node_AddDelegate}
   > n178 = SetupTransitionsBindings!(execute, self, Bind:bool=true) -> (then)
   Event On Activated()
   > Fire Event Main View Opened!(execute, View:Widget Object Reference=Self-Reference(), AdditionalInfo:Flow Analytics Additional Info Structure, bIsOverlay:bool=false) -> (then)
   n179 = Get FortItemSelectionVM()
   > UpdateFlowriderSubView(SubViewString=GetAnalyticsName(self=Get SelectedCategory(self=@n179)))
   > SetAllInteractionsEnabled(IsInteractionEnabled=true)
   > OnWidgetActivatedEvent() {is=title K2Node_CallFunction}
   n180 = Get FortSortFilterVM()
   > SetInputFieldText!(
    execute
    self:WBP UIKit Filtered Input Field Object Reference=Get FilteredInputField()
    Text:Text=Get SearchText(self:Fort Sort Filter Search Text VM Object Reference=Get SearchText(self:Fort Sort Filter VM Object Reference=@n180) -> (SearchText:Fort Sort Filter Search Text VM Object Reference)) -> (SearchText:Text)
   ) -> (then)
   n181 = Get Selected Item(self=Get LockerTileList())
   > UpdateEquipButtonState(SelectedItem=@n181)
   > UpdateDetailsButtonState(Selected item=@n181)
   > AdjustVisualsForEmoteWheel() {is=title K2Node_CallFunction}
   > Broadcast Event(
    self=GetGlobalEventRouter()
    EventData=Make BPS Set Text Chat Active State(bSetActive_5_#18=False)
    EventContext=GetOwningPlayer()
   )
   > ToggleBottomBarButtonsForFilteredInputField(IsSearchHovered=false)
   > SetVisibility(self=Get TileView_Hitbox(), InVisibility=SelfHitTestInvisible)
   > RefreshPreviewItem(self=Get FortItemSelectionVM())
   Event Destruct()
   > Clear and Invalidate Timer by Handle(Handle=Get ItemDescriptionHideTimer())
   > ToggleItemDescription(show=true)
   > Unbind Event from On Input Method Changed(self=@n171, Delegate=@n170)
   > Remove Field Value Changed Delegate!(execute, self:Widget Object Reference=@n173, FieldId:Field Notification Id Structure, Delegate:Delegate=@n172) -> (then)
   > Unbind Event from On Item Clicked(self:List View Object Reference=@n175, Delegate=@n174) {is=title K2Node_RemoveDelegate}
   > Unbind Event from On Search Added to Focus Path(self:WBP UIKit Filtered Input Field Object Reference=@n177, Delegate=@n176) {is=title K2Node_RemoveDelegate}
   > SetupTransitionsBindings(Bind=false)
   > StopListeningToEvent(self:Async Action Start Listening To Event Object Reference=Get ChatEventWatch() -> (ChatEventWatch:Async Action Start Listening To Event Object Reference))
   > Parent: Destruct()
   CreateWidgetAsync!(
    execute:Exec
    UserWidgetSoftClass:User Widget Soft Class Reference
    OwningPlayer:Player Controller Object Reference
    bSuspendInputUntilComplete:bool=false
   ) -> (then:Exec, OnComplete:Exec, OnFailure:Exec, UserWidget:User Widget Object Reference) {is=title K2Node_AsyncAction}
   ToggleItemDescription(execute=@n186.else, show=true)
   SetAllInteractionsEnabled(execute=@n184.then, IsInteractionEnabled=true)
   n182 = Get EquippedItem(self=Get CategoryItem(self=Get FortLockerCategoryFilterVM()))
   Navigate To Item(execute=@n191.Is Valid, self=Get LockerTileList(), Item=@n182)
   Reroute Node(InputPin:WBP MPLocker Sort Filter Modal Object Reference=@n62) -> (OutputPin:WBP MPLocker Sort Filter Modal Object Reference)
   Reroute Node(InputPin:Object Reference=@n115.Item) -> (OutputPin:Object Reference)
   Set LastActiveScreen(execute=@n195.then, LastActiveScreen=NewEnumerator3) -> (Output_Get=NewEnumerator0)
   n183 = Get EquippedItem(self=Get CategoryItem(self=Get FortLockerCategoryFilterVM()))
   Navigate To Item(execute=@n190.then, self=Get LockerTileList(), Item=@n183)
   > ShowCustomizationISB!(execute, self, ShowISB:bool=Get bRequiresCustomization(self=@n183)) -> (then)
   > RequestRefreshFocus(self:Common Activatable Widget Object Reference)
   > n184 = Set bIsEnabled!(execute, bIsEnabled=true, self:Widget Object Reference=Get WBP_UIKit_Button_Options()) -> (then, Output_Get)
   n185 = IsVisible(self=Get ItemDetails_VerticalBox())
   Clear and Invalidate Timer by Handle(execute=@n82.then, Handle=Get ItemDescriptionHideTimer())
   > n186 = Branch(Condition=@n185)
   n187 = Cast To AthenaLockerSearchTextVM!(execute=[@n194.then, @n193.else], Object=Get SearchText(self=Get FortSortFilterVM())) -> (then, CastFailed, AsAthena Locker Search Text VM:Athena Locker Search Text VM Object Reference) {is=title K2Node_DynamicCast}
   > n188 = UpdateCacheForRenamedItem!(
    execute
    self:Athena Locker Search Text VM Object Reference=@n187
    Item:Fort Item VM Object Reference=Get SelectedItem(self=Get FortItemSelectionVM())
   ) -> (then)
   Set LockerTileWidget(execute=[@n188.then, @n187.CastFailed])
   > n189 = Set Filtered Items(Filtered Items:Array of Fort Item VM Object References=Get FilteredItems(self=Get FortLockerCategoryFilterVM())) -> (Output_Get:Array of Fort Item VM Object References)
   > n190 = SetListOfItems!(execute, self, Items:Array of Fort Item VM Object References=@n189) -> (then) {is=title K2Node_CallFunction}
   n191 = Is Valid(exec=@n192.then, InputObject=@n182)
   n192 = Unbind Event from Sort Filter Accepted(
    execute=@n63.then
    self:WBP MPLocker Sort Filter Modal Object Reference=Get SortFilter Modal() -> (SortFilter Modal:WBP MPLocker Sort Filter Modal Object Reference)
    Delegate=@n63
   ) {is=title K2Node_RemoveDelegate}
   n193 = Get!(execute=@n104.then) -> (then, else, CustomizeMenu:Common Activatable Widget Object Reference)
   > n194 = Unbind Event from On Widget Deactivated(self=@n193, Delegate=@n104)
   n195 = Set bIsEnabled!(execute=@n65.then, bIsEnabled=true) -> (then, Output_Get)
  ]}
 FocusSearchField {nodes [
   FocusSearchField() {is=title K2Node_FunctionEntry}
   > CenterWidget!(execute, ToCenter:Widget Object Reference=Get FilteredInputField()) -> (then)
   n1 = IsConsolePlatform!() -> (ReturnValue:bool)
   n2 = IsMobilePlatform!() -> (ReturnValue:bool)
   > n3 = Branch(Condition=NOT Boolean(A=OR Boolean(A=@n1, B=@n2)))
   Return Node!(execute=[@n5.then, @n3.then], bPassThrough:bool=false)
   n4 = Get InputField(self=Get FilteredInputField())
   n5 = ToggleVirtualKeyboard(execute=@n3.else, self=Get EditableText(self=Get Block_EditableText(self=@n4)), bShow=true)
  ]}
 GetCategoryConfig {nodes [
   GetCategoryConfig() {is=title K2Node_FunctionEntry}
   n1 = Get FortLockerCategoryFilterVM()
   n2 = Get Category(self=@n1)
   > n3 = Branch(Condition=AND Boolean(A=IsValid(Object=@n1), B=IsValid(Object=@n2)))
   > Return Node!(execute, Category Config:Fort Locker Category Config Object Reference=GetConfig(self=@n2))
   Return Node(execute=@n3.else)
  ]}
 GetCategoryEquipVisibility {nodes [
   n1 = GetCategoryEquipVisibility!() -> (then, Category Item:Fort Locker Category Item VM Object Reference) {is=title K2Node_FunctionEntry}
   > Is Valid(InputObject=@n1)
   > Return Node!(
    execute
    ReturnValue:ESlateVisibility Enum="To Visibility (Boolean)"(bIsVisible=GetIsEquippable(self=@n1), TrueVisibility=SelfHitTestInvisible, FalseVisibility=Collapsed) -> (ReturnValue=Visible)
   )
  ]}
 GetCategoryOptionsVisibility {nodes [
   n1 = GetCategoryOptionsVisibility!() -> (then, Category Item:Fort Locker Category Item VM Object Reference) {is=title K2Node_FunctionEntry}
   > Is Valid(InputObject=@n1)
   > Return Node(ReturnValue="To Visibility (Boolean)"(bIsVisible=GetIsEquippable(self=@n1), TrueVisibility=Visible, FalseVisibility=Collapsed) -> (ReturnValue=Visible))
  ]}
 GetEquippedInOtherCategoryMessage {nodes [
   GetEquippedInOtherCategoryMessage() {is=title K2Node_FunctionEntry}
   n1 = Get FortLockerEquipUtilityVM()
   > Return Node!(
    execute
    Result:Text=Format Text!(
     Format="LOCTEXT(\"This item is also equipped in {Category}.\")"
     Category:Text=GetItemEquippedCategoryNames!(self:Fort Locker Equip Utility VM Object Reference=@n1) -> (ReturnValue:Text)
    ) -> (Result)
   )
  ]}
 GetPrimaryContentSetup {nodes [
   GetPrimaryContentSetup() {is=title K2Node_FunctionEntry}
   > Return Node!(execute, ReturnValue:Primary Content Setup Structure=Get PrimaryContentSetup() -> (PrimaryContentSetup:Primary Content Setup Structure))
  ]}
 GetRestrictionMessage {nodes [
   n1 = GetRestrictionMessage!() -> (then, RestrictionMessage:Text, IsEquippedInOtherCategory:bool=~, Restriction:EItemAccessRestriction Enum=UnsatisfiedExclusiveItem) {is=title K2Node_FunctionEntry}
   n2 = "Equal (Enum)"(A:EItemAccessRestriction Enum=@n1.Restriction, B:EItemAccessRestriction Enum=Unknown)
   > Return Node!(
    execute
    Result:Text=SelectText(
     A=GetEquippedInOtherCategoryMessage!(self) -> (Result:Text) {is=title K2Node_CallFunction}
     B=@n1
     bPickA=AND Boolean(A=@n1.IsEquippedInOtherCategory, B=@n2)
    )
   )
  ]}
 GetRestrictionType {nodes [
   n1 = GetRestrictionType!() -> (then, IsEquippedInOtherCategory:bool=~, Restriction:EItemAccessRestriction Enum=UnsatisfiedExclusiveItem) {is=title K2Node_FunctionEntry}
   > Return Node(Result:EItemAccessRestriction Enum=@n1.Restriction)
  ]}
 GetWarning Icon {nodes [
   n1 = GetWarning Icon!() -> (then, HasAccessRestriction:bool=~, MoreDetailedFigureIncoming:bool=~, IsEquippedInOtherCategory:bool=~) {is=title K2Node_FunctionEntry}
   n2 = NOT Boolean(A=@n1)
   > n3 = Branch(Condition=AND Boolean(A=@n1.IsEquippedInOtherCategory, B=@n2))
   > Return Node!(execute, Output_Get:Material Interface Object Reference=/FortUILibrary/Textures/MaterialInstances/MI_UIKit_Icon_LobbyImage)
   n4 = Branch(execute=@n3.else, Condition=AND Boolean(A=@n2, B=@n1.MoreDetailedFigureIncoming))
   > Return Node(Output_Get=/Game/Athena/UI/Frontend/CosmeticGating/Materials/MI_ItemJunoStyle)
   Return Node(execute=@n4.else, Output_Get=/Game/Athena/UI/Frontend/CosmeticGating/Materials/MI_UI_CosmeticGate_WarningIcon)
  ]}
 Handle Input Changed {nodes [
   n1 = Handle Input Changed!() -> (then, NewInputType:ECommonInputType Enum) {is=title K2Node_FunctionEntry}
   > n2 = Branch(Condition="Equal (Enum)"(A:ECommonInputType Enum=@n1, B:ECommonInputType Enum=Gamepad))
   > n3 = Set ShowGamepadPadding(ShowGamepadPadding:bool=true) -> (Output_Get:bool)
   UpdateSortFilterPadding(execute=[@n3.then, @n9.then])
   n4 = Get FortItemSelectionVM()
   n5 = Get SelectedItem(self=@n4)
   n6 = Get EquippedItem(self=Get SelectedCategory(self=@n4))
   > UpdateEditButtonEnabled(IsEnabled=Is Edit Enabled(SelectedHasVariants=Get bHasVariants(self=@n5), EquippedHasVariants=Get bHasVariants(self=@n6)))
   > SetContextVisibility!(execute:Exec, self:Object Reference=Get LockerTileWidget()) -> (then:Exec) {is=title K2Node_Message}
   n7 = Get NewInputType() -> (NewInputType:ECommonInputType Enum)
   > UpdateInputPadding(Selection=@n7) {is=title K2Node_CallFunction}
   > ShowItemDescription() {is=title K2Node_CallFunction}
   n8 = "Equal (Enum)"(A=@n7, B=Gamepad)
   > SetSpecialPreviewActionAllowList!(
    execute
    self:Fort Item Preview Screen Object Reference=Get FortItemPreviewWidget()
    InSpecialPreviewActionAllowList:Set of Names=Make Set!("[0]":Name=Select(Option 0:Name=SpecialPreviewAction_ViewLoadingScreen, Option 1:Name=None, Index=@n8) -> (ReturnValue:Name)) -> (Set:Set of Names) {is=title K2Node_MakeSet}
   ) -> (then)
   > Branch(Condition="Equal (Enum)"(A=Get NewInputType(), B=Touch))
   > ToggleItemDescription(show=true)
   SetVisibility(execute=@n2.else, self=Get GamepadButtonBlocker() -> (GamepadButtonBlocker:Image Object Reference), InVisibility=Collapsed)
   > n9 = Set ShowGamepadPadding(ShowGamepadPadding=false)
  ]}
 HandleSubcategoryTabSelected {nodes [
   HandleSubcategoryTabSelected!() -> (then, TabNameID:Name) {is=title K2Node_FunctionEntry}
   n1 = Get FortItemSelectionVM()
   > n2 = For Each Loop with Break!(
    Exec:Exec
    Array:Array of Fort Locker Category Item VM Object References=Get CategoryItems!(self:Fort Locker Category Group VM Object Reference=Get SelectedCategoryGroup!(self:Fort Item Selection VM Object Reference=@n1) -> (SelectedCategoryGroup:Fort Locker Category Group VM Object Reference)) -> (CategoryItems:Array of Fort Locker Category Item VM Object References)
    Break:Exec=@n6.then
   ) -> (LoopBody:Exec, Array Element:Fort Locker Category Item VM Object Reference, Array Index:int, Completed:Exec)
   n3 = GetAnalyticsName(self=@n2)
   > Branch(Condition="Equal (Name)"(A=String To Name(InString=@n3), B=Get TabNameID() -> (TabNameID:Name=None)))
   n4 = Get FortItemSelectionVM()
   > n5 = Set SelectedCategory!(execute, SelectedCategory:Fort Locker Category Item VM Object Reference=@n2, self:Fort Item Selection VM Object Reference=@n4) -> (then, Output_Get:Fort Locker Category Item VM Object Reference)
   > n6 = Set HoveredCategory!(execute, HoveredCategory:Fort Locker Category Item VM Object Reference=@n5, self:Fort Item Selection VM Object Reference=@n4) -> (then, Output_Get:Fort Locker Category Item VM Object Reference)
  ]}
 Is Edit Enabled {nodes [
   n1 = Is Edit Enabled!() -> (then, SelectedHasVariants:bool=~, EquippedHasVariants:bool=~) {is=title K2Node_FunctionEntry}
   > Return Node!(execute, Result:bool=@n1)
  ]}
 IsCustomizable {nodes [
   IsCustomizable() {is=title K2Node_FunctionEntry}
   > n1 = Get!(execute) -> (then, else, FortItemSelectionVM:Fort Item Selection VM Object Reference)
   > n2 = Get!(execute, self:Fort Item Selection VM Object Reference=@n1) -> (then, else, SelectedItem:Fort Item VM Object Reference)
   > Return Node!(execute, Customizable:bool=RequiresCustomization!(self:Fort Item VM Object Reference=@n2) -> (ReturnValue:bool))
   Return Node(execute=[@n2.else, @n1.else], Customizable=false)
  ]}
 IsOptionsEnabled {nodes [
   n1 = IsOptionsEnabled!() -> (then, SelectedItem:Fort Item VM Object Reference) {is=title K2Node_FunctionEntry}
   > n2 = Is Valid(InputObject=@n1)
   n3 = Get bIsValid!(self:Fort Item VM Object Reference=@n1) -> (bIsValid:bool)
   > Return Node!(
    execute
    Result:bool=NOR Boolean!(A:bool=Get bIsShuffleTile!(self:Fort Item VM Object Reference=@n1) -> (bIsShuffleTile:bool), B:bool=NOT Boolean(A=@n3)) -> (ReturnValue:bool)
   )
   Return Node(execute=@n2.Is Not Valid, Result=false)
  ]}
 ListBinds {nodes [
   ListBinds() {is=title K2Node_FunctionEntry}
   n1 = Get LockerTileList()
   > Is Valid(InputObject=@n1)
   > Bind Event to On Item Clicked(self=@n1, Delegate=Create Event())
   > Bind Event to On Entry Initialized(self:List View Object Reference=@n1, Delegate=Create Event()) {is=title K2Node_AddDelegate}
   > Bind Event to On Entry Released(self:List View Base Object Reference=@n1, Delegate=Create Event()) {is=title K2Node_AddDelegate}
   > Bind Event to On Entry Generated(self:List View Base Object Reference=@n1, Delegate=Create Event()) {is=title K2Node_AddDelegate}
   > Bind Event to On Item Is Hovered Changed(self:List View Object Reference=@n1, Delegate=Create Event()) {is=title K2Node_AddDelegate}
   > Bind Event to On Item Selection Changed(self:List View Object Reference=@n1, Delegate=Create Event()) {is=title K2Node_AddDelegate}
   > n2 = Cast To WBP_LockerList_Base!(execute, Object=Get LockerTileWidget()) -> (then, CastFailed, AsWBP Locker List Base:WBP Locker List Base Object Reference) {is=title K2Node_DynamicCast}
   > Bind Event to On Tile Context Menu Open(self:WBP Locker List Base Object Reference=@n2, Delegate=Create Event()) {is=title K2Node_AddDelegate}
   > Bind Event to On Tile Context Menu Closed(self:WBP Locker List Base Object Reference=@n2, Delegate=Create Event()) {is=title K2Node_AddDelegate}
   > Bind Event to Go to Locker Item Details(self:WBP Locker List Base Object Reference=@n2, Delegate=Create Event()) {is=title K2Node_AddDelegate}
   > Bind Event to Context Edit Styles Clicked(self:WBP Locker List Base Object Reference=@n2, Delegate=Create Event()) {is=title K2Node_AddDelegate}
   > Bind Event to Close Item Selection(self:WBP Locker List Base Object Reference=@n2, Delegate=Create Event()) {is=title K2Node_AddDelegate}
   > Bind Event to On Modal Changed(self:WBP Locker List Base Object Reference=@n2, Delegate=Create Event()) {is=title K2Node_AddDelegate}
   > Bind Event to On Wait for Streaming Content(self:WBP Locker List Base Object Reference=@n2, Delegate=Create Event()) {is=title K2Node_AddDelegate}
  ]}
 ListUnbinds {nodes [
   ListUnbinds() {is=title K2Node_FunctionEntry}
   n1 = Get LockerTileList()
   > n2 = Is Valid(InputObject=@n1)
   > Unbind all Events from On Entry Initialized!(execute:Exec, self:List View Object Reference=@n1) -> (then:Exec) {is=title K2Node_ClearDelegate}
   > Unbind all Events from On Entry Released!(execute:Exec, self:List View Base Object Reference=@n1) -> (then:Exec) {is=title K2Node_ClearDelegate}
   > Unbind all Events from On List View Scrolled!(execute:Exec, self:List View Object Reference=@n1) -> (then:Exec) {is=title K2Node_ClearDelegate}
   > Unbind all Events from On Entry Generated!(execute:Exec, self:List View Base Object Reference=@n1) -> (then:Exec) {is=title K2Node_ClearDelegate}
   > Unbind Event from On Item Is Hovered Changed(self:List View Object Reference=@n1, Delegate=Create Event()) {is=title K2Node_RemoveDelegate}
   > n3 = Unbind Event from On Item Selection Changed(self:List View Object Reference=@n1, Delegate=Create Event()) {is=title K2Node_RemoveDelegate}
   n4 = Cast To WBP_LockerList_Base(execute=[@n3.then, @n2.Is Not Valid], Object=Get LockerTileWidget())
   > Unbind all Events from On Tile Context Menu Open!(execute:Exec, self:WBP Locker List Base Object Reference=@n4) -> (then:Exec) {is=title K2Node_ClearDelegate}
   > Unbind all Events from On Tile Context Menu Closed!(execute:Exec, self:WBP Locker List Base Object Reference=@n4) -> (then:Exec) {is=title K2Node_ClearDelegate}
   > Unbind all Events from Go to Locker Item Details!(execute:Exec, self:WBP Locker List Base Object Reference=@n4) -> (then:Exec) {is=title K2Node_ClearDelegate}
   > Unbind all Events from Context Edit Styles Clicked!(execute:Exec, self:WBP Locker List Base Object Reference=@n4) -> (then:Exec) {is=title K2Node_ClearDelegate}
   > Unbind all Events from Close Item Selection!(execute:Exec, self:WBP Locker List Base Object Reference=@n4) -> (then:Exec) {is=title K2Node_ClearDelegate}
   > Unbind Event from On Modal Changed(self:WBP Locker List Base Object Reference=@n4, Delegate=Create Event()) {is=title K2Node_RemoveDelegate}
   > Unbind Event from On Wait for Streaming Content(self:WBP Locker List Base Object Reference=@n4, Delegate=Create Event()) {is=title K2Node_RemoveDelegate}
  ]}
 OnCategoryChanged {nodes [
   OnCategoryChanged!() -> (then, InCategory:Fort Locker Category VM Object Reference) {is=title K2Node_FunctionEntry}
   > n1 = Sequence()
   n2 = Get InCategory() -> (InCategory:Fort Locker Category VM Object Reference)
   n3 = Get FortSortFilterVM()
   > Branch(Condition=AND Boolean(A=IsValid(Object=@n2), B=IsValid(Object=@n3)))
   > n4 = Set SearchText(SearchText=Get CurrentSearchText() -> (CurrentSearchText:Text), self=Get SearchText(self=Get FortSortFilterVM()))
   > SetInputFieldText(self=Get FilteredInputField(), Text=@n4)
   n5 = Get FortItemSelectionVM()
   Is Valid(exec=@n1.then_1, InputObject=@n5)
   > SetSlotIndexOverrideAndCurrentItem!(
    execute
    self:Fort Item Preview Widget Object Reference=Get FortItemPreviewWidget()
    InSlotIndex:int=Get SlotIndex(self=Get SelectedCategory(self=@n5)) -> (SlotIndex=0)
    InItemVM:Fort Item VM Object Reference=Get FirstValidOrEquippedItem(self=Get FortLockerCategoryFilterVM())
   ) -> (then)
  ]}
 OnCategoryTabSelected {nodes [OnCategoryTabSelected!() -> (then, SelectedTab:Name) {is=title K2Node_FunctionEntry}]}
 OnContextMenuCardEditStylesClicked {nodes [OnContextMenuCardEditStylesClicked() {is=title K2Node_FunctionEntry}]}
 OnSearchAddedToFocusPath {nodes [
   OnSearchAddedToFocusPath() {is=title K2Node_FunctionEntry}
   > SetVisibility(self=Get GamepadButtonBlocker(), InVisibility=Collapsed)
   > ToggleBottomBarButtonsForFilteredInputField(IsSearchHovered=true)
  ]}
 OnSortFilterModalOpen {nodes [OnSortFilterModalOpen() {is=title K2Node_FunctionEntry}]}
 OnTileSelectionChanged {nodes [
   n1 = OnTileSelectionChanged!() -> (then, Item:Object Reference, bIsSelected:bool=~) {is=title K2Node_FunctionEntry}
   n2 = GetPlayerController!(PlayerIndex:int=0) -> (ReturnValue:Player Controller Object Reference)
   > Branch(Condition=AND Boolean(A=@n1.bIsSelected, B=IsUsingTouch(PlayerController=@n2)))
   > SetupPreviewActionBinds(self=Get LockerTileWidget(), PreviewWidget=Get FortItemPreviewWidget())
  ]}
 OnTransitionAnimationFinished {nodes [
   OnTransitionAnimationFinished() {is=title K2Node_FunctionEntry}
   > SetAllInteractionsEnabled(IsInteractionEnabled=true)
  ]}
 OnTransitionAnimationStarted {nodes [
   OnTransitionAnimationStarted() {is=title K2Node_FunctionEntry}
   > SetAllInteractionsEnabled(IsInteractionEnabled=false)
  ]}
 RegisterSTWInput {nodes [
   RegisterSTWInput() {is=title K2Node_FunctionEntry}
   n1 = Get InputField(self=Get FilteredInputField())
   > RegisterInputAction!(
    execute
    self:Fort Global UIContext Object Reference=Get FortGlobalUIContext()
    UserWidget:User Widget Object Reference=@n1
    InputActionRow:Data Table Row Handle Structure=Get TriggeringInputAction!(self:Common Button Base Object Reference=@n1) -> (TriggeringInputAction:Data Table Row Handle Structure)
    CommitedEvent:Delegate=Create Event()
    InFilterPriority:int=0
   ) -> (then)
  ]}
 STWCarouselNextInput {nodes [
   STWCarouselNextInput() {is=title K2Node_FunctionEntry}
   > SelectNextTab!(execute, self:UIKit Tab Group Object Reference=Get Tabs_Subcategory(self=Get SubcategoryCarousel()), bSuppressClickFeedback:bool=true) -> (then)
   > Return Node!(execute, bPassThrough:bool=false)
  ]}
 STWCarouselPreviousInput {nodes [
   STWCarouselPreviousInput() {is=title K2Node_FunctionEntry}
   > SelectPreviousTab!(execute, self:UIKit Tab Group Object Reference=Get Tabs_Subcategory(self=Get SubcategoryCarousel()), bSuppressClickFeedback:bool=true) -> (then)
   > Return Node(bPassThrough=false)
  ]}
 Set Back To Top Options {nodes [
   Set Back To Top Options() {is=title K2Node_FunctionEntry}
   > Branch(Condition=Get NewListBackToTop() -> (NewListBackToTop:bool))
   > Set NewListBackToTop(NewListBackToTop=false)
   n1 = GetIndexForItem!(
    self:List View Object Reference=Get LockerTileList()
    Item:Object Reference=Get FirstValidItem!(self:Fort Locker Category Filter VM Object Reference=Get FortLockerCategoryFilterVM()) -> (FirstValidItem:Fort Item VM Object Reference)
   ) -> (ReturnValue:int=0)
   > Set TargetIndex!(execute, TargetIndex:int=@n1, self:WBP MPLocker Back To Top Button Object Reference=Get WBP_MPLocker_BackToTopButton()) -> (then, Output_Get:int=0)
   n2 = Get LockerTileWidget()
   n3 = BreakVector2D(InVec=GetLocalSize!(Geometry:"Geometry Structure (by ref)"=GetCachedGeometry!(self:Widget Object Reference=@n2) -> (ReturnValue:"Geometry Structure (by ref)")) -> (ReturnValue:Vector 2D Structure))
   n4 = GetEntryWidth!(self:Tile View Object Reference=Get LockerTileList()) -> (ReturnValue:float)
   n5 = int + int(A=Floor(A=float / float(A=@n3, B=@n4)) {is=title K2Node_CallFunction})
   > Set ScrollOffset!(
    execute
    ScrollOffset:double="To Float (Integer)"(InInt=@n5)
    self:WBP MPLocker Back To Top Button Object Reference=Get WBP_MPLocker_BackToTopButton()
   ) -> (then, Output_Get:double)
   > Set TargetIndex(TargetIndex=@n1, self=Get WBP_MPLocker_BackToTopButton_Touch()) -> (Output_Get=0)
   > Set ScrollOffset(ScrollOffset="To Float (Integer)"(InInt=@n5), self=Get WBP_MPLocker_BackToTopButton_Touch())
  ]}
 Set Warning Icon {nodes [
   n1 = Set Warning Icon!() -> (then, InWarningIcon:Material Interface Object Reference) {is=title K2Node_FunctionEntry}
   > SetIconImage!(execute, self:WBP UIKit Info Box Object Reference=Get AcessRestrictionWarning(), Material:Material Interface Object Reference=@n1) -> (then)
  ]}
 SetAllInteractionsEnabled {nodes [
   SetAllInteractionsEnabled!() -> (then, IsInteractionEnabled:bool=true) {is=title K2Node_FunctionEntry}
   > SetIsInteractionEnabled!(
    execute
    self:Common Button Base Object Reference=[
     Get Button_SortFilter() -> (Button_SortFilter:WBP UIKit Button Generic Object Reference)
     Get Button_ClearSearch() -> (Button_ClearSearch:WBP UIKit Button Quiet Object Reference)
     Get Button_EditStyles() -> (Button_EditStyles:WBP UIKit Button Quiet Object Reference)
     Get WBP_UIKit_Button_Options()
     Get Button_Use() -> (Button_Use:WBP UIKit Button Quiet Object Reference)
     Get WBP_UIKit_Button_Back() -> (WBP_UIKit_Button_Back:WBP UIKit Button Quiet Object Reference)
    ]
    bInIsInteractionEnabled:bool=Get IsInteractionEnabled() -> (IsInteractionEnabled:bool)
   ) -> (then)
   > Set IgnoreInputs(IgnoreInputs:bool=NOT Boolean(A=Get IsInteractionEnabled())) -> (Output_Get:bool)
  ]}
 SetCastedItemCardUnbindings {nodes [
   n1 = SetCastedItemCardUnbindings!() -> (then, EntryWidgetObjectReference:Widget Object Reference) {is=title K2Node_FunctionEntry}
   > n2 = Cast To WBP_LockerList_Base(Object=@n1)
   > Unbind all Events from On Tile Context Menu Open(self=@n2)
   > Unbind all Events from On Tile Context Menu Closed(self=@n2)
   > Unbind all Events from Go to Locker Item Details(self=@n2)
  ]}
 SetClosedContextMenuParentEntryOpacity {nodes [
   SetClosedContextMenuParentEntryOpacity() {is=title K2Node_FunctionEntry}
   > n1 = For Each Loop(Array:Array of User Widget Object References=GetDisplayedEntryWidgets!(self:List View Base Object Reference=Get LockerTileList()) -> (ReturnValue:Array of User Widget Object References)) -> (Array Element:User Widget Object Reference)
   > n2 = Set RenderOpacity!(
    execute
    RenderOpacity:float=Get ItemCardHighlightedRenderOpacity() -> (ItemCardHighlightedRenderOpacity:float)
    self:Widget Object Reference=@n1
   ) -> (then, Output_Get:float)
   > n3 = Set RenderOpacity(RenderOpacity=@n2, self=Get SortFilterBar() -> (SortFilterBar:Horizontal Box Object Reference))
   > Set RenderOpacity(RenderOpacity=@n3, self=Get SubcategoryCarousel())
  ]}
 SetCompactUIStyle {nodes [
   SetCompactUIStyle() {is=title K2Node_FunctionEntry}
   > Branch(Condition=IsMobileGame())
   n1 = Get SizeBox() -> (SizeBox:Size Box Object Reference)
   > ClearWidthOverride(self:Size Box Object Reference=@n1)
   > Set HAlign!(
    execute
    HAlign:EHorizontalAlignment Enum=HAlign_Fill
    self:Safe Zone Slot Object Reference=SlotAsSafeBoxSlot!(Widget:Widget Object Reference=@n1) -> (ReturnValue:Safe Zone Slot Object Reference)
   ) -> (then, Output_Get:EHorizontalAlignment Enum=HAlign_Fill)
   > SetPadding!(
    execute
    self:Grid Slot Object Reference=SlotAsGridSlot(Widget=Get LockerOverlay() -> (LockerOverlay:Overlay Object Reference))
    InPadding:Margin Structure=Make Margin(Left=0, Top=8, Right=0, Bottom=96)
   ) -> (then)
   > SetPadding(self=SlotAsGridSlot(Widget=Get Locker()), InPadding=Make Margin(Left=64, Top=48, Right=-16, Bottom=16))
   > SetSize!(
    execute
    self:Horizontal Box Slot Object Reference=SlotAsHorizontalBoxSlot!(Widget:Widget Object Reference=Get Camera_Spacer_Left() -> (Camera_Spacer_Left:Spacer Object Reference)) -> (ReturnValue:Horizontal Box Slot Object Reference)
    InSize:Slate Child Size Structure=Make Slate Child Size!(Value:float=0.24, SizeRule:ESlateSizeRule Enum=Fill) -> (SlateChildSize:Slate Child Size Structure) {is=title K2Node_MakeStruct}
   ) -> (then)
   > Set BottomSlotRow!(execute, BottomSlotRow:int=5, self:WBP UIKit Item Description Base Object Reference=Get WBP_UIKit_DescriptionWithSubtitle()) -> (then, Output_Get:int=0)
   > SetPadding(self:User Widget Object Reference=Get AcessRestrictionWarning())
   > SetPadding(self:Grid Slot Object Reference=SlotAsGridSlot(Widget=Get TileView_Hitbox()), InPadding=Make Margin(Left=0, Top=8, Right=0, Bottom=96))
   > SetSidesToPad!(
    execute
    self:Safe Zone Object Reference=Get SafeZone() -> (SafeZone:Safe Zone Object Reference)
    InPadLeft:bool=true
    InPadRight:bool=true
    InPadTop:bool=true
    InPadBottom:bool=true
   ) -> (then)
   > PrintString!(execute, InString=Mobile SET, bPrintToScreen, bPrintToLog, TextColor, Duration, Key) -> (then)
   n2 = Get SubcategoryCarousel()
   > SetPadding!(
    execute
    self:Vertical Box Slot Object Reference=SlotAsVerticalBoxSlot(Widget=Get WBP_MPLocker_Heading!(self:WBP Subcategory Carousel Object Reference=@n2) -> (WBP_MPLocker_Heading:WBP MPLocker Heading Object Reference))
    InPadding_Left:float=32
    InPadding_Top:float=-22
    InPadding_Right:float=0
    InPadding_Bottom:float=36
   ) -> (then)
   n3 = SlotAsGridSlot(Widget=Get ItemDetails_VerticalBox())
   n4 = Get!(self:Grid Slot Object Reference=@n3) -> (Padding_Left:float=~, Padding_Top:float=~, Padding_Right:float=~, Padding_Bottom:float=~)
   > SetPadding(
    self:Grid Slot Object Reference=@n3
    InPadding_Left=@n4
    InPadding_Top=@n4.Padding_Top
    InPadding_Right=@n4.Padding_Right
    InPadding_Bottom=104
   )
  ]}
 SetIsEmoteWheel {nodes [
   n1 = SetIsEmoteWheel!() -> (then, IsEmoteWheel:bool=~) {is=title K2Node_FunctionEntry}
   > Set IsLoadedAsEmoteWheel(IsLoadedAsEmoteWheel:bool=@n1) -> (Output_Get:bool)
  ]}
 SetOpenedContextMenuParentEntryOpacity {nodes [
   n1 = SetOpenedContextMenuParentEntryOpacity!() -> (then, ContextMenuParentEntryWidget:Widget Object Reference) {is=title K2Node_FunctionEntry}
   > n2 = For Each Loop(Array=GetDisplayedEntryWidgets(self=Get LockerTileList()))
   > n3 = Branch(Condition="Not Equal (Object)"(A:Object Reference=@n2, B:Object Reference=@n1) -> (ReturnValue:bool=~) {is=title K2Node_PromotableOperator})
   > n4 = Set RenderOpacity(RenderOpacity=Get ItemCardUnderlayingRenderOpacity() -> (ItemCardUnderlayingRenderOpacity:float), self=@n2)
   > n5 = Set RenderOpacity(RenderOpacity=@n4, self=Get SortFilterBar())
   > Set RenderOpacity(RenderOpacity=@n5, self=Get SubcategoryCarousel())
   Set RenderOpacity(execute=@n3.else, RenderOpacity=Get ItemCardHighlightedRenderOpacity(), self=@n2)
  ]}
 SetupTransitionsBindings {nodes [
   n1 = SetupTransitionsBindings!() -> (then, Bind:bool=~) {is=title K2Node_FunctionEntry}
   > n2 = Branch(Condition=@n1)
   n3 = Create Event()
   n4 = Get Intro()
   > BindToAnimationStarted!(execute, self:User Widget Object Reference, Animation:Widget Animation Object Reference=@n4, Delegate:Delegate=@n3) -> (then)
   n5 = Create Event()
   n6 = Get SubMenu()
   > BindToAnimationStarted(Animation=@n6, Delegate=@n5)
   n7 = Create Event()
   n8 = Get Intro()
   > BindToAnimationFinished(Animation=@n8, Delegate=@n7)
   n9 = Create Event()
   n10 = Get SubMenu()
   > BindToAnimationFinished(Animation=@n10, Delegate=@n9)
   UnbindFromAnimationStarted!(execute=@n2.else, self:User Widget Object Reference, Animation:Widget Animation Object Reference=@n4, Delegate:Delegate=@n3) -> (then)
   > UnbindFromAnimationStarted(Animation=@n6, Delegate=@n5)
   > UnbindFromAnimationFinished!(execute, self:User Widget Object Reference, Animation:Widget Animation Object Reference=@n8, Delegate:Delegate=@n7) -> (then)
   > UnbindFromAnimationFinished(Animation=@n10, Delegate=@n9)
  ]}
 ShowCustomizationISB {nodes [
   n1 = ShowCustomizationISB!() -> (then, ShowISB:bool=~) {is=title K2Node_FunctionEntry}
   > n2 = Branch(Condition=@n1)
   > n3 = Get!(execute:Exec) -> (then:Exec, else:Exec, MimosaISB:Common User Widget Object Reference)
   n4 = Get!(execute=@n3.else, self:Fort Item Selection VM Object Reference=Get FortItemSelectionVM()) -> (then, else, SelectedItem:Fort Item VM Object Reference)
   > n5 = LoadClassAsset_Blocking(AssetClass=TryGetSoftCustomWidgetClass!(
     self:Fort Locker Category Config Object Reference=GetItemCategory(self=@n4)
     ViewRequest:Gameplay Tag Structure=(TagName="UI.ItemPreview.PluginBridge.Mimosa.ShopISB")
    ) -> (WidgetClass:Common User Widget Soft Class Reference, ReturnValue:bool))
   > n6 = Cast To CommonUserWidget Class!(execute:Exec, Class:Object Class Reference=@n5) -> (then:Exec, CastFailed:Exec, AsCommon User Widget:Common User Widget Class Reference) {is=title K2Node_ClassDynamicCast}
   > n7 = Create Widget(Class=@n6, OwningPlayer=GetOwningPlayer()) -> (ReturnValue:Common User Widget Object Reference)
   > n8 = Set MimosaISB(MimosaISB:Common User Widget Object Reference=@n7) -> (Output_Get:Common User Widget Object Reference)
   > AddChild(self=Get ItemDetails_VerticalBox(), Content=@n8)
   > n9 = Cast To WBP_CustomBanner_Base!(execute, Object=@n7) -> (then, CastFailed, AsWBP Custom Banner Base:WBP Custom Banner Base Object Reference) {is=title K2Node_DynamicCast}
   > SetContext!(execute, self:WBP Custom Banner Base Object Reference=@n9, Desired Area:E_MultiproductContext Enum=NewEnumerator1) -> (then)
   n10 = Get!(execute=[@n4.else, @n2.else]) -> (then, else, MimosaISB:Common User Widget Object Reference)
   > RemoveWidget!(execute, Widget:Widget Object Reference=@n10) -> (then)
   > Set MimosaISB()
  ]}
 ShowMatchedTags {nodes [
   n1 = ShowMatchedTags!() -> (then, ItemVM:Fort Item VM Object Reference, ItemCard:WBP UIKit Item Card Locker Object Reference) {is=title K2Node_FunctionEntry}
   n2 = GetMatchedTags!(self:Fort Sort Filter Search Text VM Object Reference=Get SearchText(self=Get FortSortFilterVM()), ItemVM:Fort Item VM Object Reference=@n1) -> (ReturnValue:Text)
   > n3 = Branch(Condition=TextIsEmpty(InText=@n2))
   Set ToolTipText!(
    execute=@n3.else
    ToolTipText:Text=@n2
    self:Widget Object Reference=Get ItemCard() -> (ItemCard:WBP UIKit Item Card Locker Object Reference)
   ) -> (then, Output_Get:Text)
  ]}
 TileHoverChanged {nodes [
   n1 = TileHoverChanged!() -> (then, Item:Object Reference, bIsHovered:bool=~) {is=title K2Node_FunctionEntry}
   > Branch(Condition=@n1.bIsHovered)
   > ShowItemDescription()
   n2 = Get CommonInputSubsystem()
   > Branch(Condition="Equal (Enum)"(A=GetCurrentInputType(self=@n2), B=Gamepad))
   n3 = GetListItems!(self:List View Object Reference=Get LockerTileList()) -> (ReturnValue:Array of Object References)
   n4 = Get SelectedItem(self=Get FortItemSelectionVM())
   n5 = Get LockerTileList()
   > n6 = Branch(Condition="integer < integer"(
     A=Find Item!(TargetArray:Array of Object References=@n3, ItemToFind:"Object Reference (by ref)"=@n4) -> (ReturnValue)
     B=GetNumberItemsPerRow!(self:Fort Tile View Object Reference=@n5) -> (ReturnValue:int=0)
    ))
   n7 = Get GamepadButtonBlocker()
   > SetVisibility(self=@n7, InVisibility=Collapsed)
   SetVisibility(execute=@n6.else, self=@n7, InVisibility=Visible)
  ]}
 ToggleBottomBarButtonsForFilteredInputField {nodes [
   ToggleBottomBarButtonsForFilteredInputField!() -> (then, IsSearchHovered:bool=~) {is=title K2Node_FunctionEntry}
   n1 = NOT Boolean(A=Get IsSearchHovered() -> (IsSearchHovered:bool))
   n2 = Get WBP_UIKit_Button_Options()
   > Set bIsEnabled!(execute, bIsEnabled=@n1, self:Widget Object Reference=@n2) -> (then, Output_Get)
   > SetVisibility(
    self=@n2
    InVisibility=Select(Option 0:ESlateVisibility Enum=Collapsed, Option 1:ESlateVisibility Enum=Visible, Index=@n1) -> (ReturnValue:ESlateVisibility Enum)
   )
   n3 = Get Button_ClearSearch()
   n4 = Get IsSearchHovered()
   > Set bIsEnabled(bIsEnabled=@n4, self=@n3)
   > SetVisibility(self=@n3, InVisibility=Select(Option 0=Collapsed, Option 1=Visible, Index=@n4))
   n5 = Get FortItemSelectionVM()
   n6 = Get SelectedItem(self=@n5)
   n7 = Get EquippedItem(self=Get SelectedCategory(self=@n5))
   n8 = Get Button_EditStyles()
   > Set bIsEnabled(
    bIsEnabled=Is Edit Enabled(SelectedHasVariants=Get bHasVariants(self=@n6), EquippedHasVariants=Get bHasVariants(self=@n7)) {is=title K2Node_CallFunction}
    self=@n8
   )
   n9 = Get IsSearchHovered()
   > SetVisibility(self=@n8, InVisibility=Select(Option 0=Collapsed, Option 1=Visible, Index=NOT Boolean(A=@n9)))
   n10 = Get IsSearchHovered()
   n11 = Get Button_Use()
   > Set bIsEnabled(bIsEnabled=@n10, self=@n11)
   > SetVisibility(self=@n11, InVisibility=Select(Option 0=Collapsed, Option 1=Visible, Index=@n10))
  ]}
 UnregisterSTWInput {nodes [
   UnregisterSTWInput() {is=title K2Node_FunctionEntry}
   n1 = Get InputField(self=Get FilteredInputField())
   > UnregisterInputAction!(
    execute
    self:Fort Global UIContext Object Reference=Get FortGlobalUIContext()
    UserWidget:User Widget Object Reference=@n1
    InputActionRow:Data Table Row Handle Structure=Get TriggeringInputAction(self=@n1)
   ) -> (then)
  ]}
 Update Edit Visibility {nodes [
   n1 = Update Edit Visibility!() -> (then, isEquippable:bool=~, AreVariantsSupported:bool=~) {is=title K2Node_FunctionEntry}
   > Return Node!(
    execute
    ReturnValue:ESlateVisibility Enum=Select(Option 0=Collapsed, Option 1=Visible, Index=AND Boolean(A=@n1, B=@n1.AreVariantsSupported))
   )
  ]}
 UpdateActiveItemCount {nodes [
   UpdateActiveItemCount!() -> (then, ActiveItems:int) {is=title K2Node_FunctionEntry}
   n1 = Get ActiveItems() -> (ActiveItems:int=0)
   n2 = Get StatusIndicator!(self:WBP UIKit Button Generic Base Object Reference=Get Button_SortFilter()) -> (StatusIndicator:WBP UIKit Status Indicator Object Reference)
   > SetVisibility(self=@n2, InVisibility=Select(Option 0=Collapsed, Option 1=HitTestInvisible, Index="integer > integer"(A=@n1)))
   > "SetText (Text)"(
    self=Get Indicator_Text!(self:WBP UIKit Status Indicator Base Object Reference=@n2) -> (Indicator_Text:Common Text Object Reference)
    InText="To Text (Integer)"(Value=Get ActiveItems() -> (ActiveItems=0))
   )
   > Set ShowActiveItemPadding(ShowActiveItemPadding:bool="integer > integer"(A=Get ActiveItems() -> (ActiveItems=0))) -> (Output_Get:bool)
   > UpdateSortFilterPadding()
  ]}
 UpdateCategoryPreviewScale {nodes [
   n1 = UpdateCategoryPreviewScale!() -> (then, VM:Fort Locker Category Item VM Object Reference) {is=title K2Node_FunctionEntry}
   > Is Valid(InputObject=@n1)
   n2 = GetSlotInfo!(self:Fort Locker Category Item VM Object Reference=@n1) -> (
    ReturnValue_CategoryName:Text
    ReturnValue_CategoryNamePlural:Text
    ReturnValue_CategoryNameWithPluralForms:Text
    ReturnValue_CategoryDescription:Text
    ReturnValue_DefaultEmptyTexture:Soft Object Reference
    ReturnValue_OverrideShuffleTexture:Texture 2D Soft Object Reference
    ReturnValue_CategoryIcon:Material Instance Object Reference
    ReturnValue_CategorySoundOverrides:Map of Gameplay Tag Structures to Sound Base Soft Object References
    ReturnValue_bShowInCategoryTab:bool=~
    ReturnValue_PreviewWindowScale:float=~
    ReturnValue_PreviewWindowOffset_X:double
    ReturnValue_PreviewWindowOffset_Y:double
    ReturnValue_PreviewWindowOffset_Z:double
    ReturnValue_bUsePluralCategoryInItemList:bool=~
   )
   > Set WidthOverride!(
    execute
    WidthOverride:float=float * float(A=Get CameraFrameWidth() -> (CameraFrameWidth:double), B=@n2.ReturnValue_PreviewWindowScale)
    self:Size Box Object Reference=Get SizeBox_PreviewWidget() -> (SizeBox_PreviewWidget:Size Box Object Reference)
   ) -> (then, Output_Get:float)
   > SetRenderTranslation!(
    execute
    self=Get Widget_CameraFrame() -> (Widget_CameraFrame:Size Box Object Reference)
    Translation_X:double=@n2.ReturnValue_PreviewWindowOffset_X
    Translation_Y:double=@n2.ReturnValue_PreviewWindowOffset_Y
   ) -> (then)
  ]}
 UpdateDetailsButtonState {nodes [
   n1 = UpdateDetailsButtonState!() -> (then, Selected item:Object Reference) {is=title K2Node_FunctionEntry}
   n2 = GetCategoryConfig() {is=title K2Node_CallFunction}
   n3 = IsValid(Object=@n2)
   n4 = IsValidSoftClassReference(SoftClassReference=Get CustomItemDetails(self=@n2))
   > SetVisibility(
    self=Get Button_Details() -> (Button_Details:WBP UIKit Button Quiet Object Reference)
    InVisibility="To Visibility (Boolean)"(bIsVisible=AND Boolean(A=@n4, B=@n3), TrueVisibility=Visible, FalseVisibility=Collapsed) -> (ReturnValue=Visible)
   )
   > n5 = Cast To FortItemVM(Object=@n1)
   > SetIsEnabled(self=Get Button_Details(), bInIsEnabled=true)
   SetIsEnabled(execute=@n5.CastFailed, self=Get Button_Details(), bInIsEnabled=false)
  ]}
 UpdateEditButtonEnabled {nodes [
   n1 = UpdateEditButtonEnabled!() -> (then, IsEnabled:bool=~) {is=title K2Node_FunctionEntry}
   n2 = IsInZone(self=Get FortGlobalUIContext())
   n3 = IsCustomizable() {is=title K2Node_CallFunction}
   > Set IsEditEnabled(IsEditEnabled:bool=Select(Option 0:bool=@n1, Option 1:bool=false, Index=AND Boolean(A=@n3, B=@n2)) -> (ReturnValue:bool)) -> (Output_Get:bool)
   n4 = Get Button_EditStyles()
   > Branch(Condition="Not Equal (Enum)"(
     A:ESlateVisibility Enum=GetVisibility!(self:Widget Object Reference=@n4) -> (ReturnValue:ESlateVisibility Enum=Visible)
     B:ESlateVisibility Enum=Collapsed
    ))
   n5 = Get Button_EditStyles()
   > SetIsEnabled(self=@n5, bInIsEnabled=Get IsEditEnabled() -> (IsEditEnabled:bool))
   > SetText!(
    execute
    self:Fort CTAButton Object Reference=@n5
    ButtonText:"Text (by ref)"=Select(
     Option 0:"Text (by ref)"=MakeLiteralText(Value="LOCTEXT(\"Edit\")")
     Option 1:"Text (by ref)"=MakeLiteralText(Value="LOCTEXT(\"Customize\")")
     Index=IsCustomizable()
    ) -> (ReturnValue:"Text (by ref)")
   ) -> (then)
  ]}
 UpdateEditButtonVisibility {nodes [
   n1 = UpdateEditButtonVisibility!() -> (then, Visibility:ESlateVisibility Enum=Visible) {is=title K2Node_FunctionEntry}
   > SetVisibility(self=Get Button_EditStyles(), InVisibility=@n1)
   > n2 = Branch(Condition="Equal (Enum)"(A:ESlateVisibility Enum=@n1, B:ESlateVisibility Enum=Collapsed))
   SetIsEnabled(execute=@n2.else, self=Get Button_EditStyles(), bInIsEnabled=Get IsEditEnabled())
  ]}
 UpdateEquipButtonState {nodes [
   UpdateEquipButtonState!() -> (then, SelectedItem:Object Reference) {is=title K2Node_FunctionEntry}
   n1 = Get FortItemSelectionVM()
   > UpdateEquipButtonView!(execute, self, ItemModel:Fort Item VM Object Reference=Get EquippedItem(self=Get SelectedCategory(self=@n1))) -> (then)
  ]}
 UpdateEquipButtonView {nodes [
   UpdateEquipButtonView!() -> (then, ItemModel:Fort Item VM Object Reference) {is=title K2Node_FunctionEntry}
   > n1 = Cast To FortItemVM(Object=Get Selected Item(self=Get LockerTileList()))
   > SetIsEnabled(
    self=Get Button_Equip() -> (Button_Equip:WBP UIKit Button Quiet Object Reference)
    bInIsEnabled="Not Equal (Object)"(A=@n1, B=Get ItemModel() -> (ItemModel:Fort Item VM Object Reference)) -> (ReturnValue=~)
   )
   SetIsEnabled(execute=@n1.CastFailed, self=Get Button_Equip(), bInIsEnabled=false)
  ]}
 UpdateFilteredInputFieldControllerVisuals {nodes [
   UpdateFilteredInputFieldControllerVisuals() {is=title K2Node_FunctionEntry}
   n1 = Get EditableText(self:WBP UIKit Filtered Input Field Object Reference=Get FilteredInputField()) -> (EditableText:Text)
   > n2 = Branch(Condition=NOT Boolean(A=TextIsEmpty(InText=@n1)))
   n3 = Get Button_Clear!(self:WBP UIKit Filtered Input Field Object Reference=Get FilteredInputField()) -> (Button_Clear:WBP UIKit Button Regular Object Reference)
   > SetVisibility(self=@n3, InVisibility=HitTestInvisible)
   > SetIsFocusable!(execute, self:Common Button Base Object Reference=@n3, bInIsFocusable:bool=false) -> (then)
   > SetIsEnabled(self=@n3, bInIsEnabled=true)
   > SetHideInputAction!(execute, self:Common Button Base Object Reference=Get InputField(self=Get FilteredInputField()), bInHideInputAction:bool=true) -> (then)
   SetHideInputAction(execute=@n2.else, self=Get InputField(self=Get FilteredInputField()), bInHideInputAction=false)
   > SetIsFocusable(self=@n3, bInIsFocusable=false)
   > SetShouldSelectUponReceivingFocus!(execute, self:Common Button Base Object Reference=@n3, bInShouldSelectUponReceivingFocus:bool=false) -> (then)
  ]}
 UpdateFlowriderSubView {nodes [
   UpdateFlowriderSubView!() -> (then, SubViewString:String) {is=title K2Node_FunctionEntry}
   n1 = GetDisplayName!(self:Fort Locker Category Group VM Object Reference=Get SelectedCategoryGroup(self=Get FortItemSelectionVM())) -> (ReturnValue:"Text (by ref)")
   > "Fire Event Sub View Opened (Non-Localized)"!(
    execute
    FortPC:Fort Player Controller Object Reference=GetFortPlayerControllerFromActor!(Actor:Actor Object Reference=GetOwningPlayer()) -> (ReturnValue:Fort Player Controller Object Reference)
    ViewName:String=Append!(
     A=Get Source String from Text!(Text:Text=@n1) -> (ReturnValue:String) {is=title K2Node_CallFunction}
     B=" / "
     C:String=Get SubViewString() -> (SubViewString:String)
    ) -> (ReturnValue)
    AdditionalInfo:Flow Analytics Additional Info Structure
    bIsOverlay:bool=false
   ) -> (then)
  ]}
 UpdateNavigationIfEmpty {nodes [
   UpdateNavigationIfEmpty() {is=title K2Node_FunctionEntry}
   n1 = Get FilteredItems(self=Get FortLockerCategoryFilterVM())
   > n2 = Branch(Condition="integer > integer"(A=Length(TargetArray=@n1)))
   n3 = Get SortFilterBar_Hitbox()
   > SetNavigationRuleBase!(execute, self:Widget Object Reference=@n3, Direction:EUINavigation Enum=Down, Rule:EUINavigationRule Enum=Escape) -> (then)
   > Return Node!(execute, IsTileListEmpty:bool=false)
   SetNavigationRuleBase(execute=@n2.else, self=@n3, Direction=Down, Rule=Stop)
   > SetVisibility(self=Get GamepadButtonBlocker(), InVisibility=Collapsed)
   > ShowCustomizationISB(ShowISB=false) {is=title K2Node_CallFunction}
   > RequestRefreshFocus()
   > Return Node(IsTileListEmpty=true)
  ]}
 UpdatePreview {nodes [
   n1 = UpdatePreview!() -> (then, ItemVM:Fort Item VM Object Reference) {is=title K2Node_FunctionEntry}
   > Set Item!(execute, Item:Fort Item VM Object Reference=@n1, self:Fort Locker Equip Utility VM Object Reference=Get FortLockerEquipUtilityVM()) -> (then, Output_Get:Fort Item VM Object Reference)
   > SetCurrentPreviewItem(self=Get FortItemPreviewWidget(), ItemVM=@n1)
   > MarkItemAsSeen!(
    execute
    self:Fort Locker Category VM Object Reference=Get Category(self=Get FortLockerCategoryFilterVM())
    Item:Fort Item VM Object Reference=@n1
   ) -> (then)
   > UpdateCategoryPreviewScale!(execute, self, VM:Fort Locker Category Item VM Object Reference=Get SelectedCategory(self=Get FortItemSelectionVM())) -> (then) {is=title K2Node_CallFunction}
  ]}
 UpdateScroll {nodes [
   n1 = Get TopScrollOffsetMax() -> (TopScrollOffsetMax:double)
   n2 = Get ItemOffset() -> (ItemOffset:double)
   n3 = Get FullImageSizeY() -> (FullImageSizeY:double)
   n4 = Lerp(A=0, B=@n3, Alpha=SafeDivide(A=@n2, B=@n1))
   n5 = "float > float"(A=@n1, B=@n2)
   n6 = Get ItemOffset()
   SetBrush!(
    execute
    self=Get Image_Fade_Top() -> (Image_Fade_Top:Image Object Reference)
    InBrush_TintColor:Slate Color Structure=(SpecifiedColor=(R=1,G=1,B=1,A=1),ColorUseRule=UseColor_Specified)
    InBrush_DrawAs:ESlateBrushDrawType Enum=Image
    InBrush_Tiling:ESlateBrushTileType Enum=NoTile
    InBrush_Mirroring:ESlateBrushMirrorType Enum=NoMirror
    InBrush_ImageSize_X:double=0
    InBrush_ImageSize_Y:double=SelectFloat(A=0, B=SelectFloat(A=@n4, B=@n3, bPickA=@n5), bPickA="Equal (Float)"(A=@n6))
    InBrush_Margin:Margin Structure=(Left=0,Top=0,Right=0,Bottom=0)
    InBrush_ResourceObject:Object Reference=/Locker/MPL/Materials/MI_UI_ItemsBlueMaskTop
    InBrush_OutlineSettings:Slate Brush Outline Settings Structure=(CornerRadii=(X=0,Y=0,Z=0,W=0),Color=(SpecifiedColor=(R=0,G=0,B=0,A=0),ColorUseRule=UseColor_Specified),Width=0,RoundingType=HalfHeightRadius,bUseBrushTransparency=False)
   ) -> (then)
   n7 = Get LockerTileList()
   n8 = Get LastItemIndex() -> (LastItemIndex:int=0)
   > n9 = Branch(Condition=Is Item Visible!(
     self:List View Object Reference=@n7
     Item:Object Reference=GetItemAt!(self:List View Object Reference=@n7, Index:int=@n8) -> (ReturnValue:Object Reference)
    ) -> (ReturnValue:bool))
   > n10 = Branch(Condition=OR Boolean(A=Get FirstTimeFadeCheck() -> (FirstTimeFadeCheck:bool), B=Get ResetFadeMath() -> (ResetFadeMath:bool)))
   > Set LastTileDistance(LastTileDistance:float=Get DistanceRemaining() -> (DistanceRemaining:double)) -> (Output_Get:float)
   > Set FirstTimeFadeCheck(FirstTimeFadeCheck:bool=false) -> (Output_Get:bool)
   > n11 = Set ResetFadeMath(ResetFadeMath:bool=false) -> (Output_Get:bool)
   n12 = Get LastTileDistance() -> (LastTileDistance:float)
   n13 = Get DistanceRemaining()
   SetBrush(
    execute=[@n10.else, @n11.then]
    self=Get Image_Fade_Bottom() -> (Image_Fade_Bottom:Image Object Reference)
    InBrush_TintColor=(SpecifiedColor=(R=1,G=1,B=1,A=1),ColorUseRule=UseColor_Specified)
    InBrush_DrawAs=Image
    InBrush_Tiling=NoTile
    InBrush_Mirroring=NoMirror
    InBrush_ImageSize_X=0
    InBrush_ImageSize_Y=Lerp(A=0, B=Get FullImageSizeY(), Alpha=MapRangeClamped(Value=@n13, InRangeA=@n12, InRangeB=0, OutRangeA=1, OutRangeB=0))
    InBrush_Margin=(Left=0,Top=0,Right=0,Bottom=0)
    InBrush_ResourceObject=/Locker/MPL/Materials/MI_UI_ItemsBlueMaskBottom
    InBrush_OutlineSettings=(CornerRadii=(X=0,Y=0,Z=0,W=0),Color=(SpecifiedColor=(R=0,G=0,B=0,A=0),ColorUseRule=UseColor_Specified),Width=0,RoundingType=HalfHeightRadius,bUseBrushTransparency=False)
   )
   Set ResetFadeMath(execute=@n9.else, ResetFadeMath=true)
   n14 = Get FullImageSizeY()
   > SetBrush(
    self=Get Image_Fade_Bottom()
    InBrush_TintColor=(SpecifiedColor=(R=1,G=1,B=1,A=1),ColorUseRule=UseColor_Specified)
    InBrush_DrawAs=Image
    InBrush_Tiling=NoTile
    InBrush_Mirroring=NoMirror
    InBrush_ImageSize_X=0
    InBrush_ImageSize_Y=@n14
    InBrush_Margin=(Left=0,Top=0,Right=0,Bottom=0)
    InBrush_ResourceObject=&MI_UI_ItemsBlueMaskBottom
    InBrush_OutlineSettings=(CornerRadii=(X=0,Y=0,Z=0,W=0),Color=(SpecifiedColor=(R=0,G=0,B=0,A=0),ColorUseRule=UseColor_Specified),Width=0,RoundingType=HalfHeightRadius,bUseBrushTransparency=False)
   )
   n15 = Get ItemOffset()
   n16 = Get DistanceRemaining()
   SelectFloat(A=0, B=@n14, bPickA=AND Boolean(A="Equal (Float)"(A=@n15), B="Equal (Float)"(A=@n16)))
   UpdateScroll!() -> (then, ItemOffset:double=~, DistanceRemaining:double=~) {is=title K2Node_FunctionEntry}
  ]}
 UpdateSortFilterPadding {nodes [
   UpdateSortFilterPadding() {is=title K2Node_FunctionEntry}
   n1 = Get SortButtonPaddingExpanded() -> (SortButtonPaddingExpanded:int=0)
   n2 = Get SortButtonPaddingNormal() -> (SortButtonPaddingNormal:int=0)
   n3 = Get ShowGamepadPadding() -> (ShowGamepadPadding:bool)
   n4 = Get ShowActiveItemPadding() -> (ShowActiveItemPadding:bool)
   > SetTextPadding!(
    execute
    self:WBP UIKit Button Generic Base Object Reference=Get Button_SortFilter()
    TextPadding_Left:float="To Float (Integer)"(InInt=SelectInt(A=@n1, B=@n2, bPickA=@n3))
    TextPadding_Top:float=0
    TextPadding_Right:float="To Float (Integer)"(InInt=SelectInt(A=@n1, B=@n2, bPickA=@n4))
    TextPadding_Bottom:float=0
   ) -> (then)
  ]}
 UpdateStringSuggestions {nodes [
   n1 = UpdateStringSuggestions!() -> (then, StringSuggestions:Array of Texts) {is=title K2Node_FunctionEntry}
   > n2 = For Each Loop(Array:Array of Texts=@n1) -> (Array Element:Text)
   > PrintText!(execute, InText=@n2, bPrintToScreen, bPrintToLog, TextColor, Duration, Key) -> (then)
  ]}
]
variables [
 CameraFrameWidth {type=number, default=940}
 ChatEventWatch {type=AsyncAction_StartListeningToEvent, default=null}
 ClassToCreate {type=Class, default=null}
 CompactUIOverrideEntryWidth {type=number, default=166}
 CosmeticDownloadInfoWidget {type=WBP_CosmeticDownloadInfo_C, default=null}
 CosmeticDownloadWidget {type=WBP_CosmeticDownload_C, default=null}
 CurrentSearchText {type=string, default=""}
 CustomItemDetailsClass {type=Class, default=null}
 CustomizeMenu {type=CommonActivatableWidget, default=null}
 Default Class to Create {type=Class, default=/Locker/MPL/ItemLists/WBP_LockerList_Default.WBP_LockerList_Default_C}
 EditableTextPaddingOnGamepad {type=Margin, default=<ltrb 0 0 0 0>}
 EditableTextPaddingOnKeyboard {type=Margin, default=<ltrb 0 0 36 0>}
 EmptyItemsWarning {type=WBP_LockerEmptyWarning_C, default=null}
 Filtered Items {type="array<FortItemVM>", default []}
 FirstTimeFadeCheck {type=boolean}
 Hovered Item {type=WBP_UIKit_ItemCard_Locker_C, default=null}
 IgnoreInputs {type=boolean}
 IsContextMenuOpen {type=boolean}
 IsEditEnabled {type=boolean}
 IsImpressionAvailable {type=boolean, default=true}
 IsLoadedAsEmoteWheel {type=boolean}
 ItemCardHighlightedRenderOpacity {type=number, default=1}
 ItemCardUnderlayingRenderOpacity {type=number, default=0.5}
 ItemDescriptionHideTimer {type=TimerHandle}
 ItemDetailsHideTime {type=number, default=2}
 ItemToCustomize {type=FortItemVM, default=null}
 LastActiveScreen {type=ELockerScreenList, default=Overview}
 LastItemIndex {type=integer}
 LastTileDistance {type=number}
 LeftTriggerPadingOnGamepad {type=Margin, default=<ltrb 0 0 8 0>}
 LockerTileContainer {type=Overlay, default=null}
 LockerTileList {type=FortTileView, default=null}
 LockerTileWidget {type=UserWidget, default=null}
 MimosaISB {type=CommonUserWidget, default=null}
 NeedToSetupPreviewActions {type=boolean}
 NewListBackToTop {type=boolean, default=true}
 OnCategoryTabSelected {}
 OnContextMenuCardEditStylesClicked {}
 OnSortFilterModalOpen {}
 ResetFadeMath {type=boolean}
 SearchIconPaddingOnGamepad {type=Margin, default=<ltrb 0 0 18 0>}
 SearchIconPaddingOnKeyboard {type=Margin, default=<ltrb 0 0 0 0>}
 ShowActiveItemPadding {type=boolean}
 ShowGamepadPadding {type=boolean}
 SortButtonPaddingExpanded {type=integer, default=76}
 SortButtonPaddingNormal {type=integer, default=32}
 SortFilter Modal {type=WBP_MPLocker_SortFilter_Modal_C, default=null}
 TabFlowRiderString {type=string, default=""}
]
widgets {
 info {
  inheritedWidgetCount=0
  namedSlotCount=2
  parentClass=&CommonActivatableWidget
  rootWidgetClass=&SafeZone
  widgetCount=52
 }
 widgets [
  SafeZone {widgetClassPath=&SafeZone, bIsVariable=true, widget=SafeZone, properties {padBottom=false, padTop=false}}
  SizeBox {
   widgetClassPath=&SizeBox
   bIsVariable=true
   parent=SafeZone
   widget=SizeBox
   properties {bOverride_WidthOverride=true, widthOverride=1920}
   slot {class=SafeZoneSlot, properties {hAlign=HAlign_Center}}
  }
  GridPanel_0 {
   widgetClassPath=&GridPanel
   parent=SizeBox
   widget=GridPanel_0
   properties {columnFill [1,0], rowFill [1,0]}
   slot {class=SizeBoxSlot}
  }
  WBP_CaptureForPostBufferUpdate {
   widgetClassPath=&WBP_CaptureForPostBufferUpdate
   bIsVariable=true
   parent=GridPanel_0
   slot {class=GridSlot, properties {columnSpan=2, layer=-20, rowSpan=2}}
  }
  Locker {
   widgetClassPath=&GridPanel
   bIsVariable=true
   parent=GridPanel_0
   widget=Locker
   properties {columnFill [0,1,0,0.05], rowFill [0,0,1,0]}
   slot {class=GridSlot, properties {padding=<ltrb 64 64 64 32>}}
  }
  LockerOverlay {
   widgetClassPath=&Overlay
   bIsVariable=true
   parent=Locker
   widget=LockerOverlay
   slot {class=GridSlot, properties {column=2, row=2, rowSpan=3, padding=<ltrb 0 8 0 104>}}
  }
  TileView_Hitbox {
   widgetClassPath=&Button
   parent=LockerOverlay
   widget=TileView_Hitbox
   properties {
    clipping=ClipToBounds
    isFocusable=false
    widgetStyle {
     clickedSlateSound {resourceObject=null}
     disabled {
      bIsDynamicallyLoaded=false
      drawAs=NoDrawType
      imageType=NoImage
      mirroring=NoMirror
      resourceName=None
      resourceObject=null
      tiling=NoTile
      imageSize=<0 0>
      margin=<ltrb 0 0 0 0>
      outlineSettings {
       bUseBrushTransparency=false
       roundingType=HalfHeightRadius
       width=0
       color {colorUseRule=UseColor_Specified, specifiedColor=<rgba 0 0 0 0>}
       cornerRadii=<0 0 0 0>
      }
      tintColor {colorUseRule=UseColor_Specified, specifiedColor=<rgba 1 1 1 1>}
      uVRegion {bIsValid=false, max=<0 0>, min=<0 0>}
     }
     disabledForeground {colorUseRule=UseColor_Specified, specifiedColor=<rgba 0.5271 0.5271 0.5271 1>}
     hovered {
      bIsDynamicallyLoaded=false
      drawAs=NoDrawType
      imageType=NoImage
      mirroring=NoMirror
      resourceName=None
      resourceObject=null
      tiling=NoTile
      imageSize=<32 32>
      margin=<ltrb 0 0 0 0>
      outlineSettings {
       bUseBrushTransparency=true
       roundingType=FixedRadius
       width=1
       color {colorUseRule=UseColor_Specified, specifiedColor=<rgba 0.7243 0.7243 0.7243 1>}
       cornerRadii=<4 4 4 4>
      }
      tintColor {colorUseRule=UseColor_Specified, specifiedColor=<rgba 0.7243 0.7243 0.7243 1>}
      uVRegion {bIsValid=false, max=<0 0>, min=<0 0>}
     }
     hoveredForeground {colorUseRule=UseColor_Specified, specifiedColor=<rgba 0 1 1 1>}
     hoveredSlateSound {resourceObject=null}
     normal {
      bIsDynamicallyLoaded=false
      drawAs=NoDrawType
      imageType=NoImage
      mirroring=NoMirror
      resourceName=None
      resourceObject=null
      tiling=NoTile
      imageSize=<32 32>
      margin=<ltrb 0 0 0 0>
      outlineSettings {
       bUseBrushTransparency=true
       roundingType=FixedRadius
       width=1
       color {colorUseRule=UseColor_Specified, specifiedColor=<rgba 0.6951 0.6951 0.6951 1>}
       cornerRadii=<4 4 4 4>
      }
      tintColor {colorUseRule=UseColor_Specified, specifiedColor=<rgba 0.4955 0.4955 0.4955 1>}
      uVRegion {bIsValid=false, max=<0 0>, min=<0 0>}
     }
     normalForeground {colorUseRule=UseColor_Specified, specifiedColor=<rgba 0 1 1 1>}
     normalPadding=<ltrb 0 0 0 0>
     pressed {
      bIsDynamicallyLoaded=false
      drawAs=NoDrawType
      imageType=NoImage
      mirroring=NoMirror
      resourceName=None
      resourceObject=null
      tiling=NoTile
      imageSize=<32 32>
      margin=<ltrb 0 0 0 0>
      outlineSettings {
       bUseBrushTransparency=true
       roundingType=FixedRadius
       width=1
       color {colorUseRule=UseColor_Specified, specifiedColor=<rgba 0.7243 0.7243 0.7243 1>}
       cornerRadii=<4 4 4 4>
      }
      tintColor {colorUseRule=UseColor_Specified, specifiedColor=<rgba 0.3843 0.3843 0.3843 1>}
      uVRegion {bIsValid=false, max=<0 0>, min=<0 0>}
     }
     pressedForeground {colorUseRule=UseColor_Specified, specifiedColor=<rgba 0 1 1 1>}
     pressedPadding=<ltrb 0 0 0 0>
     pressedSlateSound {resourceObject=null}
    }
   }
   slot {class=OverlaySlot, properties {horizontalAlignment=HAlign_Fill, verticalAlignment=VAlign_Fill, padding=<ltrb 0 10 0 0>}}
  }
  LockerList_Container {
   widgetClassPath=&NamedSlot
   bIsVariable=true
   parent=TileView_Hitbox
   widget=LockerList_Container
   slot {class=ButtonSlot, properties {horizontalAlignment=HAlign_Fill, verticalAlignment=VAlign_Fill, padding=<ltrb 8 8 8 16>}}
  }
  Dummy_List {widgetClassPath=/Locker/MPL/ItemLists/WBP_LockerList_Default, parent=LockerList_Container, widget=Dummy_List, slot {class=PanelSlot}}
  Image_Fade_Top {
   widgetClassPath=&Image
   bIsVariable=true
   parent=LockerOverlay
   widget=Image_Fade_Top
   properties {
    renderOpacity=0
    visibility=Collapsed
    brush {
     bIsDynamicallyLoaded=false
     drawAs=Image
     imageType=NoImage
     mirroring=NoMirror
     resourceName=None
     resourceObject=&MI_UI_ItemsBlueMaskTop
     tiling=NoTile
     imageSize=<32 0>
     margin=<ltrb 0 0 0 0>
     outlineSettings {
      bUseBrushTransparency=false
      roundingType=HalfHeightRadius
      width=0
      color {colorUseRule=UseColor_Specified, specifiedColor=<rgba 0 0 0 0>}
      cornerRadii=<0 0 0 0>
     }
     tintColor {colorUseRule=UseColor_Specified, specifiedColor=<rgba 1 1 1 1>}
     uVRegion {bIsValid=false, max=<0 0>, min=<0 0>}
    }
   }
   slot {class=OverlaySlot, properties {horizontalAlignment=HAlign_Fill, padding=<ltrb 0 0 20 0>}}
  }
  Spacer_MinItemsWidth {widgetClassPath=&Spacer, parent=LockerOverlay, widget=Spacer_MinItemsWidth, properties {size=<820 0>}, slot {class=OverlaySlot}}
  Image_Fade_Bottom {
   widgetClassPath=&Image
   bIsVariable=true
   parent=LockerOverlay
   widget=Image_Fade_Bottom
   properties {
    renderOpacity=0
    visibility=Collapsed
    brush {
     bIsDynamicallyLoaded=false
     drawAs=Image
     imageType=NoImage
     mirroring=NoMirror
     resourceName=None
     resourceObject=&MI_UI_ItemsBlueMaskBottom
     tiling=NoTile
     imageSize=<32 0>
     margin=<ltrb 0 0 0 0>
     outlineSettings {
      bUseBrushTransparency=false
      roundingType=HalfHeightRadius
      width=0
      color {colorUseRule=UseColor_Specified, specifiedColor=<rgba 0 0 0 0>}
      cornerRadii=<0 0 0 0>
     }
     tintColor {colorUseRule=UseColor_Specified, specifiedColor=<rgba 1 1 1 1>}
     uVRegion {bIsValid=false, max=<0 0>, min=<0 0>}
    }
   }
   slot {class=OverlaySlot, properties {horizontalAlignment=HAlign_Fill, verticalAlignment=VAlign_Bottom, padding=<ltrb 0 0 20 0>}}
  }
  GamepadButtonBlocker {
   widgetClassPath=&Image
   bIsVariable=true
   parent=Locker
   widget=GamepadButtonBlocker
   properties {
    brush {
     bIsDynamicallyLoaded=false
     drawAs=NoDrawType
     imageType=NoImage
     mirroring=NoMirror
     resourceName=None
     resourceObject=null
     tiling=NoTile
     imageSize=<32 32>
     margin=<ltrb 0 0 0 0>
     outlineSettings {
      bUseBrushTransparency=false
      roundingType=HalfHeightRadius
      width=0
      color {colorUseRule=UseColor_Specified, specifiedColor=<rgba 0 0 0 0>}
      cornerRadii=<0 0 0 0>
     }
     tintColor {colorUseRule=UseColor_Specified, specifiedColor=<rgba 1 1 1 1>}
     uVRegion {bIsValid=false, max=<0 0>, min=<0 0>}
    }
    colorAndOpacity=<rgba 1 1 1 0.5>
   }
   slot {class=GridSlot, properties {column=2, layer=1, row=1, padding=<ltrb 0 0 0 -5>}}
  }
  SortFilterBar_Hitbox {
   widgetClassPath=&Button
   parent=Locker
   widget=SortFilterBar_Hitbox
   properties {widgetStyle {
     clickedSlateSound {resourceObject=null}
     disabled {
      bIsDynamicallyLoaded=false
      drawAs=NoDrawType
      imageType=NoImage
      mirroring=NoMirror
      resourceName=None
      resourceObject=null
      tiling=NoTile
      imageSize=<0 0>
      margin=<ltrb 0 0 0 0>
      outlineSettings {
       bUseBrushTransparency=false
       roundingType=HalfHeightRadius
       width=0
       color {colorUseRule=UseColor_Specified, specifiedColor=<rgba 0 0 0 0>}
       cornerRadii=<0 0 0 0>
      }
      tintColor {colorUseRule=UseColor_Specified, specifiedColor=<rgba 1 1 1 1>}
      uVRegion {bIsValid=false, max=<0 0>, min=<0 0>}
     }
     disabledForeground {colorUseRule=UseColor_Specified, specifiedColor=<rgba 0.5271 0.5271 0.5271 1>}
     hovered {
      bIsDynamicallyLoaded=false
      drawAs=NoDrawType
      imageType=NoImage
      mirroring=NoMirror
      resourceName=None
      resourceObject=null
      tiling=NoTile
      imageSize=<32 32>
      margin=<ltrb 0 0 0 0>
      outlineSettings {
       bUseBrushTransparency=true
       roundingType=FixedRadius
       width=1
       color {colorUseRule=UseColor_Specified, specifiedColor=<rgba 0.7243 0.7243 0.7243 1>}
       cornerRadii=<4 4 4 4>
      }
      tintColor {colorUseRule=UseColor_Specified, specifiedColor=<rgba 0.7243 0.7243 0.7243 1>}
      uVRegion {bIsValid=false, max=<0 0>, min=<0 0>}
     }
     hoveredForeground {colorUseRule=UseColor_Specified, specifiedColor=<rgba 0 1 1 1>}
     hoveredSlateSound {resourceObject=null}
     normal {
      bIsDynamicallyLoaded=false
      drawAs=NoDrawType
      imageType=NoImage
      mirroring=NoMirror
      resourceName=None
      resourceObject=null
      tiling=NoTile
      imageSize=<32 32>
      margin=<ltrb 0 0 0 0>
      outlineSettings {
       bUseBrushTransparency=true
       roundingType=FixedRadius
       width=1
       color {colorUseRule=UseColor_Specified, specifiedColor=<rgba 0.6951 0.6951 0.6951 1>}
       cornerRadii=<4 4 4 4>
      }
      tintColor {colorUseRule=UseColor_Specified, specifiedColor=<rgba 0.4955 0.4955 0.4955 1>}
      uVRegion {bIsValid=false, max=<0 0>, min=<0 0>}
     }
     normalForeground {colorUseRule=UseColor_Specified, specifiedColor=<rgba 0 1 1 1>}
     normalPadding=<ltrb 0 0 0 0>
     pressed {
      bIsDynamicallyLoaded=false
      drawAs=NoDrawType
      imageType=NoImage
      mirroring=NoMirror
      resourceName=None
      resourceObject=null
      tiling=NoTile
      imageSize=<32 32>
      margin=<ltrb 0 0 0 0>
      outlineSettings {
       bUseBrushTransparency=true
       roundingType=FixedRadius
       width=1
       color {colorUseRule=UseColor_Specified, specifiedColor=<rgba 0.7243 0.7243 0.7243 1>}
       cornerRadii=<4 4 4 4>
      }
      tintColor {colorUseRule=UseColor_Specified, specifiedColor=<rgba 0.3843 0.3843 0.3843 1>}
      uVRegion {bIsValid=false, max=<0 0>, min=<0 0>}
     }
     pressedForeground {colorUseRule=UseColor_Specified, specifiedColor=<rgba 0 1 1 1>}
     pressedPadding=<ltrb 0 0 0 0>
     pressedSlateSound {resourceObject=null}
    }}
   slot {class=GridSlot, properties {column=2, row=1, padding=<ltrb 0 16 0 8>}}
  }
  SortFilterBar {
   widgetClassPath=&HorizontalBox
   bIsVariable=true
   parent=SortFilterBar_Hitbox
   widget=SortFilterBar
   properties {flowDirectionPreference=Culture}
   slot {class=ButtonSlot, properties {horizontalAlignment=HAlign_Fill, verticalAlignment=VAlign_Fill, padding=<ltrb 0 0 0 0>}}
  }
  FilteredInputField {
   widgetClassPath=/FortUILibrary/UIKit/InputField/WBP_UIKit_FilteredInputField
   bIsVariable=true
   parent=SortFilterBar
   widget=FilteredInputField
   properties {
    bIsFocusable=true
    hideValidationInfo=true
    hintText=Search
    isSearchField=true
    min Height=32
    min Width=32
    useHintText=true
    clearButtonPadding=<ltrb 0 0 16 0>
    desiredImageSize=<32 32>
    icon Properties {
     material=/FortUILibrary/Systems/InputField/MaterialInstances/MI_UIKit_InputFieldIcon
     image Size=<32 32>
     placement {
      horizontal Alignment=HAlign_Left
      vertical Alignment=VAlign_Center
      grid {column=0, column Span=0, layer=0, row=1, row Span=0, nudge=<0 0>}
      padding=<ltrb 20 0 0 0>
     }
    }
    input Action Row {dataTable=&InputActionDataTable, rowName=GenericLeftTrigger}
    inputActionPromptPadding=<ltrb 0 0 20 0>
    scaleOverride {enable=true, value=1.5}
   }
   slot {class=HorizontalBoxSlot, properties {padding=<ltrb 16 0 0 0>, size {sizeRule=Fill, value=1}}}
  }
  Button_SortFilter {
   widgetClassPath=&WBP_UIKit_Button_Generic
   bIsVariable=true
   parent=SortFilterBar
   widget=Button_SortFilter
   properties {
    bIsFocusable=false
    configuration=Number
    enableFocusedSelectedState=false
    mobileMinHeight=80
    statusIndicatorVerticalAlignment=VAlign_Center
    text=Sort and Filter
    useDarkInputIcon=true
    useIconBlock=false
    useStatusIndicator=true
    buttonPadding=<ltrb 0 0 0 0>
    font {
     bForceMonospaced=false
     bMaterialIsStencil=false
     fontMaterial=/FortUILibrary/UIKit/Buttons/MaterialInstances/CTAButtons/MI_UIKit_TextQuiet
     fontObject=&HeadingNow
     letterSpacing=48
     monospacedWidth=1
     size=15.75
     skewAmount=0
     typefaceFontName=76 Bold
     outlineSettings {
      bApplyOutlineToDropShadows=false
      bMiteredCorners=false
      bSeparateFillAlpha=false
      outlineMaterial=null
      outlineSize=0
      outlineColor=<rgba 0 0 0 1>
     }
    }
    promptNudge=<24 0>
    statusIndicatorPadding=<ltrb 0 0 24 0>
    textPadding=<ltrb 32 0 32 0>
    triggeringInputAction {dataTable=&InputActionDataTable, rowName=GenericRightTrigger}
   }
   slot {class=HorizontalBoxSlot, properties {padding=<ltrb 32 0 16 0>}}
  }
  ItemDetails_VerticalBox {
   widgetClassPath=&VerticalBox
   bIsVariable=true
   parent=Locker
   widget=ItemDetails_VerticalBox
   slot {
    class=GridSlot
    properties {column=1, horizontalAlignment=HAlign_Left, row=3, rowSpan=2, verticalAlignment=VAlign_Bottom, nudge=<0 16>, padding=<ltrb 0 0 0 72>}
   }
  }
  WBP_UIKit_DescriptionWithSubtitle {
   widgetClassPath=/FortUILibrary/UIKit/Items/WBP_UIKit_DescriptionWithSubtitle
   bIsVariable=true
   parent=ItemDetails_VerticalBox
   widget=WBP_UIKit_DescriptionWithSubtitle
   properties {
    bodyTextRow=3
    bottomSlotRow=5
    compatibilityRow=4
    hideBodyText=false
    titleFont {
     bForceMonospaced=false
     bMaterialIsStencil=false
     fontMaterial=null
     fontObject=&HeadingNow
     letterSpacing=0
     monospacedWidth=1
     size=30.75
     skewAmount=0
     typefaceFontName=86 Bold Italic
     outlineSettings {
      bApplyOutlineToDropShadows=false
      bMiteredCorners=false
      bSeparateFillAlpha=false
      outlineMaterial=null
      outlineSize=0
      outlineColor=<rgba 0 0 0 1>
     }
    }
   }
   slot {class=VerticalBoxSlot}
  }
  AcessRestrictionWarning {
   widgetClassPath=/FortUILibrary/UIKit/InfoBox/WBP_UIKit_InfoBox
   bIsVariable=true
   namedSlotHost=WBP_UIKit_DescriptionWithSubtitle
   widget=AcessRestrictionWarning
   properties {showIcon=false, iconSize=<40 40>}
  }
  WBP_ActionBarContainer {
   widgetClassPath=/Game/UI/Frontend/BottomBar/WBP_ActionBarContainer
   bIsVariable=true
   parent=Locker
   widget=WBP_ActionBarContainer
   slot {
    class=GridSlot
    properties {column=1, columnSpan=2, horizontalAlignment=HAlign_Right, row=4, verticalAlignment=VAlign_Bottom, padding=<ltrb 0 0 15 0>}
   }
  }
  Temp_ButtonBar {widgetClassPath=&HorizontalBox, namedSlotHost=WBP_ActionBarContainer, widget=Temp_ButtonBar}
  VisibilityBorder_BackToTop_Touch {
   widgetClassPath=&CommonHardwareVisibilityBorder
   parent=Temp_ButtonBar
   widget=VisibilityBorder_BackToTop_Touch
   properties {
    horizontalAlignment=HAlign_Right
    verticalAlignment=VAlign_Bottom
    padding=<ltrb 0 0 24 0>
    visibilityQuery {
     autoDescription=" ANY( Input.Touch )"
     tokenStreamVersion=0
     userDescription=""
     queryTokenStream [0,1,1,1,0]
     tagDictionary [{tagName=Input.Touch}]
    }
   }
   slot {class=HorizontalBoxSlot, properties {horizontalAlignment=HAlign_Right, verticalAlignment=VAlign_Bottom}}
  }
  WBP_MPLocker_BackToTopButton_Touch {
   widgetClassPath=&/WBP_MPLocker_BackToTopButton
   bIsVariable=true
   parent=VisibilityBorder_BackToTop_Touch
   widget=WBP_MPLocker_BackToTopButton_Touch
   properties {scrollOffset=7, targetIndex=2}
   slot {class=BorderSlot, properties {horizontalAlignment=HAlign_Right, verticalAlignment=VAlign_Bottom, padding=<ltrb 0 0 24 0>}}
  }
  TogglePreview {
   widgetClassPath=/FortUILibrary/UIKit/Buttons/WBP_UIKit_Button_Quiet
   bIsVariable=true
   parent=Temp_ButtonBar
   widget=TogglePreview
   properties {
    text=Toggle Preview
    visibility=Collapsed
    triggeringInputAction {dataTable=&InputActionDataTable, rowName=GenericAccept}
   }
   slot {class=HorizontalBoxSlot, properties {horizontalAlignment=HAlign_Right, verticalAlignment=VAlign_Bottom, padding=<ltrb 0 0 24 0>}}
  }
  VisibilityBorder_Details {
   widgetClassPath=&CommonHardwareVisibilityBorder
   parent=Temp_ButtonBar
   widget=VisibilityBorder_Details
   properties {
    horizontalAlignment=HAlign_Right
    verticalAlignment=VAlign_Bottom
    padding=<ltrb 0 0 24 0>
    visibilityQuery {
     autoDescription=" ANY( Input.Touch, Input.Gamepad )"
     tokenStreamVersion=0
     userDescription=""
     queryTokenStream [0,1,1,2,0,1]
     tagDictionary [{tagName=Input.Touch}, {tagName=Input.Gamepad}]
    }
   }
   slot {class=HorizontalBoxSlot, properties {horizontalAlignment=HAlign_Right, verticalAlignment=VAlign_Bottom}}
  }
  Button_Details {
   widgetClassPath=&WBP_UIKit_Button_Quiet
   bIsVariable=true
   parent=VisibilityBorder_Details
   widget=Button_Details
   properties {text=View Details, visibility=Visible, triggeringInputAction {dataTable=&InputActionDataTable, rowName=ViewDetails}}
   slot {class=BorderSlot, properties {horizontalAlignment=HAlign_Right, verticalAlignment=VAlign_Bottom, padding=<ltrb 0 0 24 0>}}
  }
  VisibilityBorder_Equip_Touch {
   widgetClassPath=&CommonHardwareVisibilityBorder
   parent=Temp_ButtonBar
   widget=VisibilityBorder_Equip_Touch
   properties {
    horizontalAlignment=HAlign_Right
    verticalAlignment=VAlign_Bottom
    padding=<ltrb 0 0 24 0>
    visibilityQuery {
     autoDescription=" ALL( Input.Touch )"
     tokenStreamVersion=0
     userDescription=""
     queryTokenStream [0,1,2,1,0]
     tagDictionary [{tagName=Input.Touch}]
    }
   }
   slot {class=HorizontalBoxSlot, properties {horizontalAlignment=HAlign_Right, verticalAlignment=VAlign_Bottom}}
  }
  Button_Equip {
   widgetClassPath=&WBP_UIKit_Button_Quiet
   bIsVariable=true
   parent=VisibilityBorder_Equip_Touch
   widget=Button_Equip
   properties {text=Equip, visibility=Visible, triggeringInputAction {dataTable=&InputActionDataTable, rowName=GenericAccept}}
   slot {class=BorderSlot, properties {horizontalAlignment=HAlign_Right, verticalAlignment=VAlign_Bottom, padding=<ltrb 0 0 24 0>}}
  }
  VisibilityBorder_Use_Touch {
   widgetClassPath=&CommonHardwareVisibilityBorder
   parent=Temp_ButtonBar
   widget=VisibilityBorder_Use_Touch
   properties {
    horizontalAlignment=HAlign_Right
    verticalAlignment=VAlign_Bottom
    padding=<ltrb 0 0 24 0>
    visibilityQuery {
     autoDescription=" ANY( Input.Gamepad, Input.Touch )"
     tokenStreamVersion=0
     userDescription=""
     queryTokenStream [0,1,1,2,0,1]
     tagDictionary [{tagName=Input.Gamepad}, {tagName=Input.Touch}]
    }
   }
   slot {class=HorizontalBoxSlot, properties {horizontalAlignment=HAlign_Right, verticalAlignment=VAlign_Bottom}}
  }
  Button_Use {
   widgetClassPath=&WBP_UIKit_Button_Quiet
   bIsVariable=true
   parent=VisibilityBorder_Use_Touch
   widget=Button_Use
   properties {text=Use, visibility=Visible, triggeringInputAction {dataTable=&InputActionDataTable, rowName=GenericForward}}
   slot {class=BorderSlot, properties {horizontalAlignment=HAlign_Right, verticalAlignment=VAlign_Bottom, padding=<ltrb 0 0 24 0>}}
  }
  Button_PreviewAction1 {
   widgetClassPath=&WBP_UIKit_Button_Quiet
   bIsVariable=true
   parent=Temp_ButtonBar
   widget=Button_PreviewAction1
   properties {
    bAutoUpdateFromInputAction=true
    text="[PH] Preview 1"
    visibility=Collapsed
    triggeringInputAction {dataTable=&InputActionDataTable, rowName=Edit}
   }
   slot {class=HorizontalBoxSlot, properties {horizontalAlignment=HAlign_Right, verticalAlignment=VAlign_Bottom, padding=<ltrb 0 0 24 0>}}
  }
  Button_PreviewAction {
   widgetClassPath=&WBP_UIKit_Button_Quiet
   bIsVariable=true
   parent=Temp_ButtonBar
   widget=Button_PreviewAction
   properties {
    bAutoUpdateFromInputAction=true
    text="[PH] Preview"
    visibility=Collapsed
    triggeringInputAction {dataTable=&InputActionDataTable, rowName=Edit}
   }
   slot {class=HorizontalBoxSlot, properties {horizontalAlignment=HAlign_Right, verticalAlignment=VAlign_Bottom, padding=<ltrb 0 0 24 0>}}
  }
  Button_EditStyles {
   widgetClassPath=&WBP_UIKit_Button_Quiet
   bIsVariable=true
   parent=Temp_ButtonBar
   widget=Button_EditStyles
   properties {text=Edit, visibility=Visible, triggeringInputAction {dataTable=&InputActionDataTable, rowName=Edit}}
   slot {class=HorizontalBoxSlot, properties {horizontalAlignment=HAlign_Right, verticalAlignment=VAlign_Bottom, padding=<ltrb 0 0 24 0>}}
  }
  VisibilityBorder_TouchGamepad {
   widgetClassPath=&CommonHardwareVisibilityBorder
   parent=Temp_ButtonBar
   widget=VisibilityBorder_TouchGamepad
   properties {
    horizontalAlignment=HAlign_Right
    verticalAlignment=VAlign_Bottom
    padding=<ltrb 0 0 24 0>
    visibilityQuery {
     autoDescription=" ANY( Input.Gamepad, Input.Touch )"
     tokenStreamVersion=0
     userDescription=""
     queryTokenStream [0,1,1,2,0,1]
     tagDictionary [{tagName=Input.Gamepad}, {tagName=Input.Touch}]
    }
   }
   slot {class=HorizontalBoxSlot, properties {horizontalAlignment=HAlign_Right, verticalAlignment=VAlign_Bottom}}
  }
  WBP_UIKit_Button_Options {
   widgetClassPath=&WBP_UIKit_Button_Quiet
   bIsVariable=true
   parent=VisibilityBorder_TouchGamepad
   widget=WBP_UIKit_Button_Options
   properties {
    bHideInputActionWithKeyboard=true
    text=Options
    triggeringInputAction {dataTable=&InputActionDataTable, rowName=ContextPopupMenu}
   }
   slot {class=BorderSlot, properties {horizontalAlignment=HAlign_Right, verticalAlignment=VAlign_Bottom, padding=<ltrb 0 0 24 0>}}
  }
  VisibilityBorder_TouchGamepad_ClearSearch {
   widgetClassPath=&CommonHardwareVisibilityBorder
   parent=Temp_ButtonBar
   widget=VisibilityBorder_TouchGamepad_ClearSearch
   properties {
    horizontalAlignment=HAlign_Right
    verticalAlignment=VAlign_Bottom
    padding=<ltrb 0 0 24 0>
    visibilityQuery {
     autoDescription=" ANY( Input.Gamepad, Input.Touch )"
     tokenStreamVersion=0
     userDescription=""
     queryTokenStream [0,1,1,2,0,1]
     tagDictionary [{tagName=Input.Gamepad}, {tagName=Input.Touch}]
    }
   }
   slot {class=HorizontalBoxSlot, properties {horizontalAlignment=HAlign_Right, verticalAlignment=VAlign_Bottom}}
  }
  Button_ClearSearch {
   widgetClassPath=&WBP_UIKit_Button_Quiet
   bIsVariable=true
   parent=VisibilityBorder_TouchGamepad_ClearSearch
   widget=Button_ClearSearch
   properties {
    bHideInputActionWithKeyboard=true
    text=Clear Search
    triggeringInputAction {dataTable=&InputActionDataTable, rowName=DiscoverReduxClearResults}
   }
   slot {class=BorderSlot, properties {horizontalAlignment=HAlign_Right, verticalAlignment=VAlign_Bottom, padding=<ltrb 0 0 24 0>}}
  }
  WBP_AthenaBottomBar_WithNotifications {
   widgetClassPath=/Game/UI/Frontend/BottomBar/WBP_AthenaBottomBar_WithNotifications
   bIsVariable=true
   parent=Temp_ButtonBar
   widget=WBP_AthenaBottomBar_WithNotifications
   properties {visibility=Collapsed}
   slot {class=HorizontalBoxSlot, properties {horizontalAlignment=HAlign_Right, verticalAlignment=VAlign_Bottom, padding=<ltrb 0 0 24 0>}}
  }
  WBP_UIKit_Button_Back {
   widgetClassPath=&WBP_UIKit_Button_Quiet
   bIsVariable=true
   parent=Temp_ButtonBar
   widget=WBP_UIKit_Button_Back
   properties {text=Back, triggeringInputAction {dataTable=&InputActionDataTable, rowName=Back}}
   slot {class=HorizontalBoxSlot, properties {horizontalAlignment=HAlign_Right, verticalAlignment=VAlign_Bottom}}
  }
  VisibilityBorder_BackToTop_NonTouch {
   widgetClassPath=&CommonHardwareVisibilityBorder
   parent=Locker
   widget=VisibilityBorder_BackToTop_NonTouch
   properties {visibilityQuery {
     autoDescription=" NONE( Input.Touch )"
     tokenStreamVersion=0
     userDescription=""
     queryTokenStream [0,1,3,1,0]
     tagDictionary [{tagName=Input.Touch}]
    }}
   slot {
    class=GridSlot
    properties {column=3, horizontalAlignment=HAlign_Left, row=2, rowSpan=3, verticalAlignment=VAlign_Bottom, padding=<ltrb 0 0 0 72>}
   }
  }
  WBP_MPLocker_BackToTopButton {
   widgetClassPath=&WBP_MPLocker_BackToTopButton
   bIsVariable=true
   parent=VisibilityBorder_BackToTop_NonTouch
   widget=WBP_MPLocker_BackToTopButton
   properties {scrollOffset=7, targetIndex=2}
   slot {class=BorderSlot, properties {padding=<ltrb 0 0 0 0>}}
  }
  SubcategoryCarousel {
   widgetClassPath=&/WBP_SubcategoryCarousel
   bIsVariable=true
   parent=Locker
   widget=SubcategoryCarousel
   slot {class=GridSlot, properties {column=2}}
  }
  Widget_CameraFrame {
   widgetClassPath=&SizeBox
   bIsVariable=true
   parent=GridPanel_0
   widget=Widget_CameraFrame
   slot {class=GridSlot, properties {layer=-1}}
  }
  HorizontalBox_0 {widgetClassPath=&HorizontalBox, parent=Widget_CameraFrame, widget=HorizontalBox_0, slot {class=SizeBoxSlot}}
  Camera_Spacer_Left {
   widgetClassPath=&Spacer
   bIsVariable=true
   parent=HorizontalBox_0
   widget=Camera_Spacer_Left
   slot {class=HorizontalBoxSlot, properties {size {sizeRule=Fill, value=0}}}
  }
  SizeBox_PreviewWidget {
   widgetClassPath=&SizeBox
   bIsVariable=true
   parent=HorizontalBox_0
   widget=SizeBox_PreviewWidget
   properties {bOverride_HeightOverride=true, bOverride_WidthOverride=true, heightOverride=1050, widthOverride=940}
   slot {class=HorizontalBoxSlot, properties {horizontalAlignment=HAlign_Center, verticalAlignment=VAlign_Center}}
  }
  FortItemPreviewWidget {
   widgetClassPath=/FortniteCoreUI/Frontend/ItemPreview/WBP_ItemPreview
   bIsVariable=true
   parent=SizeBox_PreviewWidget
   widget=FortItemPreviewWidget
   properties {
    bSupportsActivationFocus=false
    infoAlignment_V=VAlign_Bottom
    infoWidget=&self:WidgetTree.Widget_PreviewInfo_External
    itemViewObject=&self:WidgetTree.FortItemPreviewWidget.ItemViewObject
    previewActionButtons [&self:WidgetTree.Button_PreviewAction, &self:WidgetTree.Button_PreviewAction1]
   }
   slot {class=SizeBoxSlot, properties {padding=<ltrb 0 0 0 84>}}
  }
  Camera_Spacer_Right {
   widgetClassPath=&Spacer
   bIsVariable=true
   parent=HorizontalBox_0
   widget=Camera_Spacer_Right
   slot {class=HorizontalBoxSlot, properties {size {sizeRule=Fill, value=1}}}
  }
  NotificationSizeBox {
   widgetClassPath=&SizeBox
   parent=GridPanel_0
   widget=NotificationSizeBox
   properties {bOverride_MaxDesiredWidth=true, maxDesiredWidth=450}
   slot {class=GridSlot, properties {horizontalAlignment=HAlign_Left, verticalAlignment=VAlign_Top, padding=<ltrb 50 50 0 0>}}
  }
  NotificationVerticalBox {
   widgetClassPath=&VerticalBox
   bIsVariable=true
   parent=NotificationSizeBox
   widget=NotificationVerticalBox
   slot {class=SizeBoxSlot, properties {horizontalAlignment=HAlign_Left, verticalAlignment=VAlign_Top}}
  }
  Widget_PreviewInfo_External {
   widgetClassPath=/FortniteCoreUI/Frontend/ItemPreview/WBP_ItemPreview_Info
   bIsVariable=true
   parent=NotificationVerticalBox
   widget=Widget_PreviewInfo_External
   slot {class=VerticalBoxSlot, properties {horizontalAlignment=HAlign_Left, verticalAlignment=VAlign_Top, padding=<ltrb 0 0 0 10>}}
  }
 ]
}
