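# Locker widget that displays saved loadout presets in a tile view for equipping, saving, renaming, deleting and shuffle toggling.
# Pooled preset cards bind click, hover, hold and context-menu events on entry initialization and unbind on release, with separate touch paths.
# Hovering drives sub-item previews and age-gating warnings; save, rename and delete open modals before async loadout operations.
# It also fires telemetry impressions, shows a preset-cap coach mark and adapts layout for compact UI.
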WidgetBlueprint /Locker/MPL/WBP_MPLocker_Presets
animations [
 IsActive {
  binding_count=1
  length_seconds=0.5068
  object=&self:IsActive
  sequence {bindings [SubItems_ScrollBox {
     id=#1
     possessed_class=ScrollBox
     tracks [{
       type=MovieSceneFloatTrack
       display_name=Render Opacity
       sections [{type=MovieSceneFloatSection, end_frame=10, start_frame=0, channels [None {type=MovieSceneScriptingFloatChannel, keys [[0,0],[10.136,1]]}]}]
      }]
    }]}
 }
 SlideSubItems {
  binding_count=1
  length_seconds=4.9954
  object=&self:SlideSubItems
  sequence {bindings [WBP_MPLocker_Presets_C_128 {
     id=#2
     possessed_class=WBP_MPLocker_Presets_C
     tracks [{
       type=MovieSceneDoubleTrack
       display_name=Scroll Timer
       sections [{type=MovieSceneDoubleSection, end_frame=99, start_frame=0, channels [None {type=MovieSceneScriptingDoubleChannel, keys [[0,0],[99.9073,1]]}]}]
      }]
    }]}
 }
]
graphs [
 AddTelemetryImpression {nodes [
   AddTelemetryImpression() {is=title K2Node_FunctionEntry}
   > Set PresetLoadoutFlowRiderString(PresetLoadoutFlowRiderString:String) -> (Output_Get:String)
   > n1 = For Each Loop(Array:Array of User Widget Object References=GetDisplayedEntryWidgets!(self:List View Base Object Reference=Get TileView_Presets() -> (TileView_Presets:Fort Tile View Object Reference)) -> (ReturnValue:Array of User Widget Object References)) -> (Array Element:User Widget Object Reference)
   n2 = Get SectionContentsArray() -> (SectionContentsArray:Array of Flow Analytics Impression Section Content Structures)
   Branch(
    execute=@n1.Completed
    Condition="integer > integer"(A=Length(TargetArray:Array of Flow Analytics Impression Section Content Structures=@n2))
   )
   n3 = Get SelectedCategoryGroup!(self:Fort Item Selection VM Object Reference=Get FortItemSelectionVM() -> (FortItemSelectionVM:Fort Item Selection VM Object Reference)) -> (SelectedCategoryGroup:Fort Locker Category Group VM Object Reference)
   > "Fire Event Add Impression (Non-Localized)"!(
    execute
    SectionID:String=Get Source String from Text!(Text:Text=GetDisplayName!(self:Fort Locker Category Group VM Object Reference=@n3) -> (ReturnValue:"Text (by ref)")) -> (ReturnValue:String) {is=title K2Node_CallFunction}
    SectionContents:Array of Flow Analytics Impression Section Content Structures=Get SectionContentsArray()
   ) -> (then)
   n4 = Cast To WBP_UIKit_PresetCard!(execute=@n1.LoopBody, Object=@n1) -> (then, CastFailed, AsWBP UIKit Preset Card:WBP UIKit Preset Card Object Reference) {is=title K2Node_DynamicCast}
   n5 = Get IsSaveShuffleTile!(self:WBP UIKit Preset Card Object Reference=@n4) -> (IsSaveShuffleTile:bool)
   n6 = Get FortTypedModularLoadoutVM!(self:WBP UIKit Preset Card Object Reference=@n4) -> (FortTypedModularLoadoutVM:Fort Typed Modular Loadout VM Object Reference)
   > Branch(Condition=AND Boolean(A=Get bIsValid!(self:Fort Typed Modular Loadout VM Object Reference=@n6) -> (bIsValid:bool), B=NOT Boolean(A=@n5)))
   > n7 = For Each Loop(Array:Array of Fort Modular Loadout Item VM Object References=Get LoadoutItems!(self:Fort Typed Modular Loadout VM Object Reference=@n6) -> (LoadoutItems:Array of Fort Modular Loadout Item VM Object References)) -> (Array Element:Fort Modular Loadout Item VM Object Reference)
   > Is Valid(InputObject=@n7)
   n8 = GetItemTemplateId!(self:Fort Item VM Object Reference=Get Item!(self:Fort Modular Loadout Item VM Object Reference=@n7) -> (Item:Fort Item VM Object Reference)) -> (ReturnValue:String)
   n9 = Select(Option 0:String=@n8, Option 1:String=Empty, Index=IsEmpty(InString=@n8)) -> (ReturnValue:String)
   n10 = Get PresetLoadoutFlowRiderString() -> (PresetLoadoutFlowRiderString:String)
   > Set PresetLoadoutFlowRiderString(PresetLoadoutFlowRiderString=Select(Option 0=Append!(A=@n10, B=" | ", C:String=@n9) -> (ReturnValue), Option 1=@n8, Index="Equal (Integer)"(A=@n7.Array Index)))
   n11 = Get PresetLoadoutFlowRiderString()
   Add(
    execute=@n7.Completed
    TargetArray:Array of Flow Analytics Impression Section Content Structures=Get SectionContentsArray()
    NewItem:"Flow Analytics Impression Section Content Structure (by ref)"=Make Flow Analytics Impression Section Content!(ContentID:Text="To Text (String)"(InString=@n11), OfferID:Text, Row:int=-1, Column:int=-1, Index:int=-1) -> (FlowAnalyticsImpressionSectionContent:Flow Analytics Impression Section Content Structure) {is=title K2Node_MakeStruct}
   ) {is=title K2Node_CallArrayFunction}
  ]}
 BP_GetDesiredFocusTarget {nodes [
   Get Desired Focus Target()
   > Branch(Condition=IsActivated!(self:Common Activatable Widget Object Reference) -> (ReturnValue:bool))
   > n1 = Branch(Condition="integer < integer"(A=Get LastChangedIndex() -> (LastChangedIndex:int=0), B=1))
   > Return Node!(execute, ReturnValue:Widget Object Reference=Get TileView_Presets())
   n2 = Select(
    Option 0:int=Get LastChangedIndex() -> (LastChangedIndex=0)
    Option 1:int=Get LastItemIndex() -> (LastItemIndex:int=0)
    Index="integer > integer"(A=Get LastChangedIndex() -> (LastChangedIndex=0), B=Get LastItemIndex() -> (LastItemIndex=0))
   ) -> (ReturnValue:int)
   n3 = Get LastItemIndex() -> (LastItemIndex=0)
   n4 = Get TileView_Presets()
   Return Node(
    execute=@n1.else
    ReturnValue="Get (a copy)"(
     Array:Array of User Widget Object References=GetDisplayedEntryWidgets(self=@n4)
     Dimension 1="Clamp (Integer)"(Value=@n2, Min=0, Max=@n3) -> (ReturnValue=0)
    ) -> (Output:User Widget Object Reference)
   )
  ]}
 BindCardEvents {nodes [
   n1 = BindCardEvents!() -> (then, Item:Object Reference, Widget:Object Reference) {is=title K2Node_FunctionEntry}
   > n2 = Cast To WBP_UIKit_PresetCard(Object=@n1.Widget)
   > n3 = Set CachedPresetCard(CachedPresetCard:WBP UIKit Preset Card Object Reference=@n2) -> (Output_Get:WBP UIKit Preset Card Object Reference)
   > Bind Event to Disable Options Hint(self:WBP UIKit Preset Card Object Reference=@n3, Delegate=Create Event()) {is=title K2Node_AddDelegate}
   > Bind Event to Enable Options Hint(self:WBP UIKit Preset Card Object Reference=@n3, Delegate=Create Event()) {is=title K2Node_AddDelegate}
   > Bind Event to On Rename Event Fired(self:WBP UIKit Preset Card Object Reference=@n3, Delegate=Create Event()) {is=title K2Node_AddDelegate}
   > Bind Event to On Delete Event Fired(self:WBP UIKit Preset Card Object Reference=@n3, Delegate=Create Event()) {is=title K2Node_AddDelegate}
   > n4 = Cast To FortTypedModularLoadoutVM!(execute, Object=Get Item!() -> (Item:Object Reference)) -> (then, CastFailed, AsFort Typed Modular Loadout VM:Fort Typed Modular Loadout VM Object Reference) {is=title K2Node_DynamicCast}
   > n5 = Branch(Condition=Get bIsShuffle!(self:Fort Typed Modular Loadout VM Object Reference=@n4) -> (bIsShuffle:bool))
   > n6 = Set SaveShuffleButton(SaveShuffleButton:WBP UIKit Preset Card Object Reference=Get CachedPresetCard() -> (CachedPresetCard:WBP UIKit Preset Card Object Reference)) -> (Output_Get:WBP UIKit Preset Card Object Reference)
   > Bind Event to On Clicked(self:Common Button Base Object Reference=@n6, Delegate=Create Event()) {is=title K2Node_AddDelegate}
   > Bind Event to On Hovered(self:Common Button Base Object Reference=@n6, Delegate=Create Event()) {is=title K2Node_AddDelegate}
   > Bind Event to On Unhovered(self:Common Button Base Object Reference=@n6, Delegate=Create Event()) {is=title K2Node_AddDelegate}
   > Bind Event to On Selected Changed(self:Common Button Base Object Reference=@n6, Delegate=Create Event()) {is=title K2Node_AddDelegate}
   n7 = Get CachedPresetCard()
   Bind Event to On Clicked(execute=@n5.else, self=@n7, Delegate=Create Event())
   > Bind Event to On Hovered(self=@n7, Delegate=Create Event())
   > Bind Event to OnModularButtonHoldTriggered(self:UIKit Holdable Modular Button Object Reference=@n7, Delegate=Create Event()) {is=title K2Node_AddDelegate}
  ]}
 ChangeBackToTopButtonVisibility {nodes [ChangeBackToTopButtonVisibility!() -> (then, New Visibility:ESlateVisibility Enum=Visible) {is=title K2Node_FunctionEntry}]}
 CheckCoachMarks {nodes [
   CheckCoachMarks() {is=title K2Node_FunctionEntry}
   > n1 = ShouldShowOnboarding!(
    execute
    InOnboardingSeenId:"Gameplay Tag Structure (by ref)"=Get PresetCapIncreaseSeenID() -> (PresetCapIncreaseSeenID:Gameplay Tag Structure)
    SocialTutorial:ESocialTutorialType Enum=PresetCountIncreaseCoachMark
    None:%=Self-Reference()
   ) -> (then, ReturnValue:bool)
   > Branch(Condition=@n1)
   > n2 = Create Widget!(
    execute:Exec
    Class:User Widget Class Reference=/Locker/CoachMarks/WBP_Locker_PresetCap_CoachMark.WBP_Locker_PresetCap_CoachMark_C
    OwningPlayer:Player Controller Object Reference
   ) -> (then:Exec, ReturnValue:WBP Locker Preset Cap Coach Mark Object Reference)
   > Set PresetCapCoachMark(PresetCapCoachMark:WBP Locker Preset Cap Coach Mark Object Reference=@n2) -> (Output_Get:WBP Locker Preset Cap Coach Mark Object Reference)
   > AddChild(self=Get CoachMark_Attachment() -> (CoachMark_Attachment:Fort Visual Attachment Object Reference), Content=@n2)
   > ActivateWidget(self=@n2)
   > Bind Event to On Widget Deactivated(self:Common Activatable Widget Object Reference=@n2, Delegate=Create Event()) {is=title K2Node_AddDelegate}
  ]}
 Clear Shuffle Description {nodes [
   Clear Shuffle Description() {is=title K2Node_FunctionEntry}
   n1 = Get ShuffleDescription() -> (ShuffleDescription:WBP UIKit Description With Subtitle Object Reference)
   > SetVisibility(
    self=Get RarityTag!(self:WBP UIKit Item Description Base Object Reference=@n1) -> (RarityTag:WBP UIKit Tag Object Reference)
    InVisibility=Collapsed
   )
   n2 = Get Blank() -> (Blank:Text)
   > SetTitleText!(execute, self:WBP UIKit Item Description Base Object Reference=@n1, Text:Text=@n2) -> (then)
   > SetBodyText!(execute, self:WBP UIKit Item Description Base Object Reference=@n1, Text:"Text (by ref)"=@n2) -> (then)
  ]}
 DeclineModal {nodes [
   DeclineModal() {is=title K2Node_FunctionEntry}
   > FireEventElementInteractedByName!(execute, Element Name:String=Get DeclineString() -> (DeclineString:String), WidgetCallingFrom:Widget Object Reference=Self-Reference()) -> (then, Success:bool)
   > ActivateWidget()
  ]}
 DisableOptionsHint {nodes [DisableOptionsHint() {is=title K2Node_FunctionEntry}]}
 EnableOptionsHint {nodes [EnableOptionsHint() {is=title K2Node_FunctionEntry}]}
 EntryReleaseBinding {nodes [
   n1 = EntryReleaseBinding!() -> (then, Widget:Object Reference) {is=title K2Node_FunctionEntry}
   > n2 = Cast To WBP_UIKit_PresetCard(Object=@n1)
   > Unbind all Events from On Rename Event Fired!(execute:Exec, self:WBP UIKit Preset Card Object Reference=@n2) -> (then:Exec) {is=title K2Node_ClearDelegate}
   > Unbind all Events from On Delete Event Fired!(execute:Exec, self:WBP UIKit Preset Card Object Reference=@n2) -> (then:Exec) {is=title K2Node_ClearDelegate}
   > Unbind all Events from On Clicked!(execute:Exec, self:Common Button Base Object Reference=@n2) -> (then:Exec) {is=title K2Node_ClearDelegate}
   > Unbind all Events from Disable Options Hint!(execute:Exec, self:WBP UIKit Preset Card Object Reference=@n2) -> (then:Exec) {is=title K2Node_ClearDelegate}
   > Unbind all Events from Enable Options Hint!(execute:Exec, self:WBP UIKit Preset Card Object Reference=@n2) -> (then:Exec) {is=title K2Node_ClearDelegate}
   > Unbind all Events from On Clicked(self=@n2)
   > Unbind all Events from On Hovered!(execute:Exec, self:Common Button Base Object Reference=@n2) -> (then:Exec) {is=title K2Node_ClearDelegate}
   > Unbind all Events from On Unhovered!(execute:Exec, self:Common Button Base Object Reference=@n2) -> (then:Exec) {is=title K2Node_ClearDelegate}
   > Unbind all Events from On Context Menu Opened!(execute:Exec, self:WBP UIKit Preset Card Object Reference=@n2) -> (then:Exec) {is=title K2Node_ClearDelegate}
   > Unbind all Events from On Context Menu Closed!(execute:Exec, self:WBP UIKit Preset Card Object Reference=@n2) -> (then:Exec) {is=title K2Node_ClearDelegate}
   > Unbind all Events from OnModularButtonHoldTriggered!(execute:Exec, self:UIKit Holdable Modular Button Object Reference=@n2) -> (then:Exec) {is=title K2Node_ClearDelegate}
   > Unbind all Events from On Selected Changed!(execute:Exec, self:Common Button Base Object Reference=@n2) -> (then:Exec) {is=title K2Node_ClearDelegate}
  ]}
 EventGraph {nodes [
   n1 = "On Entry Generated (TileView_Presets)"!() -> (OutputDelegate, then, Widget:User Widget Object Reference) {is=title K2Node_ComponentBoundEvent}
   > n2 = Cast To WBP_UIKit_PresetCard(Object=@n1.Widget)
   > Bind Event to On Context Menu Closed(self:WBP UIKit Preset Card Object Reference=@n2, Delegate=Create Event()) {is=title K2Node_AddDelegate}
   > Bind Event to On Context Menu Opened(self:WBP UIKit Preset Card Object Reference=@n2, Delegate=Create Event()) {is=title K2Node_AddDelegate}
   > Branch(Condition=Get IsImpressionAvailable() -> (IsImpressionAvailable:bool))
   > AddTelemetryImpression() {is=title K2Node_CallFunction}
   > Set IsImpressionAvailable(IsImpressionAvailable:bool=false) -> (Output_Get:bool)
   > Delay(Duration=1)
   > Set IsImpressionAvailable(IsImpressionAvailable=true)
   "On List View Scrolled (TileView_Presets)"!() -> (OutputDelegate, then, ItemOffset:float, DistanceRemaining:float) {is=title K2Node_ComponentBoundEvent}
   n3 = GetScrollOffset!(self:List View Base Object Reference=Get TileView_Presets()) -> (ReturnValue:float)
   > Call Change Back to Top Button Visibility!(
    execute:Exec
    self:WBP MPLocker Presets Object Reference
    New Visibility:ESlateVisibility Enum=Select(Option 0:ESlateVisibility Enum=Collapsed, Option 1:ESlateVisibility Enum=Visible, Index="float >= float"(A=@n3, B=1)) -> (ReturnValue:ESlateVisibility Enum)
   ) -> (then:Exec) {is=title K2Node_CallDelegate}
   n4 = "On Entry Initialized (TileView_Presets)"!() -> (OutputDelegate, then, Item:Object Reference, Widget:User Widget Object Reference) {is=title K2Node_ComponentBoundEvent}
   > BindCardEvents!(execute, self, Item:Object Reference=@n4.Item, Widget:Object Reference=@n4.Widget) -> (then) {is=title K2Node_CallFunction}
   n5 = "On Entry Released (TileView_Presets)"!() -> (OutputDelegate, then, Widget:User Widget Object Reference) {is=title K2Node_ComponentBoundEvent}
   > EntryReleaseBinding!(execute, self, Widget:Object Reference=@n5.Widget) -> (then) {is=title K2Node_CallFunction}
   n6 = SaveShuffleOnSelectedChanged!() -> (OutputDelegate, then, Button:Common Button Base Object Reference, Selected:bool=~) {is=title K2Node_CustomEvent}
   n7 = GetOwningPlayer()
   > Branch(Condition=AND Boolean(A=@n6.Selected, B=IsUsingTouch(PlayerController=@n7)))
   > n8 = Cast To WBP_UIKit_PresetCard(Object=@n6.Button)
   > Update Presets Preview!(execute, self, PresetCard:WBP UIKit Preset Card Object Reference=@n8) -> (then)
   n9 = OnCMRenamedConfirmed!() -> (OutputDelegate, then, NewName:String) {is=title K2Node_CustomEvent}
   n10 = OnCMRenameFired!() -> (OutputDelegate, then, PresetCard:WBP UIKit Preset Card Object Reference) {is=title K2Node_CustomEvent}
   RenameModularLoadout!(
    execute:Exec=@n9.then
    Target:Fort Typed Modular Loadout VM Object Reference=Get FortTypedModularLoadoutVM(self=@n10.PresetCard)
    NewName:String=@n9.NewName
   ) -> (then:Exec, OnOperationCompleted:Exec, Success:bool=~) {is=title K2Node_AsyncAction}
   > FireEventElementInteractedByName(Element Name=Preset Rename Confirmed, WidgetCallingFrom=Self-Reference())
   > ActivateWidget()
   EnableOptions() {is=title K2Node_CustomEvent}
   > Call Enable Options Hint!(execute:Exec, self:WBP MPLocker Presets Object Reference) -> (then:Exec) {is=title K2Node_CallDelegate}
   HandleTouchLoadoutEquip() {is=title K2Node_CustomEvent}
   > HandleLoadoutEquip!(
    execute
    self
    Loadout:Fort Typed Modular Loadout VM Object Reference=Get CurrentSelectedLoadout() -> (CurrentSelectedLoadout:Fort Typed Modular Loadout VM Object Reference)
   ) -> (then)
   n11 = SaveShuffleHovered!() -> (OutputDelegate, then, Button:Common Button Base Object Reference) {is=title K2Node_CustomEvent}
   > n12 = Cast To WBP_UIKit_PresetCard(Object=@n11.Button)
   > Update Presets Preview(PresetCard=@n12)
   n13 = Toggle Selected Context!() -> (OutputDelegate, then, SelectedWidget:WBP UIKit Preset Card Object Reference) {is=title K2Node_CustomEvent}
   n14 = Get CommonInputSubsystem()
   > n15 = Branch(Condition="Equal (Enum)"(A=GetCurrentInputType(self=@n14), B=Touch))
   > Is Valid(InputObject=Get Selected Item!(self:List View Object Reference=Get TileView_Presets()) -> (ReturnValue:Object Reference) {is=title K2Node_CallFunction})
   n16 = Create Event()
   > Unbind Event from On Item Scrolled Into View(self:List View Object Reference=Get TileView_Presets(), Delegate=@n16) {is=title K2Node_RemoveDelegate}
   > Assign On Item Scrolled Into View!(execute:Exec, self:List View Object Reference=Get TileView_Presets(), Delegate:"Delegate (by ref)"=@n16) -> (then:Exec) {is=title K2Node_AssignDelegate}
   > Scroll Item Into View!(execute, self:List View Object Reference=Get TileView_Presets(), Item:Object Reference=Get Selected Item(self=Get TileView_Presets())) -> (then)
   ToggleContextMenu(execute=@n15.else, self:WBP UIKit Preset Card Object Reference=@n13.SelectedWidget)
   OnContextMenuClosed() {is=title K2Node_CustomEvent}
   n17 = Get TileView_Presets()
   > SetIsPointerScrollingEnabled!(execute, self:List View Base Object Reference=@n17, bInIsPointerScrollingEnabled:bool=true) -> (then)
   > Branch(Condition=IsInSubGame!(SubGame:ESubGame Enum=Campaign) -> (ReturnValue:bool))
   > CenterWidget!(execute, ToCenter:Widget Object Reference=Get Hovered Preset Button() -> (Hovered Preset Button:WBP UIKit Preset Card Object Reference)) -> (then)
   SaveShuffleUnHovered!() -> (OutputDelegate, then, Button:Common Button Base Object Reference) {is=title K2Node_CustomEvent}
   n18 = GetOwningPlayer()
   > Branch(Condition=NOT Boolean(A=IsUsingTouch(PlayerController=@n18)))
   > Clear Shuffle Description() {is=title K2Node_CallFunction}
   OnContextMenuOpened() {is=title K2Node_CustomEvent}
   > SetIsPointerScrollingEnabled(self=@n17, bInIsPointerScrollingEnabled=false)
   n19 = LoadoutHold!() -> (OutputDelegate, then, Button:Common Button Base Object Reference) {is=title K2Node_CustomEvent}
   > n20 = Cast To WBP_UIKit_PresetCard(Object=@n19.Button)
   > Update Presets Preview(PresetCard=@n20)
   n21 = RandomButtonClicked_Event() {is=title K2Node_CustomEvent}
   n22 = Get FortTypedModularLoadoutsVM() -> (FortTypedModularLoadoutsVM:Fort Typed Modular Loadouts VM Object Reference)
   Branch(
    execute=[@n21.then, @n44.then]
    Condition=GetCanEnableShuffle!(self:Fort Typed Modular Loadouts VM Object Reference=@n22) -> (ReturnValue:bool)
   )
   > ToggleShuffleEnabled()
   > FireEventElementInteractedByName(
    Element Name=SelectString(
     A=Preset Random On
     B=Preset Random Off
     bPickA=Get bShuffleEnabled!(self:Fort Typed Modular Loadouts VM Object Reference=@n22) -> (bShuffleEnabled:bool)
    )
    WidgetCallingFrom=Self-Reference()
   )
   n23 = OnSaveConfirmed!() -> (OutputDelegate, then, NewName:String) {is=title K2Node_CustomEvent}
   > FireEventElementInteractedByName(Element Name=Append!(A=Get SaveString() -> (SaveString:String), B=" Confirmed") -> (ReturnValue), WidgetCallingFrom=Self-Reference())
   > ActivateWidget()
   > n24 = CopyModularLoadout!(
    execute:Exec
    Source:Fort Typed Modular Loadout VM Object Reference=Get ActiveLoadout!(self:Fort Typed Modular Loadouts VM Object Reference=Get FortTypedModularLoadoutsVM()) -> (ActiveLoadout:Fort Typed Modular Loadout VM Object Reference)
    Destination:Fort Typed Modular Loadout VM Object Reference=Get SaveDestination() -> (SaveDestination:Fort Typed Modular Loadout VM Object Reference)
    NewName:String=@n23.NewName
    bEquipLoadout:bool=true
   ) -> (then:Exec, OnOperationCompleted:Exec, Success:bool=~) {is=title K2Node_AsyncAction}
   Call Preset Saved!(
    execute:Exec=@n24.OnOperationCompleted
    self:WBP MPLocker Presets Object Reference
    EndInteraction:bool=Get IsSaveMode() -> (IsSaveMode:bool)
   ) -> (then:Exec) {is=title K2Node_CallDelegate}
   n25 = HandleLoadoutEquip!() -> (OutputDelegate, then, Loadout:Fort Typed Modular Loadout VM Object Reference) {is=title K2Node_CustomEvent}
   > ShowProgressModal!(execute, Title:Text="LOCTEXT(\"Loading\")", Description:Text="LOCTEXT(\"Applying Preset\")") -> (then)
   > changeSelectedLoadout!(execute, self, item:Object Reference=@n25.Loadout) -> (then)
   > Get!(execute:Exec) -> (then:Exec, else:Exec, SelectedFortTypedLoadoutVM:Fort Typed Modular Loadout VM Object Reference)
   > n26 = EquipModularLoadout!(
    execute:Exec
    Source:Fort Typed Modular Loadout VM Object Reference=Get SelectedFortTypedLoadoutVM() -> (SelectedFortTypedLoadoutVM:Fort Typed Modular Loadout VM Object Reference)
   ) -> (then:Exec, OnOperationCompleted:Exec, Success:bool=~) {is=title K2Node_AsyncAction}
   Call Loadout Load!(execute:Exec=@n26.OnOperationCompleted, self:WBP MPLocker Presets Object Reference) -> (then:Exec) {is=title K2Node_CallDelegate}
   > HideProgressModal!(execute) -> (then)
   > FireEventElementInteractedByName(Element Name=Preset Loaded, WidgetCallingFrom=Self-Reference())
   n27 = LoadoutClick!() -> (OutputDelegate, then, Button:Common Button Base Object Reference) {is=title K2Node_CustomEvent}
   > n28 = Set LoadoutButton(LoadoutButton:Common Button Base Object Reference=@n27.Button) -> (Output_Get:Common Button Base Object Reference)
   n29 = IsUsingTouch(PlayerController=GetOwningPlayer())
   > n30 = Branch(Condition=OR Boolean(A=GetSelected!(self:Common Button Base Object Reference=@n28) -> (ReturnValue:bool), B=NOT Boolean(A=@n29)))
   n31 = Get FortTypedModularLoadoutVM(self=Cast To WBP_UIKit_PresetCard!(Object=Get LoadoutButton() -> (LoadoutButton:Common Button Base Object Reference)) -> (AsWBP UIKit Preset Card, bSuccess:bool=~))
   > n32 = Branch(Condition=OR Boolean(A=Get bIsSaveAsNew!(self:Fort Typed Modular Loadout VM Object Reference=@n31) -> (bIsSaveAsNew:bool), B=Get IsSaveMode()))
   n33 = Get LoadoutButton()
   > n34 = Cast To WBP_UIKit_PresetCard!(execute:Exec, Object=@n33) -> (then:Exec, CastFailed:Exec, AsWBP UIKit Preset Card)
   > n35 = Set CurrentSelectedLoadout(CurrentSelectedLoadout:Fort Typed Modular Loadout VM Object Reference=Get FortTypedModularLoadoutVM(self=@n34)) -> (Output_Get:Fort Typed Modular Loadout VM Object Reference)
   SaveLoadout!(execute=[@n35.then, @n39.then], self, SaveDestination:Fort Typed Modular Loadout VM Object Reference=Get CurrentSelectedLoadout()) -> (then)
   n36 = Cast To WBP_UIKit_PresetCard(execute=@n32.else, Object=@n33)
   > n37 = Get!(execute, self:WBP UIKit Preset Card Object Reference=@n36) -> (then, else, FortTypedModularLoadoutVM:Fort Typed Modular Loadout VM Object Reference)
   > HandleLoadoutEquip(Loadout=@n37) {is=title K2Node_CallFunction}
   n38 = Cast To WBP_UIKit_PresetCard(execute=@n30.else, Object=Get LoadoutButton())
   > Update Presets Preview(PresetCard=@n38)
   > Clear Shuffle Description()
   > n39 = Branch(Condition=Get IsSaveMode())
   n40 = SaveShuffleClick!() -> (OutputDelegate, then, Button:Common Button Base Object Reference) {is=title K2Node_CustomEvent}
   > n41 = Branch(Condition=Get IsSaveMode())
   > n42 = GetFirstUnusedLoadout!(execute, self) -> (then, Item:Fort Typed Modular Loadout VM Object Reference)
   > SaveLoadout(SaveDestination=@n42)
   n43 = IsUsingTouch(PlayerController=GetOwningPlayer())
   n44 = Branch(execute=@n41.else, Condition=OR Boolean(A=GetSelected(self=@n40.Button), B=NOT Boolean(A=@n43)))
   DisableOptions() {is=title K2Node_CustomEvent}
   > Call Disable Options Hint!(execute:Exec, self:WBP MPLocker Presets Object Reference) -> (then:Exec) {is=title K2Node_CallDelegate}
   n45 = BP_OnItemScrolledIntoView_Event!() -> (OutputDelegate, then, Item:Object Reference, Widget:User Widget Object Reference) {is=title K2Node_CustomEvent}
   > Unbind Event from On Item Scrolled Into View(self=Get TileView_Presets(), Delegate=@n45)
   > n46 = Cast To WBP_UIKit_PresetCard(Object=@n45.Widget)
   > ToggleContextMenu(self=@n46)
   n47 = LoadoutHover!() -> (OutputDelegate, then, Button:Common Button Base Object Reference) {is=title K2Node_CustomEvent}
   > Clear Shuffle Description()
   > n48 = Cast To WBP_UIKit_PresetCard(Object=@n47.Button)
   n49 = Get FortTypedModularLoadoutVM(self=@n48)
   > Setup Preset Sub Items!(execute, self, Items:Object Reference=@n49) -> (then)
   > Preset Hovered!(execute, self, Preset Button:WBP UIKit Preset Card Object Reference=@n48) -> (then)
   > Is Age Gated!(execute, self, Current Active Loadout:Fort Typed Modular Loadout VM Object Reference=@n49) -> (then)
   DeletedPreset!() -> (OutputDelegate, then, NewName:String) {is=title K2Node_CustomEvent}
   > Setup Preset Sub Items()
   > FireEventElementInteractedByName(Element Name=Preset Delete Confirmed, WidgetCallingFrom=Self-Reference())
   > ActivateWidget()
   n50 = GetSelectedItemPresetCard!() -> (PresetCard:WBP UIKit Preset Card Object Reference)
   > Is Valid(InputObject=@n50)
   > SetVisibility(
    self=Get ContextMenu!(self:WBP UIKit Preset Card Object Reference=@n50) -> (ContextMenu:WBP UIKit Context Menu Anchor Button Object Reference)
    InVisibility=Collapsed
   )
   n51 = OnCMDeleteFIred!() -> (OutputDelegate, then, PresetCard:WBP UIKit Preset Card Object Reference) {is=title K2Node_CustomEvent}
   > LoadDeleteModal!(execute, self, PresetCard:WBP UIKit Preset Card Object Reference=@n51.PresetCard) -> (then)
   n52 = Event PreConstruct()
   > SetRowFill!(
    execute
    self:Grid Panel Object Reference=Get ContentMover() -> (ContentMover:Grid Panel Object Reference)
    RowIndex:int=0
    Coefficient:float=SelectFloat(A=0, B=1, bPickA=@n52.IsDesignTime)
   ) -> (then)
   > SetCompactUIStyle() {is=title K2Node_CallFunction}
   Event On Activated()
   > SetVisibility(self=[Get Text_NumPresets() -> (Text_NumPresets:Common Text Object Reference), Get ShuffleDescription()], InVisibility=SelfHitTestInvisible)
   > Clear Shuffle Description()
   > Is Age Gated(Current Active Loadout=Get ActiveLoadout(self=Get FortTypedModularLoadoutsVM()))
   > n53 = Branch(Condition=IsInSubGame(SubGame=Campaign))
   > RegisterInputAction!(
    execute
    self:Fort Global UIContext Object Reference=Get FortGlobalUIContext() -> (ReturnValue:Fort Global UIContext Object Reference) {is=title K2Node_GetSubsystem}
    UserWidget:User Widget Object Reference=Self-Reference()
    InputActionRow:Data Table Row Handle Structure=Make Data Table Row Handle!(DataTable:Data Table Object Reference=&InputActionDataTable, RowName:Name=ContextPopupMenu) -> (DataTableRowHandle:Data Table Row Handle Structure) {is=title K2Node_MakeStruct}
    CommitedEvent:Delegate=Create Event()
    InFilterPriority:int=1
   ) -> (then)
   > n54 = CenterWidget(ToCenter=GetDesiredFocusTarget!(self:Common Activatable Widget Object Reference) -> (ReturnValue:Widget Object Reference))
   Delay(execute=[@n54.then, @n53.else], Duration=0.5)
   > AddTelemetryImpression()
   > Set IsImpressionAvailable(IsImpressionAvailable=true)
   > CheckCoachMarks() {is=title K2Node_CallFunction}
   Event On Deactivated()
   > Set IsImpressionAvailable(IsImpressionAvailable=false)
   > Fire Event Flush Impressions!(execute, FortPC:Fort Player Controller Object Reference) -> (then)
   > SetVisibility(self=[Get Text_NumPresets(), Get ShuffleDescription()], InVisibility=Collapsed)
   > Reset!(execute, self:Dynamic Entry Box Object Reference=Get SubItems() -> (SubItems:Dynamic Entry Box Object Reference), bDeleteWidgets:bool=false) -> (then)
   > Clear Selection(self:List View Object Reference=Get TileView_Presets())
   > Clear Shuffle Description()
   > Set LastChangedIndex(LastChangedIndex:int=0) -> (Output_Get:int=0)
   > Branch(Condition=IsInSubGame(SubGame=Campaign))
   > UnregisterInputAction!(
    execute
    self:Fort Global UIContext Object Reference=Get FortGlobalUIContext()
    UserWidget:User Widget Object Reference=Self-Reference()
    InputActionRow:Data Table Row Handle Structure=Make Data Table Row Handle(DataTable=&InputActionDataTable, RowName=ContextPopupMenu)
   ) -> (then)
   LoadRenameModal!(execute=@n10.then, self, PresetCard:WBP UIKit Preset Card Object Reference=@n10.PresetCard) -> (then)
  ]}
 GetFirstUnusedLoadout {nodes [
   GetFirstUnusedLoadout() {is=title K2Node_FunctionEntry}
   > n1 = For Each Loop with Break!(
    Exec:Exec
    Array:Array of Object References=GetListItems!(self:List View Object Reference=Get TileView_Presets()) -> (ReturnValue:Array of Object References)
    Break:Exec=@n4.then
   ) -> (LoopBody:Exec, Array Element:Object Reference, Array Index:int, Completed:Exec)
   > n2 = Cast To FortTypedModularLoadoutVM(Object=@n1)
   > n3 = Branch(Condition=Get bIsValid(self=@n2))
   n4 = Set Result(execute=@n3.else, Result:Fort Typed Modular Loadout VM Object Reference=@n2) -> (Output_Get:Fort Typed Modular Loadout VM Object Reference)
   Return Node!(
    execute=@n1.Completed
    Item:Fort Typed Modular Loadout VM Object Reference=Get Result() -> (Result:Fort Typed Modular Loadout VM Object Reference)
   )
  ]}
 GetSelectedItemPresetCard {nodes [
   GetSelectedItemPresetCard() {is=title K2Node_FunctionEntry}
   > n1 = For Each Loop(Array:Array of User Widget Object References=GetDisplayedEntryWidgets(self=Get TileView_Presets())) -> (Array Element:User Widget Object Reference)
   Return Node!(execute=@n1.Completed, PresetCard:WBP UIKit Preset Card Object Reference)
   > n2 = Cast To WBP_UIKit_PresetCard(Object=@n1)
   n3 = Get TileView_Presets()
   > Branch(Condition="Equal (Object)"(A=Get FortTypedModularLoadoutVM(self=@n2), B=Get Selected Item(self=@n3)))
   > Return Node(PresetCard=@n2)
  ]}
 Is Active State {nodes [
   Is Active State!() -> (then, isActive:bool=~, Instant:bool=~) {is=title K2Node_FunctionEntry}
   n1 = Get IsActive() -> (IsActive:Widget Animation Object Reference)
   > PlayAnimation(
    InAnimation=@n1
    StartAtTime=SelectFloat(A=GetEndTime(self=@n1), B=0, bPickA=Get Instant() -> (Instant:bool))
    PlayMode=Select(Option 0:EUMGSequencePlayMode Enum=Reverse, Option 1:EUMGSequencePlayMode Enum, Index=Get isActive() -> (isActive:bool)) -> (ReturnValue:EUMGSequencePlayMode Enum)
   )
   > n2 = Branch(Condition=Get isActive())
   n3 = Get TileView_Presets()
   > SetVisibility(self=@n3, InVisibility=Visible)
   ScrollToTop(execute=@n2.else, self:List View Base Object Reference=@n3)
   > SetVisibility(self=@n3, InVisibility=HitTestInvisible)
   Scroll Item Into View(self=Get TileView_Presets(), Item=Get ActiveLoadout(self=Get FortTypedModularLoadoutsVM()))
   SetScrollbarVisibility!(execute, self:List View Base Object Reference, InVisibility:ESlateVisibility Enum=Visible) -> (then)
   SetScrollbarVisibility(InVisibility=Hidden)
  ]}
 Is Age Gated {nodes [
   n1 = Is Age Gated!() -> (then, Current Active Loadout:Fort Typed Modular Loadout VM Object Reference) {is=title K2Node_FunctionEntry}
   > n2 = Is Valid(InputObject=@n1)
   SetVisibility(execute=@n2.Is Not Valid, self=Get WarningMessage() -> (WarningMessage:WBP UIKit Info Box Object Reference), InVisibility=Collapsed)
   n3 = Get WarningMessage()
   SetText(
    execute=@n2.Is Valid
    self:WBP UIKit Info Box Object Reference=@n3
    InText:Text=Get AgeGatingDisclaimer!(self:Fort Typed Modular Loadout VM Object Reference=@n1) -> (AgeGatingDisclaimer:Text)
   )
   > SetVisibility(
    self=@n3
    InVisibility=Select(
     Option 0:ESlateVisibility Enum=Collapsed
     Option 1:ESlateVisibility Enum=HitTestInvisible
     Index=Get bContainsAgeGatedItems!(self:Fort Typed Modular Loadout VM Object Reference=@n1) -> (bContainsAgeGatedItems:bool)
    ) -> (ReturnValue:ESlateVisibility Enum)
   )
  ]}
 LoadDeleteModal {nodes [
   n1 = LoadDeleteModal!() -> (then, PresetCard:WBP UIKit Preset Card Object Reference) {is=title K2Node_FunctionEntry}
   n2 = Get FortTypedModularLoadoutVM(self=@n1)
   > Branch(Condition=IsValid(Object=@n2))
   > n3 = Construct Object from Class(Class=/Locker/MPL/Data/DialogVM_LockerDeleteLoadout.DialogVM_LockerDeleteLoadout_C) -> (ReturnValue:Dialog VM Locker Delete Loadout Object Reference)
   > n4 = Create Widget!(execute, Class=&/WBP_NameLoadout_MP.WBP_NameLoadout_MP_C, OwningPlayer, DialogVM:UI Kit Dialog Viewmodel Object Reference=@n3) -> (then, ReturnValue:WBP Name Loadout MP Object Reference)
   > n5 = Branch(Condition=IsInSubGame(SubGame=Campaign))
   > PushContentWidgetAdvanced!(execute, PushWidget:Widget Object Reference=@n4, bHideHeader:bool=true, bHideFooter:bool=true, bHideChatWidget:bool=true) -> (then)
   > n6 = ActivateWidget(self=@n4)
   DeactivateWidget(execute=[@n6.then, @n7.then])
   > Initialize!(
    execute
    self:WBP Name Loadout MP Object Reference=@n4
    Loadout Action:EFortLoadoutActionType Enum=Delete
    Target Loadout:Fort Typed Modular Loadout VM Object Reference=@n2
    TypedModularLoadouts:Fort Typed Modular Loadouts VM Object Reference=Get FortTypedModularLoadoutsVM()
   ) -> (then)
   > Bind Event to On Action Confirmed Event(self:Fort Save Loadout Popup Object Reference=@n4, Delegate=Create Event()) {is=title K2Node_AddDelegate}
   > Bind Event to On Action Declined Event(self:Fort Save Loadout Popup Object Reference=@n4, Delegate=Create Event()) {is=title K2Node_AddDelegate}
   > FireEventElementInteractedByName(Element Name=Preset Delete, WidgetCallingFrom=Self-Reference())
   > Set DeclineString(DeclineString:String=Preset Delete Cancelled) -> (Output_Get:String)
   n7 = TryPushWidgetAsModalPanel!(execute=@n5.else, PushWidget:User Widget Object Reference=@n4) -> (then)
  ]}
 LoadRenameModal {nodes [
   n1 = LoadRenameModal!() -> (then, PresetCard:WBP UIKit Preset Card Object Reference) {is=title K2Node_FunctionEntry}
   > n2 = Construct Object from Class(Class=/Locker/MPL/Data/DialogVM_LockerRename.DialogVM_LockerRename_C) -> (ReturnValue:Dialog VM Locker Rename Object Reference)
   > n3 = Create Widget(Class=&WBP_NameLoadout_MP_C, OwningPlayer=GetOwningPlayer(), DialogVM=@n2)
   > TryPushWidgetAsModalPanel(PushWidget=@n3)
   > Initialize(
    self=@n3
    Loadout Action=Rename
    Target Loadout=Get FortTypedModularLoadoutVM(self=@n1)
    TypedModularLoadouts=Get FortTypedModularLoadoutsVM()
   )
   > DeactivateWidget()
   > Bind Event to On Action Confirmed Event(self=@n3, Delegate=Create Event())
   > Bind Event to On Action Declined Event(self=@n3, Delegate=Create Event())
   > FireEventElementInteractedByName(Element Name=Preset Rename, WidgetCallingFrom=Self-Reference())
   > Set DeclineString(DeclineString=Preset Rename Cancelled)
  ]}
 LoadoutLoad {nodes [LoadoutLoad() {is=title K2Node_FunctionEntry}]}
 OnCoachMarkDeactivated {nodes [
   OnCoachMarkDeactivated() {is=title K2Node_FunctionEntry}
   > SetFocus(self=Get TileView_Presets())
   n1 = Get PresetCapCoachMark() -> (PresetCapCoachMark:WBP Locker Preset Cap Coach Mark Object Reference)
   > Unbind all Events from On Widget Deactivated!(execute:Exec, self:Common Activatable Widget Object Reference=@n1) -> (then:Exec) {is=title K2Node_ClearDelegate}
   > RemoveFromParent(self=@n1)
  ]}
 Preset Hovered {nodes [
   n1 = Preset Hovered!() -> (then, Preset Button:WBP UIKit Preset Card Object Reference) {is=title K2Node_FunctionEntry}
   > n2 = Set Hovered Preset Button(Hovered Preset Button:WBP UIKit Preset Card Object Reference=@n1) -> (Output_Get:WBP UIKit Preset Card Object Reference)
   > Call Preset Display Item!(execute:Exec, self:WBP MPLocker Presets Object Reference, Preset Button:WBP UIKit Preset Card Object Reference=@n2) -> (then:Exec) {is=title K2Node_CallDelegate}
   > Set LastChangedIndex(LastChangedIndex=GetIndexForItem!(self:List View Object Reference=Get TileView_Presets(), Item:Object Reference=Get FortTypedModularLoadoutVM(self=@n1)) -> (ReturnValue:int=0)) -> (Output_Get=0)
  ]}
 PresetDisplayItem {nodes [PresetDisplayItem!() -> (then, Preset Button:WBP UIKit Preset Card Object Reference) {is=title K2Node_FunctionEntry}]}
 PresetSaved {nodes [PresetSaved!() -> (then, EndInteraction:bool=~) {is=title K2Node_FunctionEntry}]}
 RandomButtonClicked {nodes [RandomButtonClicked() {is=title K2Node_FunctionEntry}]}
 SaveLoadout {nodes [
   n1 = SaveLoadout!() -> (then, SaveDestination:Fort Typed Modular Loadout VM Object Reference) {is=title K2Node_FunctionEntry}
   > Set SaveDestination(SaveDestination:Fort Typed Modular Loadout VM Object Reference=@n1) -> (Output_Get:Fort Typed Modular Loadout VM Object Reference)
   > n2 = Construct Object from Class(Class=&DialogVM_LockerRename_C)
   > n3 = Create Widget(Class=&WBP_NameLoadout_MP_C, DialogVM=@n2)
   > TryPushWidgetAsModalPanel(PushWidget=@n3)
   n4 = Get SaveDestination()
   > Initialize(
    self=@n3
    Loadout Action=Select(Option 0:EFortLoadoutActionType Enum=SaveNew, Option 1:EFortLoadoutActionType Enum=Overwrite, Index=Get bIsValid(self=@n4)) -> (ReturnValue:EFortLoadoutActionType Enum)
    Target Loadout=@n4
    TypedModularLoadouts=Get FortTypedModularLoadoutsVM()
   )
   > DeactivateWidget()
   n5 = Get bIsValid(self=Get SaveDestination())
   > Set SaveString(SaveString:String=Append(A="Presets ", B=SelectString(A=Overwrite Save, B=Save New, bPickA=@n5))) -> (Output_Get:String)
   > Bind Event to On Action Confirmed Event(self=@n3, Delegate=Create Event())
   > Bind Event to On Action Declined Event(self=@n3, Delegate=Create Event())
   > FireEventElementInteractedByName(Element Name=Get SaveString(), WidgetCallingFrom=Self-Reference())
   > Set DeclineString(DeclineString=Preset Save Cancelled)
  ]}
 ScrollListToTop {nodes [
   ScrollListToTop() {is=title K2Node_FunctionEntry}
   > ScrollToTop(self=Get TileView_Presets())
  ]}
 SetCompactUIStyle {nodes [
   SetCompactUIStyle() {is=title K2Node_FunctionEntry}
   > Branch(Condition=IsUsingCompactUI())
   > SetTileAlignment!(execute, self:Fort Tile View Object Reference=Get TileView_Presets(), Alignment:EListItemAlignment Enum=LeftAligned) -> (then)
   > Set WidthOverride!(execute, WidthOverride:float=680, self:Size Box Object Reference=Get SizeBox_Mobile() -> (SizeBox_Mobile:Size Box Object Reference)) -> (then, Output_Get:float)
   n1 = SlotAsHorizontalBoxSlot!(Widget:Widget Object Reference=Get SizeBox_Mobile()) -> (ReturnValue:Horizontal Box Slot Object Reference)
   > SetSize!(
    execute
    self:Horizontal Box Slot Object Reference=@n1
    InSize:Slate Child Size Structure=Make Slate Child Size!(Value:float=1, SizeRule:ESlateSizeRule Enum=Automatic) -> (SlateChildSize:Slate Child Size Structure) {is=title K2Node_MakeStruct}
   ) -> (then)
   > SetPadding!(execute, self:Horizontal Box Slot Object Reference=@n1, InPadding:Margin Structure=Make Margin(Left=-12, Top=0, Right=0, Bottom=0)) -> (then)
   > SetVerticalAlignment(
    self:Scroll Box Slot Object Reference=SlotAsScrollBoxSlot!(Widget:Widget Object Reference=Get SubItems()) -> (ReturnValue:Scroll Box Slot Object Reference)
    InVerticalAlignment=VAlign_Center
   )
  ]}
 SetHeader {nodes [
   n1 = SetHeader!() -> (then, Category:Text) {is=title K2Node_FunctionEntry}
   n2 = Format Text!(Format="LOCTEXT(\"Saved {Category}\")", Category:Text=@n1) -> (Result)
   > Set CurrentHeaderCategory(CurrentHeaderCategory:Text=@n2) -> (Output_Get:Text)
   > Set Text(
    Text=Select(Option 0:Text=@n2, Option 1:Text=Format Text!(Format="LOCTEXT(\"Select a slot to save\")") -> (Result), Index=Get IsSaveMode()) -> (ReturnValue:Text)
    self=Get Text_Header() -> (Text_Header:Common Text Object Reference)
   )
  ]}
 SetLoadouts {nodes [
   n1 = SetLoadouts!() -> (then, Items:Array of Object References) {is=title K2Node_FunctionEntry}
   > ClearListItems(self=Get TileView_Presets())
   > Clear(TargetArray:Array of Object References=Get Loadouts() -> (Loadouts:Array of Object References))
   n2 = Length(TargetArray:Array of Object References=@n1)
   n3 = "integer < integer"(A=@n2, B=200)
   n4 = Get IsSaveMode()
   > n5 = Branch(Condition=OR Boolean(A=AND Boolean(A=@n4, B=@n3), B="Equal (Integer)"(A=@n2)))
   n6 = TryGetSaveAsNewLoadout!(self:Fort Typed Modular Loadouts VM Object Reference=Get FortTypedModularLoadoutsVM()) -> (ReturnValue:Fort Typed Modular Loadout VM Object Reference)
   > n7 = Is Valid(InputObject=@n6)
   n8 = For Each Loop(Exec=[@n15.then, @n7.Is Not Valid], Array:Array of Object References=Get Items() -> (Items:Array of Object References)) -> (Array Element:Object Reference)
   n9 = Get TileView_Presets()
   n10 = Get Loadouts()
   Set List Items!(execute=@n8.Completed, self:List View Object Reference=@n9, InListItems:Array of Object References=@n10) -> (then) {is=title K2Node_CallFunction}
   > Set LastItemIndex(LastItemIndex:int=Last Index!(TargetArray:Array of Object References=@n10) -> (ReturnValue=0)) -> (Output_Get:int=0)
   > Branch(Condition=IsActivated())
   n11 = Get LastChangedIndex() -> (LastChangedIndex=0)
   > n12 = Branch(Condition="integer < integer"(A=@n11, B=1))
   > NavigateToIndex!(execute, self:List View Object Reference=@n9, Index:int=0) -> (then)
   n13 = Get LastItemIndex() -> (LastItemIndex=0)
   n14 = "integer > integer"(A=@n11, B=@n13)
   NavigateToIndex(
    execute=@n12.else
    self=@n9
    Index="Clamp (Integer)"(Value=Select(Option 0:int=@n11, Option 1:int=@n13, Index=@n14) -> (ReturnValue:int), Min=0, Max=@n13) -> (ReturnValue=0)
   )
   Add(execute=@n8.LoopBody, TargetArray:Array of Object References=Get Loadouts(), NewItem:"Object Reference (by ref)"=@n8)
   n15 = Add(execute=@n7.Is Valid, TargetArray=Get Loadouts(), NewItem=@n6)
   n16 = Branch(execute=@n5.else, Condition=Get IsSaveMode())
   n17 = For Each Loop(Exec=[@n22.then, @n21.Is Not Valid, @n16.then], Array=Get Items())
   n18 = Get Loadouts()
   n19 = Get TileView_Presets()
   Set List Items(execute=@n17.Completed, self=@n19, InListItems=@n18)
   > Set LastItemIndex(LastItemIndex=Last Index(TargetArray=@n18) -> (ReturnValue=0)) -> (Output_Get=0)
   > Clear Selection(self=@n19)
   Add(execute=@n17.LoopBody, TargetArray=Get Loadouts(), NewItem=@n17)
   n20 = TryGetShuffleLoadout!(self:Fort Typed Modular Loadouts VM Object Reference=Get FortTypedModularLoadoutsVM()) -> (ReturnValue:Fort Typed Modular Loadout VM Object Reference)
   n21 = Is Valid(exec=@n16.else, InputObject=@n20)
   > n22 = Add(TargetArray=Get Loadouts(), NewItem=@n20)
  ]}
 SetNumLoadouts {nodes [
   n1 = SetNumLoadouts!() -> (then, NumLoadouts:int) {is=title K2Node_FunctionEntry}
   n2 = GetMaxLoadoutCount!(self:Fort Typed Modular Loadouts VM Object Reference=Get FortTypedModularLoadoutsVM()) -> (ReturnValue:int=0)
   n3 = Get Text_NumPresets()
   > "SetText (Text)"(
    self=@n3
    InText=Format Text!(Format="LOCTEXT(\"{NumLoadouts} OF {MaxLoadouts}\")", NumLoadouts:int=@n1, MaxLoadouts:int=int - int(A=@n2)) -> (Result)
   )
   n4 = Get NumLoadouts() -> (NumLoadouts:int=0)
   > SetColorAndOpacity(
    self=@n3
    InColorAndOpacity=Select(
     Option 0:Slate Color Structure=Make Slate Color(SpecifiedColor=(R=0.5776,G=0.5776,B=0.5776,A=1))
     Option 1:Slate Color Structure=Make Slate Color(SpecifiedColor=(R=1,G=0.147,B=0.1329,A=1))
     Index="integer > integer"(A=@n4, B=@n2)
    ) -> (ReturnValue:Slate Color Structure)
   )
  ]}
 SetSaveMode {nodes [
   n1 = SetSaveMode!() -> (then, SaveMode:bool=~) {is=title K2Node_FunctionEntry}
   > Set IsSaveMode(IsSaveMode:bool=@n1) -> (Output_Get:bool)
   > n2 = Set Typed Modular Loadouts(Typed Modular Loadouts:Array of Fort Typed Modular Loadout VM Object References=Get TypedModularLoadouts!(self:Fort Typed Modular Loadouts VM Object Reference=Get FortTypedModularLoadoutsVM()) -> (TypedModularLoadouts:Array of Fort Typed Modular Loadout VM Object References)) -> (Output_Get:Array of Fort Typed Modular Loadout VM Object References)
   > SetLoadouts!(execute, self, Items:Array of Object References=@n2) -> (then) {is=title K2Node_CallFunction}
  ]}
 SetScrollTimer {nodes [
   n1 = SetScrollTimer!() -> (then, ScrollProgress:double=~) {is=title K2Node_FunctionEntry}
   n2 = Get SubItems_ScrollBox() -> (SubItems_ScrollBox:Scroll Box Object Reference)
   > SetScrollOffset!(
    execute
    self:Scroll Box Object Reference=Get SubItems_ScrollBox()
    NewScrollOffset:float=Lerp(A=0, B=GetScrollOffsetOfEnd!(self:Scroll Box Object Reference=@n2) -> (ReturnValue:float), Alpha=@n1)
   ) -> (then)
  ]}
 Setup Preset Sub Items {nodes [
   n1 = Setup Preset Sub Items!() -> (then, Items:Object Reference) {is=title K2Node_FunctionEntry}
   > Reset(self=Get SubItems(), bDeleteWidgets=false)
   > n2 = Cast To FortTypedModularLoadoutVM(Object=@n1)
   > Branch(Condition=Get bIsValid(self=@n2))
   > n3 = For Each Loop(Array:Array of Fort Modular Loadout Item VM Object References=Get LoadoutItems(self=@n2)) -> (Array Element:Fort Modular Loadout Item VM Object Reference)
   > n4 = Create Entry!(execute, self:Dynamic Entry Box Object Reference=Get SubItems()) -> (then, ReturnValue:User Widget Object Reference)
   > n5 = Cast To WBP_UIKit_PresetSubTile!(execute, Object=@n4) -> (then, CastFailed, AsWBP UIKit Preset Sub Tile:WBP UIKit Preset Sub Tile Object Reference) {is=title K2Node_DynamicCast}
   > SetFortModularLoadoutItemVM!(execute, self:WBP UIKit Preset Sub Tile Object Reference=@n5, Viewmodel:Fort Modular Loadout Item VM Object Reference) -> (then)
   > SetFortModularLoadoutItemVM(self=@n5, Viewmodel=@n3)
  ]}
 Show Context Menu for Current Selected Item {nodes [
   Show Context Menu for Current Selected Item() {is=title K2Node_FunctionEntry}
   > Set ToggleSuccess(ToggleSuccess:bool=false) -> (Output_Get:bool)
   > n1 = For Each Loop with Break(Array:Array of User Widget Object References=GetDisplayedEntryWidgets(self=Get TileView_Presets()), Break=@n4.then) -> (Array Element:User Widget Object Reference)
   Return Node!(execute=@n1.Completed, bPassTrough:bool=NOT Boolean(A=Get ToggleSuccess() -> (ToggleSuccess:bool)))
   > n2 = Cast To WBP_UIKit_PresetCard(Object=@n1)
   > n3 = Toggle Context Menu If Hovered!(execute, self:WBP UIKit Preset Card Object Reference=@n2) -> (then, ToggleSucess:bool)
   > Branch(Condition=@n3)
   > n4 = Set ToggleSuccess(ToggleSuccess=true)
  ]}
 ToggleShuffleEnabled {nodes [
   ToggleShuffleEnabled() {is=title K2Node_FunctionEntry}
   n1 = Get FortTypedModularLoadoutsVM()
   > Set bShuffleEnabled!(execute, bShuffleEnabled:bool=NOT Boolean(A=Get bShuffleEnabled(self=@n1)), self:Fort Typed Modular Loadouts VM Object Reference=@n1) -> (then, Output_Get:bool)
  ]}
 Update Presets Preview {nodes [
   n1 = Update Presets Preview!() -> (then, PresetCard:WBP UIKit Preset Card Object Reference) {is=title K2Node_FunctionEntry}
   n2 = Get FortTypedModularLoadoutVM(self=@n1)
   > n3 = Branch(Condition=Get bIsShuffle(self=@n2))
   > Preset Hovered(Preset Button=@n1) {is=title K2Node_CallFunction}
   > Reset(self=Get SubItems(), bDeleteWidgets=false)
   > UpdateShuffleDescription!(execute, self, Void:bool=false) -> (then)
   > SetVisibility(self=Get WarningMessage(), InVisibility=Collapsed)
   n4 = Set CurrentSelectedLoadout(execute=@n3.else, CurrentSelectedLoadout=@n2)
   > Setup Preset Sub Items(Items=@n4) {is=title K2Node_CallFunction}
   > Preset Hovered(Preset Button=@n1)
   > Is Age Gated(Current Active Loadout=Get CurrentSelectedLoadout()) {is=title K2Node_CallFunction}
  ]}
 UpdateScroll {nodes [
   n1 = Get TopScrollOffsetMax() -> (TopScrollOffsetMax:double)
   n2 = Get ItemOffset() -> (ItemOffset:double)
   n3 = Get FullImageSizeY() -> (FullImageSizeY:double)
   n4 = Lerp(A=0, B=@n3, Alpha=SafeDivide(A=@n2, B=@n1))
   n5 = "float > float"(A=@n1, B=@n2)
   n6 = Get ItemOffset()
   SetBrush!(
    execute
    self=Get Image_Fade_Top() -> (Image_Fade_Top:Image Object Reference)
    InBrush_TintColor:Slate Color Structure=(SpecifiedColor=(R=1,G=1,B=1,A=1),ColorUseRule=UseColor_Specified)
    InBrush_DrawAs:ESlateBrushDrawType Enum=Image
    InBrush_Tiling:ESlateBrushTileType Enum=NoTile
    InBrush_Mirroring:ESlateBrushMirrorType Enum=NoMirror
    InBrush_ImageSize_X:double=0
    InBrush_ImageSize_Y:double=SelectFloat(A=0, B=SelectFloat(A=@n4, B=@n3, bPickA=@n5), bPickA="Equal (Float)"(A=@n6))
    InBrush_Margin:Margin Structure=(Left=0,Top=0,Right=0,Bottom=0)
    InBrush_ResourceObject:Object Reference=/Locker/MPL/Materials/MI_UI_ItemsBlueMaskTop
    InBrush_OutlineSettings:Slate Brush Outline Settings Structure=(CornerRadii=(X=0,Y=0,Z=0,W=0),Color=(SpecifiedColor=(R=0,G=0,B=0,A=0),ColorUseRule=UseColor_Specified),Width=0,RoundingType=HalfHeightRadius,bUseBrushTransparency=False)
   ) -> (then)
   n7 = Get TileView_Presets()
   n8 = Get LastItemIndex() -> (LastItemIndex=0)
   > n9 = Branch(Condition=Is Item Visible!(
     self:List View Object Reference=@n7
     Item:Object Reference=GetItemAt!(self:List View Object Reference=@n7, Index:int=@n8) -> (ReturnValue:Object Reference)
    ) -> (ReturnValue:bool))
   > n10 = Branch(Condition=OR Boolean(A=Get FirstTimeFadeCheck() -> (FirstTimeFadeCheck:bool), B=Get ResetFadeMath() -> (ResetFadeMath:bool)))
   > Set LastTileDistance(LastTileDistance:float=Get DistanceRemaining() -> (DistanceRemaining:double)) -> (Output_Get:float)
   > Set FirstTimeFadeCheck(FirstTimeFadeCheck:bool=false) -> (Output_Get:bool)
   > n11 = Set ResetFadeMath(ResetFadeMath:bool=false) -> (Output_Get:bool)
   n12 = Get LastTileDistance() -> (LastTileDistance:float)
   n13 = Get DistanceRemaining()
   SetBrush(
    execute=[@n10.else, @n11.then]
    self=Get Image_Fade_Bottom() -> (Image_Fade_Bottom:Image Object Reference)
    InBrush_TintColor=(SpecifiedColor=(R=1,G=1,B=1,A=1),ColorUseRule=UseColor_Specified)
    InBrush_DrawAs=Image
    InBrush_Tiling=NoTile
    InBrush_Mirroring=NoMirror
    InBrush_ImageSize_X=0
    InBrush_ImageSize_Y=Lerp(A=0, B=Get FullImageSizeY(), Alpha=MapRangeClamped(Value=@n13, InRangeA=@n12, InRangeB=0, OutRangeA=1, OutRangeB=0))
    InBrush_Margin=(Left=0,Top=0,Right=0,Bottom=0)
    InBrush_ResourceObject=/Locker/MPL/Materials/MI_UI_ItemsBlueMaskBottom
    InBrush_OutlineSettings=(CornerRadii=(X=0,Y=0,Z=0,W=0),Color=(SpecifiedColor=(R=0,G=0,B=0,A=0),ColorUseRule=UseColor_Specified),Width=0,RoundingType=HalfHeightRadius,bUseBrushTransparency=False)
   )
   Set ResetFadeMath(execute=@n9.else, ResetFadeMath=true)
   n14 = Get FullImageSizeY()
   > SetBrush(
    self=Get Image_Fade_Bottom()
    InBrush_TintColor=(SpecifiedColor=(R=1,G=1,B=1,A=1),ColorUseRule=UseColor_Specified)
    InBrush_DrawAs=Image
    InBrush_Tiling=NoTile
    InBrush_Mirroring=NoMirror
    InBrush_ImageSize_X=0
    InBrush_ImageSize_Y=@n14
    InBrush_Margin=(Left=0,Top=0,Right=0,Bottom=0)
    InBrush_ResourceObject=&MI_UI_ItemsBlueMaskBottom
    InBrush_OutlineSettings=(CornerRadii=(X=0,Y=0,Z=0,W=0),Color=(SpecifiedColor=(R=0,G=0,B=0,A=0),ColorUseRule=UseColor_Specified),Width=0,RoundingType=HalfHeightRadius,bUseBrushTransparency=False)
   )
   n15 = Get ItemOffset()
   n16 = Get DistanceRemaining()
   SelectFloat(A=0, B=@n14, bPickA=AND Boolean(A="Equal (Float)"(A=@n15), B="Equal (Float)"(A=@n16)))
   UpdateScroll!() -> (then, ItemOffset:double=~, DistanceRemaining:double=~) {is=title K2Node_FunctionEntry}
  ]}
 UpdateShuffleDescription {nodes [
   UpdateShuffleDescription!() -> (then, Void:bool=~) {is=title K2Node_FunctionEntry}
   n1 = Get FortTypedModularLoadoutsVM()
   > n2 = Is Valid(InputObject=@n1)
   n3 = Get bShuffleEnabled(self=@n1)
   n4 = Get ShuffleDescription()
   SetTitleText(execute=[@n6.else, @n2.Is Not Valid], self=@n4, Text=SelectText(A="LOCTEXT(\"Random On\")", B="LOCTEXT(\"Random Off\")", bPickA=@n3))
   n5 = Get CurrentHeaderCategory() -> (CurrentHeaderCategory:Text)
   > SetBodyText(
    self=@n4
    Text=SelectText(
     A=Format Text!(Format="LOCTEXT(\"Your \\'{Category}\\' presets are cycling in a random order. Any changes will turn it off.\")", Category:Text=@n5) -> (Result)
     B=Format Text(Format="LOCTEXT(\"Turn on to cycle your \\'{Category}\\' presets in a random order.\")", Category=@n5)
     bPickA=@n3
    )
   )
   n6 = Branch(execute=@n2.Is Valid, Condition=NOT Boolean(A=GetCanEnableShuffle(self=@n1)))
   n7 = Get ShuffleDescription()
   > SetTitleText(self=@n7, Text="LOCTEXT(\"Random Off\")")
   > SetBodyText(
    self=@n7
    Text=Format Text(Format="LOCTEXT(\"You need at least 2 \\'{Category}\\' presets in order to turn on Random.\")", Category=Get CurrentHeaderCategory())
   )
  ]}
 changeSelectedLoadout {nodes [
   n1 = changeSelectedLoadout!() -> (then, item:Object Reference) {is=title K2Node_FunctionEntry}
   > n2 = Cast To FortTypedModularLoadoutVM(Object=@n1)
   > SetSelectedFortTypedLoadoutVM!(execute, self, Viewmodel:Fort Typed Modular Loadout VM Object Reference=@n2) -> (then)
   > Setup Preset Sub Items(Items=@n2)
   > Set CurrentSelectedLoadout(CurrentSelectedLoadout=@n2)
  ]}
]
variables [
 ChangeBackToTopButtonVisibility {}
 CurrentHeaderCategory {type=string, default=""}
 CurrentSelectedLoadout {type=FortTypedModularLoadoutVM, default=null}
 DeclineString {type=string, default=""}
 DisableOptionsHint {}
 EnableOptionsHint {}
 FirstTimeFadeCheck {type=boolean}
 Hovered Preset Button {type=WBP_UIKit_PresetCard_C, default=null}
 IsImpressionAvailable {type=boolean}
 IsSaveMode {type=boolean}
 LastChangedIndex {type=integer}
 LastItemIndex {type=integer}
 LastTileDistance {type=number}
 LoadoutButton {type=CommonButtonBase, default=null}
 LoadoutLoad {}
 Loadouts {type="array<Object>", default []}
 PresetCapCoachMark {type=WBP_Locker_PresetCap_CoachMark_C, default=null}
 PresetCapIncreaseSeenID {type=GameplayTag, default {tagName=UI.Tutorials.PresetCountIncreaseCoachMark}}
 PresetDisplayItem {}
 PresetLoadoutFlowRiderString {type=string, default=""}
 PresetSaved {}
 RandomButtonClicked {}
 ResetFadeMath {type=boolean}
 SaveDestination {type=FortTypedModularLoadoutVM, default=null}
 SaveShuffleButton {type=WBP_UIKit_PresetCard_C, default=null}
 SaveString {type=string, default=""}
 ScrollTimer {type=number}
 Time {type=number}
 TimerHandle {type=TimerHandle}
 Typed Modular Loadouts {type="array<FortTypedModularLoadoutVM>", default []}
]
widgets {
 info {
  inheritedWidgetCount=0
  namedSlotCount=1
  parentClass=&CommonActivatableWidget
  rootWidgetClass=&GridPanel
  widgetCount=20
 }
 widgets [
  ContentMover {widgetClassPath=&GridPanel, bIsVariable=true, widget=ContentMover, properties {columnFill [1], rowFill [1,0.2]}}
  Maincontent {
   widgetClassPath=&GridPanel
   parent=ContentMover
   widget=Maincontent
   properties {columnFill [1,1], rowFill [0,0.1,0.9]}
   slot {class=GridSlot, properties {row=1}}
  }
  TopPadding {
   widgetClassPath=&Spacer
   bIsVariable=true
   parent=Maincontent
   widget=TopPadding
   properties {size=<1 0>}
   slot {class=GridSlot, properties {padding=<ltrb 0 18 0 0>}}
  }
  Text_Header {
   widgetClassPath=&CommonTextBlock
   bIsVariable=true
   parent=Maincontent
   widget=Text_Header
   properties {
    text=Saved Characters
    textTransformPolicy=ToUpper
    font {
     bForceMonospaced=false
     bMaterialIsStencil=false
     fontMaterial=null
     fontObject=&HeadingNow
     letterSpacing=0
     monospacedWidth=1
     size=30
     skewAmount=0
     typefaceFontName=86 Bold Italic
     outlineSettings {
      bApplyOutlineToDropShadows=false
      bMiteredCorners=false
      bSeparateFillAlpha=false
      outlineMaterial=null
      outlineSize=0
      outlineColor=<rgba 0 0 0 1>
     }
    }
    shadowOffset=<0 0>
   }
   slot {class=GridSlot, properties {horizontalAlignment=HAlign_Left, row=1, padding=<ltrb 0 12 0 16>}}
  }
  CoachMark_HBox {
   widgetClassPath=&HorizontalBox
   bIsVariable=true
   parent=Maincontent
   widget=CoachMark_HBox
   slot {class=GridSlot, properties {horizontalAlignment=HAlign_Left, layer=1, row=1, padding=<ltrb 0 60 0 16>}}
  }
  Text_NumPresets {
   widgetClassPath=&CommonTextBlock
   bIsVariable=true
   parent=CoachMark_HBox
   widget=Text_NumPresets
   properties {
    mobileFontSizeMultiplier=1.2
    text=100 OF 100
    textTransformPolicy=ToUpper
    visibility=Collapsed
    font {
     bForceMonospaced=false
     bMaterialIsStencil=false
     fontMaterial=null
     fontObject=&HeadingNow
     letterSpacing=0
     monospacedWidth=1
     size=15
     skewAmount=0
     typefaceFontName=75 Medium
     outlineSettings {
      bApplyOutlineToDropShadows=false
      bMiteredCorners=false
      bSeparateFillAlpha=false
      outlineMaterial=null
      outlineSize=0
      outlineColor=<rgba 0 0 0 1>
     }
    }
    shadowOffset=<0 0>
   }
   slot {class=HorizontalBoxSlot}
  }
  CoachMark_Attachment {
   widgetClassPath=&FortVisualAttachment
   bIsVariable=true
   parent=CoachMark_HBox
   widget=CoachMark_Attachment
   properties {navigation=&self:WidgetTree.CoachMark_Attachment.WidgetNavigation_0, contentAnchor=<0 0.5>}
   slot {class=HorizontalBoxSlot, properties {padding=<ltrb 30 0 0 0>}}
  }
  HorizontalBox_0 {
   widgetClassPath=&HorizontalBox
   parent=Maincontent
   widget=HorizontalBox_0
   slot {class=GridSlot, properties {row=2, padding=<ltrb 0 18 0 0>}}
  }
  SizeBox_Mobile {
   widgetClassPath=&SizeBox
   bIsVariable=true
   parent=HorizontalBox_0
   widget=SizeBox_Mobile
   properties {bOverride_WidthOverride=true, widthOverride=670}
   slot {class=HorizontalBoxSlot, properties {size {sizeRule=Fill, value=1}}}
  }
  Overlay_0 {widgetClassPath=&Overlay, parent=SizeBox_Mobile, widget=Overlay_0, slot {class=SizeBoxSlot}}
  CHVB_NonTouch {
   widgetClassPath=&CommonHardwareVisibilityBorder
   parent=Overlay_0
   widget=CHVB_NonTouch
   properties {visibilityQuery {
     autoDescription=" NONE( Input.Touch )"
     tokenStreamVersion=0
     userDescription=Is Not Touch
     queryTokenStream [0,1,3,1,0]
     tagDictionary [{tagName=Input.Touch}]
    }}
   slot {class=OverlaySlot, properties {horizontalAlignment=HAlign_Right, verticalAlignment=VAlign_Bottom, padding=<ltrb 0 0 -60 0>}}
  }
  WBP_MPLocker_BackToTopButton {
   widgetClassPath=&/WBP_MPLocker_BackToTopButton
   bIsVariable=true
   parent=CHVB_NonTouch
   widget=WBP_MPLocker_BackToTopButton
   properties {targetList=&self:WidgetTree.TileView_Presets}
   slot {class=BorderSlot, properties {padding=<ltrb 0 0 0 0>}}
  }
  TileView_Presets {
   widgetClassPath=/Script/FortniteUI.FortTileView
   bIsVariable=true
   parent=Overlay_0
   widget=TileView_Presets
   properties {
    bEnableMobileScrollBarThicknessOverride=true
    consumeMouseWheel=Never
    entryHeight=208
    entryWidgetClass=/FortUILibrary/UIKit/Items/WBP_UIKit_PresetCard.WBP_UIKit_PresetCard_C
    entryWidth=208
    horizontalEntrySpacing=24
    mobileScrollBarThicknessOverride=24
    navigation=&self:WidgetTree.TileView_Presets.WidgetNavigation_0
    numDesignerPreviewEntries=13
    tileAlignment=EvenlySize
    verticalEntrySpacing=24
    scrollBarStyle {
     thickness=10
     draggedThumbImage {
      bIsDynamicallyLoaded=false
      drawAs=Box
      imageType=FullColor
      mirroring=NoMirror
      resourceName=../../../Engine/Content/Slate/Common/Scrollbar_Thumb.png
      resourceObject=/FortUILibrary/Foundation/Art_ToolKit/Tools/M_UI_Circle_White
      tiling=NoTile
      imageSize=<10 10>
      margin=<ltrb 0.5 0.5 0.5 0.5>
      outlineSettings {
       bUseBrushTransparency=false
       roundingType=HalfHeightRadius
       width=0
       color {colorUseRule=UseColor_Specified, specifiedColor=<rgba 0 0 0 0>}
       cornerRadii=<0 0 0 0>
      }
      tintColor {colorUseRule=UseColor_Specified, specifiedColor=<rgba 1 1 1 1>}
      uVRegion {bIsValid=false, max=<0 0>, min=<0 0>}
     }
     horizontalBackgroundImage {
      bIsDynamicallyLoaded=false
      drawAs=NoDrawType
      imageType=NoImage
      mirroring=NoMirror
      resourceName=None
      resourceObject=null
      tiling=NoTile
      imageSize=<0 0>
      margin=<ltrb 0 0 0 0>
      outlineSettings {
       bUseBrushTransparency=false
       roundingType=HalfHeightRadius
       width=0
       color {colorUseRule=UseColor_Specified, specifiedColor=<rgba 0 0 0 0>}
       cornerRadii=<0 0 0 0>
      }
      tintColor {colorUseRule=UseColor_Specified, specifiedColor=<rgba 1 1 1 1>}
      uVRegion {bIsValid=false, max=<0 0>, min=<0 0>}
     }
     horizontalBottomSlotImage {
      bIsDynamicallyLoaded=false
      drawAs=Image
      imageType=FullColor
      mirroring=NoMirror
      resourceName=../../../Engine/Content/Slate/Common/Scrollbar_Background_Horizontal.png
      resourceObject=null
      tiling=NoTile
      imageSize=<8 8>
      margin=<ltrb 0.5 0.5 0.5 0.5>
      outlineSettings {
       bUseBrushTransparency=false
       roundingType=HalfHeightRadius
       width=0
       color {colorUseRule=UseColor_Specified, specifiedColor=<rgba 0 0 0 0>}
       cornerRadii=<0 0 0 0>
      }
      tintColor {colorUseRule=UseColor_Specified, specifiedColor=<rgba 1 1 1 1>}
      uVRegion {bIsValid=false, max=<0 0>, min=<0 0>}
     }
     horizontalTopSlotImage {
      bIsDynamicallyLoaded=false
      drawAs=Image
      imageType=FullColor
      mirroring=NoMirror
      resourceName=../../../Engine/Content/Slate/Common/Scrollbar_Background_Horizontal.png
      resourceObject=null
      tiling=NoTile
      imageSize=<8 8>
      margin=<ltrb 0 0 0 0>
      outlineSettings {
       bUseBrushTransparency=false
       roundingType=HalfHeightRadius
       width=0
       color {colorUseRule=UseColor_Specified, specifiedColor=<rgba 0 0 0 0>}
       cornerRadii=<0 0 0 0>
      }
      tintColor {colorUseRule=UseColor_Specified, specifiedColor=<rgba 1 1 1 1>}
      uVRegion {bIsValid=false, max=<0 0>, min=<0 0>}
     }
     hoveredThumbImage {
      bIsDynamicallyLoaded=false
      drawAs=Box
      imageType=FullColor
      mirroring=NoMirror
      resourceName=../../../Engine/Content/Slate/Common/Scrollbar_Thumb.png
      resourceObject=&M_UI_Circle_White
      tiling=NoTile
      imageSize=<10 10>
      margin=<ltrb 0.5 0.5 0.5 0.5>
      outlineSettings {
       bUseBrushTransparency=false
       roundingType=HalfHeightRadius
       width=0
       color {colorUseRule=UseColor_Specified, specifiedColor=<rgba 0 0 0 0>}
       cornerRadii=<0 0 0 0>
      }
      tintColor {colorUseRule=UseColor_Specified, specifiedColor=<rgba 1 1 1 1>}
      uVRegion {bIsValid=false, max=<0 0>, min=<0 0>}
     }
     normalThumbImage {
      bIsDynamicallyLoaded=false
      drawAs=Box
      imageType=FullColor
      mirroring=NoMirror
      resourceName=../../../Engine/Content/Slate/Common/Scrollbar_Thumb.png
      resourceObject=&M_UI_Circle_White
      tiling=NoTile
      imageSize=<10 10>
      margin=<ltrb 0.5 0.5 0.5 0.5>
      outlineSettings {
       bUseBrushTransparency=false
       roundingType=HalfHeightRadius
       width=0
       color {colorUseRule=UseColor_Specified, specifiedColor=<rgba 0 0 0 0>}
       cornerRadii=<0 0 0 0>
      }
      tintColor {colorUseRule=UseColor_Specified, specifiedColor=<rgba 1 1 1 1>}
      uVRegion {bIsValid=false, max=<0 0>, min=<0 0>}
     }
     verticalBackgroundImage {
      bIsDynamicallyLoaded=false
      drawAs=Box
      imageType=NoImage
      mirroring=NoMirror
      resourceName=None
      resourceObject=&M_UI_Circle_White
      tiling=NoTile
      imageSize=<10 10>
      margin=<ltrb 0.5 0.5 0.5 0.5>
      outlineSettings {
       bUseBrushTransparency=false
       roundingType=HalfHeightRadius
       width=0
       color {colorUseRule=UseColor_Specified, specifiedColor=<rgba 0 0 0 0>}
       cornerRadii=<0 0 0 0>
      }
      tintColor {colorUseRule=UseColor_Specified, specifiedColor=<rgba 1 1 1 0.2>}
      uVRegion {bIsValid=false, max=<0 0>, min=<0 0>}
     }
     verticalBottomSlotImage {
      bIsDynamicallyLoaded=false
      drawAs=NoDrawType
      imageType=FullColor
      mirroring=NoMirror
      resourceName=../../../Engine/Content/Slate/Common/Scrollbar_Background_Vertical.png
      resourceObject=null
      tiling=NoTile
      imageSize=<10 10>
      margin=<ltrb 0 0 0 0>
      outlineSettings {
       bUseBrushTransparency=false
       roundingType=HalfHeightRadius
       width=0
       color {colorUseRule=UseColor_Specified, specifiedColor=<rgba 0 0 0 0>}
       cornerRadii=<0 0 0 0>
      }
      tintColor {colorUseRule=UseColor_Specified, specifiedColor=<rgba 1 1 1 1>}
      uVRegion {bIsValid=false, max=<0 0>, min=<0 0>}
     }
     verticalTopSlotImage {
      bIsDynamicallyLoaded=false
      drawAs=NoDrawType
      imageType=FullColor
      mirroring=NoMirror
      resourceName=../../../Engine/Content/Slate/Common/Scrollbar_Background_Vertical.png
      resourceObject=null
      tiling=NoTile
      imageSize=<8 8>
      margin=<ltrb 0 0 0 0>
      outlineSettings {
       bUseBrushTransparency=false
       roundingType=HalfHeightRadius
       width=0
       color {colorUseRule=UseColor_Specified, specifiedColor=<rgba 0 0 0 0>}
       cornerRadii=<0 0 0 0>
      }
      tintColor {colorUseRule=UseColor_Specified, specifiedColor=<rgba 1 1 1 1>}
      uVRegion {bIsValid=false, max=<0 0>, min=<0 0>}
     }
    }
   }
   slot {class=OverlaySlot, properties {horizontalAlignment=HAlign_Fill, verticalAlignment=VAlign_Fill}}
  }
  Image_Fade_Top {
   widgetClassPath=&Image
   bIsVariable=true
   parent=Overlay_0
   widget=Image_Fade_Top
   properties {
    visibility=Collapsed
    brush {
     bIsDynamicallyLoaded=false
     drawAs=Image
     imageType=NoImage
     mirroring=NoMirror
     resourceName=None
     resourceObject=&MI_UI_ItemsBlueMaskTop
     tiling=NoTile
     imageSize=<32 0>
     margin=<ltrb 0 0 0 0>
     outlineSettings {
      bUseBrushTransparency=false
      roundingType=HalfHeightRadius
      width=0
      color {colorUseRule=UseColor_Specified, specifiedColor=<rgba 0 0 0 0>}
      cornerRadii=<0 0 0 0>
     }
     tintColor {colorUseRule=UseColor_Specified, specifiedColor=<rgba 1 1 1 1>}
     uVRegion {bIsValid=false, max=<0 0>, min=<0 0>}
    }
   }
   slot {class=OverlaySlot, properties {horizontalAlignment=HAlign_Fill, padding=<ltrb 0 0 20 0>}}
  }
  Image_Fade_Bottom {
   widgetClassPath=&Image
   bIsVariable=true
   parent=Overlay_0
   widget=Image_Fade_Bottom
   properties {
    visibility=Collapsed
    brush {
     bIsDynamicallyLoaded=false
     drawAs=Image
     imageType=NoImage
     mirroring=NoMirror
     resourceName=None
     resourceObject=&MI_UI_ItemsBlueMaskBottom
     tiling=NoTile
     imageSize=<32 0>
     margin=<ltrb 0 0 0 0>
     outlineSettings {
      bUseBrushTransparency=false
      roundingType=HalfHeightRadius
      width=0
      color {colorUseRule=UseColor_Specified, specifiedColor=<rgba 0 0 0 0>}
      cornerRadii=<0 0 0 0>
     }
     tintColor {colorUseRule=UseColor_Specified, specifiedColor=<rgba 1 1 1 1>}
     uVRegion {bIsValid=false, max=<0 0>, min=<0 0>}
    }
   }
   slot {class=OverlaySlot, properties {horizontalAlignment=HAlign_Fill, verticalAlignment=VAlign_Bottom, padding=<ltrb 0 0 20 0>}}
  }
  RightContent {
   widgetClassPath=&GridPanel
   parent=Maincontent
   widget=RightContent
   properties {columnFill [1,0], rowFill [1]}
   slot {class=GridSlot, properties {column=1, horizontalAlignment=HAlign_Right, row=1, rowSpan=2, padding=<ltrb 0 0 80 132>}}
  }
  SubItems_ScrollBox {
   widgetClassPath=&ScrollBox
   bIsVariable=true
   parent=RightContent
   widget=SubItems_ScrollBox
   properties {visibility=SelfHitTestInvisible}
   slot {class=GridSlot, properties {column=1, horizontalAlignment=HAlign_Right}}
  }
  SubItems {
   widgetClassPath=&DynamicEntryBox
   bIsVariable=true
   parent=SubItems_ScrollBox
   widget=SubItems
   properties {entryBoxType=Vertical, entryWidgetClass=/FortUILibrary/UIKit/Items/WBP_UIKit_PresetSubTile.WBP_UIKit_PresetSubTile_C, entrySpacing=<8 8>}
   slot {class=ScrollBoxSlot, properties {size {sizeRule=Fill, value=1}}}
  }
  ShuffleDescription {
   widgetClassPath=/FortUILibrary/UIKit/Items/WBP_UIKit_DescriptionWithSubtitle
   bIsVariable=true
   parent=RightContent
   widget=ShuffleDescription
   properties {hideBodyText=false, hideCompatibility=true, hideRarityTag=true, hideSubtitle=true, visibility=Collapsed}
   slot {class=GridSlot, properties {rowSpan=1, verticalAlignment=VAlign_Bottom, padding=<ltrb 32 0 32 0>}}
  }
  WarningMessage {
   widgetClassPath=/FortUILibrary/UIKit/InfoBox/WBP_UIKit_InfoBox
   bIsVariable=true
   namedSlotHost=ShuffleDescription
   widget=WarningMessage
   properties {visibility=Collapsed, iconSize=<64 64>}
  }
 ]
}
